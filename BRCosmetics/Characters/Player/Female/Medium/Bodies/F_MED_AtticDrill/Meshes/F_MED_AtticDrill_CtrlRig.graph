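# Control Rig for the F_MED_AtticDrill female body, correcting skirt, haori, ribbon, sleeve and backpack-attach bones as the character moves.
# Thigh, calf and spine rotations become alphas through GetAlphaFromQuatDelta, many cached in SPR variables, alongside curve-shaped pelvis-height proximity values.
# These alphas and gameplay flags like crouching, sliding, jumping and parachuting weight layered ModifyTransforms offsets.
# AlphaInterp smooths many state-driven weights, while products combine pose conditions.
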ControlRigBlueprint /BRCosmetics/Characters/Player/Female/Medium/Bodies/F_MED_AtticDrill/Meshes/F_MED_AtticDrill_CtrlRig
graphs [
 CollapseNode_1_ContainedGraph {nodes [
   n1 = RigVMFunctionEntryNode()
   &RigUnit_ModifyTransforms(
    ItemToModify=((Item=(Type=Bone,Name="dyn_L_haori_front_mstr"),Transform=(Rotation=(X=-0.0624,Y=0.003,Z=0.0002,W=-0.998),Translation=(X=-0.0728,Y=-1.5278,Z=20.4227),Scale3D=(X=1,Y=1,Z=1))))
    WeightMinimum=0
    ExecutePin=@n1.ExecutePin
    Weight=RigVMVariableNode() {variable=SPR L Leg Fwd120}.Value
   )
   > &RigUnit_ModifyTransforms(
    ItemToModify=((Item=(Type=Bone,Name="dyn_R_haori_front_mstr"),Transform=(Rotation=(X=0,Y=0,Z=0,W=1),Translation=(X=0,Y=0,Z=9.0439),Scale3D=(X=1,Y=1,Z=1))))
    WeightMinimum=0
    Weight=RigVMVariableNode() {variable=SPR R Leg Fwd120}.Value
   )
   > &RigUnit_ModifyTransforms(
    ItemToModify="((Item=(Type=Bone,Name=\"dyn_L_haori_front_mstr\"),Transform=(Rotation=(X=-0.2329,Y=0.6272,Z=0.119,W=-0.7336),Translation=(X=-17.0393,Y=-1.2997,Z=-1.7965),Scale3D=(X=1,Y=1,Z=1))),(Item=(Type=Bone,Name=\"dyn_L_haori_side_mstr\"),Transform=(Rotation=(X=-0.0796,Y ...(truncated)"
    WeightMinimum=0
    Weight=RigVMVariableNode() {variable=SPR L Leg Otw_Haori}.Value
   )
   > &RigUnit_ModifyTransforms(
    ItemToModify="((Item=(Type=Bone,Name=\"dyn_R_haori_front_mstr\"),Transform=(Rotation=(X=0.0954,Y=0.7155,Z=-0.1379,W=-0.6782),Translation=(X=-7.8214,Y=-0.7144,Z=-1.118),Scale3D=(X=1,Y=1,Z=1))),(Item=(Type=Bone,Name=\"dyn_R_haori_side_mstr\"),Transform=(Rotation=(X=-0.1039,Y= ...(truncated)"
    WeightMinimum=0
    Weight=RigVMVariableNode() {variable=SPR R Leg Otw_Haori}.Value
   )
   n2 = RigVMVariableNode() {variable=SPR L Leg Bkd90}
   > &RigUnit_ModifyTransforms(
    ItemToModify="((Item=(Type=Bone,Name=\"dyn_L_haori_side_mstr\"),Transform=(Rotation=(X=-0.0592,Y=0.6863,Z=0.3235,W=-0.6487),Translation=(X=-3.8036,Y=0.0001,Z=-0.8885),Scale3D=(X=1,Y=1,Z=1))),(Item=(Type=Bone,Name=\"dyn_C_haori_back_mstr\"),Transform=(Rotation=(X=0.0096,Y=-0 ...(truncated)"
    WeightMinimum=0
    Weight=@n2.Value
   )
   n3 = RigVMVariableNode() {variable=SPR R Leg Bkd90}
   > &RigUnit_ModifyTransforms(
    ItemToModify="((Item=(Type=Bone,Name=\"dyn_R_haori_front_mstr\"),Transform=(Rotation=(X=0.0019,Y=-0.0023,Z=-0.6245,W=-0.781),Translation=(X=0,Y=0,Z=0),Scale3D=(X=1,Y=1,Z=1))),(Item=(Type=Bone,Name=\"dyn_R_haori_side_mstr\"),Transform=(Rotation=(X=0.0135,Y=0.6682,Z=-0.2837,W ...(truncated)"
    WeightMinimum=0
    Weight=@n3.Value
   )
   > &RigUnit_ModifyTransforms(
    ItemToModify=((Item=(Type=Bone,Name="dyn_C_haori_back_mstr"),Transform=(Rotation=(X=0.0188,Y=0.7911,Z=0.0197,W=0.611),Translation=(X=7.9437,Y=0.0844,Z=-2.3821),Scale3D=(X=1,Y=1,Z=1))))
    WeightMinimum=0
    Weight=/Script/RigVM.RigVMFunction_MathDoublePow(B=0.5, A=&RigVMFunction_MathDoubleMul(B=@n2.Value, A=@n3.Value).Result) {is=struct RigVMUnitNode, template="Power::Execute(in A,in B,out Result)"}.Result
   )
   n4 = RigVMVariableNode() {variable=pub isCrouching}
   RigVMFunction_MathBoolOr_1 = RigVMAggregateNode(
    A=RigVMVariableNode() {variable=pub isSprinting}.Value
    B=RigVMVariableNode() {variable=pub isKneeSliding}.Value
    C=RigVMVariableNode() {variable=pub isparachuteOpened}.Value
    D=RigVMVariableNode() {variable=pub_YouBoreme}.Value
    E=RigVMVariableNode() {variable=pub_IsGrindRail}.Value
   )
   n5 = RigVMDispatchNode(
    False=1
    True=1.8
    Condition=&RigVMFunction_MathBoolAnd(B=&RigVMFunction_MathBoolNot(Value=@n4.Value).Result, A=@RigVMFunction_MathBoolOr_1.Result).Result
   ) {template="DISPATCH_RigVMDispatch_If(in Condition,in True,in False,out Result)"}
   n6 = &RigVMFunction_AlphaInterp(
    InterpSpeedDecreasing=5
    InterpSpeedIncreasing=5
    Scale=1
    bClampResult=False
    Value=&RigVMFunction_MathDoubleMul(B=@n5.Result, A=RigVMVariableNode() {variable=SPR L calf Bkd100}.Value).Result
   )
   > &RigUnit_ModifyTransforms(
    ItemToModify="((Item=(Type=Bone,Name=\"dyn_L_haori_front_mstr\"),Transform=(Rotation=(X=0,Y=-0.1305,Z=0,W=0.9914),Translation=(X=0,Y=0,Z=0),Scale3D=(X=1,Y=1,Z=1))),(Item=(Type=Bone,Name=\"dyn_L_haori_side_mstr\"),Transform=(Rotation=(X=0.043,Y=-0.1371,Z=-0.1213,W=0.9822),Tr ...(truncated)"
    WeightMinimum=0
    Weight=@n6.Result
   )
   n7 = &RigVMFunction_AlphaInterp(
    InterpSpeedDecreasing=5
    InterpSpeedIncreasing=5
    Scale=1
    bClampResult=False
    Value=&RigVMFunction_MathDoubleMul(B=@n5.Result, A=RigVMVariableNode() {variable=SPR R calf Bkd100}.Value).Result
   )
   > &RigUnit_ModifyTransforms(
    ItemToModify="((Item=(Type=Bone,Name=\"dyn_L_haori_front_mstr\"),Transform=(Rotation=(X=0,Y=-0.1305,Z=0,W=0.9914),Translation=(X=0,Y=0,Z=0),Scale3D=(X=1,Y=1,Z=1))),(Item=(Type=Bone,Name=\"dyn_L_haori_side_mstr\"),Transform=(Rotation=(X=0.043,Y=-0.1371,Z=-0.1213,W=0.9822),Tr ...(truncated)"
    WeightMinimum=0
    Weight=@n7.Result
   )
   n8 = &RigVMFunction_AlphaInterp(
    InterpSpeedDecreasing=5
    InterpSpeedIncreasing=15
    Scale=1
    bClampResult=False
    Value=&RigVMFunction_MathBoolToFloat(Value=RigVMVariableNode() {variable=pub isJumping}.Value).Result
   )
   > &RigUnit_ModifyTransforms(
    ItemToModify="((Item=(Type=Bone,Name=\"dyn_L_haori_front_mstr\"),Transform=(Rotation=(X=0.0788,Y=-0.2907,Z=-0.2175,W=0.9284),Translation=(X=0,Y=0,Z=0),Scale3D=(X=1,Y=1,Z=1))),(Item=(Type=Bone,Name=\"dyn_L_haori_side_mstr\"),Transform=(Rotation=(X=0.0008,Y=0.315,Z=-0.0032,W= ...(truncated)"
    WeightMinimum=0
    Weight=@n8.Result
   )
   > &RigUnit_ModifyTransforms(
    ItemToModify="((Item=(Type=Bone,Name=\"dyn_L_haori_front_1\"),Transform=(Rotation=(X=0,Y=0.0872,Z=0,W=0.9962),Translation=(X=0,Y=0,Z=0),Scale3D=(X=1,Y=1,Z=1))),(Item=(Type=Bone,Name=\"dyn_L_haori_front_2\"),Transform=(Rotation=(X=0.0764,Y=0.2798,Z=-0.0339,W=0.9564),Translat ...(truncated)"
    WeightMinimum=0
    Weight=@n8.Result
   )
   > &RigUnit_ModifyTransforms(
    ItemToModify="((Item=(Type=Bone,Name=\"dyn_L_haori_side_1\"),Transform=(Rotation=(X=0,Y=0.0872,Z=0,W=0.9962),Translation=(X=0,Y=0,Z=0),Scale3D=(X=1,Y=1,Z=1))),(Item=(Type=Bone,Name=\"dyn_L_haori_side_2\"),Transform=(Rotation=(X=0,Y=0.0872,Z=0,W=0.9962),Translation=(X=0,Y=0, ...(truncated)"
    WeightMinimum=0
    Weight=@n8.Result
   )
   > &RigUnit_ModifyTransforms(
    ItemToModify="((Item=(Type=Bone,Name=\"dyn_C_haori_back_1\"),Transform=(Rotation=(X=0,Y=0.0872,Z=0,W=0.9962),Translation=(X=0,Y=0,Z=0),Scale3D=(X=1,Y=1,Z=1))),(Item=(Type=Bone,Name=\"dyn_C_haori_back_2\"),Transform=(Rotation=(X=0,Y=0.0872,Z=0,W=0.9962),Translation=(X=0,Y=0, ...(truncated)"
    WeightMinimum=0
    Weight=@n8.Result
   )
   > &RigUnit_ModifyTransforms(
    ItemToModify="((Item=(Type=Bone,Name=\"dyn_R_haori_front_1\"),Transform=(Rotation=(X=0,Y=0.0872,Z=0,W=0.9962),Translation=(X=0,Y=0,Z=0),Scale3D=(X=1,Y=1,Z=1))),(Item=(Type=Bone,Name=\"dyn_R_haori_front_2\"),Transform=(Rotation=(X=-0.1065,Y=0.1279,Z=0.0785,W=0.9829),Translat ...(truncated)"
    WeightMinimum=0
    Weight=@n8.Result
   )
   > &RigUnit_ModifyTransforms(
    ItemToModify="((Item=(Type=Bone,Name=\"dyn_R_haori_side_1\"),Transform=(Rotation=(X=0,Y=0.0872,Z=0,W=0.9962),Translation=(X=0,Y=0,Z=0),Scale3D=(X=1,Y=1,Z=1))),(Item=(Type=Bone,Name=\"dyn_R_haori_side_2\"),Transform=(Rotation=(X=0.1667,Y=-0.1215,Z=-0.0003,W=-0.9785),Translat ...(truncated)"
    WeightMinimum=0
    Weight=@n8.Result
   )
   n9 = &RigVMFunction_AlphaInterp(
    InterpSpeedDecreasing=5
    InterpSpeedIncreasing=5
    Scale=1
    bClampResult=False
    Value=&RigVMFunction_MathBoolToFloat(Value=RigVMVariableNode() {variable=pub isKneeSliding}.Value).Result
   )
   > &RigUnit_ModifyTransforms(
    ItemToModify="((Item=(Type=Bone,Name=\"dyn_L_haori_front_mstr\"),Transform=(Rotation=(X=0.2532,Y=0.3837,Z=0.246,W=-0.8533),Translation=(X=0,Y=0,Z=0),Scale3D=(X=1,Y=1,Z=1))),(Item=(Type=Bone,Name=\"dyn_L_haori_side_mstr\"),Transform=(Rotation=(X=0.0354,Y=0.605,Z=-0.1998,W=-0 ...(truncated)"
    WeightMinimum=0
    Weight=@n9.Result
   )
   > &RigUnit_ModifyTransforms(
    ItemToModify="((Item=(Type=Bone,Name=\"dyn_L_haori_front_2\"),Transform=(Rotation=(X=0,Y=0.0558,Z=0,W=0.9984),Translation=(X=0,Y=0,Z=0),Scale3D=(X=1,Y=1,Z=1))),(Item=(Type=Bone,Name=\"dyn_L_haori_side_2\"),Transform=(Rotation=(X=0,Y=0.0435,Z=0,W=0.9991),Translation=(X=0,Y=0 ...(truncated)"
    WeightMinimum=0
    Weight=@n9.Result
   )
   n10 = &RigVMFunction_MathBoolToFloat(Value=RigVMVariableNode() {variable=pub isBaseballSliding}.Value)
   n11 = &RigVMFunction_AlphaInterp(InterpSpeedDecreasing=5, InterpSpeedIncreasing=5, Scale=1, bClampResult=False, Value=@n10.Result)
   > &RigUnit_ModifyTransforms(
    ItemToModify="((Item=(Type=Bone,Name=\"dyn_L_haori_front_mstr\"),Transform=(Rotation=(X=0.1883,Y=0.37,Z=0.4062,W=-0.814),Translation=(X=6.5861,Y=-8.8981,Z=-9.6017),Scale3D=(X=1,Y=1,Z=1))),(Item=(Type=Bone,Name=\"dyn_L_haori_side_mstr\"),Transform=(Rotation=(X=0.1408,Y=0.542 ...(truncated)"
    WeightMinimum=0
    Weight=@n11.Result
   )
   > GetAlphaFromQuatDelta_9 = RigVMFunctionReferenceNode(
    Rotation Order=ZYX
    absoluteValue=false
    child=(Type=Bone,Name="pelvis")
    debugRotation=false
    onlyTwist=true
    outputMinMax=(X=0,Y=1)
    parent=(Type=Null,Name="pelvis_null")
    premult=1
    rotationMinMax=(X=50,Y=90)
    triggerAxis=X
    twistAxis=(X=1,Y=0,Z=0)
   )
   > &RigUnit_ModifyTransforms(
    ItemToModify="((Item=(Type=Bone,Name=\"dyn_L_haori_front_mstr\"),Transform=(Rotation=(X=-0.4126,Y=-0.3135,Z=-0.0443,W=-0.8541),Translation=(X=0,Y=0,Z=0),Scale3D=(X=1,Y=1,Z=1))),(Item=(Type=Bone,Name=\"dyn_L_haori_side_mstr\"),Transform=(Rotation=(X=-0.0219,Y=0.3769,Z=-0.008 ...(truncated)"
    WeightMinimum=0
    Weight=&RigVMFunction_MathDoubleMul(A=@GetAlphaFromQuatDelta_9.alpha, B=@n10.Result).Result
   )
   > &RigUnit_ModifyTransforms(
    ItemToModify="((Item=(Type=Bone,Name=\"dyn_C_skirt_b_3\"),Transform=(Rotation=(X=0,Y=-0.6626,Z=0,W=0.749),Translation=(X=2.4577,Y=-0.0002,Z=18.4587),Scale3D=(X=1,Y=1,Z=1))),(Item=(Type=Bone,Name=\"dyn_L_skirt_c_3\"),Transform=(Rotation=(X=0,Y=-0.1898,Z=0,W=0.9818),Translati ...(truncated)"
    WeightMinimum=0
    Weight=@n11.Result
   )
   n12 = &RigVMFunction_AlphaInterp(
    InterpSpeedDecreasing=5
    InterpSpeedIncreasing=10
    Scale=1
    bClampResult=False
    Value=&RigVMFunction_MathBoolToFloat(Value=RigVMVariableNode() {variable=pub isCrouching}.Value).Result
   )
   > &RigUnit_ModifyTransforms(
    ItemToModify="((Item=(Type=Bone,Name=\"dyn_L_haori_front_mstr\"),Transform=(Rotation=(X=-0.017,Y=-0.4358,Z=0.204,W=-0.8764),Translation=(X=-5.1888,Y=-0.9999,Z=9.4967),Scale3D=(X=1,Y=1,Z=1))),(Item=(Type=Bone,Name=\"dyn_L_haori_side_mstr\"),Transform=(Rotation=(X=0.0756,Y=-0 ...(truncated)"
    WeightMinimum=0
    Weight=@n12.Result
   )
   > &RigUnit_ModifyTransforms(
    ItemToModify="((Item=(Type=Bone,Name=\"dyn_L_haori_front_mstr\"),Transform=(Rotation=(X=-0.213,Y=0.2682,Z=-0.2052,W=-0.9168),Translation=(X=3.1589,Y=3.6311,Z=-5.9759),Scale3D=(X=1,Y=1,Z=1))),(Item=(Type=Bone,Name=\"dyn_L_haori_side_mstr\"),Transform=(Rotation=(X=0.0144,Y=-0 ...(truncated)"
    WeightMinimum=0
    Weight=&RigVMFunction_MathDoubleMul(A=@n12.Result, B=RigVMVariableNode() {variable=PelvisProx}.Value).Result
   )
   n13 = &RigVMFunction_AlphaInterp(
    InterpSpeedDecreasing=5
    InterpSpeedIncreasing=10
    Scale=1
    bClampResult=False
    Value=&RigVMFunction_MathBoolToFloat(Value=RigVMVariableNode() {variable=pub isSprinting}.Value).Result
   )
   > &RigUnit_ModifyTransforms(
    ItemToModify="((Item=(Type=Bone,Name=\"dyn_L_haori_front_mstr\"),Transform=(Rotation=(X=-0.0141,Y=-0.0467,Z=0.098,W=-0.994),Translation=(X=0,Y=0,Z=0),Scale3D=(X=1,Y=1,Z=1))),(Item=(Type=Bone,Name=\"dyn_L_haori_side_mstr\"),Transform=(Rotation=(X=-0.0052,Y=0.0128,Z=-0.0253,W ...(truncated)"
    WeightMinimum=0
    Weight=@n13.Result
   )
   n14 = RigVMVariableNode() {variable=pub isSkydive}
   n15 = &RigVMFunction_MathBoolNot(Value=RigVMVariableNode() {variable=pub isparachuteOpened}.Value)
   n16 = &RigVMFunction_AlphaInterp(
    InterpSpeedDecreasing=5
    InterpSpeedIncreasing=10
    Scale=1
    bClampResult=False
    Value=&RigVMFunction_MathDoubleMul(
     A=&RigVMFunction_MathBoolToFloat(Value=@n14.Value).Result
     B=&RigVMFunction_MathBoolToFloat(Value=@n15.Result).Result
    ).Result
   )
   > &RigUnit_ModifyTransforms(
    ItemToModify="((Item=(Type=Bone,Name=\"dyn_L_haori_front_mstr\"),Transform=(Rotation=(X=-0.0015,Y=0.0891,Z=0.0168,W=-0.9959),Translation=(X=0,Y=0,Z=0),Scale3D=(X=1,Y=1,Z=1))),(Item=(Type=Bone,Name=\"dyn_L_haori_side_mstr\"),Transform=(Rotation=(X=0,Y=0.0915,Z=0,W=-0.9958),T ...(truncated)"
    WeightMinimum=0
    Weight=@n16.Result
   )
   n17 = &RigVMFunction_AlphaInterp(
    InterpSpeedDecreasing=5
    InterpSpeedIncreasing=10
    Scale=1
    bClampResult=False
    Value=&RigVMFunction_MathBoolToFloat(Value=RigVMVariableNode() {variable=pub isparachuteOpened}.Value).Result
   )
   > &RigUnit_ModifyTransforms(
    ItemToModify="((Item=(Type=Bone,Name=\"dyn_L_haori_side_mstr\"),Transform=(Rotation=(X=0,Y=0.1317,Z=0,W=-0.9913),Translation=(X=0,Y=0,Z=0),Scale3D=(X=1,Y=1,Z=1))),(Item=(Type=Bone,Name=\"dyn_C_haori_back_mstr\"),Transform=(Rotation=(X=0,Y=0.1409,Z=0,W=-0.99),Translation=(X= ...(truncated)"
    WeightMinimum=0
    Weight=@n17.Result
   )
   n18 = &RigVMFunction_AlphaInterp(
    InterpSpeedDecreasing=5
    InterpSpeedIncreasing=5
    Scale=1
    bClampResult=False
    Value=&RigVMFunction_MathBoolToFloat(Value=RigVMVariableNode() {variable=pub isTacticalSprinting}.Value).Result
   )
   > &RigUnit_ModifyTransforms(
    ItemToModify="((Item=(Type=Bone,Name=\"dyn_L_haori_side_mstr\"),Transform=(Rotation=(X=-0.0076,Y=-0.0868,Z=-0.0868,W=0.9924),Translation=(X=0,Y=0,Z=0),Scale3D=(X=1,Y=1,Z=1))),(Item=(Type=Bone,Name=\"dyn_C_haori_back_mstr\"),Transform=(Rotation=(X=0,Y=-0.0872,Z=0,W=0.9962),T ...(truncated)"
    WeightMinimum=0
    Weight=@n18.Result
   )
   n19 = RigVMVariableNode() {variable=pub isCrouching}
   n20 = &RigVMFunction_AlphaInterp(
    InterpSpeedDecreasing=5
    InterpSpeedIncreasing=5
    Scale=1
    bClampResult=False
    Value=&RigVMFunction_MathDoubleMul(A=&RigVMFunction_MathBoolToFloat(Value=@n19.Value).Result, B=RigVMVariableNode() {variable=pub isMoveRight}.Value).Result
   )
   > &RigUnit_ModifyTransforms(
    ItemToModify="((Item=(Type=Bone,Name=\"dyn_C_haori_back_mstr\"),Transform=(Rotation=(X=-0.0244,Y=0.1414,Z=0.0111,W=-0.9896),Translation=(X=0,Y=0,Z=0),Scale3D=(X=1,Y=1,Z=1))),(Item=(Type=Bone,Name=\"dyn_R_haori_front_mstr\"),Transform=(Rotation=(X=0.0792,Y=0.0402,Z=0.0333,W= ...(truncated)"
    WeightMinimum=0
    Weight=@n20.Result
   )
   > &RigUnit_ModifyTransforms(
    ItemToModify="((Item=(Type=Bone,Name=\"dyn_L_sleeveSmallest_1\"),Transform=(Rotation=(X=0,Y=0,Z=0,W=-1),Translation=(X=0,Y=0,Z=0),Scale3D=(X=0.3451,Y=1,Z=1))),(Item=(Type=Bone,Name=\"dyn_L_sleeveSmall_1\"),Transform=(Rotation=(X=-0.0045,Y=0,Z=0,W=-1),Translation=(X=0,Y=0,Z= ...(truncated)"
    WeightMinimum=0
    Weight=&RigVMFunction_MathBoolToFloat(Value=RigVMVariableNode() {variable=pub isparachuteOpened}.Value).Result
   )
   > GetAlphaFromQuatDelta_14 = RigVMFunctionReferenceNode(
    Rotation Order=ZYX
    child=(Type=Bone,Name="thigh_l")
    debugRotation=false
    outputMinMax=(X=0,Y=1)
    parent=(Type=Null,Name="thigh_l_null")
    premult=1
    rotationMinMax=(X=0,Y=-102)
    triggerAxis=Z
    twistAxis=(X=1,Y=0,Z=0)
   )
   > GetAlphaFromQuatDelta_15 = RigVMFunctionReferenceNode(
    Rotation Order=ZYX
    child=(Type=Bone,Name="thigh_r")
    debugRotation=false
    outputMinMax=(X=0,Y=1)
    parent=(Type=Null,Name="thigh_r_null")
    premult=1
    rotationMinMax=(X=0,Y=-95)
    triggerAxis=Z
    twistAxis=(X=1,Y=0,Z=0)
   )
   > GetAlphaFromQuatDelta_1 = RigVMFunctionReferenceNode(
    Rotation Order=ZYX
    child=(Type=Bone,Name="thigh_l")
    debugRotation=false
    onlyTwist=true
    outputMinMax=(X=0,Y=1)
    parent=(Type=Null,Name="thigh_l_null")
    premult=1
    rotationMinMax=(X=0,Y=-26)
    triggerAxis=X
    twistAxis=(X=1,Y=0,Z=0)
   )
   > GetAlphaFromQuatDelta_2 = RigVMFunctionReferenceNode(
    Rotation Order=ZYX
    child=(Type=Bone,Name="thigh_r")
    debugRotation=false
    onlyTwist=true
    outputMinMax=(X=0,Y=1)
    parent=(Type=Null,Name="thigh_r_null")
    premult=1
    rotationMinMax=(X=0,Y=-23)
    triggerAxis=X
    twistAxis=(X=1,Y=0,Z=0)
   )
   Multiply_11 = RigVMAggregateNode(
    C=@GetAlphaFromQuatDelta_1.alpha
    A=@GetAlphaFromQuatDelta_14.alpha
    B=@GetAlphaFromQuatDelta_15.alpha
    D=@GetAlphaFromQuatDelta_2.alpha
    E=&RigVMFunction_MathBoolToFloat(Value=RigVMVariableNode() {variable=pub isPlayingEasternBloc}.Value).Result
    F=RigVMVariableNode() {variable=PelvisProx_Easterbloc}.Value
   )
   > &RigUnit_ModifyTransforms(
    ItemToModify="((Item=(Type=Bone,Name=\"base_C_skirt_a_1\"),Transform=(Rotation=(X=0,Y=-0.3728,Z=0,W=-0.9279),Translation=(X=0,Y=0,Z=0),Scale3D=(X=1,Y=1,Z=1))),(Item=(Type=Bone,Name=\"base_L_skirt_a_1\"),Transform=(Rotation=(X=0,Y=0.8244,Z=0,W=0.5661),Translation=(X=0,Y=0,Z= ...(truncated)"
    WeightMinimum=0
    Weight=@Multiply_11.Result
   )
   > GetAlphaFromQuatDelta_3 = RigVMFunctionReferenceNode(
    Rotation Order=ZYX
    child=(Type=Bone,Name="foot_l")
    debugRotation=false
    onlyTwist=false
    outputMinMax=(X=0,Y=1)
    parent=(Type=Null,Name="foot_l_null")
    premult=1
    rotationMinMax=(X=40,Y=70)
    triggerAxis=Z
    twistAxis=(X=1,Y=0,Z=0)
   )
   > &RigUnit_ModifyTransforms(
    ItemToModify=((Item=(Type=Bone,Name="dyn_L_bottom_ribbon_base"),Transform=(Rotation=(X=0.0074,Y=-0.0138,Z=-0.0613,W=-0.998),Translation=(X=0.2724,Y=-4.3151,Z=2.6981),Scale3D=(X=1,Y=1,Z=1))))
    WeightMinimum=0
    Weight=@GetAlphaFromQuatDelta_3.alpha
   )
   > GetAlphaFromQuatDelta_4 = RigVMFunctionReferenceNode(
    Rotation Order=ZYX
    child=(Type=Bone,Name="foot_r")
    debugRotation=false
    onlyTwist=false
    outputMinMax=(X=0,Y=1)
    parent=(Type=Null,Name="foot_r_null")
    premult=1
    rotationMinMax=(X=40,Y=70)
    triggerAxis=Z
    twistAxis=(X=1,Y=0,Z=0)
   )
   > &RigUnit_ModifyTransforms(
    ItemToModify=((Item=(Type=Bone,Name="dyn_R_bottom_ribbon_base"),Transform=(Rotation=(X=0,Y=0,Z=0,W=-1),Translation=(X=0,Y=3.0908,Z=-2.32),Scale3D=(X=1,Y=1,Z=1))))
    WeightMinimum=0
    Weight=@GetAlphaFromQuatDelta_4.alpha
   )
   n21 = RigVMVariableNode() {variable=pub isPickaxeSwing}
   n22 = &RigVMFunction_MathBoolNot(Value=RigVMVariableNode() {variable=pub isCrouching}.Value)
   n23 = &RigVMFunction_AlphaInterp(
    InterpSpeedDecreasing=5
    InterpSpeedIncreasing=5
    Scale=1
    bClampResult=False
    Value=&RigVMFunction_MathDoubleMul(
     A=&RigVMFunction_MathBoolToFloat(Value=@n21.Value).Result
     B=&RigVMFunction_MathBoolToFloat(Value=@n22.Result).Result
    ).Result
   )
   > &RigUnit_ModifyTransforms(
    ItemToModify=((Item=(Type=Bone,Name="dyn_R_haori_front_mstr"),Transform=(Rotation=(X=0,Y=0,Z=0.1736,W=0.9848),Translation=(X=-5,Y=0,Z=0),Scale3D=(X=1,Y=1,Z=1))))
    WeightMinimum=0
    Weight=@n23.Result
   )
   > GetAlphaFromQuatDelta_10 = RigVMFunctionReferenceNode(
    Rotation Order=ZYX
    child=(Type=Bone,Name="thigh_l")
    debugRotation=false
    onlyTwist=false
    outputMinMax=(X=0,Y=1)
    parent=(Type=Null,Name="thigh_l_null")
    premult=1
    rotationMinMax=(X=0,Y=15)
    triggerAxis=Y
    twistAxis=(X=1,Y=0,Z=0)
   )
   > GetAlphaFromQuatDelta_11 = RigVMFunctionReferenceNode(
    Rotation Order=ZYX
    child=(Type=Bone,Name="thigh_r")
    debugRotation=false
    onlyTwist=false
    outputMinMax=(X=0,Y=1)
    parent=(Type=Null,Name="thigh_r_null")
    premult=1
    rotationMinMax=(X=0,Y=-20)
    triggerAxis=Z
    twistAxis=(X=1,Y=0,Z=0)
   )
   Multiply_20 = RigVMAggregateNode(
    A=@GetAlphaFromQuatDelta_10.alpha
    B=@GetAlphaFromQuatDelta_11.alpha
    D=&RigVMFunction_MathBoolToFloat(Value=RigVMVariableNode() {variable=pub isPickaxeSwing}.Value).Result
    C=RigVMVariableNode() {variable=pub isMoveRight}.Value
   )
   > &RigUnit_ModifyTransforms(
    ItemToModify="((Item=(Type=Bone,Name=\"base_C_skirt_a_1\"),Transform=(Rotation=(X=0.0035,Y=0.0933,Z=-0.1133,W=-0.9892),Translation=(X=0,Y=0,Z=0),Scale3D=(X=1,Y=1,Z=1))),(Item=(Type=Bone,Name=\"base_R_skirt_a_1\"),Transform=(Rotation=(X=-0.0001,Y=0.2094,Z=-0.0946,W=-0.9732), ...(truncated)"
    WeightMinimum=0
    Weight=@Multiply_20.Result
   )
   n24 = &RigVMFunction_AlphaInterp(
    InterpSpeedDecreasing=5
    InterpSpeedIncreasing=5
    Scale=1
    bClampResult=False
    Value=&RigVMFunction_MathBoolToFloat(Value=RigVMVariableNode() {variable=pub_IsTacticalJump}.Value).Result
   )
   > n25 = &RigUnit_ModifyTransforms(
    ItemToModify="((Item=(Type=Bone,Name=\"dyn_C_haori_back_2\"),Transform=(Rotation=(X=0,Y=0.0872,Z=0,W=0.9962),Translation=(X=0,Y=0,Z=0),Scale3D=(X=1,Y=1,Z=1))),(Item=(Type=Bone,Name=\"dyn_L_haori_side_2\"),Transform=(Rotation=(X=0.0091,Y=0.0886,Z=0.1057,W=0.9904),Translation ...(truncated)"
    WeightMinimum=0
    Weight=@n24.Result
   )
   RigVMFunctionReturnNode(ExecutePin=@n25.ExecutePin)
  ]}
 CollapseNode_ContainedGraph {nodes [
   n1 = RigVMFunctionEntryNode()
   RigVMFunction_Sequence = &RigVMFunction_Sequence(ExecuteContext=@n1.ExecuteContext)
   GetAlphaFromQuatDelta_14 = RigVMFunctionReferenceNode(
    Rotation Order=ZYX
    absoluteValue=false
    child=(Type=Bone,Name="thigh_l")
    debugRotation=false
    onlyTwist=false
    outputMinMax=(X=0,Y=1)
    parent=(Type=Null,Name="thigh_l_null")
    premult=1
    rotationMinMax=(X=0,Y=-120)
    triggerAxis=Z
    twistAxis=(X=1,Y=0,Z=0)
    ExecuteContext=@RigVMFunction_Sequence.A
   )
   > RigVMVariableNode(Value=@GetAlphaFromQuatDelta_14.alpha) {variable=SPR L Leg Fwd120}
   > GetAlphaFromQuatDelta_15 = RigVMFunctionReferenceNode(
    Rotation Order=ZYX
    child=(Type=Bone,Name="thigh_r")
    outputMinMax=(X=0,Y=1)
    parent=(Type=Null,Name="thigh_r_null")
    premult=1
    rotationMinMax=(X=0,Y=-120)
    triggerAxis=Z
    twistAxis=(X=1,Y=0,Z=0)
   )
   > RigVMVariableNode(Value=@GetAlphaFromQuatDelta_15.alpha) {variable=SPR R Leg Fwd120}
   > GetAlphaFromQuatDelta_2 = RigVMFunctionReferenceNode(
    Rotation Order=ZYX
    child=(Type=Bone,Name="thigh_l")
    outputMinMax=(X=0,Y=1)
    parent=(Type=Null,Name="thigh_l_null")
    premult=1
    rotationMinMax=(X=0,Y=90)
    triggerAxis=Z
    twistAxis=(X=1,Y=0,Z=0)
   )
   > RigVMVariableNode(Value=@GetAlphaFromQuatDelta_2.alpha) {variable=SPR L Leg Bkd90}
   > GetAlphaFromQuatDelta_3 = RigVMFunctionReferenceNode(
    Rotation Order=ZYX
    child=(Type=Bone,Name="thigh_r")
    outputMinMax=(X=0,Y=1)
    parent=(Type=Null,Name="thigh_r_null")
    premult=1
    rotationMinMax=(X=0,Y=90)
    triggerAxis=Z
    twistAxis=(X=1,Y=0,Z=0)
   )
   > RigVMVariableNode(Value=@GetAlphaFromQuatDelta_3.alpha) {variable=SPR R Leg Bkd90}
   > GetAlphaFromQuatDelta_4 = RigVMFunctionReferenceNode(
    Rotation Order=ZYX
    absoluteValue=false
    child=(Type=Bone,Name="thigh_l")
    debugRotation=false
    onlyTwist=false
    outputMinMax=(X=0,Y=1)
    parent=(Type=Null,Name="thigh_l_null")
    premult=1
    rotationMinMax=(X=0,Y=-90)
    triggerAxis=Y
    twistAxis=(X=1,Y=0,Z=0)
   )
   > RigVMVariableNode(Value=@GetAlphaFromQuatDelta_4.alpha) {variable=SPR L Leg Otw}
   > GetAlphaFromQuatDelta_5 = RigVMFunctionReferenceNode(
    Rotation Order=ZYX
    child=(Type=Bone,Name="thigh_r")
    debugRotation=false
    outputMinMax=(X=0,Y=1)
    parent=(Type=Null,Name="thigh_r_null")
    premult=1
    rotationMinMax=(X=0,Y=-90)
    triggerAxis=Y
    twistAxis=(X=1,Y=0,Z=0)
   )
   > RigVMVariableNode(Value=@GetAlphaFromQuatDelta_5.alpha) {variable=SPR R Leg Otw}
   > GetAlphaFromQuatDelta_1 = RigVMFunctionReferenceNode(
    Rotation Order=ZYX
    absoluteValue=false
    child=(Type=Bone,Name="thigh_l")
    debugRotation=false
    onlyTwist=false
    outputMinMax=(X=0,Y=1)
    parent=(Type=Null,Name="thigh_l_null")
    premult=1
    rotationMinMax=(X=-10,Y=-90)
    triggerAxis=Y
    twistAxis=(X=1,Y=0,Z=0)
   )
   > RigVMVariableNode(Value=@GetAlphaFromQuatDelta_1.alpha) {variable=SPR L Leg Otw_Haori}
   > GetAlphaFromQuatDelta_16 = RigVMFunctionReferenceNode(
    Rotation Order=ZYX
    child=(Type=Bone,Name="thigh_r")
    debugRotation=false
    outputMinMax=(X=0,Y=1)
    parent=(Type=Null,Name="thigh_r_null")
    premult=1
    rotationMinMax=(X=-5,Y=-90)
    triggerAxis=Y
    twistAxis=(X=1,Y=0,Z=0)
   )
   > RigVMVariableNode(Value=@GetAlphaFromQuatDelta_16.alpha) {variable=SPR R Leg Otw_Haori}
   > GetAlphaFromQuatDelta_12 = RigVMFunctionReferenceNode(
    Rotation Order=ZYX
    child=(Type=Bone,Name="thigh_l")
    debugRotation=false
    outputMinMax=(X=0,Y=1)
    parent=(Type=Null,Name="thigh_l_null")
    premult=1
    rotationMinMax=(X=0,Y=30)
    triggerAxis=Y
    twistAxis=(X=1,Y=0,Z=0)
   )
   > RigVMVariableNode(Value=@GetAlphaFromQuatDelta_12.alpha) {variable=SPR L Leg Itw}
   > GetAlphaFromQuatDelta_13 = RigVMFunctionReferenceNode(
    Rotation Order=ZYX
    child=(Type=Bone,Name="thigh_r")
    outputMinMax=(X=0,Y=1)
    parent=(Type=Null,Name="thigh_r_null")
    premult=1
    rotationMinMax=(X=0,Y=30)
    triggerAxis=Y
    twistAxis=(X=1,Y=0,Z=0)
   )
   > RigVMVariableNode(Value=@GetAlphaFromQuatDelta_13.alpha) {variable=SPR R Leg Itw}
   > GetAlphaFromQuatDelta_8 = RigVMFunctionReferenceNode(
    Rotation Order=ZYX
    child=(Type=Bone,Name="calf_l")
    debugRotation=false
    outputMinMax=(X=0,Y=1)
    parent=(Type=Null,Name="calf_l_null")
    premult=1
    rotationMinMax=(X=0,Y=120)
    triggerAxis=Z
    twistAxis=(X=1,Y=0,Z=0)
   )
   > RigVMVariableNode(Value=@GetAlphaFromQuatDelta_8.alpha) {variable=SPR L calf Bkd100}
   > GetAlphaFromQuatDelta_9 = RigVMFunctionReferenceNode(
    Rotation Order=ZYX
    child=(Type=Bone,Name="calf_r")
    debugRotation=false
    outputMinMax=(X=0,Y=1)
    parent=(Type=Null,Name="calf_r_null")
    premult=1
    rotationMinMax=(X=0,Y=120)
    triggerAxis=Z
    twistAxis=(X=1,Y=0,Z=0)
   )
   > RigVMVariableNode(Value=@GetAlphaFromQuatDelta_9.alpha) {variable=SPR R calf Bkd100}
   > GetAlphaFromQuatDelta_10 = RigVMFunctionReferenceNode(
    Rotation Order=ZYX
    child=(Type=Bone,Name="calf_l")
    debugRotation=false
    outputMinMax=(X=0,Y=1)
    parent=(Type=Null,Name="calf_l_null")
    premult=1
    rotationMinMax=(X=100,Y=120)
    triggerAxis=Z
    twistAxis=(X=1,Y=0,Z=0)
   )
   > RigVMVariableNode(Value=@GetAlphaFromQuatDelta_10.alpha) {variable=SPR L calf Bkd140}
   > GetAlphaFromQuatDelta_11 = RigVMFunctionReferenceNode(
    Rotation Order=ZYX
    child=(Type=Bone,Name="calf_r")
    debugRotation=false
    outputMinMax=(X=0,Y=1)
    parent=(Type=Null,Name="calf_r_null")
    premult=1
    rotationMinMax=(X=100,Y=120)
    triggerAxis=Z
    twistAxis=(X=1,Y=0,Z=0)
   )
   > RigVMVariableNode(Value=@GetAlphaFromQuatDelta_11.alpha) {variable=SPR R calf Bkd140}
   > GetAlphaFromQuatDelta_6 = RigVMFunctionReferenceNode(
    Rotation Order=ZYX
    child=(Type=Bone,Name="calf_l")
    debugRotation=false
    onlyTwist=true
    outputMinMax=(X=0,Y=1)
    parent=(Type=Null,Name="calf_l_null")
    premult=1
    rotationMinMax=(X=0,Y=-90)
    triggerAxis=X
    twistAxis=(X=1,Y=0,Z=0)
   )
   > RigVMVariableNode(Value=@GetAlphaFromQuatDelta_6.alpha) {variable=SPR L calf BkdTwist}
   > GetAlphaFromQuatDelta_7 = RigVMFunctionReferenceNode(
    Rotation Order=ZYX
    child=(Type=Bone,Name="calf_r")
    debugRotation=false
    onlyTwist=false
    outputMinMax=(X=0,Y=1)
    parent=(Type=Null,Name="calf_r_null")
    premult=1
    rotationMinMax=(X=0,Y=-90)
    triggerAxis=X
    twistAxis=(X=1,Y=0,Z=0)
   )
   > RigVMVariableNode(Value=@GetAlphaFromQuatDelta_7.alpha) {variable=SPR R calf BkdTwist}
   n2 = &RigVMFunction_AnimEvalRichCurve(
    Curve="(EditorCurveData=(Keys=((InterpMode=RCIM_Cubic,TangentMode=RCTM_Break,TangentWeightMode=RCTWM_WeightedNone,Time=0,Value=0,ArriveTangent=0,ArriveTangentWeight=0,LeaveTangent=0,LeaveTangentWeight=0),(InterpMode=RCIM_Cubic,TangentMode=RCTM_Break,TangentWeight ...(truncated)"
    SourceMinimum=0.5
    TargetMinimum=0
    Value=&RigVMFunction_MathDoubleRemap(
     SourceMaximum=30
     SourceMinimum=60
     TargetMaximum=0
     TargetMinimum=1
     bClamp=False
     Value=&RigUnit_GetRelativeTransformForItem(Child=(Type=Bone,Name="pelvis"), Parent=(Type=Bone,Name="root"), bChildInitial=False, bParentInitial=False).RelativeTransform.Translation.Z
    ).Result
   )
   > RigVMVariableNode(Value=@n2.Result) {variable=PelvisProx}
   n3 = &RigVMFunction_AnimEvalRichCurve(
    Curve="(EditorCurveData=(Keys=((InterpMode=RCIM_Cubic,TangentMode=RCTM_Break,TangentWeightMode=RCTWM_WeightedNone,Time=0,Value=0,ArriveTangent=0,ArriveTangentWeight=0,LeaveTangent=0,LeaveTangentWeight=0),(InterpMode=RCIM_Cubic,TangentMode=RCTM_Break,TangentWeight ...(truncated)"
    SourceMinimum=0.5
    TargetMinimum=0
    Value=&RigVMFunction_MathDoubleRemap(
     SourceMaximum=30
     SourceMinimum=90
     TargetMaximum=0
     TargetMinimum=1
     bClamp=False
     Value=&RigUnit_GetRelativeTransformForItem(Child=(Type=Bone,Name="pelvis"), Parent=(Type=Bone,Name="root"), bChildInitial=False, bParentInitial=False).RelativeTransform.Translation.Z
    ).Result
   )
   > RigVMVariableNode(Value=@n3.Result) {variable=PelvisProx_Easterbloc}
   > GetAlphaFromQuatDelta_17 = RigVMFunctionReferenceNode(
    Rotation Order=ZYX
    child=(Type=Bone,Name="spine_05")
    debugRotation=false
    onlyTwist=false
    outputMinMax=(X=0,Y=1)
    parent=(Type=Null,Name="spine_05_null")
    premult=1
    rotationMinMax=(X=-50,Y=-60)
    triggerAxis=Z
    twistAxis=(X=1,Y=0,Z=0)
   )
   n4 = &RigVMFunction_MathBoolToFloat(Value=RigVMVariableNode() {variable=pub isparachuteOpened}.Value)
   > &RigUnit_ModifyTransforms(
    ItemToModify="((Item=(Type=Bone,Name=\"thigh_l\"),Transform=(Rotation=(X=0,Y=0,Z=-0.4226,W=0.9063),Translation=(X=0,Y=0,Z=0),Scale3D=(X=1,Y=1,Z=1))),(Item=(Type=Bone,Name=\"thigh_r\"),Transform=(Rotation=(X=0,Y=0,Z=-0.4226,W=0.9063),Translation=(X=0,Y=0,Z=0),Scale3D=(X=1,Y= ...(truncated)"
    WeightMinimum=0
    Weight=&RigVMFunction_MathDoubleMul(A=@GetAlphaFromQuatDelta_17.alpha, B=@n4.Result).Result
   )
   > GetAlphaFromQuatDelta_23 = RigVMFunctionReferenceNode(
    Rotation Order=ZYX
    child=(Type=Bone,Name="spine_05")
    debugRotation=false
    onlyTwist=false
    outputMinMax=(X=0,Y=1)
    parent=(Type=Null,Name="spine_05_null")
    premult=1
    rotationMinMax=(X=0,Y=-40)
    triggerAxis=Z
    twistAxis=(X=1,Y=0,Z=0)
   )
   > &RigUnit_ModifyTransforms(
    ItemToModify="((Item=(Type=Bone,Name=\"pelvis_skirt_offset\"),Transform=(Rotation=(X=0,Y=0,Z=-0.1736,W=0.9848),Translation=(X=0,Y=0,Z=0),Scale3D=(X=1,Y=1,Z=1))),(Item=(Type=Bone,Name=\"pelvis_skirt_haori\"),Transform=(Rotation=(X=0,Y=0,Z=-0.1736,W=0.9848),Translation=(X=0,Y ...(truncated)"
    WeightMinimum=0
    Weight=&RigVMFunction_MathDoubleMul(A=@GetAlphaFromQuatDelta_23.alpha, B=@n4.Result).Result
   )
   > GetAlphaFromQuatDelta_21 = RigVMFunctionReferenceNode(
    Rotation Order=ZYX
    child=(Type=Bone,Name="spine_05")
    debugRotation=false
    onlyTwist=true
    outputMinMax=(X=0,Y=1)
    parent=(Type=Null,Name="spine_05_null")
    premult=1
    rotationMinMax=(X=0,Y=10)
    triggerAxis=X
    twistAxis=(X=1,Y=0,Z=0)
   )
   RigVMFunction_MathBoolToFloat = &RigVMFunction_MathBoolToFloat(Value=RigVMVariableNode() {variable=pub isPickaxeSwing}.Value)
   Multiply = &RigVMFunction_MathDoubleMul(A=@GetAlphaFromQuatDelta_21.alpha, B=@RigVMFunction_MathBoolToFloat.Result)
   AlphaInterp = &RigVMFunction_AlphaInterp(InterpSpeedDecreasing=5, InterpSpeedIncreasing=5, Scale=1, bClampResult=False, Value=@Multiply.Result)
   > &RigUnit_ModifyTransforms(
    ItemToModify=((Item=(Type=Bone,Name="attach_backpack"),Transform=(Rotation=(X=0.0435,Y=0.0038,Z=0.0871,W=0.9952),Translation=(X=0,Y=0,Z=0),Scale3D=(X=1,Y=1,Z=1))))
    WeightMinimum=0
    Weight=@AlphaInterp.Result
   )
   > GetAlphaFromQuatDelta_22 = RigVMFunctionReferenceNode(
    Rotation Order=ZYX
    child=(Type=Bone,Name="spine_05")
    debugRotation=false
    onlyTwist=true
    outputMinMax=(X=0,Y=1)
    parent=(Type=Null,Name="spine_05_null")
    premult=1
    rotationMinMax=(X=0,Y=-10)
    triggerAxis=X
    twistAxis=(X=1,Y=0,Z=0)
   )
   n5 = &RigVMFunction_AlphaInterp(
    InterpSpeedDecreasing=5
    InterpSpeedIncreasing=5
    Scale=1
    bClampResult=False
    Value=&RigVMFunction_MathDoubleMul(A=@GetAlphaFromQuatDelta_22.alpha, B=@RigVMFunction_MathBoolToFloat.Result).Result
   )
   > &RigUnit_ModifyTransforms(
    ItemToModify=((Item=(Type=Bone,Name="attach_backpack"),Transform=(Rotation=(X=0.0472,Y=0.0396,Z=-0.0889,W=0.9941),Translation=(X=0,Y=0,Z=0),Scale3D=(X=1,Y=1,Z=1))))
    WeightMinimum=0
    Weight=@n5.Result
   )
   > GetAlphaFromQuatDelta_18 = RigVMFunctionReferenceNode(
    Rotation Order=ZYX
    child=(Type=Bone,Name="upperarm_l")
    debugRotation=false
    onlyTwist=false
    outputMinMax=(X=0,Y=1)
    parent=(Type=Null,Name="upperarm_l_null")
    premult=1
    rotationMinMax=(X=0,Y=-47)
    triggerAxis=X
    twistAxis=(X=1,Y=0,Z=0)
   )
   > GetAlphaFromQuatDelta_19 = RigVMFunctionReferenceNode(
    Rotation Order=ZYX
    child=(Type=Bone,Name="upperarm_l")
    debugRotation=false
    onlyTwist=false
    outputMinMax=(X=0,Y=1)
    parent=(Type=Null,Name="upperarm_l_null")
    premult=1
    rotationMinMax=(X=0,Y=-47)
    triggerAxis=Z
    twistAxis=(X=1,Y=0,Z=0)
   )
   > &RigUnit_ModifyTransforms(
    ItemToModify=((Item=(Type=Bone,Name="attach_backpack"),Transform=(Rotation=(X=0,Y=0,Z=-0.0872,W=0.9962),Translation=(X=0,Y=-2,Z=0),Scale3D=(X=1,Y=1,Z=1))))
    WeightMinimum=0
    Weight=&RigVMFunction_MathDoubleMul(A=@GetAlphaFromQuatDelta_19.alpha, B=@RigVMFunction_MathBoolToFloat.Result).Result
   )
   > GetAlphaFromQuatDelta_20 = RigVMFunctionReferenceNode(
    Rotation Order=ZYX
    child=(Type=Bone,Name="upperarm_r")
    debugRotation=false
    onlyTwist=false
    outputMinMax=(X=0,Y=1)
    parent=(Type=Null,Name="upperarm_r_null")
    premult=1
    rotationMinMax=(X=0,Y=-20)
    triggerAxis=Z
    twistAxis=(X=1,Y=0,Z=0)
   )
   > &RigUnit_ModifyTransforms(
    ItemToModify=((Item=(Type=Bone,Name="attach_backpack"),Transform=(Rotation=(X=0,Y=0,Z=0.0872,W=0.9962),Translation=(X=0,Y=-3,Z=0),Scale3D=(X=1,Y=1,Z=1))))
    WeightMinimum=0
    Weight=&RigVMFunction_MathDoubleMul(A=@GetAlphaFromQuatDelta_20.alpha, B=@RigVMFunction_MathBoolToFloat.Result).Result
   )
   > GetAlphaFromQuatDelta_24 = RigVMFunctionReferenceNode(
    Rotation Order=ZYX
    child=(Type=Bone,Name="head")
    debugRotation=false
    onlyTwist=false
    outputMinMax=(X=0,Y=1)
    parent=(Type=Null,Name="head_null")
    premult=1
    rotationMinMax=(X=0,Y=-50)
    triggerAxis=Z
    twistAxis=(X=1,Y=0,Z=0)
   )
   n6 = RigVMVariableNode() {variable=pub_IsTargeting}
   n7 = &RigVMFunction_AlphaInterp(
    InterpSpeedDecreasing=5
    InterpSpeedIncreasing=10
    Scale=1
    bClampResult=False
    Value=&RigVMFunction_MathDoubleMul(A=@GetAlphaFromQuatDelta_24.alpha, B=&RigVMFunction_MathBoolToFloat(Value=@n6.Value).Result).Result
   )
   > &RigUnit_ModifyTransforms(
    ItemToModify=((Item=(Type=Bone,Name="attach_backpack"),Transform=(Rotation=(X=0,Y=0,Z=0,W=1),Translation=(X=0,Y=4,Z=0),Scale3D=(X=1,Y=1,Z=1))))
    WeightMinimum=0
    Weight=@n7.Result
   )
   RigVMFunctionReturnNode(ExecuteContext=@RigVMFunction_Sequence.B)
  ]}
 CollapseNode_ContainedGraph {nodes [
   n1 = RigVMFunctionEntryNode()
   RigVMFunction_Sequence = RigVMAggregateNode(ExecuteContext=@n1.ExecutePin)
   Skirt = RigVMCollapseNode(ExecutePin=@RigVMFunction_Sequence.B)
   Haori = RigVMCollapseNode(ExecutePin=@RigVMFunction_Sequence.C)
   RigVMFunctionReturnNode(ExecutePin=@RigVMFunction_Sequence.D)
  ]}
 CollapseNode_ContainedGraph {nodes [
   n1 = RigVMFunctionEntryNode()
   n2 = RigVMVariableNode() {variable=SPR L Leg Fwd120}
   &RigUnit_ModifyTransforms(
    ItemToModify="((Item=(Type=Bone,Name=\"base_C_skirt_a_1\"),Transform=(Rotation=(X=0.0427,Y=-0.968,Z=-0.1687,W=0.1808),Translation=(X=-10.9014,Y=6.7533,Z=-6.442),Scale3D=(X=1,Y=1,Z=1))),(Item=(Type=Bone,Name=\"base_L_skirt_a_1\"),Transform=(Rotation=(X=-0.007,Y=-0.9324,Z=0.0 ...(truncated)"
    WeightMinimum=0
    ExecutePin=@n1.ExecutePin
    Weight=@n2.Value
   )
   n3 = RigVMVariableNode() {variable=SPR R Leg Fwd120}
   > &RigUnit_ModifyTransforms(
    ItemToModify="((Item=(Type=Bone,Name=\"base_C_skirt_a_1\"),Transform=(Rotation=(X=-0.0797,Y=0.9728,Z=0.0797,W=-0.2025),Translation=(X=-10.8025,Y=0.0004,Z=-6.2364),Scale3D=(X=1,Y=1,Z=1))),(Item=(Type=Bone,Name=\"base_R_skirt_b_1\"),Transform=(Rotation=(X=-0.0117,Y=0.0935,Z=0 ...(truncated)"
    WeightMinimum=0
    Weight=@n3.Value
   )
   Multiply = &RigVMFunction_MathDoubleMul(A=@n2.Value, B=@n3.Value)
   > &RigUnit_ModifyTransforms(
    ItemToModify=((Item=(Type=Bone,Name="base_C_skirt_a_1"),Transform=(Rotation=(X=0,Y=0.9114,Z=0,W=0.4115),Translation=(X=0.3965,Y=0,Z=-0.3497),Scale3D=(X=1,Y=1,Z=1))))
    WeightMinimum=0
    Weight=&RigVMFunction_MathDoublePow(B=0.5, A=@Multiply.Result) {template="Power::Execute(in A,in B,out Result)"}.Result
   )
   n4 = RigVMVariableNode() {variable=SPR L Leg Bkd90}
   > &RigUnit_ModifyTransforms(
    ItemToModify="((Item=(Type=Bone,Name=\"base_C_skirt_b_1\"),Transform=(Rotation=(X=-0.0093,Y=0.9543,Z=0.0029,W=-0.2988),Translation=(X=0,Y=0,Z=0),Scale3D=(X=1,Y=1,Z=1))),(Item=(Type=Bone,Name=\"base_L_skirt_b_1\"),Transform=(Rotation=(X=-0.0112,Y=-0.0117,Z=0.6863,W=-0.7271), ...(truncated)"
    WeightMinimum=0
    Weight=@n4.Value
   )
   n5 = RigVMVariableNode() {variable=SPR R Leg Bkd90}
   > &RigUnit_ModifyTransforms(
    ItemToModify="((Item=(Type=Bone,Name=\"base_C_skirt_b_1\"),Transform=(Rotation=(X=-0.0895,Y=-0.8814,Z=0.0278,W=0.463),Translation=(X=0,Y=0,Z=0),Scale3D=(X=1,Y=1,Z=1))),(Item=(Type=Bone,Name=\"base_R_skirt_c_1\"),Transform=(Rotation=(X=-0.0024,Y=-0.8226,Z=0.0896,W=0.5616),Tr ...(truncated)"
    WeightMinimum=0
    Weight=@n5.Value
   )
   Power = &RigVMFunction_MathDoublePow(B=0.6, A=&RigVMFunction_MathDoubleMul(A=@n4.Value, B=@n5.Value).Result) {template="Power::Execute(in A,in B,out Result)"}
   > &RigUnit_ModifyTransforms(
    ItemToModify="((Item=(Type=Bone,Name=\"base_C_skirt_b_1\"),Transform=(Rotation=(X=0.0769,Y=0.9373,Z=0.0401,W=0.3375),Translation=(X=0,Y=0,Z=0),Scale3D=(X=1,Y=1,Z=1))),(Item=(Type=Bone,Name=\"base_C_skirt_a_1\"),Transform=(Rotation=(X=-0.0376,Y=0.3903,Z=-0.0881,W=0.9157),Tra ...(truncated)"
    WeightMinimum=0
    Weight=@Power.Result
   )
   > &RigUnit_ModifyTransforms(
    ItemToModify="((Item=(Type=Bone,Name=\"base_L_skirt_a_1\"),Transform=(Rotation=(X=0.2185,Y=-0.7,Z=-0.444,W=0.5149),Translation=(X=0,Y=0,Z=0),Scale3D=(X=1,Y=1,Z=1))),(Item=(Type=Bone,Name=\"base_L_skirt_b_1\"),Transform=(Rotation=(X=0,Y=-0.7626,Z=0,W=0.6468),Translation=(X=- ...(truncated)"
    WeightMinimum=0
    Weight=RigVMVariableNode() {variable=SPR L Leg Otw}.Value
   )
   > &RigUnit_ModifyTransforms(
    ItemToModify="((Item=(Type=Bone,Name=\"base_R_skirt_c_1\"),Transform=(Rotation=(X=-0.1036,Y=-0.2117,Z=-0.695,W=0.6792),Translation=(X=0,Y=0,Z=0),Scale3D=(X=1,Y=1,Z=1))),(Item=(Type=Bone,Name=\"base_R_skirt_b_1\"),Transform=(Rotation=(X=0,Y=-0.7163,Z=0,W=0.6978),Translation= ...(truncated)"
    WeightMinimum=0
    Weight=RigVMVariableNode() {variable=SPR R Leg Otw}.Value
   )
   RigVMFunction_MathBoolNot = &RigVMFunction_MathBoolNot(Value=RigVMVariableNode() {variable=pub isCrouching}.Value)
   RigVMFunction_MathBoolOr = RigVMAggregateNode(
    B=RigVMVariableNode() {variable=pub isKneeSliding}.Value
    C=RigVMVariableNode() {variable=pub isparachuteOpened}.Value
    D=RigVMVariableNode() {variable=pub_YouBoreme}.Value
    A=RigVMVariableNode() {variable=pub isSprinting}.Value
   )
   RigVMFunction_MathBoolAnd = &RigVMFunction_MathBoolAnd(B=@RigVMFunction_MathBoolNot.Result, A=@RigVMFunction_MathBoolOr.Result)
   n6 = RigVMDispatchNode(False=1, True=1.8, Condition=@RigVMFunction_MathBoolAnd.Result) {template="DISPATCH_RigVMDispatch_If(in Condition,in True,in False,out Result)"}
   AlphaInterp = &RigVMFunction_AlphaInterp(
    InterpSpeedDecreasing=5
    InterpSpeedIncreasing=10
    Scale=1
    bClampResult=False
    Value=&RigVMFunction_MathDoubleMul(B=@n6.Result, A=RigVMVariableNode() {variable=SPR L calf Bkd100}.Value).Result
   )
   > &RigUnit_ModifyTransforms(
    ItemToModify="((Item=(Type=Bone,Name=\"base_C_skirt_b_1\"),Transform=(Rotation=(X=0,Y=-0.2164,Z=0,W=0.9763),Translation=(X=0,Y=0,Z=0),Scale3D=(X=1,Y=1,Z=1))),(Item=(Type=Bone,Name=\"base_L_skirt_b_1\"),Transform=(Rotation=(X=0,Y=-0.1305,Z=0,W=0.9914),Translation=(X=0,Y=0,Z= ...(truncated)"
    WeightMinimum=0
    Weight=@AlphaInterp.Result
   )
   n7 = RigVMVariableNode() {variable=SPR R calf Bkd100}
   n8 = &RigVMFunction_AlphaInterp(
    InterpSpeedDecreasing=5
    InterpSpeedIncreasing=10
    Scale=1
    bClampResult=False
    Value=&RigVMFunction_MathDoubleMul(B=@n6.Result, A=@n7.Value).Result
   )
   > &RigUnit_ModifyTransforms(
    ItemToModify="((Item=(Type=Bone,Name=\"base_C_skirt_b_1\"),Transform=(Rotation=(X=0,Y=-0.2164,Z=0,W=0.9763),Translation=(X=0,Y=0,Z=0),Scale3D=(X=1,Y=1,Z=1))),(Item=(Type=Bone,Name=\"base_L_skirt_c_1\"),Transform=(Rotation=(X=0,Y=-0.2164,Z=0,W=0.9763),Translation=(X=0,Y=0,Z= ...(truncated)"
    WeightMinimum=0
    Weight=@n8.Result
   )
   n9 = RigVMVariableNode() {variable=pub_YouBoreme}
   > &RigUnit_ModifyTransforms(
    ItemToModify=((Item=(Type=Bone,Name="dyn_R_skirt_c_3"),Transform=(Rotation=(X=0,Y=-0.5736,Z=0,W=0.8192),Translation=(X=0,Y=0,Z=0),Scale3D=(X=1,Y=1,Z=1))))
    WeightMinimum=0
    Weight=&RigVMFunction_MathDoubleMul(B=&RigVMFunction_MathBoolToFloat(Value=@n9.Value).Result, A=@n7.Value).Result
   )
   n10 = &RigVMFunction_AlphaInterp(
    InterpSpeedDecreasing=5
    InterpSpeedIncreasing=5
    Scale=1
    bClampResult=False
    Value=&RigVMFunction_MathBoolToFloat(Value=RigVMVariableNode() {variable=pub isKneeSliding}.Value).Result
   )
   > &RigUnit_ModifyTransforms(
    ItemToModify="((Item=(Type=Bone,Name=\"base_C_skirt_a_1\"),Transform=(Rotation=(X=0,Y=0.2336,Z=0,W=-0.9723),Translation=(X=0,Y=0,Z=0),Scale3D=(X=1,Y=1,Z=1))),(Item=(Type=Bone,Name=\"base_C_skirt_b_1\"),Transform=(Rotation=(X=0,Y=0.4688,Z=0,W=-0.8833),Translation=(X=-13.2174 ...(truncated)"
    WeightMinimum=0
    Weight=@n10.Result
   )
   n11 = &RigVMFunction_AlphaInterp(
    InterpSpeedDecreasing=5
    InterpSpeedIncreasing=5
    Scale=1
    bClampResult=False
    Value=&RigVMFunction_MathBoolToFloat(Value=RigVMVariableNode() {variable=pub isBaseballSliding}.Value).Result
   )
   > &RigUnit_ModifyTransforms(
    ItemToModify="((Item=(Type=Bone,Name=\"base_L_skirt_c_1\"),Transform=(Rotation=(X=0.0847,Y=0.2294,Z=-0.0303,W=-0.9692),Translation=(X=7.3988,Y=21.2427,Z=3.5201),Scale3D=(X=1,Y=1,Z=1))),(Item=(Type=Bone,Name=\"base_C_skirt_b_1\"),Transform=(Rotation=(X=-0.07,Y=0.2378,Z=-0.01 ...(truncated)"
    WeightMinimum=0
    Weight=@n11.Result
   )
   n12 = RigVMVariableNode() {variable=pub isSkydive}
   n13 = RigVMVariableNode() {variable=pub isparachuteOpened}
   > &RigUnit_ModifyTransforms(
    ItemToModify="((Item=(Type=Bone,Name=\"dyn_R_skirt_c_2\"),Transform=(Rotation=(X=0,Y=-0.2164,Z=0,W=0.9763),Translation=(X=0,Y=0,Z=0),Scale3D=(X=1,Y=1,Z=1))),(Item=(Type=Bone,Name=\"dyn_C_skirt_b_2\"),Transform=(Rotation=(X=0,Y=-0.1906,Z=0,W=0.9817),Translation=(X=0,Y=0,Z=0) ...(truncated)"
    WeightMinimum=0
    Weight=&RigVMFunction_MathBoolToFloat(Value=&RigVMFunction_MathBoolOr(A=@n12.Value, B=@n13.Value).Result).Result
   )
   n14 = &RigVMFunction_AlphaInterp(
    InterpSpeedDecreasing=5
    InterpSpeedIncreasing=10
    Scale=1
    bClampResult=False
    Value=&RigVMFunction_MathBoolToFloat(Value=RigVMVariableNode() {variable=pub isJumping}.Value).Result
   )
   > &RigUnit_ModifyTransforms(
    ItemToModify="((Item=(Type=Bone,Name=\"base_L_skirt_c_1\"),Transform=(Rotation=(X=-0.039,Y=-0.2399,Z=0.0449,W=0.969),Translation=(X=0,Y=0,Z=0),Scale3D=(X=0.9255,Y=1,Z=1))),(Item=(Type=Bone,Name=\"base_C_skirt_b_1\"),Transform=(Rotation=(X=0,Y=0.2518,Z=0,W=-0.9678),Translati ...(truncated)"
    WeightMinimum=0
    Weight=@n14.Result
   )
   > &RigUnit_ModifyTransforms(
    ItemToModify="((Item=(Type=Bone,Name=\"dyn_C_skirt_a_1\"),Transform=(Rotation=(X=0,Y=0.0872,Z=0,W=0.9962),Translation=(X=0,Y=0,Z=0),Scale3D=(X=1,Y=1,Z=1))),(Item=(Type=Bone,Name=\"dyn_C_skirt_a_2\"),Transform=(Rotation=(X=0,Y=0.0872,Z=0,W=0.9962),Translation=(X=0,Y=0,Z=0),S ...(truncated)"
    WeightMinimum=0
    Weight=@n14.Result
   )
   > &RigUnit_ModifyTransforms(
    ItemToModify="((Item=(Type=Bone,Name=\"dyn_C_skirt_b_1\"),Transform=(Rotation=(X=0,Y=0.0872,Z=0,W=0.9962),Translation=(X=0,Y=0,Z=0),Scale3D=(X=1,Y=1,Z=1))),(Item=(Type=Bone,Name=\"dyn_C_skirt_b_2\"),Transform=(Rotation=(X=0,Y=0.0872,Z=0,W=0.9962),Translation=(X=0,Y=0,Z=0),S ...(truncated)"
    WeightMinimum=0
    Weight=@n14.Result
   )
   > &RigUnit_ModifyTransforms(
    ItemToModify="((Item=(Type=Bone,Name=\"dyn_R_skirt_c_1\"),Transform=(Rotation=(X=0,Y=0.0872,Z=0,W=0.9962),Translation=(X=0,Y=0,Z=0),Scale3D=(X=1,Y=1,Z=1))),(Item=(Type=Bone,Name=\"dyn_R_skirt_c_2\"),Transform=(Rotation=(X=0,Y=0.0872,Z=0,W=0.9962),Translation=(X=0,Y=0,Z=0),S ...(truncated)"
    WeightMinimum=0
    Weight=@n14.Result
   )
   > &RigUnit_ModifyTransforms(
    ItemToModify="((Item=(Type=Bone,Name=\"dyn_R_skirt_b_1\"),Transform=(Rotation=(X=0,Y=0.0872,Z=0,W=0.9962),Translation=(X=0,Y=0,Z=0),Scale3D=(X=1,Y=1,Z=1))),(Item=(Type=Bone,Name=\"dyn_R_skirt_b_2\"),Transform=(Rotation=(X=0,Y=0.0872,Z=0,W=0.9962),Translation=(X=0,Y=0,Z=0),S ...(truncated)"
    WeightMinimum=0
    Weight=@n14.Result
   )
   > &RigUnit_ModifyTransforms(
    ItemToModify="((Item=(Type=Bone,Name=\"dyn_L_skirt_b_1\"),Transform=(Rotation=(X=0,Y=0.0872,Z=0,W=0.9962),Translation=(X=0,Y=0,Z=0),Scale3D=(X=1,Y=1,Z=1))),(Item=(Type=Bone,Name=\"dyn_L_skirt_b_2\"),Transform=(Rotation=(X=0,Y=0.0872,Z=0,W=0.9962),Translation=(X=0,Y=0,Z=0),S ...(truncated)"
    WeightMinimum=0
    Weight=@n14.Result
   )
   > &RigUnit_ModifyTransforms(
    ItemToModify="((Item=(Type=Bone,Name=\"dyn_L_skirt_c_1\"),Transform=(Rotation=(X=0,Y=0.0872,Z=0,W=0.9962),Translation=(X=0,Y=0,Z=0),Scale3D=(X=1,Y=1,Z=1))),(Item=(Type=Bone,Name=\"dyn_L_skirt_c_2\"),Transform=(Rotation=(X=0,Y=0.0872,Z=0,W=0.9962),Translation=(X=0,Y=0,Z=0),S ...(truncated)"
    WeightMinimum=0
    Weight=@n14.Result
   )
   n15 = RigVMVariableNode() {variable=pub isCrouching}
   n16 = &RigVMFunction_AlphaInterp(
    InterpSpeedDecreasing=5
    InterpSpeedIncreasing=5
    Scale=1
    bClampResult=False
    Value=&RigVMFunction_MathDoubleMul(B=&RigVMFunction_MathBoolToFloat(Value=@n15.Value).Result, A=RigVMVariableNode() {variable=pub isMoveBackward}.Value).Result
   )
   > &RigUnit_ModifyTransforms(
    ItemToModify="((Item=(Type=Bone,Name=\"base_C_skirt_a_1\"),Transform=(Rotation=(X=-0.0326,Y=0.0295,Z=0.2311,W=-0.9719),Translation=(X=0,Y=0,Z=0),Scale3D=(X=1,Y=1,Z=1))),(Item=(Type=Bone,Name=\"base_L_skirt_b_1\"),Transform=(Rotation=(X=0.1671,Y=-0.032,Z=-0.1434,W=-0.9749),T ...(truncated)"
    WeightMinimum=0
    Weight=@n16.Result
   )
   n17 = &RigVMFunction_AlphaInterp(
    InterpSpeedDecreasing=5
    InterpSpeedIncreasing=5
    Scale=1
    bClampResult=False
    Value=&RigVMFunction_MathBoolToFloat(Value=RigVMVariableNode() {variable=pub isPickaxeSwing}.Value).Result
   )
   > &RigUnit_ModifyTransforms(
    ItemToModify="((Item=(Type=Bone,Name=\"base_R_skirt_a_1\"),Transform=(Rotation=(X=0.0022,Y=0.0479,Z=-0.0452,W=-0.9978),Translation=(X=0,Y=0,Z=0),Scale3D=(X=1,Y=1,Z=1))),(Item=(Type=Bone,Name=\"base_R_skirt_b_1\"),Transform=(Rotation=(X=0.008,Y=-0.0621,Z=0.0005,W=0.998),Tran ...(truncated)"
    WeightMinimum=0
    Weight=@n17.Result
   )
   n18 = &RigVMFunction_AlphaInterp(
    InterpSpeedDecreasing=5
    InterpSpeedIncreasing=3
    Scale=1
    bClampResult=False
    Value=&RigVMFunction_MathBoolToFloat(Value=RigVMVariableNode() {variable=pub isCrouching}.Value).Result
   )
   > &RigUnit_ModifyTransforms(
    ItemToModify="((Item=(Type=Bone,Name=\"base_C_skirt_b_1\"),Transform=(Rotation=(X=0,Y=-0.4714,Z=0,W=-0.8819),Translation=(X=0,Y=0,Z=0),Scale3D=(X=0.9922,Y=1,Z=1))),(Item=(Type=Bone,Name=\"base_L_skirt_c_1\"),Transform=(Rotation=(X=0.0659,Y=-0.474,Z=0.1107,W=-0.871),Translat ...(truncated)"
    WeightMinimum=0
    Weight=@n18.Result
   )
   > &RigUnit_ModifyTransforms(
    ItemToModify="((Item=(Type=Bone,Name=\"dyn_L_skirt_c_2\"),Transform=(Rotation=(X=0,Y=-0.4394,Z=0,W=0.8983),Translation=(X=0,Y=0,Z=0),Scale3D=(X=0.6843,Y=1,Z=1))),(Item=(Type=Bone,Name=\"dyn_C_skirt_b_2\"),Transform=(Rotation=(X=0,Y=-0.5736,Z=0,W=0.8192),Translation=(X=0,Y=0 ...(truncated)"
    WeightMinimum=0
    Weight=@n18.Result
   )
   RigVMFunction_MathBoolToFloat = &RigVMFunction_MathBoolToFloat(Value=RigVMVariableNode() {variable=pub isCrouching}.Value)
   > &RigUnit_ModifyTransforms(
    ItemToModify="((Item=(Type=Bone,Name=\"base_C_skirt_a_1\"),Transform=(Rotation=(X=0.0137,Y=-0.1198,Z=0.2002,W=0.9723),Translation=(X=-2.5798,Y=0.0001,Z=9.0303),Scale3D=(X=1,Y=1,Z=1))),(Item=(Type=Bone,Name=\"base_L_skirt_a_1\"),Transform=(Rotation=(X=0.0024,Y=-0.0011,Z=0.09 ...(truncated)"
    WeightMinimum=0
    Weight=&RigVMFunction_MathDoubleMul(A=@RigVMFunction_MathBoolToFloat.Result, B=RigVMVariableNode() {variable=PelvisProx}.Value).Result
   )
   Multiply_15 = RigVMAggregateNode(
    C=&RigVMFunction_MathBoolToFloat(Value=RigVMVariableNode() {variable=pub isCrouching}.Value).Result
    A=RigVMVariableNode() {variable=pub isMoveForward}.Value
    B=RigVMVariableNode() {variable=PelvisProx}.Value
   )
   n19 = &RigVMFunction_AlphaInterp(InterpSpeedDecreasing=5, InterpSpeedIncreasing=5, Scale=1, bClampResult=False, Value=@Multiply_15.Result)
   > &RigUnit_ModifyTransforms(
    ItemToModify="((Item=(Type=Bone,Name=\"base_C_skirt_a_1\"),Transform=(Rotation=(X=-0.0114,Y=-0.13,Z=-0.0864,W=0.9877),Translation=(X=0,Y=0,Z=0),Scale3D=(X=1,Y=1,Z=1))),(Item=(Type=Bone,Name=\"dyn_C_haori_back_mstr\"),Transform=(Rotation=(X=0,Y=0.0593,Z=0,W=-0.9982),Translat ...(truncated)"
    WeightMinimum=0
    Weight=@n19.Result
   )
   > &RigUnit_ModifyTransforms(
    ItemToModify=((Item=(Type=Bone,Name="dyn_R_sleeveHole"),Transform=(Rotation=(X=0,Y=0,Z=0.0872,W=0.9962),Translation=(X=0,Y=0,Z=0),Scale3D=(X=1,Y=1,Z=1))))
    Weight=1
    WeightMinimum=0
   )
   > GetAlphaFromQuatDelta_5 = RigVMFunctionReferenceNode(
    Rotation Order=ZYX
    child=(Type=Bone,Name="calf_l")
    debugRotation=false
    outputMinMax=(X=0,Y=1)
    parent=(Type=Null,Name="calf_l_null")
    premult=1
    rotationMinMax=(X=0,Y=130)
    triggerAxis=Z
    twistAxis=(X=1,Y=0,Z=0)
   )
   n20 = &RigVMFunction_MathBoolToFloat(Value=RigVMVariableNode() {variable=pub_IsGrindRail}.Value)
   > &RigUnit_ModifyTransforms(
    ItemToModify="((Item=(Type=Bone,Name=\"base_L_skirt_b_1\"),Transform=(Rotation=(X=0,Y=-0.5,Z=0,W=0.866),Translation=(X=0,Y=0,Z=0),Scale3D=(X=1,Y=1,Z=1))),(Item=(Type=Bone,Name=\"base_C_skirt_a_1\"),Transform=(Rotation=(X=0,Y=-0.1736,Z=0,W=0.9848),Translation=(X=-10,Y=0,Z=0) ...(truncated)"
    WeightMinimum=0
    Weight=&RigVMFunction_MathDoubleMul(A=@GetAlphaFromQuatDelta_5.alpha, B=@n20.Result).Result
   )
   > GetAlphaFromQuatDelta_6 = RigVMFunctionReferenceNode(
    Rotation Order=ZYX
    child=(Type=Bone,Name="calf_r")
    debugRotation=false
    outputMinMax=(X=0,Y=1)
    parent=(Type=Null,Name="calf_r_null")
    premult=1
    rotationMinMax=(X=0,Y=130)
    triggerAxis=Z
    twistAxis=(X=1,Y=0,Z=0)
   )
   > &RigUnit_ModifyTransforms(
    ItemToModify="((Item=(Type=Bone,Name=\"base_R_skirt_b_1\"),Transform=(Rotation=(X=0.0449,Y=-0.2549,Z=0.1677,W=0.9513),Translation=(X=-5,Y=0,Z=0),Scale3D=(X=1,Y=1,Z=1))),(Item=(Type=Bone,Name=\"base_C_skirt_b_1\"),Transform=(Rotation=(X=0,Y=-0.2588,Z=0,W=0.9659),Translation= ...(truncated)"
    WeightMinimum=0
    Weight=&RigVMFunction_MathDoubleMul(A=@GetAlphaFromQuatDelta_6.alpha, B=@n20.Result).Result
   )
   n21 = &RigVMFunction_AlphaInterp(
    InterpSpeedDecreasing=5
    InterpSpeedIncreasing=5
    Scale=1
    bClampResult=False
    Value=&RigVMFunction_MathBoolToFloat(Value=RigVMVariableNode() {variable=pub isSprinting}.Value).Result
   )
   > &RigUnit_ModifyTransforms(
    ItemToModify="((Item=(Type=Bone,Name=\"base_C_skirt_a_1\"),Transform=(Rotation=(X=0,Y=0,Z=0,W=1),Translation=(X=0,Y=0,Z=10),Scale3D=(X=1,Y=1,Z=1))),(Item=(Type=Bone,Name=\"attach_backpack\"),Transform=(Rotation=(X=0,Y=0,Z=0,W=1),Translation=(X=0,Y=-2,Z=0),Scale3D=(X=1,Y=1,Z ...(truncated)"
    WeightMinimum=0
    Weight=@n21.Result
   )
   > GetAlphaFromQuatDelta_7 = RigVMFunctionReferenceNode(
    Rotation Order=ZYX
    child=(Type=Bone,Name="thigh_r")
    debugRotation=false
    outputMinMax=(X=0,Y=1)
    parent=(Type=Null,Name="thigh_r_null")
    premult=1
    rotationMinMax=(X=0,Y=28)
    triggerAxis=Y
    twistAxis=(X=1,Y=0,Z=0)
   )
   > GetAlphaFromQuatDelta_8 = RigVMFunctionReferenceNode(
    Rotation Order=ZYX
    child=(Type=Bone,Name="thigh_r")
    debugRotation=false
    outputMinMax=(X=0,Y=1)
    parent=(Type=Null,Name="thigh_r_null")
    premult=1
    rotationMinMax=(X=0,Y=-80)
    triggerAxis=Z
    twistAxis=(X=1,Y=0,Z=0)
   )
   Multiply_9 = RigVMAggregateNode(
    B=@GetAlphaFromQuatDelta_7.alpha
    A=@GetAlphaFromQuatDelta_8.alpha
    C=&RigVMFunction_MathBoolToFloat(Value=RigVMVariableNode() {variable=pub isCrouching}.Value).Result
    D=RigVMVariableNode() {variable=PelvisProx}.Value
   )
   > &RigUnit_ModifyTransforms(
    ItemToModify=((Item=(Type=Bone,Name="pelvis_skirt_offset"),Transform=(Rotation=(X=0.2927,Y=-0.0721,Z=0.0221,W=-0.9532),Translation=(X=0,Y=0,Z=0),Scale3D=(X=1,Y=1,Z=1))))
    WeightMinimum=0
    Weight=@Multiply_9.Result
   )
   > n22 = &RigUnit_ModifyTransforms(
    ItemToModify="((Item=(Type=Bone,Name=\"base_L_skirt_b_1\"),Transform=(Rotation=(X=0,Y=-0.0262,Z=0,W=0.9997),Translation=(X=0,Y=0,Z=0),Scale3D=(X=1,Y=1,Z=1))),(Item=(Type=Bone,Name=\"dyn_L_haori_front_3\"),Transform=(Rotation=(X=0,Y=-0.0872,Z=0,W=0.9962),Translation=(X=0,Y=0 ...(truncated)"
    WeightMinimum=0
    Weight=&RigVMFunction_MathBoolToFloat(Value=RigVMVariableNode() {variable=pub_IsFrontEnd}.Value).Result
   )
   RigVMFunctionReturnNode(ExecutePin=@n22.ExecutePin)
   &RigUnit_ModifyTransforms(
    ExecutePin=()
    ItemToModify="((Item=(Type=Bone,Name=\"base_C_skirt_b_1\"),Transform=(Rotation=(X=0.118,Y=-0.8429,Z=-0.0673,W=0.5206),Translation=(X=0,Y=0,Z=0),Scale3D=(X=1,Y=1,Z=1))),(Item=(Type=Bone,Name=\"base_L_skirt_b_1\"),Transform=(Rotation=(X=0.0098,Y=0.0148,Z=-0.5535,W=0.8326),Tra ...(truncated)"
    Weight=1
    WeightMinimum=0
   )
   &RigUnit_ModifyTransforms(
    ExecutePin=()
    ItemToModify="((Item=(Type=Bone,Name=\"dyn_R_skirt_c_2\"),Transform=(Rotation=(X=0,Y=-0.2164,Z=0,W=0.9763),Translation=(X=0,Y=0,Z=0),Scale3D=(X=1,Y=1,Z=1))),(Item=(Type=Bone,Name=\"dyn_C_skirt_b_2\"),Transform=(Rotation=(X=0,Y=-0.1736,Z=0,W=0.9848),Translation=(X=0,Y=0,Z=0) ...(truncated)"
    Weight=1
    WeightMinimum=0
   )
  ]}
 Multiply_13_ContainedGraph {nodes [
   n1 = RigVMFunctionEntryNode()
   n2 = &RigVMFunction_MathDoubleMul(B=@n1.D, A=&RigVMFunction_MathDoubleMul(B=@n1.C, A=&RigVMFunction_MathDoubleMul(A=@n1.A, B=@n1.B).Result).Result)
   RigVMFunctionReturnNode(Result=&RigVMFunction_MathDoubleMul(B=@n1.F, A=&RigVMFunction_MathDoubleMul(B=@n1.E, A=@n2.Result).Result).Result)
  ]}
 Multiply_16_ContainedGraph {nodes [
   n1 = RigVMFunctionEntryNode()
   RigVMFunctionReturnNode(Result=&RigVMFunction_MathDoubleMul(B=@n1.C, A=&RigVMFunction_MathDoubleMul(A=@n1.A, B=@n1.B).Result).Result)
  ]}
 Multiply_19_ContainedGraph {nodes [
   n1 = RigVMFunctionEntryNode()
   n2 = &RigVMFunction_MathDoubleMul(A=@n1.A, B=@n1.B)
   RigVMFunctionReturnNode(Result=&RigVMFunction_MathDoubleMul(B=@n1.D, A=&RigVMFunction_MathDoubleMul(B=@n1.C, A=@n2.Result).Result).Result)
  ]}
 Multiply_21_ContainedGraph {nodes [
   n1 = RigVMFunctionEntryNode()
   n2 = &RigVMFunction_MathDoubleMul(A=@n1.A, B=@n1.B)
   RigVMFunctionReturnNode(Result=&RigVMFunction_MathDoubleMul(B=@n1.D, A=&RigVMFunction_MathDoubleMul(B=@n1.C, A=@n2.Result).Result).Result)
  ]}
 New Function_ContainedGraph {nodes [
   RigVMFunctionEntryNode()
   RigVMFunction_MathBoolToInteger = &RigVMFunction_MathBoolToInteger(Value=RigVMVariableNode() {variable=onlyTwist}.Value)
   RigVMFunction_MathQuaternionSwingTwist = &RigVMFunction_MathQuaternionSwingTwist(
    Input=&RigUnit_GetRelativeTransformForItem(
     bChildInitial=False
     bParentInitial=False
     Child=RigVMVariableNode() {variable=child}.Value
     Parent=RigVMVariableNode() {variable=parent}.Value
    ).RelativeTransform.Rotation
    TwistAxis=RigVMVariableNode() {variable=twistAxis}.Value
   )
   n1 = RigVMDispatchNode(
    Values=((),())
    Index=@RigVMFunction_MathBoolToInteger.Result
    Values.0=@RigVMFunction_MathQuaternionSwingTwist.Swing
    Values.1=@RigVMFunction_MathQuaternionSwingTwist.Twist
   ) {template="DISPATCH_RigVMDispatch_SelectInt32(in Index,in Values,out Result)"}
   > RigVMVariableNode(Value=@n1.Result) {variable=entry_outputQuat}
   RigVMFunction_MathQuaternionToEuler = &RigVMFunction_MathQuaternionToEuler(Value=@n1.Result, RotationOrder=RigVMVariableNode() {variable=Rotation Order}.Value)
   > DISPATCH_RigVMDispatch_Print = RigVMDispatchNode(
    ScreenColor=(R=1,G=1,B=1,A=1)
    ScreenDuration=1.5
    Value=@RigVMFunction_MathQuaternionToEuler.Result
    Enabled=RigVMVariableNode() {variable=debugRotation}.Value
   ) {template="DISPATCH_RigVMDispatch_Print(in Prefix,in Value,in Enabled,in ScreenDuration,in ScreenColor)"}
   Cast = RigVMDispatchNode(Value=RigVMVariableNode() {variable=triggerAxis}.Value) {template="DISPATCH_RigVMDispatch_CastEnumToInt(in Value,out Result)"}
   DISPATCH_RigVMDispatch_SelectInt32 = RigVMDispatchNode(
    Values=(0,0,0)
    Index=@Cast.Result
    Values.0=@RigVMFunction_MathQuaternionToEuler.Result.X
    Values.1=@RigVMFunction_MathQuaternionToEuler.Result.Y
    Values.2=@RigVMFunction_MathQuaternionToEuler.Result.Z
   ) {template="DISPATCH_RigVMDispatch_SelectInt32(in Index,in Values,out Result)"}
   n2 = RigVMVariableNode() {variable=rotationMinMax}
   n3 = RigVMVariableNode() {variable=outputMinMax}
   Remap = &RigVMFunction_MathDoubleRemap(
    bClamp=False
    Value=@DISPATCH_RigVMDispatch_SelectInt32.Result
    SourceMinimum=@n2.Value.X
    SourceMaximum=@n2.Value.Y
    TargetMinimum=@n3.Value.X
    TargetMaximum=@n3.Value.Y
   )
   Absolute = &RigVMFunction_MathDoubleAbs(Value=@Remap.Result)
   n4 = RigVMVariableNode() {variable=absoluteValue}
   Multiply = &RigVMFunction_MathDoubleMul(
    A=&RigVMFunction_MathDoubleClamp(Value=RigVMDispatchNode(True=@Absolute.Result, False=@Remap.Result, Condition=@n4.Value) {template="DISPATCH_RigVMDispatch_If(in Condition,in True,in False,out Result)"}.Result).Result
    B=RigVMVariableNode() {variable=premult}.Value
   )
   > RigVMFunctionReturnNode(alpha=@Multiply.Result, rawQuat=RigVMVariableNode() {variable=entry_outputQuat}.Value)
  ]}
 RigVMFunctionLibrary {nodes [GetAlphaFromQuatDelta = RigVMCollapseNode(
    ExecuteContext=()
    Rotation Order=ZYX
    child=(Type=Bone,Name="None")
    outputMinMax=(X=0,Y=1)
    parent=(Type=Bone,Name="None")
    premult=1
    rotationMinMax=(X=0,Y=0)
    twistAxis=(X=1,Y=0,Z=0)
   )]}
 RigVMFunction_MathBoolOr_2_ContainedGraph {nodes [
   n1 = RigVMFunctionEntryNode()
   RigVMFunction_MathBoolOr = &RigVMFunction_MathBoolOr(A=@n1.A, B=@n1.B)
   RigVMFunctionReturnNode(Result=&RigVMFunction_MathBoolOr(B=@n1.D, A=&RigVMFunction_MathBoolOr(B=@n1.C, A=@RigVMFunction_MathBoolOr.Result).Result).Result)
  ]}
 RigVMFunction_MathBoolOr_2_ContainedGraph {nodes [
   n1 = RigVMFunctionEntryNode()
   n2 = &RigVMFunction_MathBoolOr(B=@n1.C, A=&RigVMFunction_MathBoolOr(A=@n1.A, B=@n1.B).Result)
   RigVMFunctionReturnNode(Result=&RigVMFunction_MathBoolOr(B=@n1.E, A=&RigVMFunction_MathBoolOr(B=@n1.D, A=@n2.Result).Result).Result)
  ]}
 RigVMFunction_Sequence_1_ContainedGraph {nodes [
   n1 = RigVMFunctionEntryNode()
   n2 = &RigVMFunction_Sequence(ExecuteContext=@n1.ExecuteContext)
   n3 = &RigVMFunction_Sequence(ExecuteContext=@n2.B)
   RigVMFunctionReturnNode(B=@n2.A, C=@n3.A, D=@n3.B)
  ]}
 RigVMModel {nodes [
   RigUnit_BeginExecution = &RigUnit_BeginExecution()
   /Script/FortniteGame.RigUnit_FortDeformRig(
    Bool=(bUpperArm=True,bElbow=True,bWrist=True,bGroin=true,bGlute=true,bKnee=true,bPec=false,bBicep=false,bLat=false)
    Curve="(bLUpperArmUpAlpha=False,bRUpperArmUpAlpha=False,bLUpperArmFrAlpha=False,bRUpperArmFrAlpha=False,bLUpperArmBkAlpha=False,bRUpperArmBkAlpha=False,bLWristUpAlpha=False,bRWristUpAlpha=False,bLWristDnAlpha=False,bRWristDnAlpha=False,bLKneeAlpha=false,bRKneeAlp ...(truncated)"
    Debug=(bUpperArmDebug=False,bWristDebug=False,bLatDebug=False,bGroinGluteDebug=False)
    Value="(UpperArmUp=(Rotation=(X=0.0,Y=-0.0,Z=0.0,W=1.0),Translation=(X=0.0,Y=0.0,Z=8.0),Scale3D=(X=0.70,Y=1.0,Z=1.0)),UpperArmFwdFr=(Rotation=(X=0.0,Y=-0.0,Z=0.0,W=1.0),Translation=(X=0.0,Y=10.0,Z=0.0),Scale3D=(X=0.50,Y=1.0,Z=1.0)),UpperArmFwdBk=(Rotation=(X=0.0, ...(truncated)"
    ExecutePin=@RigUnit_BeginExecution.ExecutePin
   ) {is=struct RigVMUnitNode}
   > vnodes = RigVMCollapseNode()
   SkirtFix = RigVMCollapseNode(ExecutePin=@vnodes.ExecuteContext)
  ]}
]
hierarchy [
 root {loc=<0 0 0>, rot=<0 0 0 1>}
 pelvis {parent=root, loc=<0 -0.5301 98.0571>, rot=<0.0127 0.707 -0.0127 -0.707>}
 spine_01 {parent=pelvis, loc=<6.1208 -0.9328 0>, rot=<0 0 0.034 -0.9994>}
 spine_02 {parent=spine_01, loc=<6.5997 -0.2095 0>, rot=<0 0 0.0238 -0.9997>}
 spine_03 {parent=spine_02, loc=<6.4914 -0.1237 0>, rot=<0 0 0.0149 -0.9999>}
 spine_04 {parent=spine_03, loc=<6.9598 0.0794 0>, rot=<0 0 -0.0132 -0.9999>}
 spine_05 {parent=spine_04, loc=<6.9987 0 0>, rot=<0 0 -0.0015 -1>}
 neck_01 {parent=spine_05, loc=<11.3138 0.1529 0>, rot=<0 0 -0.2003 -0.9797>}
 neck_02 {parent=neck_01, loc=<4.5627 0 0>, rot=<0 0 0.0504 -0.9987>}
 head {parent=neck_02, loc=<5.3914 0 0>, rot=<0 0 0.115 -0.9934>}
 clavicle_l {parent=spine_05, loc=<5.4831 -0.1053 -3.2946>, rot=<0.0465 -0.7289 -0.009 -0.6829>}
 upperarm_l {parent=clavicle_l, loc=<9.8658 -1.2136 -0.4105>, rot=<0.0164 -0.3245 -0.0269 -0.9454>}
 lowerarm_l {parent=upperarm_l, loc=<26.465 -0.8345 0.259>, rot=<0.0036 0.0161 -0.2298 -0.9731>}
 hand_l {parent=lowerarm_l, loc=<22.7355 -0.3203 -0.6105>, rot=<-0.6784 -0.0223 0.0735 -0.7307>}
 thumb_01_l {parent=hand_l, loc=<1.3118 -0.3191 -2.459>, rot=<0.6546 -0.4042 0.0664 -0.6354>}
 thumb_02_l {parent=thumb_01_l, loc=<5.9336 -0.0448 0.0401>, rot=<-0.0037 0.0213 0.298 -0.9543>}
 thumb_03_l {parent=thumb_02_l, loc=<2.2223 0.3682 -0.0907>, rot=<0.0118 -0.013 0.0066 -0.9998>}
 index_metacarpal_l {parent=hand_l, loc=<1.6535 0.0183 -2.96>, rot=<0.0158 -0.0124 -0.0173 -0.9996>}
 index_01_l {parent=index_metacarpal_l, loc=<8.3008 0.0357 -0.0524>, rot=<-0.0044 0.0485 0.0924 -0.9945>}
 index_02_l {parent=index_01_l, loc=<3.3587 -0.3004 0.0452>, rot=<-0.002 0.0027 0.3225 -0.9466>}
 index_03_l {parent=index_02_l, loc=<2.5578 -0.123 -0.0252>, rot=<-0.002 0.0067 0.2558 -0.9667>}
 middle_metacarpal_l {parent=hand_l, loc=<2.1856 -0.1725 -0.5984>, rot=<0.006 0.0168 -0.0329 -0.9993>}
 middle_01_l {parent=middle_metacarpal_l, loc=<7.3275 -0.2447 0.0129>, rot=<-0.0208 0.0042 0.0858 -0.9961>}
 middle_02_l {parent=middle_01_l, loc=<4.0123 0.1401 0.1175>, rot=<-0.0218 0.0258 0.3512 -0.9357>}
 middle_03_l {parent=middle_02_l, loc=<2.6077 0.0534 0.1373>, rot=<0.0082 0.0188 0.1708 -0.9851>}
 ring_metacarpal_l {parent=hand_l, loc=<2.2074 -0.3625 1.3638>, rot=<-0.0284 0.0722 -0.0458 -0.9959>}
 ring_01_l {parent=ring_metacarpal_l, loc=<7.0295 -0.6197 -0.1446>, rot=<-0.0489 -0.0252 0.1892 -0.9804>}
 ring_02_l {parent=ring_01_l, loc=<3.6174 0.4647 -0.1027>, rot=<0 0 0.2806 -0.9598>}
 ring_03_l {parent=ring_02_l, loc=<2.4152 0.114 0.0199>, rot=<0 0 0.1189 -0.9929>}
 pinky_metacarpal_l {parent=hand_l, loc=<1.7616 -0.9753 3.3814>, rot=<-0.0397 0.0814 -0.069 -0.9935>}
 pinky_01_l {parent=pinky_metacarpal_l, loc=<7.1709 -0.4831 -0.1094>, rot=<-0.0654 -0.0035 0.2242 -0.9723>}
 pinky_02_l {parent=pinky_01_l, loc=<2.8274 0.0893 -0.1043>, rot=<0.0037 0.0011 0.2854 -0.9584>}
 pinky_03_l {parent=pinky_02_l, loc=<2.2046 0.0307 -0.0078>, rot=<-0.0002 0.0098 0.1445 -0.9895>}
 lowerarm_twist_02_l {parent=lowerarm_l, loc=<10.6707 0 0>, rot=<0 0 0 1>}
 dyn_L_sleeveHole {parent=lowerarm_twist_02_l, loc=<-3.0203 0 0>, rot=<0 0 0 1>}
 dyn_L_sleeveBig_1 {parent=dyn_L_sleeveHole, loc=<9.1107 -4.9872 -10.7497>, rot=<0.6027 -0.5705 -0.4073 -0.3813>}
 dyn_L_sleeveBig_2 {parent=dyn_L_sleeveBig_1, loc=<8.5314 0 0>, rot=<0 0 0 1>}
 dyn_L_sleeveSmall_1 {parent=lowerarm_l, loc=<5.5207 -1.3181 -9.2739>, rot=<0.5478 -0.5692 -0.3614 -0.4953>}
 dyn_L_sleeveSmall_2 {parent=dyn_L_sleeveSmall_1, loc=<6.9159 0 0>, rot=<0 0 0 1>}
 upperarm_twist_01_l {parent=upperarm_l, loc=<7.2135 -0.1134 0>, rot=<0 0 0 1>}
 upperarm_twist_02_l {parent=upperarm_l, loc=<19.7998 -0.1134 0>, rot=<0 0 0 1>}
 dyn_L_sleeveSmallest_1 {parent=upperarm_twist_02_l, loc=<1.8593 -0.9815 -8.6792>, rot=<0.6052 -0.4231 -0.4288 -0.5204>}
 clavicle_r {parent=spine_05, loc=<5.4831 -0.1053 3.2946>, rot=<0.7289 0.0465 -0.6829 0.009>}
 upperarm_r {parent=clavicle_r, loc=<-9.8658 1.2136 0.4105>, rot=<0.0164 -0.3245 -0.0269 -0.9454>}
 lowerarm_r {parent=upperarm_r, loc=<-26.465 0.8345 -0.259>, rot=<0.0036 0.0161 -0.2298 -0.9731>}
 hand_r {parent=lowerarm_r, loc=<-22.7355 0.3203 0.6106>, rot=<-0.6784 -0.0223 0.0735 -0.7307>}
 thumb_01_r {parent=hand_r, loc=<-1.3118 0.3191 2.459>, rot=<0.6546 -0.4042 0.0664 -0.6354>}
 thumb_02_r {parent=thumb_01_r, loc=<-5.9336 0.0448 -0.0401>, rot=<-0.0037 0.0213 0.298 -0.9543>}
 thumb_03_r {parent=thumb_02_r, loc=<-2.2223 -0.3682 0.0907>, rot=<0.0118 -0.013 0.0066 -0.9998>}
 index_metacarpal_r {parent=hand_r, loc=<-1.6535 -0.0183 2.96>, rot=<0.0158 -0.0124 -0.0173 -0.9996>}
 index_01_r {parent=index_metacarpal_r, loc=<-8.3008 -0.0357 0.0524>, rot=<-0.0044 0.0485 0.0924 -0.9945>}
 index_02_r {parent=index_01_r, loc=<-3.3587 0.3004 -0.0452>, rot=<-0.002 0.0027 0.3225 -0.9466>}
 index_03_r {parent=index_02_r, loc=<-2.5578 0.123 0.0252>, rot=<-0.002 0.0067 0.2558 -0.9667>}
 middle_metacarpal_r {parent=hand_r, loc=<-2.1856 0.1725 0.5984>, rot=<0.006 0.0168 -0.0329 -0.9993>}
 middle_01_r {parent=middle_metacarpal_r, loc=<-7.3275 0.2447 -0.0129>, rot=<-0.0208 0.0042 0.0858 -0.9961>}
 middle_02_r {parent=middle_01_r, loc=<-4.0123 -0.1401 -0.1175>, rot=<-0.0218 0.0258 0.3512 -0.9357>}
 middle_03_r {parent=middle_02_r, loc=<-2.6077 -0.0534 -0.1373>, rot=<0.0082 0.0188 0.1708 -0.9851>}
 ring_metacarpal_r {parent=hand_r, loc=<-2.2074 0.3625 -1.3638>, rot=<-0.0284 0.0722 -0.0458 -0.9959>}
 ring_01_r {parent=ring_metacarpal_r, loc=<-7.0295 0.6197 0.1446>, rot=<-0.0489 -0.0252 0.1892 -0.9804>}
 ring_02_r {parent=ring_01_r, loc=<-3.6174 -0.4647 0.1027>, rot=<0 0 0.2806 -0.9598>}
 ring_03_r {parent=ring_02_r, loc=<-2.4152 -0.114 -0.0199>, rot=<0 0 0.1189 -0.9929>}
 pinky_metacarpal_r {parent=hand_r, loc=<-1.7616 0.9753 -3.3814>, rot=<-0.0397 0.0814 -0.069 -0.9935>}
 pinky_01_r {parent=pinky_metacarpal_r, loc=<-7.1709 0.4831 0.1094>, rot=<-0.0654 -0.0035 0.2242 -0.9723>}
 pinky_02_r {parent=pinky_01_r, loc=<-2.8274 -0.0893 0.1043>, rot=<0.0037 0.0011 0.2854 -0.9584>}
 pinky_03_r {parent=pinky_02_r, loc=<-2.2046 -0.0307 0.0078>, rot=<-0.0002 0.0098 0.1445 -0.9895>}
 lowerarm_twist_02_r {parent=lowerarm_r, loc=<-10.6707 0 0>, rot=<0 0 0 -1>}
 dyn_R_sleeveHole {parent=lowerarm_twist_02_r, loc=<3.0202 0 0.0002>, rot=<0 0 0 1>}
 dyn_R_sleeveBig_1 {parent=dyn_R_sleeveHole, loc=<-9.1107 4.9872 10.7495>, rot=<0.6027 -0.5705 -0.4073 -0.3813>}
 dyn_R_sleeveBig_2 {parent=dyn_R_sleeveBig_1, loc=<-8.5315 0 -0.0001>, rot=<0 0 0 -1>}
 dyn_R_sleeveSmall_1 {parent=lowerarm_r, loc=<-5.5205 1.318 9.2735>, rot=<0.5478 -0.5692 -0.3614 -0.4953>}
 dyn_R_sleeveSmall_2 {parent=dyn_R_sleeveSmall_1, loc=<-6.9161 0.0001 0>, rot=<0 0 0 1>}
 upperarm_twist_01_r {parent=upperarm_r, loc=<-7.2135 0.1134 0>, rot=<0 0 0 -1>}
 upperarm_twist_02_r {parent=upperarm_r, loc=<-19.7998 0.1134 0>, rot=<0 0 0 -1>}
 dyn_R_sleeveSmallest_1 {parent=upperarm_twist_02_r, loc=<-1.8596 0.9815 8.6795>, rot=<0.6052 -0.4231 -0.4288 -0.5204>}
 attach_backpack {parent=spine_05, loc=<-0.123 -13.473 0>, rot=<-0.0093 -0.707 0.0093 -0.707>}
 Backpack_BR {parent=attach_backpack, type=Null, loc=<0 -14.4 1>, rot=<-0.1045 0 0 0.9945>}
 attach_cape {parent=spine_05, loc=<9.017 -11.714 0>, rot=<0.0616 -0.7044 -0.0616 -0.7044>}
 cape {parent=attach_cape, type=Null, loc=<0 -14.3 0>, rot=<0 0 0 1>}
 head_null {parent=spine_05, type=Null, loc=<20.6573 3.5489 0>, rot=<0 0 -0.036 -0.9994>}
 upperarm_l_null {parent=spine_05, type=Null, loc=<4.5519 -1.8922 -13.0371>, rot=<-0.0385 0.9118 0.0238 0.4081>}
 upperarm_r_null {parent=spine_05, type=Null, loc=<4.5519 -1.8922 13.0371>, rot=<-0.9118 -0.0385 0.4081 -0.0238>}
 thigh_l {parent=pelvis, loc=<-1.0855 -0.2109 -8.7053>, rot=<0.0009 0.0355 0.0155 -0.9993>}
 calf_l {parent=thigh_l, loc=<-42.5284 -0.0019 -0.7297>, rot=<0.0006 -0.0161 -0.0391 -0.9991>}
 foot_l {parent=calf_l, loc=<-46.7401 -2.5014 -0.592>, rot=<0.0427 -0.0238 0.0353 -0.9982>}
 ball_l {parent=foot_l, loc=<-3.5658 10.8066 -0.1544>, rot=<0.0001 -0.0001 -0.7063 -0.7079>}
 dyn_L_flipflop {parent=foot_l, loc=<-8.0004 10.8169 -0.1544>, rot=<0.0001 -0.0001 -0.7063 -0.7079>}
 calf_twist_02_l {parent=calf_l, loc=<-19.1546 -1.7943 -0.2638>, rot=<0.0237 0.0095 -0.0035 -0.9997>}
 calf_twist_01_l {parent=calf_l, loc=<-35.8483 -1.9197 -0.5794>, rot=<0.0237 0.0095 -0.0035 -0.9997>}
 dyn_L_top_ribbon_base {parent=calf_l, loc=<-24.488 -8.6564 -0.7134>, rot=<-0.0309 -0.7201 0.0276 -0.6927>}
 dyn_L_top_ribbon_A {parent=dyn_L_top_ribbon_base, loc=<-0.3316 -0.1604 -0.016>, rot=<0.223 -0.9745 0.0054 -0.0235>}
 dyn_L_top_ribbon_B {parent=dyn_L_top_ribbon_base, loc=<0.3233 -0.1771 0.0057>, rot=<-0.0022 0.0086 0.2479 -0.9688>}
 dyn_L_mid_ribbon_base {parent=calf_l, loc=<-34.4616 -7.2116 -0.6814>, rot=<-0.0309 -0.7201 0.0276 -0.6927>}
 dyn_L_mid_ribbon_A {parent=dyn_L_mid_ribbon_base, loc=<-0.3186 -0.1795 -0.0463>, rot=<0.2509 -0.9653 0.0181 -0.0697>}
 dyn_L_mid_ribbon_B {parent=dyn_L_mid_ribbon_base, loc=<0.306 -0.2041 -0.0249>, rot=<0.0118 -0.0389 0.289 -0.9565>}
 dyn_L_bottom_ribbon_base {parent=calf_l, loc=<-43.4199 -6.3504 -0.3262>, rot=<-0.0309 -0.7201 0.0276 -0.6927>}
 dyn_L_bottom_ribbon_A {parent=dyn_L_bottom_ribbon_base, loc=<-0.3403 -0.1414 -0.0107>, rot=<0.1955 -0.9806 0.0031 -0.0153>}
 dyn_L_bottom_ribbon_B {parent=dyn_L_bottom_ribbon_base, loc=<0.3393 -0.1435 -0.013>, rot=<0.0038 -0.0188 0.1985 -0.9799>}
 dyn_L_yukata_back_mstr {parent=calf_l, loc=<-0.7423 -7.2571 2.1425>, rot=<0.1101 -0.9916 -0.0606 -0.0304>}
 dyn_L_yukata_back_1 {parent=dyn_L_yukata_back_mstr, loc=<0 0 0>, rot=<0 0 0 1>}
 deform_knee_in_base_l {parent=calf_l, loc=<-1.5335 -5.0733 0.3766>, rot=<0.0017 0.0216 0.0001 -0.9998>}
 deform_knee_in_l {parent=deform_knee_in_base_l, loc=<0 0 0>, rot=<0 0 0 1>}
 foot_l_null {parent=calf_l, type=Null, loc=<-46.7401 -2.5014 -0.592>, rot=<0.0427 -0.0238 0.0353 -0.9982>}
 thigh_twist_01_l {parent=thigh_l, loc=<-27.952 0 0>, rot=<0.0013 0 0 -1>}
 calf_l_null {parent=thigh_l, type=Null, loc=<-42.5284 -0.0019 -0.7297>, rot=<0.0006 -0.0161 -0.0391 -0.9991>}
 thigh_r {parent=pelvis, loc=<-1.0855 -0.2109 8.7053>, rot=<0.0355 -0.0009 0.9993 0.0155>}
 calf_r {parent=thigh_r, loc=<42.5284 0.0019 0.7297>, rot=<0.0006 -0.0161 -0.0391 -0.9991>}
 foot_r {parent=calf_r, loc=<46.7401 2.5014 0.592>, rot=<0.0427 -0.0238 0.0353 -0.9982>}
 ball_r {parent=foot_r, loc=<3.5658 -10.8066 0.1544>, rot=<0.0001 -0.0001 -0.7063 -0.7079>}
 dyn_R_flipflop {parent=foot_r, loc=<8.0004 -10.8169 0.1545>, rot=<0.0001 -0.0001 -0.7063 -0.7079>}
 calf_twist_02_r {parent=calf_r, loc=<19.1546 1.7943 0.2638>, rot=<0.0237 0.0095 -0.0035 -0.9997>}
 calf_twist_01_r {parent=calf_r, loc=<35.8483 1.9197 0.5794>, rot=<0.0237 0.0095 -0.0035 -0.9997>}
 dyn_R_top_ribbon_base {parent=calf_r, loc=<24.4879 8.6563 0.7134>, rot=<-0.0309 -0.7201 0.0276 -0.6927>}
 dyn_R_top_ribbon_A {parent=dyn_R_top_ribbon_base, loc=<0.3315 0.1604 0.016>, rot=<0.223 -0.9745 0.0054 -0.0235>}
 dyn_R_top_ribbon_B {parent=dyn_R_top_ribbon_base, loc=<-0.3233 0.1771 -0.0057>, rot=<-0.0022 0.0086 0.2479 -0.9688>}
 dyn_R_mid_ribbon_base {parent=calf_r, loc=<34.4616 7.2117 0.6814>, rot=<-0.0309 -0.7201 0.0276 -0.6927>}
 dyn_R_mid_ribbon_A {parent=dyn_R_mid_ribbon_base, loc=<0.3186 0.1795 0.0463>, rot=<0.2509 -0.9653 0.0181 -0.0697>}
 dyn_R_mid_ribbon_B {parent=dyn_R_mid_ribbon_base, loc=<-0.306 0.2041 0.0249>, rot=<0.0118 -0.0389 0.289 -0.9565>}
 dyn_R_bottom_ribbon_base {parent=calf_r, loc=<43.4199 6.3504 0.3261>, rot=<-0.0309 -0.7201 0.0276 -0.6927>}
 dyn_R_bottom_ribbon_A {parent=dyn_R_bottom_ribbon_base, loc=<0.3403 0.1414 0.0107>, rot=<0.1955 -0.9806 0.0031 -0.0153>}
 dyn_R_bottom_ribbon_B {parent=dyn_R_bottom_ribbon_base, loc=<-0.3394 0.1434 0.0131>, rot=<0.0038 -0.0188 0.1985 -0.9799>}
 dyn_R_yukata_back_mstr {parent=calf_r, loc=<0.7423 7.2571 -2.1425>, rot=<0.1101 -0.9916 -0.0606 -0.0304>}
 dyn_R_yukata_back_1 {parent=dyn_R_yukata_back_mstr, loc=<0 0 0>, rot=<0 0 0 1>}
 deform_knee_in_base_r {parent=calf_r, loc=<1.5335 5.0733 -0.3766>, rot=<0.0017 0.0216 0.0001 -0.9998>}
 deform_knee_in_r {parent=deform_knee_in_base_r, loc=<0 0 0>, rot=<0 0 0 1>}
 foot_r_null {parent=calf_r, type=Null, loc=<46.7401 2.5014 0.592>, rot=<0.0427 -0.0238 0.0353 -0.9982>}
 thigh_twist_01_r {parent=thigh_r, loc=<27.952 0 0>, rot=<0.0013 0 0 -1>}
 calf_r_null {parent=thigh_r, type=Null, loc=<42.5284 0.0019 0.7297>, rot=<0.0006 -0.0161 -0.0391 -0.9991>}
 thigh_l_null {parent=pelvis, type=Null, loc=<-1.0855 -0.2109 -8.7053>, rot=<0.0009 0.0355 0.0155 -0.9993>}
 thigh_r_null {parent=pelvis, type=Null, loc=<-1.0855 -0.2109 8.7053>, rot=<0.0355 -0.0009 0.9993 0.0155>}
 spine_05_null {parent=pelvis, type=Null, loc=<32.9717 -4.2232 0>, rot=<0 0 0.058 -0.9983>}
 pelvis_skirt_offset {parent=pelvis, loc=<0 0 0>, rot=<0 0 0 1>}
 base_C_skirt_a_1 {parent=pelvis_skirt_offset, loc=<-0.8659 8.3629 0.0001>, rot=<-0.0127 -0.707 -0.707 -0.0127>}
 dyn_C_skirt_a_1 {parent=base_C_skirt_a_1, loc=<16.3581 0 0>, rot=<0 0 0 1>}
 dyn_C_skirt_a_2 {parent=dyn_C_skirt_a_1, loc=<0 0 0>, rot=<0 0 0 1>}
 dyn_C_skirt_a_3 {parent=dyn_C_skirt_a_2, loc=<16.5916 0 0>, rot=<0 0 0 1>}
 base_C_skirt_b_1 {parent=pelvis_skirt_offset, loc=<-1.61 -12.396 0.0001>, rot=<-0.0031 -0.7071 0.7071 0.0031>}
 dyn_C_skirt_b_1 {parent=base_C_skirt_b_1, loc=<16.3581 0 0>, rot=<0 0 0 1>}
 dyn_C_skirt_b_2 {parent=dyn_C_skirt_b_1, loc=<0 0 0>, rot=<0 0 0 1>}
 dyn_C_skirt_b_3 {parent=dyn_C_skirt_b_2, loc=<16.5916 0 0>, rot=<0 0 0 1>}
 dyn_C_skirt_b_4 {parent=dyn_C_skirt_b_3, loc=<12.8164 0 0>, rot=<0 0.1324 0 -0.9912>}
 base_L_skirt_a_1 {parent=pelvis_skirt_offset, loc=<-0.9136 7.0322 -10.7183>, rot=<-0.0224 -0.8188 -0.5732 -0.0213>}
 dyn_L_skirt_a_1 {parent=base_L_skirt_a_1, loc=<16.3581 0 0>, rot=<0 0 0 1>}
 dyn_L_skirt_a_2 {parent=dyn_L_skirt_a_1, loc=<0 0 0>, rot=<0 0 0 1>}
 dyn_L_skirt_a_3 {parent=dyn_L_skirt_a_2, loc=<16.5916 0 0>, rot=<0 0 0 1>}
 base_L_skirt_b_1 {parent=pelvis_skirt_offset, loc=<-1.2434 -2.168 -16.2648>, rot=<-0.0179 -0.9998 0 0>}
 dyn_L_skirt_b_1 {parent=base_L_skirt_b_1, loc=<16.3581 0 0>, rot=<0 0 0 -1>}
 dyn_L_skirt_b_2 {parent=dyn_L_skirt_b_1, loc=<0 0 0>, rot=<0 0 0 1>}
 dyn_L_skirt_b_3 {parent=dyn_L_skirt_b_2, loc=<16.5916 0 0>, rot=<0 0 0 1>}
 dyn_L_skirt_b_4 {parent=dyn_L_skirt_b_3, loc=<12.9541 0 0>, rot=<0 0.1754 0 -0.9845>}
 base_L_skirt_c_1 {parent=pelvis_skirt_offset, loc=<-1.5459 -10.6077 -9.3728>, rot=<0 0 0 1>}
 dyn_L_skirt_c_1 {parent=base_L_skirt_c_1, loc=<-16.3575 0.1449 0>, rot=<-0.0036 -0.8191 0.5736 0.0025>}
 dyn_L_skirt_c_2 {parent=dyn_L_skirt_c_1, loc=<0 0 0>, rot=<0 0 0 1>}
 dyn_L_skirt_c_3 {parent=dyn_L_skirt_c_2, loc=<16.5916 0 0>, rot=<0 0 0 1>}
 dyn_L_skirt_c_4 {parent=dyn_L_skirt_c_3, loc=<12.9541 0 0>, rot=<0 0.157 0 -0.9876>}
 base_R_skirt_a_1 {parent=pelvis_skirt_offset, loc=<-0.9136 7.0322 10.7183>, rot=<0.0103 0.5735 0.819 0.0147>}
 dyn_R_skirt_a_1 {parent=base_R_skirt_a_1, loc=<16.3581 0 0>, rot=<0 0 0 1>}
 dyn_R_skirt_a_2 {parent=dyn_R_skirt_a_1, loc=<0 0 0>, rot=<0 0 0 -1>}
 dyn_R_skirt_a_3 {parent=dyn_R_skirt_a_2, loc=<16.5916 0 0>, rot=<0 0 0 -1>}
 base_R_skirt_b_1 {parent=pelvis_skirt_offset, loc=<-1.2434 -2.168 16.2648>, rot=<0 0 0.9998 0.0179>}
 dyn_R_skirt_b_1 {parent=base_R_skirt_b_1, loc=<16.3581 0 0>, rot=<0 0 0 1>}
 dyn_R_skirt_b_2 {parent=dyn_R_skirt_b_1, loc=<0 0 0>, rot=<0 0 0 1>}
 dyn_R_skirt_b_3 {parent=dyn_R_skirt_b_2, loc=<16.5916 0 0>, rot=<0 0 0 1>}
 dyn_R_skirt_b_4 {parent=dyn_R_skirt_b_3, loc=<12.954 0 0>, rot=<0 0.1798 0 -0.9837>}
 base_R_skirt_c_1 {parent=pelvis_skirt_offset, loc=<-1.5459 -10.6077 9.3728>, rot=<-0.0025 -0.5736 0.8191 0.0036>}
 dyn_R_skirt_c_1 {parent=base_R_skirt_c_1, loc=<16.3581 0 0>, rot=<0 0 0 1>}
 dyn_R_skirt_c_2 {parent=dyn_R_skirt_c_1, loc=<0 0 0>, rot=<0 0 0 1>}
 dyn_R_skirt_c_3 {parent=dyn_R_skirt_c_2, loc=<16.5916 0 0>, rot=<0 0 0 1>}
 dyn_R_skirt_c_4 {parent=dyn_R_skirt_c_3, loc=<12.9541 0 0>, rot=<0 0.1595 0 -0.9872>}
 pelvis_skirt_haori {parent=pelvis, loc=<0 0 0>, rot=<0 0 0 1>}
 dyn_L_haori_front_mstr {parent=pelvis_skirt_haori, loc=<5.5379 3.4724 -17.2862>, rot=<0.0526 -0.9847 -0.1361 -0.0955>}
 dyn_L_haori_front_1 {parent=dyn_L_haori_front_mstr, loc=<0 0 0>, rot=<0 0 0 1>}
 dyn_L_haori_front_2 {parent=dyn_L_haori_front_1, loc=<17 0 0>, rot=<0 0 0 1>}
 dyn_L_haori_front_3 {parent=dyn_L_haori_front_2, loc=<17 0 0>, rot=<0 0 0 1>}
 dyn_L_haori_front_4 {parent=dyn_L_haori_front_3, loc=<17 0 0>, rot=<0 0 0 1>}
 dyn_L_haori_side_mstr {parent=pelvis_skirt_haori, loc=<5.6551 -9.8488 -13.3436>, rot=<0.039 -0.8852 0.4523 -0.1018>}
 dyn_L_haori_side_1 {parent=dyn_L_haori_side_mstr, loc=<0 0 0>, rot=<0 0 0 1>}
 dyn_L_haori_side_2 {parent=dyn_L_haori_side_1, loc=<17 0 0>, rot=<0 0 0 1>}
 dyn_L_haori_side_3 {parent=dyn_L_haori_side_2, loc=<17 0 0>, rot=<0 0 0 1>}
 dyn_L_haori_side_4 {parent=dyn_L_haori_side_3, loc=<17 0 0>, rot=<0 0 0 1>}
 dyn_C_haori_back_mstr {parent=pelvis_skirt_haori, loc=<4.7337 -14.0213 0>, rot=<0.0674 -0.7039 0.7039 -0.0674>}
 dyn_C_haori_back_1 {parent=dyn_C_haori_back_mstr, loc=<0 0 0>, rot=<0 0 0 1>}
 dyn_C_haori_back_2 {parent=dyn_C_haori_back_1, loc=<17 0 0>, rot=<0 0 0 1>}
 dyn_C_haori_back_3 {parent=dyn_C_haori_back_2, loc=<17 0 0>, rot=<0 0 0 1>}
 dyn_C_haori_back_4 {parent=dyn_C_haori_back_3, loc=<17 0 0>, rot=<0 0 0 1>}
 dyn_R_haori_front_mstr {parent=pelvis_skirt_haori, loc=<5.5378 3.4724 17.2862>, rot=<0.0955 0.1361 0.9847 -0.0526>}
 dyn_R_haori_front_1 {parent=dyn_R_haori_front_mstr, loc=<0 0 0>, rot=<0 0 0 1>}
 dyn_R_haori_front_2 {parent=dyn_R_haori_front_1, loc=<17 0 0>, rot=<0 0 0 1>}
 dyn_R_haori_front_3 {parent=dyn_R_haori_front_2, loc=<17 0 -0.0001>, rot=<0 0 0 1>}
 dyn_R_haori_front_4 {parent=dyn_R_haori_front_3, loc=<17 0 0>, rot=<0 0 0 1>}
 dyn_R_haori_side_mstr {parent=pelvis_skirt_haori, loc=<5.6547 -9.8488 13.3436>, rot=<0.1018 -0.4523 0.8852 -0.039>}
 dyn_R_haori_side_1 {parent=dyn_R_haori_side_mstr, loc=<0 0 0>, rot=<0 0 0 1>}
 dyn_R_haori_side_2 {parent=dyn_R_haori_side_1, loc=<17 0 0.0001>, rot=<0 0 0 1>}
 dyn_R_haori_side_3 {parent=dyn_R_haori_side_2, loc=<17 0 -0.0001>, rot=<0 0 0 1>}
 dyn_R_haori_side_4 {parent=dyn_R_haori_side_3, loc=<17 0 0>, rot=<0 0 0 1>}
 pelvis_null {parent=root, type=Null, loc=<0 -0.5301 98.0571>, rot=<0.0127 0.707 -0.0127 -0.707>}
]
variables [
 PelvisProx {type=double, default=0}
 PelvisProx_Easterbloc {type=double, default=0}
 SPR L Leg Bkd90 {type=double, default=0}
 SPR L Leg Fwd120 {type=double, default=0}
 SPR L Leg Fwd90 {type=double, default=0}
 SPR L Leg Itw {type=double, default=0}
 SPR L Leg Otw {type=double, default=0}
 SPR L Leg Otw_Haori {type=double, default=0}
 SPR L calf Bkd100 {type=double, default=0}
 SPR L calf Bkd140 {type=double, default=0}
 SPR L calf BkdTwist {type=double, default=0}
 SPR R Leg Bkd90 {type=double, default=0}
 SPR R Leg Fwd120 {type=double, default=0}
 SPR R Leg Fwd90 {type=double, default=0}
 SPR R Leg Itw {type=double, default=0}
 SPR R Leg Otw {type=double, default=0}
 SPR R Leg Otw_Haori {type=double, default=0}
 SPR R calf Bkd100 {type=double, default=0}
 SPR R calf Bkd140 {type=double, default=0}
 SPR R calf BkdTwist {type=double, default=0}
 pub isBaseballSliding {type=bool, default=False}
 pub isCrouchStrafeLeft {type=double, default=0}
 pub isCrouchStrafeRight {type=double, default=0}
 pub isCrouching {type=bool, default=False}
 "pub isDBNO " {type=bool, default=False}
 pub isFrontEnd {type=bool, default=False}
 pub isJumping {type=bool, default=False}
 pub isKneeSliding {type=bool, default=False}
 pub isMoveBackward {type=double, default=0}
 pub isMoveForward {type=double, default=0}
 pub isMoveLeft {type=double, default=0}
 pub isMoveRight {type=double, default=0}
 pub isPickaxeSwing {type=bool, default=False}
 pub isPlayingEasternBloc {type=bool, default=False}
 pub isPlayingEmote {type=bool, default=False}
 pub isSkydive {type=bool, default=False}
 pub isSprinting {type=bool, default=False}
 pub isTacticalSprinting {type=bool, default=False}
 pub isparachuteOpened {type=bool, default=False}
 pub_IsFrontEnd {type=bool, default=False}
 pub_IsGrindRail {type=bool, default=False}
 pub_IsTacticalJump {type=bool, default=False}
 pub_IsTargeting {type=bool, default=False}
 pub_YouBoreme {type=bool, default=False}
]
