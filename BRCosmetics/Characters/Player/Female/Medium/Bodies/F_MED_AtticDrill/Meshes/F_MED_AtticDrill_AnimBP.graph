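# Post-process style cosmetic AnimBP for the F_MED_AtticDrill player body, adding secondary motion to its haori, sleeves and ribbons.
# It copies the attached parent's pose into a Control Rig driven by movement and action states, offsets the backpack socket, then caches the result.
# Chained root-relative, spring-based AnimDynamics nodes simulate cloth bones from cached poses.
# Additive bone transforms finally adjust knees, sleeves, skirt and haori; several first-segment haori dynamics nodes are unconnected.
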AnimBlueprint /BRCosmetics/Characters/Player/Female/Medium/Bodies/F_MED_AtticDrill/Meshes/F_MED_AtticDrill_AnimBP
target_skeleton=&/F_MED_AtticDrill_Skeleton
graphs [
 AnimGraph {nodes [
   Anim Dynamics - Bone: dyn_L_haori_front_1!(
    ComponentPose
    Alpha=1
    GravityOverride:Vector="0,0,-980"
    SimSpaceSettings:Anim Phys Sim Space Settings Structure=(SimSpaceAngularAlpha=0,MaxAngularVelocity=10000,MaxAngularAcceleration=10000,ExternalAngularVelocity=(X=0,Y=0,Z=0))
    ExternalForce="0,0,0"
   ) -> (Pose) {
    angularDampingOverride=1
    angularSpringConstant=12000
    bAngularSpring=true
    bOverrideAngularDamping=true
    bUseGravityOverride=true
    gravityScale=0.35
    simulationSpace=RootRelative
    boundBone {boneName=dyn_L_haori_front_1}
    componentAppliedLinearAccClamp=<10000 10000 10000>
    componentLinearAccScale=<1 1 1>
    componentLinearVelScale=<1 1 1>
    gravityOverride {z=-980}
    physicsBodyDefinitions [{
      boundBone {boneName=dyn_L_haori_front_1}
      boxExtents=<30 50 50>
      constraintSetup {angularLimitsMax=<8 0 10>, angularLimitsMin=<-8 -20 -10>, angularTarget=<1 0 0>}
      localJointOffset=<8 0 0>
     }]
   }
   Anim Dynamics - Bone: dyn_C_haori_back_1!(
    ComponentPose
    Alpha=1
    GravityOverride:Vector="0,0,-980"
    SimSpaceSettings:Anim Phys Sim Space Settings Structure=(SimSpaceAngularAlpha=0,MaxAngularVelocity=10000,MaxAngularAcceleration=10000,ExternalAngularVelocity=(X=0,Y=0,Z=0))
    ExternalForce="0,0,0"
   ) -> (Pose) {
    angularDampingOverride=1
    angularSpringConstant=12000
    bAngularSpring=true
    bOverrideAngularDamping=true
    bUseGravityOverride=true
    gravityScale=0.35
    simulationSpace=RootRelative
    boundBone {boneName=dyn_C_haori_back_1}
    componentAppliedLinearAccClamp=<10000 10000 10000>
    componentLinearAccScale=<1 1 1>
    componentLinearVelScale=<1 1 1>
    gravityOverride {z=-980}
    physicsBodyDefinitions [{
      boundBone {boneName=dyn_C_haori_back_1}
      boxExtents=<30 50 50>
      constraintSetup {angularLimitsMax=<8 0 10>, angularLimitsMin=<-8 -20 -10>, angularTarget=<1 0 0>}
      localJointOffset=<8 0 0>
     }]
   }
   Anim Dynamics - Bone: dyn_R_haori_front_1!(
    ComponentPose
    Alpha=1
    GravityOverride:Vector="0,0,-980"
    SimSpaceSettings:Anim Phys Sim Space Settings Structure=(SimSpaceAngularAlpha=0,MaxAngularVelocity=10000,MaxAngularAcceleration=10000,ExternalAngularVelocity=(X=0,Y=0,Z=0))
    ExternalForce="0,0,0"
   ) -> (Pose) {
    angularDampingOverride=1
    angularSpringConstant=12000
    bAngularSpring=true
    bOverrideAngularDamping=true
    bUseGravityOverride=true
    gravityScale=0.35
    simulationSpace=RootRelative
    boundBone {boneName=dyn_R_haori_front_1}
    componentAppliedLinearAccClamp=<10000 10000 10000>
    componentLinearAccScale=<1 1 1>
    componentLinearVelScale=<1 1 1>
    gravityOverride {z=-980}
    physicsBodyDefinitions [{
      boundBone {boneName=dyn_R_haori_front_1}
      boxExtents=<30 50 50>
      constraintSetup {angularLimitsMax=<8 0 10>, angularLimitsMin=<-8 -20 -10>, angularTarget=<1 0 0>}
      localJointOffset=<8 0 0>
     }]
   }
   Anim Dynamics - Bone: dyn_R_haori_side_1!(
    ComponentPose
    Alpha=1
    GravityOverride:Vector="0,0,-980"
    SimSpaceSettings:Anim Phys Sim Space Settings Structure=(SimSpaceAngularAlpha=0,MaxAngularVelocity=10000,MaxAngularAcceleration=10000,ExternalAngularVelocity=(X=0,Y=0,Z=0))
    ExternalForce="0,0,0"
   ) -> (Pose) {
    angularDampingOverride=1
    angularSpringConstant=12000
    bAngularSpring=true
    bOverrideAngularDamping=true
    bUseGravityOverride=true
    gravityScale=0.35
    simulationSpace=RootRelative
    boundBone {boneName=dyn_R_haori_side_1}
    componentAppliedLinearAccClamp=<10000 10000 10000>
    componentLinearAccScale=<1 1 1>
    componentLinearVelScale=<1 1 1>
    gravityOverride {z=-980}
    physicsBodyDefinitions [{
      boundBone {boneName=dyn_R_haori_side_1}
      boxExtents=<30 50 50>
      constraintSetup {angularLimitsMax=<8 0 10>, angularLimitsMin=<-8 -20 -10>, angularTarget=<1 0 0>}
      localJointOffset=<8 0 0>
     }]
   }
   Anim Dynamics - Bone: dyn_L_haori_side_1!(
    ComponentPose
    Alpha=1
    GravityOverride:Vector="0,0,-980"
    SimSpaceSettings:Anim Phys Sim Space Settings Structure=(SimSpaceAngularAlpha=0,MaxAngularVelocity=10000,MaxAngularAcceleration=10000,ExternalAngularVelocity=(X=0,Y=0,Z=0))
    ExternalForce="0,0,0"
   ) -> (Pose) {
    angularDampingOverride=1
    angularSpringConstant=12000
    bAngularSpring=true
    bOverrideAngularDamping=true
    bUseGravityOverride=true
    gravityScale=0.35
    simulationSpace=RootRelative
    boundBone {boneName=dyn_L_haori_side_1}
    componentAppliedLinearAccClamp=<10000 10000 10000>
    componentLinearAccScale=<1 1 1>
    componentLinearVelScale=<1 1 1>
    gravityOverride {z=-980}
    physicsBodyDefinitions [{
      boundBone {boneName=dyn_L_haori_side_1}
      boxExtents=<30 50 50>
      constraintSetup {angularLimitsMax=<7 0 10>, angularLimitsMin=<-7 -20 -10>, angularTarget=<1 0 0>}
      localJointOffset=<8 0 0>
     }]
   }
   n1 = Anim Dynamics - Bone: dyn_L_sleeveHole!(
    ComponentPose=Local To Component(LocalPose=Use cached pose 'CR_Cache_Haori'())
    Alpha=1
    GravityOverride:Vector="0,0,-980"
    SimSpaceSettings:Anim Phys Sim Space Settings Structure=(SimSpaceAngularAlpha=0,MaxAngularVelocity=10000,MaxAngularAcceleration=10000,ExternalAngularVelocity=(X=0,Y=0,Z=0))
    ExternalForce="0,0,0"
   ) -> (Pose) {
    angularDampingOverride=1
    angularSpringConstant=7000
    bAngularSpring=true
    bOverrideAngularDamping=true
    bUseGravityOverride=true
    gravityScale=0.35
    simulationSpace=RootRelative
    boundBone {boneName=dyn_L_sleeveHole}
    componentAppliedLinearAccClamp=<10000 10000 10000>
    componentLinearAccScale=<1 1 1>
    componentLinearVelScale=<1 1 1>
    gravityOverride {z=-980}
    physicsBodyDefinitions [{
      boundBone {boneName=dyn_L_sleeveHole}
      boxExtents=<30 40 40>
      constraintSetup {angularLimitsMax=<10 0 10>, angularLimitsMin=<-10 -30 -10>, angularTarget=<1 0 0>}
      localJointOffset=<3 0 0>
     }]
   }
   n2 = Anim Dynamics - Bone: dyn_L_sleeveBig_1!(
    ComponentPose=@n1
    Alpha=1
    GravityOverride:Vector="0,0,-980"
    SimSpaceSettings:Anim Phys Sim Space Settings Structure=(SimSpaceAngularAlpha=0,MaxAngularVelocity=10000,MaxAngularAcceleration=10000,ExternalAngularVelocity=(X=0,Y=0,Z=0))
    ExternalForce="0,0,0"
   ) -> (Pose) {
    angularDampingOverride=1
    angularSpringConstant=1000
    bAngularSpring=true
    bOverrideAngularDamping=true
    bUseGravityOverride=true
    gravityScale=0.35
    simulationSpace=RootRelative
    boundBone {boneName=dyn_L_sleeveBig_1}
    componentAppliedLinearAccClamp=<10000 10000 10000>
    componentLinearAccScale=<1 1 1>
    componentLinearVelScale=<1 1 1>
    gravityOverride {z=-980}
    physicsBodyDefinitions [{
      boundBone {boneName=dyn_L_sleeveBig_1}
      boxExtents=<30 40 40>
      constraintSetup {angularLimitsMax=<15 80 40>, angularLimitsMin=<-15 -80 -40>, angularTarget=<1 0 0>}
      localJointOffset=<4 0 0>
     }]
   }
   n3 = Anim Dynamics - Bone: dyn_L_sleeveBig_2!(
    ComponentPose=@n2
    Alpha=1
    GravityOverride:Vector="0,0,-980"
    SimSpaceSettings:Anim Phys Sim Space Settings Structure=(SimSpaceAngularAlpha=0,MaxAngularVelocity=10000,MaxAngularAcceleration=10000,ExternalAngularVelocity=(X=0,Y=0,Z=0))
    ExternalForce="0,0,0"
   ) -> (Pose) {
    angularDampingOverride=1
    angularSpringConstant=1000
    bAngularSpring=true
    bOverrideAngularDamping=true
    bUseGravityOverride=true
    gravityScale=0.35
    simulationSpace=RootRelative
    boundBone {boneName=dyn_L_sleeveBig_2}
    componentAppliedLinearAccClamp=<10000 10000 10000>
    componentLinearAccScale=<1 1 1>
    componentLinearVelScale=<1 1 1>
    gravityOverride {z=-980}
    physicsBodyDefinitions [{
      boundBone {boneName=dyn_L_sleeveBig_2}
      boxExtents=<30 40 40>
      constraintSetup {angularLimitsMax=<15 80 40>, angularLimitsMin=<-15 -80 -40>, angularTarget=<1 0 0>}
      localJointOffset=<4 0 0>
     }]
   }
   n4 = Anim Dynamics - Bone: dyn_L_sleeveSmall_1!(
    ComponentPose=@n3
    Alpha=1
    GravityOverride:Vector="0,0,-980"
    SimSpaceSettings:Anim Phys Sim Space Settings Structure=(SimSpaceAngularAlpha=0,MaxAngularVelocity=10000,MaxAngularAcceleration=10000,ExternalAngularVelocity=(X=0,Y=0,Z=0))
    ExternalForce="0,0,0"
   ) -> (Pose) {
    angularDampingOverride=1
    angularSpringConstant=1000
    bAngularSpring=true
    bOverrideAngularDamping=true
    bUseGravityOverride=true
    gravityScale=0.35
    simulationSpace=RootRelative
    boundBone {boneName=dyn_L_sleeveSmall_1}
    componentAppliedLinearAccClamp=<10000 10000 10000>
    componentLinearAccScale=<1 1 1>
    componentLinearVelScale=<1 1 1>
    gravityOverride {z=-980}
    physicsBodyDefinitions [{
      boundBone {boneName=dyn_L_sleeveSmall_1}
      boxExtents=<30 40 40>
      constraintSetup {angularLimitsMax=<15 80 40>, angularLimitsMin=<-15 -80 -40>, angularTarget=<1 0 0>}
      localJointOffset=<4 0 0>
     }]
   }
   n5 = Anim Dynamics - Bone: dyn_L_sleeveSmall_2!(
    ComponentPose=@n4
    Alpha=1
    GravityOverride:Vector="0,0,-980"
    SimSpaceSettings:Anim Phys Sim Space Settings Structure=(SimSpaceAngularAlpha=0,MaxAngularVelocity=10000,MaxAngularAcceleration=10000,ExternalAngularVelocity=(X=0,Y=0,Z=0))
    ExternalForce="0,0,0"
   ) -> (Pose) {
    angularDampingOverride=1
    angularSpringConstant=1000
    bAngularSpring=true
    bOverrideAngularDamping=true
    bUseGravityOverride=true
    gravityScale=0.35
    simulationSpace=RootRelative
    boundBone {boneName=dyn_L_sleeveSmall_2}
    componentAppliedLinearAccClamp=<10000 10000 10000>
    componentLinearAccScale=<1 1 1>
    componentLinearVelScale=<1 1 1>
    gravityOverride {z=-980}
    physicsBodyDefinitions [{
      boundBone {boneName=dyn_L_sleeveSmall_2}
      boxExtents=<30 40 40>
      constraintSetup {angularLimitsMax=<15 80 40>, angularLimitsMin=<-15 -80 -40>, angularTarget=<1 0 0>}
      localJointOffset=<4 0 0>
     }]
   }
   n6 = Anim Dynamics - Bone: dyn_L_sleeveSmallest_1!(
    ComponentPose=@n5
    Alpha=1
    GravityOverride:Vector="0,0,-980"
    SimSpaceSettings:Anim Phys Sim Space Settings Structure=(SimSpaceAngularAlpha=0,MaxAngularVelocity=10000,MaxAngularAcceleration=10000,ExternalAngularVelocity=(X=0,Y=0,Z=0))
    ExternalForce="0,0,0"
   ) -> (Pose) {
    angularDampingOverride=1
    angularSpringConstant=1000
    bAngularSpring=true
    bOverrideAngularDamping=true
    bUseGravityOverride=true
    gravityScale=0.35
    simulationSpace=RootRelative
    boundBone {boneName=dyn_L_sleeveSmallest_1}
    componentAppliedLinearAccClamp=<10000 10000 10000>
    componentLinearAccScale=<1 1 1>
    componentLinearVelScale=<1 1 1>
    gravityOverride {z=-980}
    physicsBodyDefinitions [{
      boundBone {boneName=dyn_L_sleeveSmallest_1}
      boxExtents=<30 40 40>
      constraintSetup {angularLimitsMax=<15 80 40>, angularLimitsMin=<-15 -80 -40>, angularTarget=<1 0 0>}
      localJointOffset=<4 0 0>
     }]
   }
   n7 = Anim Dynamics - Bone: dyn_R_sleeveHole!(
    ComponentPose=@n6
    Alpha=1
    GravityOverride:Vector="0,0,-980"
    SimSpaceSettings:Anim Phys Sim Space Settings Structure=(SimSpaceAngularAlpha=0,MaxAngularVelocity=10000,MaxAngularAcceleration=10000,ExternalAngularVelocity=(X=0,Y=0,Z=0))
    ExternalForce="0,0,0"
   ) -> (Pose) {
    angularDampingOverride=1
    angularSpringConstant=7000
    bAngularSpring=true
    bOverrideAngularDamping=true
    bUseGravityOverride=true
    gravityScale=0.35
    simulationSpace=RootRelative
    boundBone {boneName=dyn_R_sleeveHole}
    componentAppliedLinearAccClamp=<10000 10000 10000>
    componentLinearAccScale=<1 1 1>
    componentLinearVelScale=<1 1 1>
    gravityOverride {z=-980}
    physicsBodyDefinitions [{
      boundBone {boneName=dyn_R_sleeveHole}
      boxExtents=<30 40 40>
      constraintSetup {angularLimitsMax=<10 0 10>, angularLimitsMin=<-10 -30 -10>, angularTarget=<1 0 0>}
      localJointOffset=<-3 0 0>
     }]
   }
   n8 = Anim Dynamics - Bone: dyn_R_sleeveBig_1!(
    ComponentPose=@n7
    Alpha=1
    GravityOverride:Vector="0,0,-980"
    SimSpaceSettings:Anim Phys Sim Space Settings Structure=(SimSpaceAngularAlpha=0,MaxAngularVelocity=10000,MaxAngularAcceleration=10000,ExternalAngularVelocity=(X=0,Y=0,Z=0))
    ExternalForce="0,0,0"
   ) -> (Pose) {
    angularDampingOverride=1
    angularSpringConstant=1000
    bAngularSpring=true
    bOverrideAngularDamping=true
    bUseGravityOverride=true
    gravityScale=0.35
    simulationSpace=RootRelative
    boundBone {boneName=dyn_R_sleeveBig_1}
    componentAppliedLinearAccClamp=<10000 10000 10000>
    componentLinearAccScale=<1 1 1>
    componentLinearVelScale=<1 1 1>
    gravityOverride {z=-980}
    physicsBodyDefinitions [{
      boundBone {boneName=dyn_R_sleeveBig_1}
      boxExtents=<30 40 40>
      constraintSetup {angularLimitsMax=<15 80 40>, angularLimitsMin=<-15 -80 -40>, angularTarget=<1 0 0>}
      localJointOffset=<-4 0 0>
     }]
   }
   n9 = Anim Dynamics - Bone: dyn_R_sleeveBig_2!(
    ComponentPose=@n8
    Alpha=1
    GravityOverride:Vector="0,0,-980"
    SimSpaceSettings:Anim Phys Sim Space Settings Structure=(SimSpaceAngularAlpha=0,MaxAngularVelocity=10000,MaxAngularAcceleration=10000,ExternalAngularVelocity=(X=0,Y=0,Z=0))
    ExternalForce="0,0,0"
   ) -> (Pose) {
    angularDampingOverride=1
    angularSpringConstant=1000
    bAngularSpring=true
    bOverrideAngularDamping=true
    bUseGravityOverride=true
    gravityScale=0.35
    simulationSpace=RootRelative
    boundBone {boneName=dyn_R_sleeveBig_2}
    componentAppliedLinearAccClamp=<10000 10000 10000>
    componentLinearAccScale=<1 1 1>
    componentLinearVelScale=<1 1 1>
    gravityOverride {z=-980}
    physicsBodyDefinitions [{
      boundBone {boneName=dyn_R_sleeveBig_2}
      boxExtents=<30 40 40>
      constraintSetup {angularLimitsMax=<15 80 40>, angularLimitsMin=<-15 -80 -40>, angularTarget=<1 0 0>}
      localJointOffset=<-4 0 0>
     }]
   }
   n10 = Anim Dynamics - Bone: dyn_R_sleeveSmall_1!(
    ComponentPose=@n9
    Alpha=1
    GravityOverride:Vector="0,0,-980"
    SimSpaceSettings:Anim Phys Sim Space Settings Structure=(SimSpaceAngularAlpha=0,MaxAngularVelocity=10000,MaxAngularAcceleration=10000,ExternalAngularVelocity=(X=0,Y=0,Z=0))
    ExternalForce="0,0,0"
   ) -> (Pose) {
    angularDampingOverride=1
    angularSpringConstant=1000
    bAngularSpring=true
    bOverrideAngularDamping=true
    bUseGravityOverride=true
    gravityScale=0.35
    simulationSpace=RootRelative
    boundBone {boneName=dyn_R_sleeveSmall_1}
    componentAppliedLinearAccClamp=<10000 10000 10000>
    componentLinearAccScale=<1 1 1>
    componentLinearVelScale=<1 1 1>
    gravityOverride {z=-980}
    physicsBodyDefinitions [{
      boundBone {boneName=dyn_R_sleeveSmall_1}
      boxExtents=<30 40 40>
      constraintSetup {angularLimitsMax=<15 80 40>, angularLimitsMin=<-15 -80 -40>, angularTarget=<1 0 0>}
      localJointOffset=<-4 0 0>
     }]
   }
   n11 = Anim Dynamics - Bone: dyn_R_sleeveSmall_2!(
    ComponentPose=@n10
    Alpha=1
    GravityOverride:Vector="0,0,-980"
    SimSpaceSettings:Anim Phys Sim Space Settings Structure=(SimSpaceAngularAlpha=0,MaxAngularVelocity=10000,MaxAngularAcceleration=10000,ExternalAngularVelocity=(X=0,Y=0,Z=0))
    ExternalForce="0,0,0"
   ) -> (Pose) {
    angularDampingOverride=1
    angularSpringConstant=1000
    bAngularSpring=true
    bOverrideAngularDamping=true
    bUseGravityOverride=true
    gravityScale=0.35
    simulationSpace=RootRelative
    boundBone {boneName=dyn_R_sleeveSmall_2}
    componentAppliedLinearAccClamp=<10000 10000 10000>
    componentLinearAccScale=<1 1 1>
    componentLinearVelScale=<1 1 1>
    gravityOverride {z=-980}
    physicsBodyDefinitions [{
      boundBone {boneName=dyn_R_sleeveSmall_2}
      boxExtents=<30 40 40>
      constraintSetup {angularLimitsMax=<15 80 40>, angularLimitsMin=<-15 -80 -40>, angularTarget=<1 0 0>}
      localJointOffset=<-4 0 0>
     }]
   }
   n12 = Anim Dynamics - Bone: dyn_R_sleeveSmallest_1!(
    ComponentPose=@n11
    Alpha=1
    GravityOverride:Vector="0,0,-980"
    SimSpaceSettings:Anim Phys Sim Space Settings Structure=(SimSpaceAngularAlpha=0,MaxAngularVelocity=10000,MaxAngularAcceleration=10000,ExternalAngularVelocity=(X=0,Y=0,Z=0))
    ExternalForce="0,0,0"
   ) -> (Pose) {
    angularDampingOverride=1
    angularSpringConstant=1000
    bAngularSpring=true
    bOverrideAngularDamping=true
    bUseGravityOverride=true
    gravityScale=0.35
    simulationSpace=RootRelative
    boundBone {boneName=dyn_R_sleeveSmallest_1}
    componentAppliedLinearAccClamp=<10000 10000 10000>
    componentLinearAccScale=<1 1 1>
    componentLinearVelScale=<1 1 1>
    gravityOverride {z=-980}
    physicsBodyDefinitions [{
      boundBone {boneName=dyn_R_sleeveSmallest_1}
      boxExtents=<30 40 40>
      constraintSetup {angularLimitsMax=<15 80 40>, angularLimitsMin=<-15 -80 -40>, angularTarget=<1 0 0>}
      localJointOffset=<-4 0 0>
     }]
   }
   n13 = Anim Dynamics - Bone: dyn_L_top_ribbon_A!(
    ComponentPose=@n12
    Alpha=1
    SimSpaceSettings:Anim Phys Sim Space Settings Structure=(SimSpaceAngularAlpha=0,MaxAngularVelocity=10000,MaxAngularAcceleration=10000,ExternalAngularVelocity=(X=0,Y=0,Z=0))
    ExternalForce="0,0,0"
   ) -> (Pose) {
    angularDampingOverride=1
    angularSpringConstant=1000
    bAngularSpring=true
    bOverrideAngularDamping=true
    bUseGravityOverride=true
    gravityScale=0.35
    simulationSpace=RootRelative
    boundBone {boneName=dyn_L_top_ribbon_A}
    componentAppliedLinearAccClamp=<10000 10000 10000>
    componentLinearAccScale=<1 1 1>
    componentLinearVelScale=<1 1 1>
    gravityOverride {z=-980}
    physicsBodyDefinitions [{
      boundBone {boneName=dyn_L_top_ribbon_A}
      boxExtents=<11 20 20>
      constraintSetup {angularLimitsMax=<0 5 0>, angularLimitsMin=<0 -5 -25>, angularTarget=<1 0 0>}
      localJointOffset=<2 0 0>
     }]
   }
   n14 = Anim Dynamics - Bone: dyn_L_top_ribbon_B!(
    ComponentPose=@n13
    Alpha=1
    SimSpaceSettings:Anim Phys Sim Space Settings Structure=(SimSpaceAngularAlpha=0,MaxAngularVelocity=10000,MaxAngularAcceleration=10000,ExternalAngularVelocity=(X=0,Y=0,Z=0))
    ExternalForce="0,0,0"
   ) -> (Pose) {
    angularDampingOverride=1
    angularSpringConstant=1000
    bAngularSpring=true
    bOverrideAngularDamping=true
    bUseGravityOverride=true
    gravityScale=0.35
    simulationSpace=RootRelative
    boundBone {boneName=dyn_L_top_ribbon_B}
    componentAppliedLinearAccClamp=<10000 10000 10000>
    componentLinearAccScale=<1 1 1>
    componentLinearVelScale=<1 1 1>
    gravityOverride {z=-980}
    physicsBodyDefinitions [{
      boundBone {boneName=dyn_L_top_ribbon_B}
      boxExtents=<11 20 20>
      constraintSetup {angularLimitsMax=<0 5 0>, angularLimitsMin=<0 -5 -25>, angularTarget=<1 0 0>}
      localJointOffset=<2 0 0>
     }]
   }
   n15 = Anim Dynamics - Bone: dyn_L_mid_ribbon_B!(
    ComponentPose=@n14
    Alpha=1
    SimSpaceSettings:Anim Phys Sim Space Settings Structure=(SimSpaceAngularAlpha=0,MaxAngularVelocity=10000,MaxAngularAcceleration=10000,ExternalAngularVelocity=(X=0,Y=0,Z=0))
    ExternalForce="0,0,0"
   ) -> (Pose) {
    angularDampingOverride=1
    angularSpringConstant=1000
    bAngularSpring=true
    bOverrideAngularDamping=true
    bUseGravityOverride=true
    gravityScale=0.35
    simulationSpace=RootRelative
    boundBone {boneName=dyn_L_mid_ribbon_B}
    componentAppliedLinearAccClamp=<10000 10000 10000>
    componentLinearAccScale=<1 1 1>
    componentLinearVelScale=<1 1 1>
    gravityOverride {z=-980}
    physicsBodyDefinitions [{
      boundBone {boneName=dyn_L_mid_ribbon_B}
      boxExtents=<11 20 20>
      constraintSetup {angularLimitsMax=<0 5 0>, angularLimitsMin=<0 -5 -20>, angularTarget=<1 0 0>}
      localJointOffset=<2 0 0>
     }]
   }
   n16 = Anim Dynamics - Bone: dyn_L_mid_ribbon_A!(
    ComponentPose=@n15
    Alpha=1
    SimSpaceSettings:Anim Phys Sim Space Settings Structure=(SimSpaceAngularAlpha=0,MaxAngularVelocity=10000,MaxAngularAcceleration=10000,ExternalAngularVelocity=(X=0,Y=0,Z=0))
    ExternalForce="0,0,0"
   ) -> (Pose) {
    angularDampingOverride=1
    angularSpringConstant=1000
    bAngularSpring=true
    bOverrideAngularDamping=true
    bUseGravityOverride=true
    gravityScale=0.35
    simulationSpace=RootRelative
    boundBone {boneName=dyn_L_mid_ribbon_A}
    componentAppliedLinearAccClamp=<10000 10000 10000>
    componentLinearAccScale=<1 1 1>
    componentLinearVelScale=<1 1 1>
    gravityOverride {z=-980}
    physicsBodyDefinitions [{
      boundBone {boneName=dyn_L_mid_ribbon_A}
      boxExtents=<11 20 20>
      constraintSetup {angularLimitsMax=<0 5 0>, angularLimitsMin=<0 -5 -20>, angularTarget=<1 0 0>}
      localJointOffset=<2 0 0>
     }]
   }
   n17 = Anim Dynamics - Bone: dyn_L_bottom_ribbon_A!(
    ComponentPose=@n16
    Alpha=1
    SimSpaceSettings:Anim Phys Sim Space Settings Structure=(SimSpaceAngularAlpha=0,MaxAngularVelocity=10000,MaxAngularAcceleration=10000,ExternalAngularVelocity=(X=0,Y=0,Z=0))
    ExternalForce="0,0,0"
   ) -> (Pose) {
    angularDampingOverride=1
    angularSpringConstant=1000
    bAngularSpring=true
    bOverrideAngularDamping=true
    bUseGravityOverride=true
    gravityScale=0.35
    simulationSpace=RootRelative
    boundBone {boneName=dyn_L_bottom_ribbon_A}
    componentAppliedLinearAccClamp=<10000 10000 10000>
    componentLinearAccScale=<1 1 1>
    componentLinearVelScale=<1 1 1>
    gravityOverride {z=-980}
    physicsBodyDefinitions [{
      boundBone {boneName=dyn_L_bottom_ribbon_A}
      boxExtents=<11 20 20>
      constraintSetup {angularLimitsMax=<0 5 0>, angularLimitsMin=<0 -5 -30>, angularTarget=<1 0 0>}
      localJointOffset=<2 0 0>
     }]
   }
   n18 = Anim Dynamics - Bone: dyn_L_bottom_ribbon_B!(
    ComponentPose=@n17
    Alpha=1
    SimSpaceSettings:Anim Phys Sim Space Settings Structure=(SimSpaceAngularAlpha=0,MaxAngularVelocity=10000,MaxAngularAcceleration=10000,ExternalAngularVelocity=(X=0,Y=0,Z=0))
    ExternalForce="0,0,0"
   ) -> (Pose) {
    angularDampingOverride=1
    angularSpringConstant=1000
    bAngularSpring=true
    bOverrideAngularDamping=true
    bUseGravityOverride=true
    gravityScale=0.35
    simulationSpace=RootRelative
    boundBone {boneName=dyn_L_bottom_ribbon_B}
    componentAppliedLinearAccClamp=<10000 10000 10000>
    componentLinearAccScale=<1 1 1>
    componentLinearVelScale=<1 1 1>
    gravityOverride {z=-980}
    physicsBodyDefinitions [{
      boundBone {boneName=dyn_L_bottom_ribbon_B}
      boxExtents=<11 20 20>
      constraintSetup {angularLimitsMax=<0 5 0>, angularLimitsMin=<0 -5 -30>, angularTarget=<1 0 0>}
      localJointOffset=<2 0 0>
     }]
   }
   n19 = Anim Dynamics - Bone: dyn_R_top_ribbon_A!(
    ComponentPose=@n18
    Alpha=1
    SimSpaceSettings:Anim Phys Sim Space Settings Structure=(SimSpaceAngularAlpha=0,MaxAngularVelocity=10000,MaxAngularAcceleration=10000,ExternalAngularVelocity=(X=0,Y=0,Z=0))
    ExternalForce="0,0,0"
   ) -> (Pose) {
    angularDampingOverride=1
    angularSpringConstant=1000
    bAngularSpring=true
    bOverrideAngularDamping=true
    bUseGravityOverride=true
    gravityScale=0.35
    simulationSpace=RootRelative
    boundBone {boneName=dyn_R_top_ribbon_A}
    componentAppliedLinearAccClamp=<10000 10000 10000>
    componentLinearAccScale=<1 1 1>
    componentLinearVelScale=<1 1 1>
    gravityOverride {z=-980}
    physicsBodyDefinitions [{
      boundBone {boneName=dyn_R_top_ribbon_A}
      boxExtents=<11 20 20>
      constraintSetup {angularLimitsMax=<0 5 0>, angularLimitsMin=<0 -5 -25>, angularTarget=<1 0 0>}
      localJointOffset=<2 0 0>
     }]
   }
   n20 = Anim Dynamics - Bone: dyn_R_top_ribbon_B!(
    ComponentPose=@n19
    Alpha=1
    SimSpaceSettings:Anim Phys Sim Space Settings Structure=(SimSpaceAngularAlpha=0,MaxAngularVelocity=10000,MaxAngularAcceleration=10000,ExternalAngularVelocity=(X=0,Y=0,Z=0))
    ExternalForce="0,0,0"
   ) -> (Pose) {
    angularDampingOverride=1
    angularSpringConstant=1000
    bAngularSpring=true
    bOverrideAngularDamping=true
    bUseGravityOverride=true
    gravityScale=0.35
    simulationSpace=RootRelative
    boundBone {boneName=dyn_R_top_ribbon_B}
    componentAppliedLinearAccClamp=<10000 10000 10000>
    componentLinearAccScale=<1 1 1>
    componentLinearVelScale=<1 1 1>
    gravityOverride {z=-980}
    physicsBodyDefinitions [{
      boundBone {boneName=dyn_R_top_ribbon_B}
      boxExtents=<11 20 20>
      constraintSetup {angularLimitsMax=<0 5 0>, angularLimitsMin=<0 -5 -25>, angularTarget=<1 0 0>}
      localJointOffset=<2 0 0>
     }]
   }
   n21 = Anim Dynamics - Bone: dyn_R_mid_ribbon_A!(
    ComponentPose=@n20
    Alpha=1
    SimSpaceSettings:Anim Phys Sim Space Settings Structure=(SimSpaceAngularAlpha=0,MaxAngularVelocity=10000,MaxAngularAcceleration=10000,ExternalAngularVelocity=(X=0,Y=0,Z=0))
    ExternalForce="0,0,0"
   ) -> (Pose) {
    angularDampingOverride=1
    angularSpringConstant=1000
    bAngularSpring=true
    bOverrideAngularDamping=true
    bUseGravityOverride=true
    gravityScale=0.35
    simulationSpace=RootRelative
    boundBone {boneName=dyn_R_mid_ribbon_A}
    componentAppliedLinearAccClamp=<10000 10000 10000>
    componentLinearAccScale=<1 1 1>
    componentLinearVelScale=<1 1 1>
    gravityOverride {z=-980}
    physicsBodyDefinitions [{
      boundBone {boneName=dyn_R_mid_ribbon_A}
      boxExtents=<11 20 20>
      constraintSetup {angularLimitsMax=<0 5 0>, angularLimitsMin=<0 -5 -20>, angularTarget=<1 0 0>}
      localJointOffset=<-2 0 0>
     }]
   }
   n22 = Anim Dynamics - Bone: dyn_R_mid_ribbon_B!(
    ComponentPose=@n21
    Alpha=1
    SimSpaceSettings:Anim Phys Sim Space Settings Structure=(SimSpaceAngularAlpha=0,MaxAngularVelocity=10000,MaxAngularAcceleration=10000,ExternalAngularVelocity=(X=0,Y=0,Z=0))
    ExternalForce="0,0,0"
   ) -> (Pose) {
    angularDampingOverride=1
    angularSpringConstant=1000
    bAngularSpring=true
    bOverrideAngularDamping=true
    bUseGravityOverride=true
    gravityScale=0.35
    simulationSpace=RootRelative
    boundBone {boneName=dyn_R_mid_ribbon_B}
    componentAppliedLinearAccClamp=<10000 10000 10000>
    componentLinearAccScale=<1 1 1>
    componentLinearVelScale=<1 1 1>
    gravityOverride {z=-980}
    physicsBodyDefinitions [{
      boundBone {boneName=dyn_R_mid_ribbon_B}
      boxExtents=<11 20 20>
      constraintSetup {angularLimitsMax=<0 5 0>, angularLimitsMin=<0 -5 -20>, angularTarget=<1 0 0>}
      localJointOffset=<-2 0 0>
     }]
   }
   n23 = Anim Dynamics - Bone: dyn_R_bottom_ribbon_A!(
    ComponentPose=@n22
    Alpha=1
    SimSpaceSettings:Anim Phys Sim Space Settings Structure=(SimSpaceAngularAlpha=0,MaxAngularVelocity=10000,MaxAngularAcceleration=10000,ExternalAngularVelocity=(X=0,Y=0,Z=0))
    ExternalForce="0,0,0"
   ) -> (Pose) {
    angularDampingOverride=1
    angularSpringConstant=1000
    bAngularSpring=true
    bOverrideAngularDamping=true
    bUseGravityOverride=true
    gravityScale=0.35
    simulationSpace=RootRelative
    boundBone {boneName=dyn_R_bottom_ribbon_A}
    componentAppliedLinearAccClamp=<10000 10000 10000>
    componentLinearAccScale=<1 1 1>
    componentLinearVelScale=<1 1 1>
    gravityOverride {z=-980}
    physicsBodyDefinitions [{
      boundBone {boneName=dyn_R_bottom_ribbon_A}
      boxExtents=<11 20 20>
      constraintSetup {angularLimitsMax=<0 5 0>, angularLimitsMin=<0 -5 -30>, angularTarget=<1 0 0>}
      localJointOffset=<-2 0 0>
     }]
   }
   n24 = Anim Dynamics - Bone: dyn_R_bottom_ribbon_B!(
    ComponentPose=@n23
    Alpha=1
    SimSpaceSettings:Anim Phys Sim Space Settings Structure=(SimSpaceAngularAlpha=0,MaxAngularVelocity=10000,MaxAngularAcceleration=10000,ExternalAngularVelocity=(X=0,Y=0,Z=0))
    ExternalForce="0,0,0"
   ) -> (Pose) {
    angularDampingOverride=1
    angularSpringConstant=1000
    bAngularSpring=true
    bOverrideAngularDamping=true
    bUseGravityOverride=true
    gravityScale=0.35
    simulationSpace=RootRelative
    boundBone {boneName=dyn_R_bottom_ribbon_B}
    componentAppliedLinearAccClamp=<10000 10000 10000>
    componentLinearAccScale=<1 1 1>
    componentLinearVelScale=<1 1 1>
    gravityOverride {z=-980}
    physicsBodyDefinitions [{
      boundBone {boneName=dyn_R_bottom_ribbon_B}
      boxExtents=<11 20 20>
      constraintSetup {angularLimitsMax=<0 5 0>, angularLimitsMin=<0 -5 -30>, angularTarget=<1 0 0>}
      localJointOffset=<-2 0 0>
     }]
   }
   n25 = "Transform (Modify) Bone - Bone: deform_knee_in_base_l"!(ComponentPose=@n24, bAlphaBoolEnabled:bool=True, Translation="3.430,-13.707,-0.349", Rotation="-8.887,-3.323,0.666", Scale="1,1,1") -> (Pose) {
    alphaInputType=Bool
    rotationMode=BMM_Additive
    rotationSpace=BCS_BoneSpace
    translationMode=BMM_Additive
    translationSpace=BCS_BoneSpace
    alphaBoolBlend {blendInTime=0.1, blendOutTime=0.1}
    boneToModify {boneName=deform_knee_in_base_l}
    rotation=<rot -8.887 -3.323 0.666>
    translation=<3.43 -13.707 -0.349>
   }
   n26 = "Transform (Modify) Bone - Bone: deform_knee_in_base_r"!(ComponentPose=@n25, bAlphaBoolEnabled:bool=True, Translation="0.446,12.702,-0.000", Rotation="-2.647,-5.650,-3.013", Scale="1,1,1") -> (Pose) {
    alphaInputType=Bool
    rotationMode=BMM_Additive
    rotationSpace=BCS_BoneSpace
    translationMode=BMM_Additive
    translationSpace=BCS_BoneSpace
    alphaBoolBlend {blendInTime=0.1, blendOutTime=0.1}
    boneToModify {boneName=deform_knee_in_base_r}
    rotation=<rot -2.647 -5.65 -3.013>
    translation=<0.446 12.702>
   }
   n27 = "Transform (Modify) Bone - Bone: dyn_L_sleeveSmallest_1"!(ComponentPose=@n26, bAlphaBoolEnabled:bool=True, Translation="0,0,0", Rotation="0,-10,0", Scale="0.5,1,1") -> (Pose) {
    alphaInputType=Bool
    rotationMode=BMM_Additive
    rotationSpace=BCS_BoneSpace
    scaleMode=BMM_Additive
    scaleSpace=BCS_BoneSpace
    translationMode=BMM_Additive
    translationSpace=BCS_BoneSpace
    alphaBoolBlend {blendInTime=0.1, blendOutTime=0.1}
    boneToModify {boneName=dyn_L_sleeveSmallest_1}
    rotation {yaw=-10}
    scale {x=0.5}
   }
   n28 = "Transform (Modify) Bone - Bone: dyn_R_sleeveSmallest_1"!(ComponentPose=@n27, bAlphaBoolEnabled:bool=True, Translation="0,0,0", Rotation="0,-10,0", Scale="0.5,1,1") -> (Pose) {
    alphaInputType=Bool
    rotationMode=BMM_Additive
    rotationSpace=BCS_BoneSpace
    scaleMode=BMM_Additive
    scaleSpace=BCS_BoneSpace
    translationMode=BMM_Additive
    translationSpace=BCS_BoneSpace
    alphaBoolBlend {blendInTime=0.1, blendOutTime=0.1}
    boneToModify {boneName=dyn_R_sleeveSmallest_1}
    rotation {yaw=-10}
    scale {x=0.5}
   }
   n29 = "Transform (Modify) Bone - Bone: dyn_R_sleeveBig_1"!(ComponentPose=@n28, bAlphaBoolEnabled:bool=True, Translation="0,0,5", Rotation="0,0,0", Scale="1,1,1") -> (Pose) {
    alphaInputType=Bool
    rotationMode=BMM_Additive
    rotationSpace=BCS_BoneSpace
    scaleMode=BMM_Additive
    scaleSpace=BCS_BoneSpace
    translationMode=BMM_Additive
    translationSpace=BCS_BoneSpace
    alphaBoolBlend {blendInTime=0.1, blendOutTime=0.1}
    boneToModify {boneName=dyn_R_sleeveBig_1}
    translation {z=5}
   }
   n30 = "Transform (Modify) Bone - Bone: dyn_L_skirt_c_2"!(ComponentPose=@n29, bAlphaBoolEnabled:bool=True, Translation="0,0,5", Rotation="20,-10,0", Scale="1,1,1") -> (Pose) {
    alphaInputType=Bool
    rotationMode=BMM_Additive
    rotationSpace=BCS_BoneSpace
    scaleMode=BMM_Additive
    scaleSpace=BCS_BoneSpace
    translationMode=BMM_Additive
    translationSpace=BCS_BoneSpace
    alphaBoolBlend {blendInTime=0.1, blendOutTime=0.1}
    boneToModify {boneName=dyn_L_skirt_c_2}
    rotation {pitch=20, yaw=-10}
    translation {z=5}
   }
   n31 = "Transform (Modify) Bone - Bone: dyn_C_skirt_b_2"!(ComponentPose=@n30, bAlphaBoolEnabled:bool=True, Translation="0,0,0", Rotation="10,0,0", Scale="1,1,1") -> (Pose) {
    alphaInputType=Bool
    rotationMode=BMM_Additive
    rotationSpace=BCS_BoneSpace
    scaleMode=BMM_Additive
    scaleSpace=BCS_BoneSpace
    translationMode=BMM_Additive
    translationSpace=BCS_BoneSpace
    alphaBoolBlend {blendInTime=0.1, blendOutTime=0.1}
    boneToModify {boneName=dyn_C_skirt_b_2}
    rotation {pitch=10}
   }
   n32 = "Transform (Modify) Bone - Bone: dyn_R_skirt_c_2"!(ComponentPose=@n31, bAlphaBoolEnabled:bool=True, Translation="0,0,0", Rotation="20,20,0", Scale="1,1,1") -> (Pose) {
    alphaInputType=Bool
    rotationMode=BMM_Additive
    rotationSpace=BCS_BoneSpace
    scaleMode=BMM_Additive
    scaleSpace=BCS_BoneSpace
    translationMode=BMM_Additive
    translationSpace=BCS_BoneSpace
    alphaBoolBlend {blendInTime=0.1, blendOutTime=0.1}
    boneToModify {boneName=dyn_R_skirt_c_2}
    rotation {pitch=20, yaw=20}
   }
   n33 = "Transform (Modify) Bone - Bone: dyn_L_skirt_b_2"!(ComponentPose=@n32, bAlphaBoolEnabled:bool=True, Translation="0,0,0", Rotation="0,0,0", Scale="0.6,1,1") -> (Pose) {
    alphaInputType=Bool
    rotationMode=BMM_Additive
    rotationSpace=BCS_BoneSpace
    scaleMode=BMM_Additive
    scaleSpace=BCS_BoneSpace
    translationMode=BMM_Additive
    translationSpace=BCS_BoneSpace
    alphaBoolBlend {blendInTime=0.1, blendOutTime=0.1}
    boneToModify {boneName=dyn_L_skirt_b_2}
    scale {x=0.6}
   }
   n34 = "Transform (Modify) Bone - Bone: dyn_C_haori_back_2"!(ComponentPose=@n33, bAlphaBoolEnabled:bool=True, Translation="0,0,0", Rotation="15,0,0", Scale="1,1,1") -> (Pose) {
    alphaInputType=Bool
    rotationMode=BMM_Additive
    rotationSpace=BCS_BoneSpace
    scaleMode=BMM_Additive
    scaleSpace=BCS_BoneSpace
    translationMode=BMM_Additive
    translationSpace=BCS_BoneSpace
    alphaBoolBlend {blendInTime=0.1, blendOutTime=0.1}
    boneToModify {boneName=dyn_C_haori_back_2}
    rotation {pitch=15}
   }
   n35 = "Transform (Modify) Bone - Bone: dyn_L_haori_side_2"!(ComponentPose=@n34, bAlphaBoolEnabled:bool=True, Translation="0,0,0", Rotation="8,0,0", Scale="1,1,1") -> (Pose) {
    alphaInputType=Bool
    rotationMode=BMM_Additive
    rotationSpace=BCS_BoneSpace
    scaleMode=BMM_Additive
    scaleSpace=BCS_BoneSpace
    translationMode=BMM_Additive
    translationSpace=BCS_BoneSpace
    alphaBoolBlend {blendInTime=0.1, blendOutTime=0.1}
    boneToModify {boneName=dyn_L_haori_side_2}
    rotation {pitch=8}
   }
   n36 = "Transform (Modify) Bone - Bone: dyn_R_haori_side_2"!(ComponentPose=@n35, bAlphaBoolEnabled:bool=True, Translation="0,0,0", Rotation="8,-5,0", Scale="1,1,1") -> (Pose) {
    alphaInputType=Bool
    rotationMode=BMM_Additive
    rotationSpace=BCS_BoneSpace
    scaleMode=BMM_Additive
    scaleSpace=BCS_BoneSpace
    translationMode=BMM_Additive
    translationSpace=BCS_BoneSpace
    alphaBoolBlend {blendInTime=0.1, blendOutTime=0.1}
    boneToModify {boneName=dyn_R_haori_side_2}
    rotation {pitch=8, yaw=-5}
   }
   n37 = "Transform (Modify) Bone - Bone: dyn_L_skirt_a_1"!(ComponentPose=@n36, bAlphaBoolEnabled:bool=True, Translation="0,0,0", Rotation="25,0,0", Scale="1,1,1") -> (Pose) {
    alphaInputType=Bool
    rotationMode=BMM_Additive
    rotationSpace=BCS_BoneSpace
    translationMode=BMM_Additive
    translationSpace=BCS_BoneSpace
    boneToModify {boneName=dyn_L_skirt_a_1}
    rotation {pitch=25}
   }
   n38 = "Transform (Modify) Bone - Bone: dyn_R_skirt_b_1"!(ComponentPose=@n37, bAlphaBoolEnabled:bool=True, Translation="0,0,5", Rotation="5,-5,0", Scale="1,1,1") -> (Pose) {
    alphaInputType=Bool
    rotationMode=BMM_Additive
    rotationSpace=BCS_BoneSpace
    translationMode=BMM_Additive
    translationSpace=BCS_BoneSpace
    boneToModify {boneName=dyn_R_skirt_b_1}
    rotation {pitch=5, yaw=-5}
    translation {z=5}
   }
   n39 = "Transform (Modify) Bone - Bone: dyn_R_skirt_a_1"!(ComponentPose=@n38, bAlphaBoolEnabled:bool=True, Translation="0,0,0", Rotation="20,0,0", Scale="1,1,1") -> (Pose) {
    alphaInputType=Bool
    rotationMode=BMM_Additive
    rotationSpace=BCS_BoneSpace
    translationMode=BMM_Additive
    translationSpace=BCS_BoneSpace
    boneToModify {boneName=dyn_R_skirt_a_1}
    rotation {pitch=20}
   }
   n40 = "Transform (Modify) Bone - Bone: dyn_R_haori_front_2"!(ComponentPose=@n39, bAlphaBoolEnabled:bool=True, Translation="0,0,0", Rotation="10,0,0", Scale="1,1,1") -> (Pose) {
    alphaInputType=Bool
    rotationMode=BMM_Additive
    rotationSpace=BCS_BoneSpace
    translationMode=BMM_Additive
    translationSpace=BCS_BoneSpace
    boneToModify {boneName=dyn_R_haori_front_2}
    rotation {pitch=10}
   }
   Output Pose(Result=Component To Local(ComponentPose=@n40)) {layerGroup=None}
   n41 = Control Rig!(
    Source=Copy Pose From Mesh() {bUseAttachedParent=true}
    Alpha
    pub isKneeSliding:bool=Get bIsKneeSliding()
    pub isBaseballSliding:bool=Get bIsBaseballSliding()
    pub isCrouching:bool=Get bIsCrouching()
    pub isPlayingEasternBloc:bool=Get IsEasterBloc() -> (IsEasterBloc:bool)
    pub isSprinting:bool=Get bIsSprinting()
    pub isJumping:bool=Get bIsJumping()
    pub isMoveForward:double=Get moveForward() -> (moveForward:double)
    pub isMoveBackward:double=Get moveBackward() -> (moveBackward:double)
    pub isMoveLeft:double=Get strafeLeft float() -> (strafeLeft float:double)
    pub isMoveRight:double=Get strafeRight float() -> (strafeRight float:double)
    pub isparachuteOpened:bool=Get bIsParachuteOpened()
    pub isSkydive:bool=Get bIsSkyDiving()
    pub isTacticalSprinting:bool=Get bIsTacticalSprinting()
    pub isPickaxeSwing:bool=Get bIsPlayingMeleeAnim()
    pub_YouBoreme:bool=Get IsYouBoreme() -> (IsYouBoreme:bool)
    pub_IsGrindRail:bool=Get bIsOnGrindRail()
    pub_IsTargeting:bool=Get bIsTargeting()
    pub_IsTacticalJump:bool=Get TacticalJumping() -> (TacticalJumping:bool)
    pub_IsFrontEnd:bool=Get bIsInFrontEnd()
   ) -> (Pose) {controlRigAssetReference {blueprintRigClass=&/F_MED_AtticDrill_CtrlRig.F_MED_AtticDrill_CtrlRig_C}}
   n42 = "Transform (Modify) Bone - Bone: attach_backpack"!(ComponentPose=Local To Component(LocalPose=@n41), bAlphaBoolEnabled:bool=True, Translation="0,-5,5", Rotation="0,0,25") -> (Pose) {
    alphaInputType=Bool
    rotationMode=BMM_Additive
    rotationSpace=BCS_BoneSpace
    translationMode=BMM_Additive
    translationSpace=BCS_BoneSpace
    alphaBoolBlend {blendInTime=0.2, blendOutTime=0.2}
    boneToModify {boneName=attach_backpack}
    rotation {roll=25}
    translation {y=-5, z=5}
   }
   Save cached pose 'CR_Cache_Input'(Pose=Component To Local(ComponentPose=@n42)) {cacheName=CR_Cache_Input}
   n43 = Anim Dynamics - Bone: dyn_L_haori_front_2!(
    ComponentPose=Local To Component(LocalPose=Use cached pose 'CR_Cache_Input'())
    Alpha=1
    GravityOverride:Vector="0,0,-980"
    SimSpaceSettings:Anim Phys Sim Space Settings Structure=(SimSpaceAngularAlpha=0,MaxAngularVelocity=10000,MaxAngularAcceleration=10000,ExternalAngularVelocity=(X=0,Y=0,Z=0))
    ExternalForce="0,0,0"
   ) -> (Pose) {
    angularDampingOverride=1
    angularSpringConstant=12000
    bAngularSpring=true
    bOverrideAngularDamping=true
    bUseGravityOverride=true
    gravityScale=0.35
    simulationSpace=RootRelative
    boundBone {boneName=dyn_L_haori_front_2}
    componentAppliedLinearAccClamp=<10000 10000 10000>
    componentLinearAccScale=<1 1 1>
    componentLinearVelScale=<1 1 1>
    gravityOverride {z=-980}
    physicsBodyDefinitions [{
      boundBone {boneName=dyn_L_haori_front_2}
      boxExtents=<30 50 50>
      constraintSetup {angularLimitsMax=<8 0 10>, angularLimitsMin=<-8 -20 -10>, angularTarget=<1 0 0>}
      localJointOffset=<8 0 0>
     }]
   }
   n44 = Anim Dynamics - Bone: dyn_L_haori_front_3!(
    ComponentPose=@n43
    Alpha=1
    GravityOverride:Vector="0,0,-980"
    SimSpaceSettings:Anim Phys Sim Space Settings Structure=(SimSpaceAngularAlpha=0,MaxAngularVelocity=10000,MaxAngularAcceleration=10000,ExternalAngularVelocity=(X=0,Y=0,Z=0))
    ExternalForce="0,0,0"
   ) -> (Pose) {
    angularDampingOverride=1
    angularSpringConstant=12000
    bAngularSpring=true
    bOverrideAngularDamping=true
    bUseGravityOverride=true
    gravityScale=0.35
    simulationSpace=RootRelative
    boundBone {boneName=dyn_L_haori_front_3}
    componentAppliedLinearAccClamp=<10000 10000 10000>
    componentLinearAccScale=<1 1 1>
    componentLinearVelScale=<1 1 1>
    gravityOverride {z=-980}
    physicsBodyDefinitions [{
      boundBone {boneName=dyn_L_haori_front_3}
      boxExtents=<30 50 50>
      constraintSetup {angularLimitsMax=<8 0 10>, angularLimitsMin=<-8 -20 -10>, angularTarget=<1 0 0>}
      localJointOffset=<8 0 0>
     }]
   }
   n45 = Anim Dynamics - Bone: dyn_L_haori_front_4!(
    ComponentPose=@n44
    Alpha=1
    GravityOverride:Vector="0,0,-980"
    SimSpaceSettings:Anim Phys Sim Space Settings Structure=(SimSpaceAngularAlpha=0,MaxAngularVelocity=10000,MaxAngularAcceleration=10000,ExternalAngularVelocity=(X=0,Y=0,Z=0))
    ExternalForce="0,0,0"
   ) -> (Pose) {
    angularDampingOverride=1
    angularSpringConstant=12000
    bAngularSpring=true
    bOverrideAngularDamping=true
    bUseGravityOverride=true
    gravityScale=0.35
    simulationSpace=RootRelative
    boundBone {boneName=dyn_L_haori_front_4}
    componentAppliedLinearAccClamp=<10000 10000 10000>
    componentLinearAccScale=<1 1 1>
    componentLinearVelScale=<1 1 1>
    gravityOverride {z=-980}
    physicsBodyDefinitions [{
      boundBone {boneName=dyn_L_haori_front_4}
      boxExtents=<30 50 50>
      constraintSetup {angularLimitsMax=<8 0 10>, angularLimitsMin=<-8 -20 -10>, angularTarget=<1 0 0>}
      localJointOffset=<8 0 0>
     }]
   }
   n46 = Anim Dynamics - Bone: dyn_L_haori_side_2!(
    ComponentPose=@n45
    Alpha=1
    GravityOverride:Vector="0,0,-980"
    SimSpaceSettings:Anim Phys Sim Space Settings Structure=(SimSpaceAngularAlpha=0,MaxAngularVelocity=10000,MaxAngularAcceleration=10000,ExternalAngularVelocity=(X=0,Y=0,Z=0))
    ExternalForce="0,0,0"
   ) -> (Pose) {
    angularDampingOverride=1
    angularSpringConstant=12000
    bAngularSpring=true
    bOverrideAngularDamping=true
    bUseGravityOverride=true
    gravityScale=0.35
    simulationSpace=RootRelative
    boundBone {boneName=dyn_L_haori_side_2}
    componentAppliedLinearAccClamp=<10000 10000 10000>
    componentLinearAccScale=<1 1 1>
    componentLinearVelScale=<1 1 1>
    gravityOverride {z=-980}
    physicsBodyDefinitions [{
      boundBone {boneName=dyn_L_haori_side_2}
      boxExtents=<30 50 50>
      constraintSetup {angularLimitsMax=<7 0 10>, angularLimitsMin=<-7 -20 -10>, angularTarget=<1 0 0>}
      localJointOffset=<8 0 0>
     }]
   }
   n47 = Anim Dynamics - Bone: dyn_L_haori_side_3!(
    ComponentPose=@n46
    Alpha=1
    GravityOverride:Vector="0,0,-980"
    SimSpaceSettings:Anim Phys Sim Space Settings Structure=(SimSpaceAngularAlpha=0,MaxAngularVelocity=10000,MaxAngularAcceleration=10000,ExternalAngularVelocity=(X=0,Y=0,Z=0))
    ExternalForce="0,0,0"
   ) -> (Pose) {
    angularDampingOverride=1
    angularSpringConstant=12000
    bAngularSpring=true
    bOverrideAngularDamping=true
    bUseGravityOverride=true
    gravityScale=0.35
    simulationSpace=RootRelative
    boundBone {boneName=dyn_L_haori_side_3}
    componentAppliedLinearAccClamp=<10000 10000 10000>
    componentLinearAccScale=<1 1 1>
    componentLinearVelScale=<1 1 1>
    gravityOverride {z=-980}
    physicsBodyDefinitions [{
      boundBone {boneName=dyn_L_haori_side_3}
      boxExtents=<30 50 50>
      constraintSetup {angularLimitsMax=<7 0 10>, angularLimitsMin=<-7 -20 -10>, angularTarget=<1 0 0>}
      localJointOffset=<8 0 0>
     }]
   }
   n48 = Anim Dynamics - Bone: dyn_L_haori_side_4!(
    ComponentPose=@n47
    Alpha=1
    GravityOverride:Vector="0,0,-980"
    SimSpaceSettings:Anim Phys Sim Space Settings Structure=(SimSpaceAngularAlpha=0,MaxAngularVelocity=10000,MaxAngularAcceleration=10000,ExternalAngularVelocity=(X=0,Y=0,Z=0))
    ExternalForce="0,0,0"
   ) -> (Pose) {
    angularDampingOverride=1
    angularSpringConstant=12000
    bAngularSpring=true
    bOverrideAngularDamping=true
    bUseGravityOverride=true
    gravityScale=0.35
    simulationSpace=RootRelative
    boundBone {boneName=dyn_L_haori_side_4}
    componentAppliedLinearAccClamp=<10000 10000 10000>
    componentLinearAccScale=<1 1 1>
    componentLinearVelScale=<1 1 1>
    gravityOverride {z=-980}
    physicsBodyDefinitions [{
      boundBone {boneName=dyn_L_haori_side_4}
      boxExtents=<30 50 50>
      constraintSetup {angularLimitsMax=<7 0 10>, angularLimitsMin=<-7 -20 -10>, angularTarget=<1 0 0>}
      localJointOffset=<8 0 0>
     }]
   }
   n49 = Anim Dynamics - Bone: dyn_C_haori_back_2!(
    ComponentPose=@n48
    Alpha=1
    GravityOverride:Vector="0,0,-980"
    SimSpaceSettings:Anim Phys Sim Space Settings Structure=(SimSpaceAngularAlpha=0,MaxAngularVelocity=10000,MaxAngularAcceleration=10000,ExternalAngularVelocity=(X=0,Y=0,Z=0))
    ExternalForce="0,0,0"
   ) -> (Pose) {
    angularDampingOverride=1
    angularSpringConstant=12000
    bAngularSpring=true
    bOverrideAngularDamping=true
    bUseGravityOverride=true
    gravityScale=0.35
    simulationSpace=RootRelative
    boundBone {boneName=dyn_C_haori_back_2}
    componentAppliedLinearAccClamp=<10000 10000 10000>
    componentLinearAccScale=<1 1 1>
    componentLinearVelScale=<1 1 1>
    gravityOverride {z=-980}
    physicsBodyDefinitions [{
      boundBone {boneName=dyn_C_haori_back_2}
      boxExtents=<30 50 50>
      constraintSetup {angularLimitsMax=<8 0 10>, angularLimitsMin=<-8 -20 -10>, angularTarget=<1 0 0>}
      localJointOffset=<8 0 0>
     }]
   }
   n50 = Anim Dynamics - Bone: dyn_C_haori_back_3!(
    ComponentPose=@n49
    Alpha=1
    GravityOverride:Vector="0,0,-980"
    SimSpaceSettings:Anim Phys Sim Space Settings Structure=(SimSpaceAngularAlpha=0,MaxAngularVelocity=10000,MaxAngularAcceleration=10000,ExternalAngularVelocity=(X=0,Y=0,Z=0))
    ExternalForce="0,0,0"
   ) -> (Pose) {
    angularDampingOverride=1
    angularSpringConstant=12000
    bAngularSpring=true
    bOverrideAngularDamping=true
    bUseGravityOverride=true
    gravityScale=0.35
    simulationSpace=RootRelative
    boundBone {boneName=dyn_C_haori_back_3}
    componentAppliedLinearAccClamp=<10000 10000 10000>
    componentLinearAccScale=<1 1 1>
    componentLinearVelScale=<1 1 1>
    gravityOverride {z=-980}
    physicsBodyDefinitions [{
      boundBone {boneName=dyn_C_haori_back_3}
      boxExtents=<30 50 50>
      constraintSetup {angularLimitsMax=<8 0 10>, angularLimitsMin=<-8 -20 -10>, angularTarget=<1 0 0>}
      localJointOffset=<8 0 0>
     }]
   }
   n51 = Anim Dynamics - Bone: dyn_C_haori_back_4!(
    ComponentPose=@n50
    Alpha=1
    GravityOverride:Vector="0,0,-980"
    SimSpaceSettings:Anim Phys Sim Space Settings Structure=(SimSpaceAngularAlpha=0,MaxAngularVelocity=10000,MaxAngularAcceleration=10000,ExternalAngularVelocity=(X=0,Y=0,Z=0))
    ExternalForce="0,0,0"
   ) -> (Pose) {
    angularDampingOverride=1
    angularSpringConstant=12000
    bAngularSpring=true
    bOverrideAngularDamping=true
    bUseGravityOverride=true
    gravityScale=0.35
    simulationSpace=RootRelative
    boundBone {boneName=dyn_C_haori_back_4}
    componentAppliedLinearAccClamp=<10000 10000 10000>
    componentLinearAccScale=<1 1 1>
    componentLinearVelScale=<1 1 1>
    gravityOverride {z=-980}
    physicsBodyDefinitions [{
      boundBone {boneName=dyn_C_haori_back_4}
      boxExtents=<30 50 50>
      constraintSetup {angularLimitsMax=<8 0 10>, angularLimitsMin=<-8 -20 -10>, angularTarget=<1 0 0>}
      localJointOffset=<8 0 0>
     }]
   }
   n52 = Anim Dynamics - Bone: dyn_R_haori_front_2!(
    ComponentPose=@n51
    Alpha=1
    GravityOverride:Vector="0,0,-980"
    SimSpaceSettings:Anim Phys Sim Space Settings Structure=(SimSpaceAngularAlpha=0,MaxAngularVelocity=10000,MaxAngularAcceleration=10000,ExternalAngularVelocity=(X=0,Y=0,Z=0))
    ExternalForce="0,0,0"
   ) -> (Pose) {
    angularDampingOverride=1
    angularSpringConstant=12000
    bAngularSpring=true
    bOverrideAngularDamping=true
    bUseGravityOverride=true
    gravityScale=0.35
    simulationSpace=RootRelative
    boundBone {boneName=dyn_R_haori_front_2}
    componentAppliedLinearAccClamp=<10000 10000 10000>
    componentLinearAccScale=<1 1 1>
    componentLinearVelScale=<1 1 1>
    gravityOverride {z=-980}
    physicsBodyDefinitions [{
      boundBone {boneName=dyn_R_haori_front_2}
      boxExtents=<30 50 50>
      constraintSetup {angularLimitsMax=<8 0 10>, angularLimitsMin=<-8 -20 -10>, angularTarget=<1 0 0>}
      localJointOffset=<8 0 0>
     }]
   }
   n53 = Anim Dynamics - Bone: dyn_R_haori_front_3!(
    ComponentPose=@n52
    Alpha=1
    GravityOverride:Vector="0,0,-980"
    SimSpaceSettings:Anim Phys Sim Space Settings Structure=(SimSpaceAngularAlpha=0,MaxAngularVelocity=10000,MaxAngularAcceleration=10000,ExternalAngularVelocity=(X=0,Y=0,Z=0))
    ExternalForce="0,0,0"
   ) -> (Pose) {
    angularDampingOverride=1
    angularSpringConstant=12000
    bAngularSpring=true
    bOverrideAngularDamping=true
    bUseGravityOverride=true
    gravityScale=0.35
    simulationSpace=RootRelative
    boundBone {boneName=dyn_R_haori_front_3}
    componentAppliedLinearAccClamp=<10000 10000 10000>
    componentLinearAccScale=<1 1 1>
    componentLinearVelScale=<1 1 1>
    gravityOverride {z=-980}
    physicsBodyDefinitions [{
      boundBone {boneName=dyn_R_haori_front_3}
      boxExtents=<30 50 50>
      constraintSetup {angularLimitsMax=<8 0 10>, angularLimitsMin=<-8 -20 -10>, angularTarget=<1 0 0>}
      localJointOffset=<8 0 0>
     }]
   }
   n54 = Anim Dynamics - Bone: dyn_R_haori_front_4!(
    ComponentPose=@n53
    Alpha=1
    GravityOverride:Vector="0,0,-980"
    SimSpaceSettings:Anim Phys Sim Space Settings Structure=(SimSpaceAngularAlpha=0,MaxAngularVelocity=10000,MaxAngularAcceleration=10000,ExternalAngularVelocity=(X=0,Y=0,Z=0))
    ExternalForce="0,0,0"
   ) -> (Pose) {
    angularDampingOverride=1
    angularSpringConstant=12000
    bAngularSpring=true
    bOverrideAngularDamping=true
    bUseGravityOverride=true
    gravityScale=0.35
    simulationSpace=RootRelative
    boundBone {boneName=dyn_R_haori_front_4}
    componentAppliedLinearAccClamp=<10000 10000 10000>
    componentLinearAccScale=<1 1 1>
    componentLinearVelScale=<1 1 1>
    gravityOverride {z=-980}
    physicsBodyDefinitions [{
      boundBone {boneName=dyn_R_haori_front_4}
      boxExtents=<30 50 50>
      constraintSetup {angularLimitsMax=<8 0 10>, angularLimitsMin=<-8 -20 -10>, angularTarget=<1 0 0>}
      localJointOffset=<8 0 0>
     }]
   }
   n55 = Anim Dynamics - Bone: dyn_R_haori_side_2!(
    ComponentPose=@n54
    Alpha=1
    GravityOverride:Vector="0,0,-980"
    SimSpaceSettings:Anim Phys Sim Space Settings Structure=(SimSpaceAngularAlpha=0,MaxAngularVelocity=10000,MaxAngularAcceleration=10000,ExternalAngularVelocity=(X=0,Y=0,Z=0))
    ExternalForce="0,0,0"
   ) -> (Pose) {
    angularDampingOverride=1
    angularSpringConstant=12000
    bAngularSpring=true
    bOverrideAngularDamping=true
    bUseGravityOverride=true
    gravityScale=0.35
    simulationSpace=RootRelative
    boundBone {boneName=dyn_R_haori_side_2}
    componentAppliedLinearAccClamp=<10000 10000 10000>
    componentLinearAccScale=<1 1 1>
    componentLinearVelScale=<1 1 1>
    gravityOverride {z=-980}
    physicsBodyDefinitions [{
      boundBone {boneName=dyn_R_haori_side_2}
      boxExtents=<30 50 50>
      constraintSetup {angularLimitsMax=<8 0 10>, angularLimitsMin=<-8 -20 -10>, angularTarget=<1 0 0>}
      localJointOffset=<8 0 0>
     }]
   }
   n56 = Anim Dynamics - Bone: dyn_R_haori_side_3!(
    ComponentPose=@n55
    Alpha=1
    GravityOverride:Vector="0,0,-980"
    SimSpaceSettings:Anim Phys Sim Space Settings Structure=(SimSpaceAngularAlpha=0,MaxAngularVelocity=10000,MaxAngularAcceleration=10000,ExternalAngularVelocity=(X=0,Y=0,Z=0))
    ExternalForce="0,0,0"
   ) -> (Pose) {
    angularDampingOverride=1
    angularSpringConstant=12000
    bAngularSpring=true
    bOverrideAngularDamping=true
    bUseGravityOverride=true
    gravityScale=0.35
    simulationSpace=RootRelative
    boundBone {boneName=dyn_R_haori_side_3}
    componentAppliedLinearAccClamp=<10000 10000 10000>
    componentLinearAccScale=<1 1 1>
    componentLinearVelScale=<1 1 1>
    gravityOverride {z=-980}
    physicsBodyDefinitions [{
      boundBone {boneName=dyn_R_haori_side_3}
      boxExtents=<30 50 50>
      constraintSetup {angularLimitsMax=<8 0 10>, angularLimitsMin=<-8 -20 -10>, angularTarget=<1 0 0>}
      localJointOffset=<8 0 0>
     }]
   }
   n57 = Anim Dynamics - Bone: dyn_R_haori_side_4!(
    ComponentPose=@n56
    Alpha=1
    GravityOverride:Vector="0,0,-980"
    SimSpaceSettings:Anim Phys Sim Space Settings Structure=(SimSpaceAngularAlpha=0,MaxAngularVelocity=10000,MaxAngularAcceleration=10000,ExternalAngularVelocity=(X=0,Y=0,Z=0))
    ExternalForce="0,0,0"
   ) -> (Pose) {
    angularDampingOverride=1
    angularSpringConstant=12000
    bAngularSpring=true
    bOverrideAngularDamping=true
    bUseGravityOverride=true
    gravityScale=0.35
    simulationSpace=RootRelative
    boundBone {boneName=dyn_R_haori_side_4}
    componentAppliedLinearAccClamp=<10000 10000 10000>
    componentLinearAccScale=<1 1 1>
    componentLinearVelScale=<1 1 1>
    gravityOverride {z=-980}
    physicsBodyDefinitions [{
      boundBone {boneName=dyn_R_haori_side_4}
      boxExtents=<30 50 50>
      constraintSetup {angularLimitsMax=<8 0 10>, angularLimitsMin=<-8 -20 -10>, angularTarget=<1 0 0>}
      localJointOffset=<8 0 0>
     }]
   }
   Save cached pose 'CR_Cache_Haori'(Pose=Component To Local(ComponentPose=@n57)) {cacheName=CR_Cache_Haori}
  ]}
 EventGraph {nodes [
   Event BlueprintUpdateAnimation()
   > n1 = Sequence()
   > Set Sim Space Settings(Sim Space Settings:Anim Phys Sim Space Settings Structure=Get Sim Space Settings_BR() -> (Sim Space Settings_BR:Anim Phys Sim Space Settings Structure)) -> (Output_Get:Anim Phys Sim Space Settings Structure)
   > Branch(Condition=Get bIsInFrontEnd())
   > Set Sim Space Settings(Sim Space Settings=Get Sim Space Settings_FE() -> (Sim Space Settings_FE:Anim Phys Sim Space Settings Structure))
   K2Node_Composite_0 = Emote!(InputPin:Exec=@n1.then_1) -> (then_1:Exec) {is=title K2Node_Composite}
   > n2 = "Gravity Override State Function (Parameter Asset) (Nativized)"(
    CharacterPartAnimInstance=Self-Reference()
    GameTimeInSeconds=GetGameTimeInSeconds()
    Parameters=&/F_MED_AtticDrill_Parameters
   )
   > Set GravityOverride(GravityOverride=@n2)
   > Set RBAN Settings Out(RBAN Settings Out=@n2.OutRigidBodyAnimNodeParameters)
   Event OnBeginJumping()
   > Set TacticalJumping(TacticalJumping:bool=Get bIsTacticalSprinting()) -> (Output_Get:bool)
   n3 = Event OnLeftLegHidden() {is=title K2Node_Event}
   Set IsWearingBoots(execute=[@n3.then, @n4.then], IsWearingBoots:bool=true) -> (Output_Get:bool)
   Event OnEnterGrindRail() {is=title K2Node_Event}
   > Set isOnGrindRail float(isOnGrindRail float:double=1) -> (Output_Get:double)
   n4 = Event OnRightLegHidden() {is=title K2Node_Event}
   n5 = Event OnLeftLegVisible() {is=title K2Node_Event}
   Set IsWearingBoots(execute=[@n5.then, @n6.then], IsWearingBoots=false)
   Event OnExitGrindRail() {is=title K2Node_Event}
   > Set isOnGrindRail float(isOnGrindRail float=0)
   Event OnEndFalling()
   > Set TacticalJumping(TacticalJumping=false)
   n6 = Event OnRightLegVisible() {is=title K2Node_Event}
  ]}
 Emote {
  path=EventGraph.K2Node_Composite_0.Emote
  nodes [
   Inputs!() -> (InputPin:Exec)
   > n1 = Sequence!(execute) -> (then_0, then_1, then_2:Exec)
   > n2 = PlayerMovement(AnimInstance=Self-Reference(), XDotProductCoefficient=1, YDotProductCoefficient=1, ZDotProductCoefficient=1)
   n3 = BreakVector(InVec=MakeVector(X=@n2, Y=@n2.YMovement))
   > Set strafeLeft float(strafeLeft float:double="Absolute (Float)"(A="Clamp (Float)"(Value=@n3, Min=-1, Max=0))) -> (Output_Get:double)
   > Set strafeRight float(strafeRight float:double="Clamp (Float)"(Value=@n3, Min=0, Max=1)) -> (Output_Get:double)
   > Set moveForward(moveForward:double="Clamp (Float)"(Value=@n3.Y, Min=0, Max=1)) -> (Output_Get:double)
   > Set moveBackward(moveBackward:double="Absolute (Float)"(A="Clamp (Float)"(Value=@n3.Y, Min=-1, Max=0))) -> (Output_Get:double)
   > n4 = Branch(Condition=OR Boolean!(A=Get bIsJumping(), B=Get bIsFalling(), C:bool=Get bIsKneeSliding(), D:bool=Get bIsBaseballSliding()) -> (ReturnValue))
   > Set AlphaDyn(AlphaDyn:double=0.3) -> (Output_Get:double)
   > Set AlphaDynHaori(AlphaDynHaori:double=0.5) -> (Output_Get:double)
   Set AlphaDyn(execute=@n4.else, AlphaDyn=0.5)
   > Set AlphaDynHaori(AlphaDynHaori=1)
   n5 = GetAnimInstance(self=GetOwnerSkeletalMeshComponent())
   Is Valid(exec=@n1.then_1, InputObject=@n5)
   > Set IsEasterBloc(IsEasterBloc:bool=Montage_IsPlaying(self=@n5, Montage=/BRCosmetics/Animation/Game/MainPlayer/Emotes/Eastern_Bloc/Emote_EasternBloc_CMF_M)) -> (Output_Get:bool)
   > Set IsYouBoreme(IsYouBoreme:bool=Montage_IsPlaying(self=@n5, Montage=/BRCosmetics/Animation/Game/MainPlayer/Emotes/YouBoreMe/Emote_You_Bore_Me_CMF_M)) -> (Output_Get:bool)
   > Set SleevesFix(SleevesFix:bool=NOT Boolean(A=Get bIsParachuteOpened())) -> (Output_Get:bool)
   Outputs!(then_1:Exec=@n1.then_2)
  ]
 }
]
variables [
 AlphaDyn {type=number}
 AlphaDynHaori {type=number}
 GravityOverride {type=Vector}
 is parachuteoffset {type=boolean}
 isCrouchStrafeLeft float {type=number}
 isCrouchStrafeRight float {type=number}
 IsEasterBloc {type=boolean}
 isOnGrindRail float {type=number}
 IsWearingBoots {type=boolean}
 IsYouBoreme {type=boolean}
 moveBackward {type=number}
 moveBackwardStanding {type=boolean}
 moveForward {type=number}
 RBAN Settings Out {type=FortRigidBodyAnimNodeParameters}
 Sim Space Settings {
  type=AnimPhysSimSpaceSettings
  default {maxAngularAcceleration=10000, maxAngularVelocity=10000, simSpaceAngularAlpha=0, externalAngularVelocity=<0 0 0>}
 }
 Sim Space Settings_BR {
  type=AnimPhysSimSpaceSettings
  default {maxAngularAcceleration=10000, maxAngularVelocity=10000, simSpaceAngularAlpha=0, externalAngularVelocity=<0 0 0>}
 }
 Sim Space Settings_FE {
  type=AnimPhysSimSpaceSettings
  default {maxAngularAcceleration=10000, maxAngularVelocity=10000, simSpaceAngularAlpha=0.3, externalAngularVelocity=<0.2 0.2 0.2>}
 }
 SleevesFix {type=boolean}
 strafeLeft float {type=number}
 strafeRight float {type=number}
 TacticalJumping {type=boolean}
]
