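# Collision-only physics asset for the AtticDrill female medium player cosmetic body, named as the colliders for its cape.
# Capsules cover the head, clavicles, spine, pelvis and feet, with doubled capsules on the middle spine; arm, leg and neck bodies are empty.
# Five four-bone haori chains (back, left/right front and side) carry offset capsules, wider on back and sides, so the dynamic coat bones also act as colliders.
# No constraints are defined.
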PhysicsAsset /BRCosmetics/Characters/Player/Female/Medium/Bodies/F_MED_AtticDrill/Meshes/F_MED_AtticDrill_CapeColliders
bodies {
 calf_l []
 calf_r []
 clavicle_l [{shapeType=Capsule, length=6.5623, radius=7.5194, center=<3.9566 -1.0862 -2.1127>, rotation=<rot -77.5566 177.3915 7.7314>}]
 clavicle_r [{shapeType=Capsule, length=6.5623, radius=7.5194, center=<-3.9566 1.0862 2.1127>, rotation=<rot -77.5566 177.3915 -172.2686>}]
 dyn_C_haori_back_1 [{shapeType=Capsule, length=5, radius=15, shapeName=dyn_C_haori_back_1_capsule, center=<12.5709 -0.2631 13.3779>, rotation=<rot 90 0 0>}]
 dyn_C_haori_back_2 [{shapeType=Capsule, length=5, radius=15, shapeName=dyn_C_haori_back_2_capsule, center=<15.0569 -0.0001 14.2426>, rotation=<rot 90 0 0>}]
 dyn_C_haori_back_3 [{shapeType=Capsule, length=5, radius=15, shapeName=dyn_C_haori_back_3_capsule, center=<11.4922 0 14.4125>, rotation=<rot 90 0 0>}]
 dyn_C_haori_back_4 [{shapeType=Capsule, length=5, radius=15, shapeName=dyn_C_haori_back_4_capsule, center=<2.844 -0.1089 13.9303>, rotation=<rot 90 0 0>}]
 dyn_L_haori_front_1 [{shapeType=Capsule, length=5, radius=10, shapeName=dyn_L_haori_front_1_capsule, center=<5.4402 0.3021 -1.6927>, rotation=<rot 90 0 0>}]
 dyn_L_haori_front_2 [{shapeType=Capsule, length=5, radius=10, shapeName=dyn_L_haori_front_2_capsule, center=<11.6675 1.7364 -0.8236>, rotation=<rot 90 0 0>}]
 dyn_L_haori_front_3 [{shapeType=Capsule, length=5, radius=10, shapeName=dyn_L_haori_front_3_capsule, center=<16.0573 1.5886 -0.654>, rotation=<rot 90 0 0>}]
 dyn_L_haori_front_4 [{shapeType=Capsule, length=5, radius=10, shapeName=dyn_L_haori_front_4_capsule, center=<6.813 0.4127 -0.1538>, rotation=<rot 90 0 0>}]
 dyn_L_haori_side_1 [{shapeType=Capsule, length=5, radius=15, shapeName=dyn_L_haori_side_1_capsule, center=<12.0366 -2.4549 13.8995>, rotation=<rot 90 0 0>}]
 dyn_L_haori_side_2 [{shapeType=Capsule, length=5, radius=15, shapeName=dyn_L_haori_side_2_capsule, center=<14.8097 -3.2509 13.8778>, rotation=<rot 90 0 0>}]
 dyn_L_haori_side_3 [{shapeType=Capsule, length=5, radius=15, shapeName=dyn_L_haori_side_3_capsule, center=<14.5228 -1.2613 14.3196>, rotation=<rot 90 0 0>}]
 dyn_L_haori_side_4 [{shapeType=Capsule, length=5, radius=15, shapeName=dyn_L_haori_side_4_capsule, center=<6.4353 2.7879 11.9901>, rotation=<rot 90 0 0>}]
 dyn_R_haori_front_1 [{shapeType=Capsule, length=5, radius=10, shapeName=dyn_R_haori_front_1_capsule, center=<9.2913 -0.5662 -1.4544>, rotation=<rot 90 0 0>}]
 dyn_R_haori_front_2 [{shapeType=Capsule, length=5, radius=10, shapeName=dyn_R_haori_front_2_capsule, center=<13.8489 -1.9652 -0.8811>, rotation=<rot 90 0 0>}]
 dyn_R_haori_front_3 [{shapeType=Capsule, length=5, radius=10, shapeName=dyn_R_haori_front_3_capsule, center=<16.0566 -1.5893 -0.6541>, rotation=<rot 90 0 0>}]
 dyn_R_haori_front_4 [{shapeType=Capsule, length=5, radius=10, shapeName=dyn_R_haori_front_4_capsule, center=<6.8118 -0.4093 -0.1024>, rotation=<rot 90 0 0>}]
 dyn_R_haori_side_1 [{shapeType=Capsule, length=5, radius=15, shapeName=dyn_R_haori_side_1_capsule, center=<11.8095 6.6678 12.4805>, rotation=<rot 90 0 0>}]
 dyn_R_haori_side_2 [{shapeType=Capsule, length=5, radius=15, shapeName=dyn_R_haori_side_2_capsule, center=<14.8091 7.4592 16.1036>, rotation=<rot 90 0 0>}]
 dyn_R_haori_side_3 [{shapeType=Capsule, length=5, radius=15, shapeName=dyn_R_haori_side_3_capsule, center=<14.5731 5.4831 16.576>, rotation=<rot 90 0 0>}]
 dyn_R_haori_side_4 [{shapeType=Capsule, length=5, radius=15, shapeName=dyn_R_haori_side_4_capsule, center=<6.4349 -0.4328 11.9901>, rotation=<rot 90 0 0>}]
 foot_l [{shapeType=Capsule, length=11.0765, radius=8.0344, center=<-4.1811 6.0027 -0.338>, rotation=<rot 0.0388 1.4701 -89.3982>}]
 foot_r [{shapeType=Capsule, length=11.0765, radius=8.0344, center=<4.1811 -6.0027 0.338>, rotation=<rot 0.0388 1.4701 90.6018>}]
 head [{shapeType=Capsule, length=0, radius=18, center=<9.0697 0.0746 0>, rotation=<rot 0 -90 0>}]
 lowerarm_l []
 lowerarm_r []
 neck_02 []
 pelvis [{shapeType=Capsule, length=20.5963, radius=11.1441, center=<-2.4471 -2.5289 -0.1981>}]
 spine_01 [{shapeType=Capsule, length=15.6686, radius=19.3034, center=<-21.5671 -28.6108 0>}]
 spine_02 [
  {shapeType=Capsule, length=19.8832, radius=18.2957, center=<-9.6327 -24.2869 -0.5739>}
  {shapeType=Capsule, length=20.6875, radius=9.3125, center=<-1.8317 1.9362 0>, rotation=<rot 90 0 0>}
 ]
 spine_03 [
  {shapeType=Capsule, length=22.2365, radius=14.7572, center=<0.0574 -20.2739 0>}
  {shapeType=Capsule, length=20.7188, radius=9.2812, center=<-1.2163 0.549 0>, rotation=<rot 90 180 -175.0732>}
 ]
 spine_04 [
  {shapeType=Capsule, length=25.0578, radius=14.0544, center=<5.6234 -13.2747 0>}
  {shapeType=Capsule, length=10.0312, radius=9.2812, center=<-1.2326 1.308 0>, rotation=<rot 90 0 0>}
 ]
 spine_05 [{shapeType=Capsule, length=9.678, radius=12.3571, center=<0.6969 1.91 0>}]
 thigh_l []
 thigh_r []
 upperarm_l []
 upperarm_r []
}
constraints []
