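# Physics asset for the F_MED_AtticDrill female player body, defining secondary motion for its haori and yukata garment bones.
# Five haori chains (center back, left and right front and side) each link four capsules to a tiny master body with tight swing and twist limits.
# Two-box yukata back chains attach to each calf with zero-degree twist limits.
# Pelvis, thigh and calf capsules provide the lower-body collision shapes.
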PhysicsAsset /BRCosmetics/Characters/Player/Female/Medium/Bodies/F_MED_AtticDrill/Meshes/F_MED_AtticDrill_Physics
bodies {
 calf_l [{shapeType=Capsule, length=39.0543, radius=9.5, shapeName=calf_r_capsule, center=<-28.7843 -2.5 0>, rotation=<rot 90 -2 0>}]
 calf_r [{shapeType=Capsule, length=41.0543, radius=9.5, shapeName=calf_r_capsule, center=<28.7843 2.5 0>, rotation=<rot 90 -2 0>}]
 dyn_C_haori_back_1 [{shapeType=Capsule, length=20, radius=7, shapeName=dyn_C_haori_back_1_capsule, center=<5 0 0>, rotation=<rot 0 0 90>}]
 dyn_C_haori_back_2 [{shapeType=Capsule, length=20, radius=7, shapeName=dyn_C_haori_back_2_capsule, center=<5 0 0>, rotation=<rot 0 0 90>}]
 dyn_C_haori_back_3 [{shapeType=Capsule, length=20, radius=7, shapeName=dyn_C_haori_back_3_capsule, center=<5 0 0>, rotation=<rot 0 0 90>}]
 dyn_C_haori_back_4 [{shapeType=Capsule, length=20, radius=6.5, shapeName=dyn_C_haori_back_4_capsule, center=<5 0 1>, rotation=<rot 0 0 90>}]
 dyn_C_haori_back_mstr [{shapeType=Capsule, length=0.505, radius=0.505, shapeName=dyn_C_haori_back_mstr_capsule, rotation=<rot 0 -90 0>}]
 dyn_L_haori_front_1 [{shapeType=Capsule, length=20, radius=7, shapeName=dyn_L_haori_front_1_capsule, center=<5 0 1.5>, rotation=<rot 0 0 90>}]
 dyn_L_haori_front_2 [{shapeType=Capsule, length=20, radius=7, shapeName=dyn_L_haori_front_2_capsule, center=<5 0 1.5>, rotation=<rot 0 0 90>}]
 dyn_L_haori_front_3 [{shapeType=Capsule, length=20, radius=7, shapeName=dyn_L_haori_front_3_capsule, center=<5 0 6.5>, rotation=<rot 0 0 90>}]
 dyn_L_haori_front_4 [{shapeType=Capsule, length=20, radius=7, shapeName=dyn_L_haori_front_4_capsule, center=<5 0 4>, rotation=<rot 0 0 90>}]
 dyn_L_haori_front_mstr [{shapeType=Capsule, length=0.505, radius=0.505, shapeName=dyn_L_haori_front_mstr_capsule, rotation=<rot 0 -90 0>}]
 dyn_L_haori_side_1 [{shapeType=Capsule, length=20, radius=7, shapeName=dyn_L_haori_side_1_capsule, center=<5 0 0>, rotation=<rot 0 0 90>}]
 dyn_L_haori_side_2 [{shapeType=Capsule, length=20, radius=7, shapeName=dyn_L_haori_side_2_capsule, center=<5 0 0>, rotation=<rot 0 0 90>}]
 dyn_L_haori_side_3 [{shapeType=Capsule, length=20, radius=7, shapeName=dyn_L_haori_side_3_capsule, center=<5 0 6>, rotation=<rot 0 0 90>}]
 dyn_L_haori_side_4 [{shapeType=Capsule, length=20, radius=6.5, shapeName=dyn_L_haori_side_4_capsule, center=<5 0 4>, rotation=<rot 0 0 90>}]
 dyn_L_haori_side_mstr [{shapeType=Capsule, length=0.505, radius=0.505, shapeName=dyn_L_haori_side_mstr_capsule, rotation=<rot 0 -90 0>}]
 dyn_L_yukata_back_1 [{shapeType=Box, extentX=21.8913, extentY=5.8913, extentZ=30, shapeName=dyn_L_yukata_back_1_box, center=<5 -6.5 0>}]
 dyn_L_yukata_back_2 [{shapeType=Box, extentX=21.8913, extentY=5.8913, extentZ=30, shapeName=dyn_L_yukata_back_2_box, center=<5 -4.5 0>, rotation=<rot 0 10 0>}]
 dyn_R_haori_front_1 [{shapeType=Capsule, length=20, radius=7, shapeName=dyn_R_haori_front_1_capsule, center=<-5 0 -1.5>, rotation=<rot 0 0 90>}]
 dyn_R_haori_front_2 [{shapeType=Capsule, length=20, radius=7, shapeName=dyn_R_haori_front_2_capsule, center=<-5 0 -1.5>, rotation=<rot 0 0 90>}]
 dyn_R_haori_front_3 [{shapeType=Capsule, length=20, radius=7, shapeName=dyn_R_haori_front_3_capsule, center=<-5 0 -6.5>, rotation=<rot 0 0 90>}]
 dyn_R_haori_front_4 [{shapeType=Capsule, length=20, radius=7, shapeName=dyn_R_haori_front_4_capsule, center=<-5 0 -4>, rotation=<rot 0 0 90>}]
 dyn_R_haori_front_mstr [{shapeType=Capsule, length=0.505, radius=0.505, shapeName=dyn_R_haori_front_mstr_capsule, rotation=<rot 0 -90 0>}]
 dyn_R_haori_side_1 [{shapeType=Capsule, length=20, radius=7, shapeName=dyn_R_haori_side_1_capsule, center=<-5 0 0>, rotation=<rot 0 0 90>}]
 dyn_R_haori_side_2 [{shapeType=Capsule, length=20, radius=7, shapeName=dyn_R_haori_side_2_capsule, center=<-5 0 0>, rotation=<rot 0 0 90>}]
 dyn_R_haori_side_3 [{shapeType=Capsule, length=20, radius=7, shapeName=dyn_R_haori_side_3_capsule, center=<-5 0 -6>, rotation=<rot 0 0 90>}]
 dyn_R_haori_side_4 [{shapeType=Capsule, length=20, radius=6.5, shapeName=dyn_R_haori_side_4_capsule, center=<-5 0 -4>, rotation=<rot 0 0 90>}]
 dyn_R_haori_side_mstr [{shapeType=Capsule, length=0.505, radius=0.505, shapeName=dyn_R_haori_side_mstr_capsule, rotation=<rot 0 -90 0>}]
 dyn_R_yukata_back_1 [{shapeType=Box, extentX=21.8913, extentY=5.8913, extentZ=30, shapeName=dyn_R_yukata_back_1_box, center=<-5 6.5 0>}]
 dyn_R_yukata_back_2 [{shapeType=Box, extentX=21.8913, extentY=5.8913, extentZ=30, shapeName=dyn_R_yukata_back_2_box, center=<-5 4.5 0>, rotation=<rot 0 10 0>}]
 pelvis [{shapeType=Capsule, length=6, radius=8, shapeName=pelvis_capsule}]
 thigh_l [{shapeType=Capsule, length=33, radius=8.5, shapeName=thigh_r_capsule, center=<-23.0995 1 2.2622>, rotation=<rot 90 0 0>}]
 thigh_r [{shapeType=Capsule, length=33, radius=8.5, shapeName=thigh_r_capsule, center=<23.0995 -1 -2.2622>, rotation=<rot 90 0 0>}]
}
constraints [
 {bone1Name=dyn_C_haori_back_1, bone2Name=dyn_C_haori_back_mstr, swing1LimitDegrees=10, swing2LimitDegrees=30, twistLimitDegrees=5}
 {bone1Name=dyn_C_haori_back_2, bone2Name=dyn_C_haori_back_1, swing1LimitDegrees=10, swing2LimitDegrees=30, twistLimitDegrees=5}
 {bone1Name=dyn_C_haori_back_3, bone2Name=dyn_C_haori_back_2, swing1LimitDegrees=10, swing2LimitDegrees=30, twistLimitDegrees=5}
 {bone1Name=dyn_C_haori_back_4, bone2Name=dyn_C_haori_back_3, swing1LimitDegrees=10, swing2LimitDegrees=30, twistLimitDegrees=5}
 {bone1Name=dyn_L_haori_front_1, bone2Name=dyn_L_haori_front_mstr, swing1LimitDegrees=10, swing2LimitDegrees=30, twistLimitDegrees=5}
 {bone1Name=dyn_L_haori_front_2, bone2Name=dyn_L_haori_front_1, swing1LimitDegrees=10, swing2LimitDegrees=30, twistLimitDegrees=5}
 {bone1Name=dyn_L_haori_front_3, bone2Name=dyn_L_haori_front_2, swing1LimitDegrees=10, swing2LimitDegrees=30, twistLimitDegrees=5}
 {bone1Name=dyn_L_haori_front_4, bone2Name=dyn_L_haori_front_3, swing1LimitDegrees=10, swing2LimitDegrees=30, twistLimitDegrees=5}
 {bone1Name=dyn_L_haori_side_1, bone2Name=dyn_L_haori_side_mstr, swing1LimitDegrees=10, swing2LimitDegrees=30, twistLimitDegrees=5}
 {bone1Name=dyn_L_haori_side_2, bone2Name=dyn_L_haori_side_1, swing1LimitDegrees=10, swing2LimitDegrees=30, twistLimitDegrees=5}
 {bone1Name=dyn_L_haori_side_3, bone2Name=dyn_L_haori_side_2, swing1LimitDegrees=10, swing2LimitDegrees=30, twistLimitDegrees=5}
 {bone1Name=dyn_L_haori_side_4, bone2Name=dyn_L_haori_side_3, swing1LimitDegrees=10, swing2LimitDegrees=30, twistLimitDegrees=5}
 {bone1Name=dyn_R_haori_front_1, bone2Name=dyn_R_haori_front_mstr, swing1LimitDegrees=10, swing2LimitDegrees=30, twistLimitDegrees=5}
 {bone1Name=dyn_R_haori_front_2, bone2Name=dyn_R_haori_front_1, swing1LimitDegrees=10, swing2LimitDegrees=30, twistLimitDegrees=5}
 {bone1Name=dyn_R_haori_front_3, bone2Name=dyn_R_haori_front_2, swing1LimitDegrees=10, swing2LimitDegrees=30, twistLimitDegrees=5}
 {bone1Name=dyn_R_haori_front_4, bone2Name=dyn_R_haori_front_3, swing1LimitDegrees=10, swing2LimitDegrees=30, twistLimitDegrees=5}
 {bone1Name=dyn_R_haori_side_1, bone2Name=dyn_R_haori_side_mstr, swing1LimitDegrees=10, swing2LimitDegrees=30, twistLimitDegrees=5}
 {bone1Name=dyn_R_haori_side_2, bone2Name=dyn_R_haori_side_1, swing1LimitDegrees=10, swing2LimitDegrees=30, twistLimitDegrees=5}
 {bone1Name=dyn_R_haori_side_3, bone2Name=dyn_R_haori_side_2, swing1LimitDegrees=10, swing2LimitDegrees=30, twistLimitDegrees=5}
 {bone1Name=dyn_R_haori_side_4, bone2Name=dyn_R_haori_side_3, swing1LimitDegrees=10, swing2LimitDegrees=30, twistLimitDegrees=5}
 {bone1Name=dyn_L_yukata_back_1, bone2Name=calf_l, swing1LimitDegrees=30, swing2LimitDegrees=5, twistLimitDegrees=0}
 {bone1Name=dyn_L_yukata_back_2, bone2Name=dyn_L_yukata_back_1, swing1LimitDegrees=30, swing2LimitDegrees=5, twistLimitDegrees=0}
 {bone1Name=dyn_R_yukata_back_1, bone2Name=calf_r, swing1LimitDegrees=30, swing2LimitDegrees=5, twistLimitDegrees=0}
 {bone1Name=dyn_R_yukata_back_2, bone2Name=dyn_R_yukata_back_1, swing1LimitDegrees=30, swing2LimitDegrees=5, twistLimitDegrees=0}
]
