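# Control Rig for the AtticDrill character's face accessory part, posing long hair chains, mainly back hair, against body motion.
# Head, clavicle, spine and pelvis readings become driver variables, some curve-shaped or smoothed.
# Weighted parent constraints tie hair control nulls to body and haori bones, alongside twist-bone and spline IK stages.
# Layered corrective offsets blend by those drivers and movement states like sliding, skydiving and sprinting, with tactical jumps branching separately.
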ControlRigBlueprint /BRCosmetics/Characters/Player/Female/Medium/Bodies/F_MED_AtticDrill/Meshes/Parts/F_MED_AtticDrill_FaceAcc_CtrlRig
graphs [
 CollapseNode_ContainedGraph {nodes [
   RigVMFunctionEntryNode()
   > RigVMFunction_Sequence = RigVMAggregateNode()
   GetAlphaFromQuatDelta = RigVMFunctionReferenceNode(
    Rotation Order=ZYX
    child=(Type=Bone,Name="head")
    debugRotation=false
    outputMinMax=(X=0,Y=1)
    parent=(Type=Null,Name="head_null")
    premult=1
    rotationMinMax=(X=0,Y=-20)
    triggerAxis=Z
    twistAxis=(X=1,Y=0,Z=0)
    ExecuteContext=@RigVMFunction_Sequence.A
   )
   n1 = &RigVMFunction_AnimEvalRichCurve(
    Curve="(EditorCurveData=(Keys=((InterpMode=RCIM_Cubic,TangentMode=RCTM_User,ArriveTangent=0.6623,LeaveTangent=0.6623),(InterpMode=RCIM_Cubic,TangentMode=RCTM_User,Time=0.5,Value=0.4,ArriveTangent=0.8898,LeaveTangent=0.8898),(InterpMode=RCIM_Cubic,TangentMode=RCTM ...(truncated)"
    SourceMinimum=0
    TargetMinimum=0
    Value=@GetAlphaFromQuatDelta.alpha
   )
   > RigVMVariableNode(Value=@n1.Result) {variable=HeadRotUp}
   > GetAlphaFromQuatDelta_1 = RigVMFunctionReferenceNode(
    Rotation Order=ZYX
    child=(Type=Bone,Name="head")
    debugRotation=false
    outputMinMax=(X=0,Y=1)
    parent=(Type=Null,Name="head_null")
    premult=1
    rotationMinMax=(X=-20,Y=-40)
    triggerAxis=Z
    twistAxis=(X=1,Y=0,Z=0)
   )
   > RigVMVariableNode(Value=@GetAlphaFromQuatDelta_1.alpha) {variable=HeadRotUpExtreme}
   > GetAlphaFromQuatDelta_2 = RigVMFunctionReferenceNode(
    Rotation Order=ZYX
    child=(Type=Bone,Name="head")
    outputMinMax=(X=0,Y=1)
    parent=(Type=Bone,Name="neck_01")
    premult=1
    rotationMinMax=(X=-20,Y=20)
    triggerAxis=Y
    twistAxis=(X=1,Y=0,Z=0)
   )
   > RigVMVariableNode(Value=@GetAlphaFromQuatDelta_2.alpha) {variable=HeadYrotation}
   > GetAlphaFromQuatDelta_3 = RigVMFunctionReferenceNode(
    Rotation Order=ZYX
    child=(Type=Bone,Name="head")
    debugRotation=false
    outputMinMax=(X=0,Y=1)
    parent=(Type=Bone,Name="neck_01")
    premult=1
    rotationMinMax=(X=-50,Y=-65)
    triggerAxis=Z
    twistAxis=(X=1,Y=0,Z=0)
   )
   n2 = &RigVMFunction_AlphaInterp(
    InterpSpeedDecreasing=5
    InterpSpeedIncreasing=25
    Scale=1
    bClampResult=true
    Value=&RigVMFunction_MathBoolToFloat(Value=RigVMVariableNode() {variable=Is Crouching}.Value).Result
   )
   Multiply = &RigVMFunction_MathDoubleMul(B=@n2.Result, A=@GetAlphaFromQuatDelta_3.alpha)
   > RigVMVariableNode(Value=@Multiply.Result) {variable=HeadLookingUpCrouching}
   > GetAlphaFromQuatDelta_9 = RigVMFunctionReferenceNode(
    Rotation Order=ZYX
    child=(Type=Bone,Name="head")
    debugRotation=false
    outputMinMax=(X=0,Y=1)
    parent=(Type=Null,Name="head_null")
    premult=1
    rotationMinMax=(X=0,Y=20)
    triggerAxis=Z
    twistAxis=(X=1,Y=0,Z=0)
   )
   > RigVMVariableNode(Value=@GetAlphaFromQuatDelta_9.alpha) {variable=HeadRotDown}
   GetAlphaFromQuatDelta_5 = RigVMFunctionReferenceNode(
    Rotation Order=ZYX
    child=(Type=Bone,Name="clavicle_l")
    outputMinMax=(X=0,Y=1)
    parent=(Type=Null,Name="clavicle_l_null")
    premult=1
    rotationMinMax=(X=0,Y=60)
    triggerAxis=Z
    twistAxis=(X=1,Y=0,Z=0)
    ExecuteContext=@RigVMFunction_Sequence.B
   )
   > RigVMVariableNode(Value=@GetAlphaFromQuatDelta_5.alpha) {variable=Clavicle_front_rotation_l}
   > GetAlphaFromQuatDelta_6 = RigVMFunctionReferenceNode(
    Rotation Order=ZYX
    child=(Type=Bone,Name="clavicle_r")
    debugRotation=false
    outputMinMax=(X=0,Y=1)
    parent=(Type=Null,Name="clavicle_r_null")
    premult=1
    rotationMinMax=(X=0,Y=60)
    triggerAxis=Z
    twistAxis=(X=1,Y=0,Z=0)
   )
   > RigVMVariableNode(Value=@GetAlphaFromQuatDelta_6.alpha) {variable=Clavicle_front_rotation_r}
   > GetAlphaFromQuatDelta_7 = RigVMFunctionReferenceNode(
    Rotation Order=ZYX
    child=(Type=Bone,Name="clavicle_l")
    outputMinMax=(X=0,Y=1)
    parent=(Type=Null,Name="clavicle_l_null")
    premult=1
    rotationMinMax=(X=0,Y=-60)
    triggerAxis=Y
    twistAxis=(X=1,Y=0,Z=0)
   )
   > RigVMVariableNode(Value=@GetAlphaFromQuatDelta_7.alpha) {variable=Clavicle_up_rotation_l}
   > GetAlphaFromQuatDelta_8 = RigVMFunctionReferenceNode(
    Rotation Order=ZYX
    child=(Type=Bone,Name="clavicle_r")
    debugRotation=false
    outputMinMax=(X=0,Y=1)
    parent=(Type=Null,Name="clavicle_r_null")
    premult=1
    rotationMinMax=(X=0,Y=-60)
    triggerAxis=Y
    twistAxis=(X=1,Y=0,Z=0)
   )
   > RigVMVariableNode(Value=@GetAlphaFromQuatDelta_8.alpha) {variable=Clavicle_up_rotation_r}
   > GetAlphaFromQuatDelta_10 = RigVMFunctionReferenceNode(
    Rotation Order=ZYX
    child=(Type=Bone,Name="clavicle_r")
    debugRotation=false
    outputMinMax=(X=0,Y=1)
    parent=(Type=Null,Name="clavicle_r_null")
    premult=1
    rotationMinMax=(X=0,Y=45)
    triggerAxis=Y
    twistAxis=(X=1,Y=0,Z=0)
   )
   RigVMVariableNode(
    Value=&RigVMFunction_MathDoubleRemap(
     SourceMaximum=65
     SourceMinimum=44
     TargetMaximum=0
     TargetMinimum=1
     Value=&RigUnit_GetRelativeTransformForItem(Child=(Type=Bone,Name="pelvis"), Parent=(Type=Bone,Name="root"), bChildInitial=False, bParentInitial=False).RelativeTransform.Translation.Z
    ).Result
    ExecuteContext=@RigVMFunction_Sequence.C
   ) {variable=CrouchingInPlace}
   n3 = &RigVMFunction_AnimEvalRichCurve(
    Curve="(EditorCurveData=(Keys=((InterpMode=RCIM_Cubic,TangentMode=RCTM_Break),(InterpMode=RCIM_Cubic,TangentMode=RCTM_Break,Time=0.45,ArriveTangent=0.0163,LeaveTangent=1.4085),(Time=0.6,Value=1),(Time=0.74,Value=1),(InterpMode=RCIM_Cubic,TangentMode=RCTM_Break,Ti ...(truncated)"
    SourceMinimum=0
    TargetMinimum=0
    Value=&RigVMFunction_MathDoubleRemap(
     SourceMaximum=100
     SourceMinimum=0
     TargetMaximum=1
     TargetMinimum=0
     Value=&RigUnit_GetRelativeTransformForItem(Child=(Type=Bone,Name="pelvis"), Parent=(Type=Bone,Name="root"), bChildInitial=False, bParentInitial=False).RelativeTransform.Translation.Z
    ).Result
   )
   > RigVMVariableNode(Value=@n3.Result) {variable=CrouchingMovingInPlace}
   > GetAlphaFromQuatDelta_4 = RigVMFunctionReferenceNode(
    Rotation Order=ZYX
    child=(Type=Bone,Name="spine_05")
    onlyTwist=true
    outputMinMax=(X=0,Y=1)
    parent=(Type=Bone,Name="spine_03")
    premult=1
    rotationMinMax=(X=0,Y=35)
    twistAxis=(X=1,Y=0,Z=0)
   )
   n4 = &RigVMFunction_AnimEvalRichCurve(
    Curve="(EditorCurveData=(Keys=((InterpMode=RCIM_Cubic,TangentMode=RCTM_User,ArriveTangent=0.6623,LeaveTangent=0.6623),(InterpMode=RCIM_Cubic,TangentMode=RCTM_User,Time=0.5,Value=0.4,ArriveTangent=0.8898,LeaveTangent=0.8898),(InterpMode=RCIM_Cubic,TangentMode=RCTM ...(truncated)"
    SourceMinimum=0
    TargetMinimum=0
    Value=@GetAlphaFromQuatDelta_4.alpha
   )
   > RigVMVariableNode(Value=@n4.Result) {variable=TorsoTwistRight}
   Remap = &RigVMFunction_MathDoubleRemap(
    SourceMaximum=1
    SourceMinimum=0
    TargetMaximum=0
    TargetMinimum=1
    bClamp=False
    Value=&RigVMFunction_MathBoolToFloat(Value=RigVMVariableNode() {variable=Is Jumping}.Value).Result
   )
   n5 = &RigVMFunction_MathBoolToFloat(Value=RigVMVariableNode() {variable=is floating}.Value)
   > RigVMVariableNode(Value=&RigVMFunction_MathDoubleMul(
     A=@Remap.Result
     B=&RigVMFunction_MathDoubleRemap(SourceMaximum=1, SourceMinimum=0, TargetMaximum=0.5, TargetMinimum=0, bClamp=False, Value=@n5.Result).Result
    ).Result) {variable=ArmBackFix}
   RigVMFunctionReturnNode(ExecuteContext=@RigVMFunction_Sequence.D)
  ]}
 CollapseNode_ContainedGraph {nodes [
   n1 = RigVMFunctionEntryNode()
   RigVMFunction_ControlFlowBranch = &RigVMFunction_ControlFlowBranch(ExecuteContext=@n1.ExecuteContext, Condition=RigVMVariableNode() {variable=Pub_IsTacticalJumps}.Value)
   RigVMFunctionReturnNode(ExecuteContext=@RigVMFunction_ControlFlowBranch.Completed)
   GetAlphaFromQuatDelta_11 = RigVMFunctionReferenceNode(
    Rotation Order=ZYX
    absoluteValue=false
    child=(Type=Bone,Name="spine_05")
    debugRotation=false
    onlyTwist=false
    outputMinMax=(X=0,Y=1)
    parent=(Type=Null,Name="spine_05_null")
    premult=1
    rotationMinMax=(X=50,Y=-75)
    triggerAxis=Z
    twistAxis=(X=1,Y=0,Z=0)
    ExecuteContext=@RigVMFunction_ControlFlowBranch.False
   )
   n2 = RigVMVariableNode() {variable=Pub_IsParachuteOpened}
   > &RigUnit_ModifyTransforms(
    ItemToModify="((Item=(Type=Bone,Name=\"dyn_R_haori_side_mstr\"),Transform=(Rotation=(X=0.0921,Y=-0.1134,Z=0.3696,W=-0.9176),Translation=(X=0,Y=0,Z=0),Scale3D=(X=1,Y=1,Z=1))),(Item=(Type=Bone,Name=\"dyn_C_haori_back_mstr\"),Transform=(Rotation=(X=0,Y=0.501,Z=0,W=0.8654),Tran ...(truncated)"
    WeightMinimum=0
    Weight=&RigVMFunction_MathDoubleMul(A=@GetAlphaFromQuatDelta_11.alpha, B=&RigVMFunction_MathBoolToFloat(Value=@n2.Value).Result).Result
   )
   > Constraints = RigVMCollapseNode()
   RigVMFunction_Sequence = &RigVMFunction_Sequence(ExecuteContext=@Constraints.ExecuteContext)
   TwistBones = RigVMCollapseNode(ExecuteContext=@RigVMFunction_Sequence.A)
   > IKSpline = RigVMCollapseNode()
   > RigVMFunction_Sequence_1 = RigVMAggregateNode()
   n3 = RigVMVariableNode() {variable=TorsoTwistRight}
   &RigUnit_ModifyTransforms(
    ItemToModify="((Item=(Type=Null,Name=\"dyn_backHair_r_04_ctrl_null\"),Transform=(Rotation=(X=0,Y=0,Z=0,W=1),Translation=(X=0,Y=-2.112,Z=2),Scale3D=(X=1,Y=1,Z=1))),(Item=(Type=Null,Name=\"dyn_backHair_r_05_ctrl_null\"),Transform=(Rotation=(X=0,Y=0,Z=0,W=-1),Translation=(X=0, ...(truncated)"
    WeightMinimum=0
    ExecutePin=@RigVMFunction_Sequence_1.B
    Weight=@n3.Value
   )
   > &RigUnit_ModifyTransforms(
    ItemToModify="((Item=(Type=Null,Name=\"dyn_backHair_r_03_ctrl_null\"),Transform=(Rotation=(X=0,Y=0,Z=0,W=1),Translation=(X=0,Y=0,Z=3),Scale3D=(X=1,Y=1,Z=1))),(Item=(Type=Null,Name=\"dyn_backHair_r_02_ctrl_null\"),Transform=(Rotation=(X=0,Y=0,Z=0,W=1),Translation=(X=0,Y=0,Z= ...(truncated)"
    WeightMinimum=0
    Weight=@n3.Value
   )
   > GetAlphaFromQuatDelta_2 = RigVMFunctionReferenceNode(
    Rotation Order=ZYX
    child=(Type=Bone,Name="spine_05")
    onlyTwist=true
    outputMinMax=(X=0,Y=1)
    parent=(Type=Bone,Name="spine_03")
    premult=1
    rotationMinMax=(X=0,Y=-35)
    twistAxis=(X=1,Y=0,Z=0)
   )
   > &RigUnit_ModifyTransforms(
    ItemToModify="((Item=(Type=Null,Name=\"dyn_backHair_l_04_ctrl_null\"),Transform=(Rotation=(X=0,Y=0,Z=0,W=1),Translation=(X=0,Y=-10,Z=6),Scale3D=(X=1,Y=1,Z=1))),(Item=(Type=Null,Name=\"dyn_backHair_l_05_ctrl_null\"),Transform=(Rotation=(X=0,Y=0,Z=0,W=-1),Translation=(X=0,Y=- ...(truncated)"
    WeightMinimum=0
    Weight=@GetAlphaFromQuatDelta_2.alpha
   )
   > GetAlphaFromQuatDelta_3 = RigVMFunctionReferenceNode(
    Rotation Order=ZYX
    child=(Type=Bone,Name="clavicle_l")
    debugRotation=false
    outputMinMax=(X=0,Y=1)
    parent=(Type=Null,Name="clavicle_l_null")
    premult=1
    rotationMinMax=(X=0,Y=-45)
    triggerAxis=Z
    twistAxis=(X=1,Y=0,Z=0)
   )
   Remap = &RigVMFunction_MathDoubleRemap(
    SourceMaximum=1
    SourceMinimum=0
    TargetMaximum=0
    TargetMinimum=1
    bClamp=False
    Value=&RigVMFunction_MathBoolToFloat(Value=RigVMVariableNode() {variable=is floating}.Value).Result
   )
   n4 = &RigVMFunction_MathDoubleMul(A=@GetAlphaFromQuatDelta_3.alpha, B=@Remap.Result)
   > &RigUnit_ModifyTransforms(
    ItemToModify="((Item=(Type=Null,Name=\"dyn_backHair_l_03_ctrl_null\"),Transform=(Rotation=(X=0,Y=0,Z=0,W=1),Translation=(X=-1,Y=-4.2739,Z=11.8897),Scale3D=(X=1,Y=1,Z=1))),(Item=(Type=Null,Name=\"dyn_backHair_l_04_ctrl_null\"),Transform=(Rotation=(X=0,Y=0,Z=0,W=-1),Translati ...(truncated)"
    WeightMinimum=0
    Weight=@n4.Result
   )
   > &RigUnit_ModifyTransforms(
    ItemToModify="((Item=(Type=Null,Name=\"dyn_backHair_c_03_ctrl_null\"),Transform=(Rotation=(X=0,Y=0,Z=0,W=1),Translation=(X=0,Y=-3,Z=0),Scale3D=(X=1,Y=1,Z=1))),(Item=(Type=Null,Name=\"dyn_backHair_c_04_ctrl_null\"),Transform=(Rotation=(X=0,Y=0,Z=0,W=1),Translation=(X=0,Y=-1, ...(truncated)"
    WeightMinimum=0
    Weight=@n4.Result
   )
   > GetAlphaFromQuatDelta_4 = RigVMFunctionReferenceNode(
    Rotation Order=ZYX
    child=(Type=Bone,Name="clavicle_r")
    debugRotation=false
    outputMinMax=(X=0,Y=1)
    parent=(Type=Null,Name="clavicle_r_null")
    premult=1
    rotationMinMax=(X=0,Y=-45)
    triggerAxis=Z
    twistAxis=(X=1,Y=0,Z=0)
   )
   n5 = RigVMVariableNode() {variable=is floating}
   n6 = &RigVMFunction_MathDoubleMul(
    A=@GetAlphaFromQuatDelta_4.alpha
    B=&RigVMFunction_MathDoubleRemap(
     SourceMaximum=1
     SourceMinimum=0
     TargetMaximum=0
     TargetMinimum=1
     bClamp=False
     Value=&RigVMFunction_MathBoolToFloat(Value=@n5.Value).Result
    ).Result
   )
   > &RigUnit_ModifyTransforms(
    ItemToModify="((Item=(Type=Null,Name=\"dyn_backHair_r_03_ctrl_null\"),Transform=(Translation=(X=0,Y=4,Z=10),Scale3D=(X=1,Y=1,Z=1))),(Item=(Type=Null,Name=\"dyn_backHair_r_04_ctrl_null\"),Transform=(Rotation=(X=0,Y=0,Z=0,W=1),Translation=(X=0,Y=4,Z=12),Scale3D=(X=1,Y=1,Z=1)) ...(truncated)"
    WeightMinimum=0
    Weight=@n6.Result
   )
   > &RigUnit_ModifyTransforms(
    ItemToModify="((Item=(Type=Null,Name=\"dyn_backHair_c_04_ctrl_null\"),Transform=(Rotation=(X=0,Y=0,Z=0,W=1),Translation=(X=0,Y=2,Z=5),Scale3D=(X=1,Y=1,Z=1))),(Item=(Type=Null,Name=\"dyn_backHair_c_03_ctrl_null\"),Transform=(Rotation=(X=0,Y=0,Z=0,W=1),Translation=(X=0,Y=1,Z= ...(truncated)"
    WeightMinimum=0
    Weight=@n6.Result
   )
   > &RigUnit_ModifyTransforms(
    ItemToModify="((Item=(Type=Null,Name=\"dyn_backHair_l_03_ctrl_null\"),Transform=(Rotation=(X=0,Y=0,Z=0,W=-1),Translation=(X=0,Y=4,Z=3),Scale3D=(X=1,Y=1,Z=1))),(Item=(Type=Null,Name=\"dyn_backHair_l_04_ctrl_null\"),Transform=(Rotation=(X=0,Y=0,Z=0,W=-1),Translation=(X=0,Y=-1 ...(truncated)"
    WeightMinimum=0
    Weight=RigVMVariableNode() {variable=Clavicle_up_rotation_l}.Value
   )
   > &RigUnit_ModifyTransforms(
    ItemToModify="((Item=(Type=Null,Name=\"dyn_backHair_r_03_ctrl_null\"),Transform=(Rotation=(X=0,Y=0,Z=0,W=-1),Translation=(X=0,Y=-4,Z=6),Scale3D=(X=1,Y=1,Z=1))),(Item=(Type=Null,Name=\"dyn_backHair_r_04_ctrl_null\"),Transform=(Rotation=(X=0,Y=0,Z=0,W=-1),Translation=(X=0,Y=1 ...(truncated)"
    WeightMinimum=0
    Weight=RigVMVariableNode() {variable=Clavicle_up_rotation_r}.Value
   )
   GetAlphaFromQuatDelta_5 = RigVMFunctionReferenceNode(
    Rotation Order=ZYX
    absoluteValue=false
    child=(Type=Bone,Name="upperarm_r")
    debugRotation=false
    outputMinMax=(X=0,Y=1)
    parent=(Type=Null,Name="upperarm_r_null")
    premult=1
    rotationMinMax=(X=0,Y=-30)
    triggerAxis=Z
    twistAxis=(X=1,Y=0,Z=0)
    ExecuteContext=@RigVMFunction_Sequence_1.C
   )
   n7 = &RigVMFunction_AnimEvalRichCurve(
    Curve="(EditorCurveData=(Keys=((InterpMode=RCIM_Cubic,TangentMode=RCTM_User,ArriveTangent=0.6623,LeaveTangent=0.6623),(InterpMode=RCIM_Cubic,TangentMode=RCTM_User,Time=0.5,Value=0.4,ArriveTangent=0.8898,LeaveTangent=0.8898),(InterpMode=RCIM_Cubic,TangentMode=RCTM ...(truncated)"
    SourceMinimum=0
    TargetMinimum=0
    Value=&RigVMFunction_MathDoubleMul(A=@GetAlphaFromQuatDelta_5.alpha, B=RigVMVariableNode() {variable=ArmBackFix}.Value).Result
   )
   > &RigUnit_ModifyTransforms(
    ItemToModify="((Item=(Type=Null,Name=\"dyn_backHair_r_01_ctrl_null\"),Transform=(Rotation=(X=0,Y=0,Z=0,W=1),Translation=(X=0,Y=5,Z=7),Scale3D=(X=1,Y=1,Z=1))),(Item=(Type=Null,Name=\"dyn_backHair_r_02_ctrl_null\"),Transform=(Rotation=(X=0,Y=0,Z=0,W=1),Translation=(X=0,Y=5,Z= ...(truncated)"
    WeightMinimum=0
    Weight=@n7.Result
   )
   > GetAlphaFromQuatDelta_6 = RigVMFunctionReferenceNode(
    Rotation Order=ZYX
    absoluteValue=false
    child=(Type=Bone,Name="upperarm_l")
    debugRotation=false
    outputMinMax=(X=0,Y=1)
    parent=(Type=Null,Name="upperarm_l_null")
    premult=1
    rotationMinMax=(X=0,Y=-30)
    triggerAxis=Z
    twistAxis=(X=1,Y=0,Z=0)
   )
   n8 = &RigVMFunction_AnimEvalRichCurve(
    Curve="(EditorCurveData=(Keys=((InterpMode=RCIM_Cubic,TangentMode=RCTM_User,ArriveTangent=0.6623,LeaveTangent=0.6623),(InterpMode=RCIM_Cubic,TangentMode=RCTM_User,Time=0.5,Value=0.4,ArriveTangent=0.8898,LeaveTangent=0.8898),(InterpMode=RCIM_Cubic,TangentMode=RCTM ...(truncated)"
    SourceMinimum=0
    TargetMinimum=0
    Value=&RigVMFunction_MathDoubleMul(A=@GetAlphaFromQuatDelta_6.alpha, B=RigVMVariableNode() {variable=ArmBackFix}.Value).Result
   )
   > &RigUnit_ModifyTransforms(
    ItemToModify="((Item=(Type=Null,Name=\"dyn_backHair_l_01_ctrl_null\"),Transform=(Rotation=(X=0,Y=0,Z=0,W=1),Translation=(X=0,Y=0,Z=3),Scale3D=(X=1,Y=1,Z=1))),(Item=(Type=Null,Name=\"dyn_backHair_l_02_ctrl_null\"),Transform=(Rotation=(X=0,Y=0,Z=0,W=1),Translation=(X=0,Y=0,Z= ...(truncated)"
    WeightMinimum=0
    Weight=@n8.Result
   )
   n9 = RigVMVariableNode() {variable=WearingCape}
   RigVMFunction_MathBoolAnd = &RigVMFunction_MathBoolAnd(A=RigVMVariableNode() {variable=Is Jumping}.Value, B=@n9.Value)
   AlphaInterp = &RigVMFunction_AlphaInterp(
    InterpSpeedDecreasing=15
    InterpSpeedIncreasing=15
    Scale=1
    bClampResult=False
    Value=RigVMDispatchNode(False=0, True=1, Condition=@RigVMFunction_MathBoolAnd.Result) {template="DISPATCH_RigVMDispatch_If(in Condition,in True,in False,out Result)"}.Result
   )
   > &RigUnit_ModifyTransforms(
    ItemToModify="((Item=(Type=Null,Name=\"dyn_backHair_l_06_ctrl_null\"),Transform=(Rotation=(X=0,Y=0,Z=0,W=1),Translation=(X=0,Y=0,Z=-10),Scale3D=(X=1,Y=1,Z=1))),(Item=(Type=Null,Name=\"dyn_backHair_c_06_ctrl_null\"),Transform=(Rotation=(X=0,Y=0,Z=0,W=1),Translation=(X=0,Y=0, ...(truncated)"
    WeightMinimum=0
    Weight=@AlphaInterp.Result
   )
   RigVMFunction_MathBoolToFloat = &RigVMFunction_MathBoolToFloat(Value=RigVMVariableNode() {variable=Pub_Sprinting}.Value)
   > &RigUnit_ModifyTransforms(
    ItemToModify="((Item=(Type=Null,Name=\"dyn_backHair_l_06_ctrl_null\"),Transform=(Rotation=(X=0,Y=0,Z=0,W=1),Translation=(X=0,Y=0,Z=-5),Scale3D=(X=1,Y=1,Z=1))),(Item=(Type=Null,Name=\"dyn_backHair_c_06_ctrl_null\"),Transform=(Rotation=(X=0,Y=0,Z=0,W=1),Translation=(X=0,Y=0,Z ...(truncated)"
    WeightMinimum=0
    Weight=&RigVMFunction_MathDoubleMul(A=@RigVMFunction_MathBoolToFloat.Result, B=&RigVMFunction_MathBoolToFloat(Value=@n9.Value).Result).Result
   )
   > GetAlphaFromQuatDelta_7 = RigVMFunctionReferenceNode(
    Rotation Order=ZYX
    absoluteValue=false
    child=(Type=Bone,Name="spine_05")
    debugRotation=false
    onlyTwist=true
    outputMinMax=(X=0,Y=1)
    parent=(Type=Null,Name="spine_05_null")
    premult=1
    rotationMinMax=(X=0,Y=-30)
    triggerAxis=X
    twistAxis=(X=1,Y=0,Z=0)
   )
   n10 = &RigVMFunction_MathBoolToFloat(Value=RigVMVariableNode() {variable=Pub_TacticalSprint}.Value)
   > &RigUnit_ModifyTransforms(
    ItemToModify="((Item=(Type=Null,Name=\"dyn_backHair_l_03_ctrl_null\"),Transform=(Rotation=(X=0,Y=0,Z=0,W=1),Translation=(X=0,Y=-10,Z=10),Scale3D=(X=1,Y=1,Z=1))),(Item=(Type=Null,Name=\"dyn_backHair_l_04_ctrl_null\"),Transform=(Rotation=(X=0,Y=0,Z=0,W=-1),Translation=(X=0,Y= ...(truncated)"
    WeightMinimum=0
    Weight=&RigVMFunction_MathDoubleMul(A=@GetAlphaFromQuatDelta_7.alpha, B=@n10.Result).Result
   )
   > GetAlphaFromQuatDelta_8 = RigVMFunctionReferenceNode(
    Rotation Order=ZYX
    absoluteValue=false
    child=(Type=Bone,Name="spine_05")
    debugRotation=false
    onlyTwist=true
    outputMinMax=(X=0,Y=1)
    parent=(Type=Null,Name="spine_05_null")
    premult=1
    rotationMinMax=(X=0,Y=30)
    triggerAxis=X
    twistAxis=(X=1,Y=0,Z=0)
   )
   > &RigUnit_ModifyTransforms(
    ItemToModify="((Item=(Type=Null,Name=\"dyn_backHair_r_03_ctrl_null\"),Transform=(Rotation=(X=0,Y=0,Z=0,W=-1),Translation=(X=0,Y=10,Z=10),Scale3D=(X=1,Y=1,Z=1))),(Item=(Type=Null,Name=\"dyn_backHair_r_04_ctrl_null\"),Transform=(Rotation=(X=0,Y=0,Z=0,W=-1),Translation=(X=0,Y= ...(truncated)"
    WeightMinimum=0
    Weight=&RigVMFunction_MathDoubleMul(A=@GetAlphaFromQuatDelta_8.alpha, B=@n10.Result).Result
   )
   n11 = &RigVMFunction_AlphaInterp(
    InterpSpeedDecreasing=5
    InterpSpeedIncreasing=1
    Scale=1
    bClampResult=False
    Value=&RigVMFunction_MathBoolToFloat(Value=RigVMVariableNode() {variable=is floating}.Value).Result
   )
   > &RigUnit_ModifyTransforms(
    ItemToModify="((Item=(Type=Null,Name=\"dyn_backHair_r_04_ctrl_null\"),Transform=(Rotation=(X=0,Y=0,Z=0,W=1),Translation=(X=0,Y=0,Z=10),Scale3D=(X=1,Y=1,Z=1))),(Item=(Type=Null,Name=\"dyn_backHair_r_01_ctrl_null\"),Transform=(Rotation=(X=0,Y=0,Z=0,W=1),Translation=(X=0,Y=0,Z ...(truncated)"
    WeightMinimum=0
    Weight=@n11.Result
   )
   n12 = RigVMVariableNode() {variable=Pub_WearingPack}
   RigVMFunction_MathBoolNot = &RigVMFunction_MathBoolNot(Value=@n12.Value)
   RigVMFunction_MathBoolOr = RigVMAggregateNode(
    A=RigVMVariableNode() {variable=Pub_kneeSliding}.Value
    B=RigVMVariableNode() {variable=Pub_BasebalSliding}.Value
    C=RigVMVariableNode() {variable=Pub_IsParachuteOpened}.Value
   )
   > &RigUnit_ModifyTransforms(
    ItemToModify="((Item=(Type=Null,Name=\"dyn_backHair_r_06_ctrl_null\"),Transform=(Rotation=(X=0,Y=0,Z=0,W=1),Translation=(X=0,Y=0,Z=5),Scale3D=(X=1,Y=1,Z=1))),(Item=(Type=Null,Name=\"dyn_backHair_c_06_ctrl_null\"),Transform=(Rotation=(X=0,Y=0,Z=0,W=1),Translation=(X=0,Y=0,Z= ...(truncated)"
    WeightMinimum=0
    Weight=&RigVMFunction_MathBoolToFloat(Value=&RigVMFunction_MathBoolAnd(B=@RigVMFunction_MathBoolNot.Result, A=@RigVMFunction_MathBoolOr.Result).Result).Result
   )
   n13 = RigVMVariableNode() {variable=Pub_kneeSliding}
   > &RigUnit_ModifyTransforms(
    ItemToModify="((Item=(Type=Null,Name=\"dyn_backHair_r_06_ctrl_null\"),Transform=(Rotation=(X=0,Y=0,Z=0,W=1),Translation=(X=0,Y=0,Z=2.5),Scale3D=(X=1,Y=1,Z=1))),(Item=(Type=Null,Name=\"dyn_backHair_c_06_ctrl_null\"),Transform=(Rotation=(X=0,Y=0,Z=0,W=1),Translation=(X=0,Y=0, ...(truncated)"
    WeightMinimum=0
    Weight=&RigVMFunction_MathBoolToFloat(Value=&RigVMFunction_MathBoolAnd(B=@n12.Value, A=@n13.Value).Result).Result
   )
   n14 = RigVMVariableNode() {variable=Pub_BasebalSliding}
   n15 = RigVMVariableNode() {variable=Pub_WearingPack}
   > &RigUnit_ModifyTransforms(
    ItemToModify="((Item=(Type=Null,Name=\"dyn_backHair_r_06_ctrl_null\"),Transform=(Rotation=(X=0,Y=0,Z=0,W=1),Translation=(X=0,Y=0,Z=10),Scale3D=(X=1,Y=1,Z=1))),(Item=(Type=Null,Name=\"dyn_backHair_c_06_ctrl_null\"),Transform=(Rotation=(X=0,Y=0,Z=0,W=1),Translation=(X=0,Y=0,Z ...(truncated)"
    WeightMinimum=0
    Weight=&RigVMFunction_MathBoolToFloat(Value=&RigVMFunction_MathBoolAnd(A=@n14.Value, B=@n15.Value).Result).Result
   )
   > GetAlphaFromQuatDelta_9 = RigVMFunctionReferenceNode(
    Rotation Order=ZYX
    absoluteValue=false
    child=(Type=Bone,Name="head")
    debugRotation=false
    onlyTwist=false
    outputMinMax=(X=0,Y=1)
    parent=(Type=Null,Name="head_null")
    premult=1
    rotationMinMax=(X=0,Y=10)
    triggerAxis=Z
    twistAxis=(X=1,Y=0,Z=0)
   )
   > &RigUnit_ModifyTransforms(
    ItemToModify=((Item=(Type=Null,Name="dyn_backHair_r_04_ctrl_null"),Transform=(Rotation=(X=0,Y=0,Z=0,W=1),Translation=(X=0,Y=0,Z=10),Scale3D=(X=1,Y=1,Z=1))))
    WeightMinimum=0
    Weight=@GetAlphaFromQuatDelta_9.alpha
   )
   > GetAlphaFromQuatDelta_10 = RigVMFunctionReferenceNode(
    Rotation Order=ZYX
    absoluteValue=false
    child=(Type=Bone,Name="head")
    debugRotation=false
    onlyTwist=false
    outputMinMax=(X=0,Y=1)
    parent=(Type=Null,Name="head_null")
    premult=1
    rotationMinMax=(X=0,Y=-8)
    triggerAxis=Z
    twistAxis=(X=1,Y=0,Z=0)
   )
   > &RigUnit_ModifyTransforms(
    ItemToModify="((Item=(Type=Null,Name=\"dyn_backHair_r_06_ctrl_null\"),Transform=(Rotation=(X=0,Y=0,Z=0,W=1),Translation=(X=0,Y=0,Z=3),Scale3D=(X=1,Y=1,Z=1))),(Item=(Type=Null,Name=\"dyn_backHair_c_06_ctrl_null\"),Transform=(Rotation=(X=0,Y=0,Z=0,W=1),Translation=(X=0,Y=0,Z= ...(truncated)"
    WeightMinimum=0
    Weight=@GetAlphaFromQuatDelta_10.alpha
   )
   n16 = RigVMVariableNode() {variable=Pub_WearingPack}
   n17 = RigVMVariableNode() {variable=Pub_IsSkydive}
   > &RigUnit_ModifyTransforms(
    ItemToModify="((Item=(Type=Null,Name=\"dyn_backHair_r_06_ctrl_null\"),Transform=(Rotation=(X=0,Y=0,Z=0,W=1),Translation=(X=0,Y=0,Z=6),Scale3D=(X=1,Y=1,Z=1))),(Item=(Type=Null,Name=\"dyn_backHair_c_06_ctrl_null\"),Transform=(Rotation=(X=0,Y=0,Z=0,W=1),Translation=(X=0,Y=0,Z= ...(truncated)"
    WeightMinimum=0
    Weight=&RigVMFunction_MathDoubleMul(
     B=&RigVMFunction_MathBoolToFloat(Value=@n16.Value).Result
     A=&RigVMFunction_MathBoolToFloat(Value=@n17.Value).Result
    ).Result
   )
   RigVMFunction_MathBoolOr_1 = RigVMAggregateNode(
    A=RigVMVariableNode() {variable=Pub_kneeSliding}.Value
    B=RigVMVariableNode() {variable=Pub_BasebalSliding}.Value
    C=RigVMVariableNode() {variable=Pub_Sprinting}.Value
    D=RigVMVariableNode() {variable=Pub_TacticalSprint}.Value
   )
   n18 = &RigVMFunction_AlphaInterp(
    InterpSpeedDecreasing=5
    InterpSpeedIncreasing=5
    Scale=1
    bClampResult=False
    Value=&RigVMFunction_MathBoolToFloat(Value=&RigVMFunction_MathBoolNot(Value=@RigVMFunction_MathBoolOr_1.Result).Result).Result
   )
   > &RigUnit_ModifyTransforms(
    ItemToModify="((Item=(Type=Null,Name=\"dyn_backHair_l_03_ctrl_null\"),Transform=(Rotation=(X=0,Y=0,Z=0,W=1),Translation=(X=0,Y=0,Z=-5),Scale3D=(X=1,Y=1,Z=1))),(Item=(Type=Null,Name=\"dyn_backHair_c_03_ctrl_null\"),Transform=(Rotation=(X=0,Y=0,Z=0,W=1),Translation=(X=0,Y=0,Z ...(truncated)"
    WeightMinimum=0
    Weight=@n18.Result
   )
   > &RigUnit_ModifyTransforms(
    ItemToModify="((Item=(Type=Null,Name=\"dyn_backHair_l_04_ctrl_null\"),Transform=(Rotation=(X=0,Y=0,Z=0,W=1),Translation=(X=0,Y=0,Z=-5),Scale3D=(X=1,Y=1,Z=1))),(Item=(Type=Null,Name=\"dyn_backHair_c_04_ctrl_null\"),Transform=(Rotation=(X=0,Y=0,Z=0,W=1),Translation=(X=0,Y=0,Z ...(truncated)"
    WeightMinimum=0
    Weight=@n18.Result
   )
   n19 = RigVMVariableNode() {variable=Pub_TacticalSprint}
   n20 = RigVMVariableNode() {variable=Pub_WearingPack}
   n21 = &RigVMFunction_AlphaInterp(
    InterpSpeedDecreasing=10
    InterpSpeedIncreasing=10
    Scale=1
    bClampResult=False
    Value=&RigVMFunction_MathBoolToFloat(Value=&RigVMFunction_MathBoolAnd(A=@n19.Value, B=@n20.Value).Result).Result
   )
   > &RigUnit_ModifyTransforms(
    ItemToModify="((Item=(Type=Null,Name=\"dyn_backHair_l_06_ctrl_null\"),Transform=(Rotation=(X=0,Y=0,Z=0,W=1),Translation=(X=0,Y=0,Z=6),Scale3D=(X=1,Y=1,Z=1))),(Item=(Type=Null,Name=\"dyn_backHair_c_06_ctrl_null\"),Transform=(Rotation=(X=0,Y=0,Z=0,W=1),Translation=(X=0,Y=0,Z= ...(truncated)"
    WeightMinimum=0
    Weight=@n21.Result
   )
   n22 = &RigVMFunction_AlphaInterp(
    InterpSpeedDecreasing=5
    InterpSpeedIncreasing=5
    Scale=1
    bClampResult=False
    Value=&RigVMFunction_MathBoolToFloat(Value=RigVMVariableNode() {variable=Pub_TacticalSprint}.Value).Result
   )
   > &RigUnit_ModifyTransforms(
    ItemToModify="((Item=(Type=Bone,Name=\"dyn_backHair_l_02\"),Transform=(Rotation=(X=0.1714,Y=0.0505,Z=0.0353,W=0.9833),Translation=(X=-5,Y=0,Z=15),Scale3D=(X=1,Y=1,Z=1))),(Item=(Type=Bone,Name=\"dyn_backHair_c_02\"),Transform=(Rotation=(X=0,Y=0,Z=0.0262,W=0.9997),Translation ...(truncated)"
    WeightMinimum=0
    Weight=@n22.Result
   )
   > &RigUnit_ModifyTransforms(
    ItemToModify="((Item=(Type=Bone,Name=\"dyn_backHair_r_02\"),Transform=(Rotation=(X=0,Y=0,Z=-0.0436,W=0.999),Translation=(X=0,Y=0,Z=15),Scale3D=(X=1,Y=1,Z=1))),(Item=(Type=Bone,Name=\"dyn_backHair_c_02\"),Transform=(Rotation=(X=0,Y=0,Z=-0.0262,W=0.9997),Translation=(X=0,Y=0, ...(truncated)"
    WeightMinimum=0
    Weight=@n22.Result
   )
   > GetAlphaFromQuatDelta_12 = RigVMFunctionReferenceNode(
    Rotation Order=ZYX
    absoluteValue=false
    child=(Type=Bone,Name="upperarm_l")
    debugRotation=false
    onlyTwist=false
    outputMinMax=(X=0,Y=1)
    parent=(Type=Null,Name="upperarm_l_null")
    premult=1
    rotationMinMax=(X=0,Y=10)
    triggerAxis=Z
    twistAxis=(X=1,Y=0,Z=0)
   )
   n23 = &RigVMFunction_MathBoolToFloat(Value=RigVMVariableNode() {variable=Pub_SwingPickaxe}.Value)
   n24 = &RigVMFunction_AlphaInterp(
    InterpSpeedDecreasing=5
    InterpSpeedIncreasing=5
    Scale=1
    bClampResult=False
    Value=&RigVMFunction_MathDoubleMul(A=@GetAlphaFromQuatDelta_12.alpha, B=@n23.Result).Result
   )
   > &RigUnit_ModifyTransforms(
    ItemToModify="((Item=(Type=Bone,Name=\"dyn_backHair_l_02\"),Transform=(Rotation=(X=0,Y=-0.1804,Z=0,W=0.9836),Translation=(X=3.5373,Y=4.3729,Z=12.1597),Scale3D=(X=1,Y=1,Z=1))),(Item=(Type=Bone,Name=\"dyn_backHair_c_02\"),Transform=(Rotation=(X=-0.0005,Y=-0.2299,Z=0.0019,W=0. ...(truncated)"
    WeightMinimum=0
    Weight=@n24.Result
   )
   > GetAlphaFromQuatDelta_13 = RigVMFunctionReferenceNode(
    Rotation Order=ZYX
    absoluteValue=false
    child=(Type=Bone,Name="upperarm_r")
    debugRotation=false
    onlyTwist=false
    outputMinMax=(X=0,Y=1)
    parent=(Type=Null,Name="upperarm_r_null")
    premult=1
    rotationMinMax=(X=0,Y=-10)
    triggerAxis=Z
    twistAxis=(X=1,Y=0,Z=0)
   )
   n25 = &RigVMFunction_AlphaInterp(
    InterpSpeedDecreasing=5
    InterpSpeedIncreasing=5
    Scale=1
    bClampResult=False
    Value=&RigVMFunction_MathDoubleMul(A=@GetAlphaFromQuatDelta_13.alpha, B=@n23.Result).Result
   )
   > &RigUnit_ModifyTransforms(
    ItemToModify="((Item=(Type=Bone,Name=\"dyn_backHair_r_02\"),Transform=(Rotation=(X=-0.0015,Y=0.1955,Z=0.0076,W=-0.9807),Translation=(X=2.3971,Y=0.6364,Z=3.0932),Scale3D=(X=1,Y=1,Z=1))),(Item=(Type=Bone,Name=\"dyn_backHair_c_02\"),Transform=(Rotation=(X=0,Y=0,Z=0,W=1),Transl ...(truncated)"
    WeightMinimum=0
    Weight=@n25.Result
   )
   n26 = RigVMVariableNode() {variable=Pub_IsSkydive}
   n27 = RigVMVariableNode() {variable=Pub_IsFrontEnd}
   > &RigUnit_ModifyTransforms(
    ItemToModify="((Item=(Type=Bone,Name=\"dyn_backHair_c_02\"),Transform=(Rotation=(X=0,Y=0,Z=0,W=1),Translation=(X=0,Y=0,Z=15),Scale3D=(X=1,Y=1,Z=1))),(Item=(Type=Bone,Name=\"dyn_backHair_l_02\"),Transform=(Rotation=(X=0,Y=0,Z=0,W=1),Translation=(X=0,Y=0,Z=15),Scale3D=(X=1,Y= ...(truncated)"
    WeightMinimum=0
    Weight=&RigVMFunction_MathDoubleMul(
     A=&RigVMFunction_MathBoolToFloat(Value=@n26.Value).Result
     B=&RigVMFunction_MathBoolToFloat(Value=@n27.Value).Result
    ).Result
   )
   > &RigUnit_ModifyTransforms(
    ItemToModify="((Item=(Type=Bone,Name=\"dyn_backHair_l_04\"),Transform=(Rotation=(X=0,Y=0,Z=0,W=1),Translation=(X=0,Y=0,Z=0),Scale3D=(X=1,Y=1,Z=1))),(Item=(Type=Bone,Name=\"dyn_backHair_c_04\"),Transform=(Rotation=(X=0,Y=0,Z=0,W=1),Translation=(X=0,Y=0,Z=15),Scale3D=(X=1,Y=1 ...(truncated)"
    WeightMinimum=0
    Weight=&RigVMFunction_MathBoolToFloat(Value=RigVMVariableNode() {variable=Pub_IsSkydive}.Value).Result
   )
   > &RigUnit_ModifyTransforms(
    ItemToModify="((Item=(Type=Bone,Name=\"dyn_backHair_l_04\"),Transform=(Rotation=(X=0,Y=0,Z=0,W=1),Translation=(X=0,Y=0,Z=10),Scale3D=(X=1,Y=1,Z=1))),(Item=(Type=Bone,Name=\"dyn_backHair_c_04\"),Transform=(Rotation=(X=0,Y=0,Z=0,W=1),Translation=(X=0,Y=0,Z=10),Scale3D=(X=1,Y= ...(truncated)"
    WeightMinimum=0
    Weight=&RigVMFunction_MathBoolToFloat(Value=RigVMVariableNode() {variable=Pub_IsParachuteOpened}.Value).Result
   )
   > &RigUnit_ModifyTransforms(
    ItemToModify="((Item=(Type=Bone,Name=\"dyn_backHair_l_03\"),Transform=(Rotation=(X=0,Y=-0.2588,Z=0,W=0.9659),Translation=(X=0,Y=0,Z=0),Scale3D=(X=1,Y=1,Z=1))),(Item=(Type=Bone,Name=\"dyn_backHair_c_03\"),Transform=(Rotation=(X=0,Y=-0.2588,Z=0,W=0.9659),Translation=(X=0,Y=0, ...(truncated)"
    WeightMinimum=0
    Weight=&RigVMFunction_MathBoolToFloat(Value=RigVMVariableNode() {variable=Pub_BasebalSliding}.Value).Result
   )
   n28 = &RigVMFunction_AlphaInterp(
    InterpSpeedDecreasing=5
    InterpSpeedIncreasing=5
    Scale=1
    bClampResult=true
    Value=RigVMDispatchNode(False=1, True=0, Condition=RigVMVariableNode() {variable=Is Crouching}.Value) {template="DISPATCH_RigVMDispatch_If(in Condition,in True,in False,out Result)"}.Result
   )
   n29 = RigVMVariableNode() {variable=HeadRotUp}
   &RigUnit_ModifyTransforms(
    ItemToModify="((Item=(Type=Null,Name=\"dyn_backHair_l_01_ctrl_null\"),Transform=(Rotation=(X=0,Y=0,Z=0,W=-1),Translation=(X=0,Y=0,Z=2),Scale3D=(X=1,Y=1,Z=1))),(Item=(Type=Null,Name=\"dyn_backHair_r_01_ctrl_null\"),Transform=(Rotation=(X=0,Y=0,Z=0,W=-1),Translation=(X=0,Y=0, ...(truncated)"
    WeightMinimum=0
    Weight=&RigVMFunction_MathDoubleMul(B=@n28.Result, A=@n29.Value).Result
    ExecutePin=@RigVMFunction_Sequence_1.A
   )
   > &RigUnit_ModifyTransforms(
    ItemToModify="((Item=(Type=Null,Name=\"dyn_backHair_l_01_ctrl_null\"),Transform=(Rotation=(X=0,Y=0,Z=0,W=-1),Translation=(X=0,Y=0,Z=2),Scale3D=(X=1,Y=1,Z=1))),(Item=(Type=Null,Name=\"dyn_backHair_r_01_ctrl_null\"),Transform=(Rotation=(X=0,Y=0,Z=0,W=-1),Translation=(X=0,Y=0, ...(truncated)"
    WeightMinimum=0
    Weight=@n29.Value
   )
   > GetAlphaFromQuatDelta = RigVMFunctionReferenceNode(
    Rotation Order=ZYX
    child=(Type=Bone,Name="head")
    debugRotation=false
    onlyTwist=true
    outputMinMax=(X=0,Y=1)
    parent=(Type=Bone,Name="neck_01")
    premult=1
    rotationMinMax=(X=0,Y=-45)
    twistAxis=(X=1,Y=0,Z=0)
   )
   > &RigUnit_ModifyTransforms(
    ItemToModify="((Item=(Type=Null,Name=\"dyn_backHair_r_02_ctrl_null\"),Transform=(Rotation=(X=0,Y=0,Z=0,W=1),Translation=(X=0,Y=0,Z=1),Scale3D=(X=1,Y=1,Z=1))),(Item=(Type=Null,Name=\"dyn_backHair_r_03_ctrl_null\"),Transform=(Rotation=(X=0,Y=0,Z=0,W=-1),Translation=(X=0,Y=0,Z ...(truncated)"
    WeightMinimum=0
    Weight=@GetAlphaFromQuatDelta.alpha
   )
   > GetAlphaFromQuatDelta_1 = RigVMFunctionReferenceNode(
    Rotation Order=ZYX
    child=(Type=Bone,Name="head")
    outputMinMax=(X=0,Y=1)
    parent=(Type=Bone,Name="neck_01")
    premult=1
    rotationMinMax=(X=0,Y=45)
    twistAxis=(X=1,Y=0,Z=0)
   )
   > &RigUnit_ModifyTransforms(
    ItemToModify="((Item=(Type=Null,Name=\"dyn_backHair_r_01_ctrl_null\"),Transform=(Rotation=(X=0,Y=0,Z=0,W=1),Translation=(X=-0.2341,Y=-3,Z=-0.9853),Scale3D=(X=1,Y=1,Z=1))),(Item=(Type=Null,Name=\"dyn_backHair_r_02_ctrl_null\"),Transform=(Rotation=(X=0,Y=0,Z=0,W=-1),Translati ...(truncated)"
    WeightMinimum=0
    Weight=@GetAlphaFromQuatDelta_1.alpha
   )
   n30 = RigVMVariableNode() {variable=HeadRotUpExtreme}
   > &RigUnit_ModifyTransforms(
    ItemToModify="((Item=(Type=Null,Name=\"dyn_backHair_r_03_ctrl_null\"),Transform=(Rotation=(X=0,Y=0,Z=0,W=-1),Translation=(X=2,Y=-1,Z=0),Scale3D=(X=1,Y=1,Z=1))),(Item=(Type=Null,Name=\"dyn_backHair_l_02_ctrl_null\"),Transform=(Rotation=(X=0,Y=0,Z=0,W=-1),Translation=(X=7.621 ...(truncated)"
    WeightMinimum=0
    Weight=@n30.Value
   )
   > &RigUnit_ModifyTransforms(
    ItemToModify="((Item=(Type=Null,Name=\"dyn_backHair_r_01_ctrl_null\"),Transform=(Rotation=(X=0,Y=0,Z=0,W=1),Translation=(X=-2,Y=0,Z=0),Scale3D=(X=1,Y=1,Z=1))),(Item=(Type=Null,Name=\"dyn_backHair_r_02_ctrl_null\"),Transform=(Rotation=(X=0,Y=0,Z=0,W=1),Translation=(X=1,Y=0,Z ...(truncated)"
    WeightMinimum=0
    Weight=@n30.Value
   )
   > &RigUnit_ModifyTransforms(
    ItemToModify="((Item=(Type=Null,Name=\"dyn_backHair_r_01_ctrl_null\"),Transform=(Rotation=(X=0,Y=0,Z=0,W=1),Translation=(X=-1,Y=-2,Z=1),Scale3D=(X=1,Y=1,Z=1))),(Item=(Type=Null,Name=\"dyn_frontHair_r_02_ctr_nulll\"),Transform=(Rotation=(X=0,Y=0,Z=0,W=1),Translation=(X=3,Y=- ...(truncated)"
    WeightMinimum=0
    Weight=&RigVMFunction_MathDoubleRemap(SourceMaximum=0.5, SourceMinimum=0, TargetMaximum=0, TargetMinimum=1, Value=RigVMVariableNode() {variable=HeadYrotation}.Value).Result
   )
   > &RigUnit_ModifyTransforms(
    ItemToModify="((Item=(Type=Null,Name=\"dyn_backHair_r_01_ctrl_null\"),Transform=(Rotation=(X=0,Y=0,Z=0,W=1),Translation=(X=-5,Y=2,Z=15),Scale3D=(X=1,Y=1,Z=1))),(Item=(Type=Null,Name=\"dyn_backHair_r_02_ctrl_null\"),Transform=(Rotation=(X=0,Y=0,Z=0,W=1),Translation=(X=0,Y=0, ...(truncated)"
    WeightMinimum=0
    Weight=RigVMVariableNode() {variable=HeadLookingUpCrouching}.Value
   )
   n31 = RigVMVariableNode() {variable=HeadYrotation}
   Multiply = &RigVMFunction_MathDoubleMul(
    B=RigVMDispatchNode(False=1, True=0, Condition=RigVMVariableNode() {variable=Is Crouching}.Value) {template="DISPATCH_RigVMDispatch_If(in Condition,in True,in False,out Result)"}.Result
    A=&RigVMFunction_MathDoubleMul(
     B=&RigVMFunction_MathDoubleRemap(SourceMaximum=0.5, SourceMinimum=0, TargetMaximum=0, TargetMinimum=1, Value=@n31.Value).Result
     A=RigVMVariableNode() {variable=HeadRotUp}.Value
    ).Result
   )
   n32 = &RigVMFunction_MathDoubleRemap(SourceMaximum=0.5, SourceMinimum=0, TargetMaximum=1, TargetMinimum=0, Value=@Multiply.Result)
   > &RigUnit_ModifyTransforms(
    ItemToModify="((Item=(Type=Null,Name=\"dyn_backHair_r_02_ctrl_null\"),Transform=(Rotation=(X=0,Y=0,Z=0,W=1),Translation=(X=3,Y=-5,Z=0),Scale3D=(X=1,Y=1,Z=1))),(Item=(Type=Null,Name=\"dyn_backHair_r_03_ctrl_null\"),Transform=(Rotation=(X=0,Y=0,Z=0,W=1),Translation=(X=0,Y=-3, ...(truncated)"
    WeightMinimum=0
    Weight=@n32.Result
   )
   > &RigUnit_ModifyTransforms(
    ItemToModify="((Item=(Type=Null,Name=\"dyn_backHair_l_02_ctrl_null\"),Transform=(Rotation=(X=0,Y=0,Z=0,W=1),Translation=(X=0,Y=-6,Z=0),Scale3D=(X=1,Y=1,Z=1))),(Item=(Type=Null,Name=\"dyn_backHair_l_03_ctrl_null\"),Transform=(Rotation=(X=0,Y=0,Z=0,W=1),Translation=(X=0,Y=-3, ...(truncated)"
    WeightMinimum=0
    Weight=@n32.Result
   )
   RigVMCollapseNode(ExecutePin=@RigVMFunction_ControlFlowBranch.True)
  ]}
 CollapseNode_ContainedGraph {nodes [
   RigVMFunctionEntryNode()
   > &RigUnit_ParentConstraint(
    AdvancedSettings=(RotationOrderForFilter=XZY)
    Child=(Type=Null,Name="dyn_backHair_c_01_ctrl_null")
    Filter=(TranslationFilter=(bX=True,bY=True,bZ=True),RotationFilter=(bX=True,bY=True,bZ=True),ScaleFilter=(bX=False,bY=False,bZ=False))
    Parents=((Item=(Type=Bone,Name="head"),Weight=0.5),(Item=(Type=Bone,Name="neck_01"),Weight=0.5))
    Weight=1
   )
   > &RigUnit_ParentConstraint(
    AdvancedSettings=(RotationOrderForFilter=XZY)
    Child=(Type=Null,Name="dyn_backHair_c_02_ctrl_null")
    Parents="((Item=(Type=Bone,Name=\"upperarm_twist_01_l\"),Weight=0.0049),(Item=(Type=Bone,Name=\"spine_03\"),Weight=0.0066),(Item=(Type=Bone,Name=\"upperarm_twist_01_r\"),Weight=0.0092),(Item=(Type=Bone,Name=\"clavicle_r\"),Weight=0.0459),(Item=(Type=Bone,Name=\"clavicle_l\") ...(truncated)"
    Weight=1
   )
   > &RigUnit_ParentConstraint(
    AdvancedSettings=(RotationOrderForFilter=XZY)
    Child=(Type=Null,Name="dyn_backHair_c_03_ctrl_null")
    Parents="((Item=(Type=Bone,Name=\"dyn_C_haori_back_1\"),Weight=0.002),(Item=(Type=Bone,Name=\"spine_05\"),Weight=0.0087),(Item=(Type=Bone,Name=\"spine_01\"),Weight=0.0283),(Item=(Type=Bone,Name=\"upperarm_twist_01_r\"),Weight=0.0441),(Item=(Type=Bone,Name=\"upperarm_twist_0 ...(truncated)"
    Weight=1
   )
   > &RigUnit_ParentConstraint(
    AdvancedSettings=(RotationOrderForFilter=XZY)
    Child=(Type=Null,Name="dyn_backHair_c_04_ctrl_null")
    Parents="((Item=(Type=Bone,Name=\"dyn_L_haori_side_2\"),Weight=0.003),(Item=(Type=Bone,Name=\"dyn_C_haori_back_2\"),Weight=0.0082),(Item=(Type=Bone,Name=\"spine_03\"),Weight=0.0122),(Item=(Type=Bone,Name=\"dyn_R_haori_side_1\"),Weight=0.076),(Item=(Type=Bone,Name=\"dyn_L_ha ...(truncated)"
    Weight=1
   )
   > &RigUnit_ParentConstraint(
    AdvancedSettings=(RotationOrderForFilter=XZY)
    Child=(Type=Null,Name="dyn_backHair_c_05_ctrl_null")
    Parents="((Item=(Type=Bone,Name=\"dyn_C_haori_back_3\"),Weight=0.032),(Item=(Type=Bone,Name=\"dyn_R_haori_side_2\"),Weight=0.072),(Item=(Type=Bone,Name=\"dyn_L_haori_side_2\"),Weight=0.0721),(Item=(Type=Bone,Name=\"dyn_R_haori_side_1\"),Weight=0.0769),(Item=(Type=Bone,Name ...(truncated)"
    Weight=1
   )
   > &RigUnit_ParentConstraint(
    AdvancedSettings=(RotationOrderForFilter=XZY)
    Child=(Type=Null,Name="dyn_backHair_c_06_ctrl_null")
    Parents="((Item=(Type=Bone,Name=\"dyn_C_haori_back_1\"),Weight=0.0403),(Item=(Type=Bone,Name=\"dyn_L_haori_side_3\"),Weight=0.0471),(Item=(Type=Bone,Name=\"dyn_R_haori_side_3\"),Weight=0.0471),(Item=(Type=Bone,Name=\"dyn_L_haori_side_2\"),Weight=0.0704),(Item=(Type=Bone,Na ...(truncated)"
    Weight=1
   )
   > &RigUnit_ParentConstraint(
    AdvancedSettings=(RotationOrderForFilter=XZY)
    Child=(Type=Null,Name="dyn_backHair_l_01_ctrl_null")
    Filter=(TranslationFilter=(bX=True,bY=True,bZ=True),RotationFilter=(bX=True,bY=True,bZ=True),ScaleFilter=(bX=False,bY=False,bZ=False))
    Parents=((Item=(Type=Bone,Name="head"),Weight=0.5),(Item=(Type=Bone,Name="neck_01"),Weight=0.5))
    Weight=1
   )
   > &RigUnit_ParentConstraint(
    AdvancedSettings=(RotationOrderForFilter=XZY)
    Child=(Type=Null,Name="dyn_backHair_l_02_ctrl_null")
    Parents="((Item=(Type=Bone,Name=\"spine_03\"),Weight=0.0001),(Item=(Type=Bone,Name=\"pelvis\"),Weight=0.0015),(Item=(Type=Bone,Name=\"upperarm_twist_02_l\"),Weight=0.0188),(Item=(Type=Bone,Name=\"neck_01\"),Weight=0.0402),(Item=(Type=Bone,Name=\"spine_05\"),Weight=0.1209),(I ...(truncated)"
    Weight=1
   )
   > &RigUnit_ParentConstraint(
    AdvancedSettings=(RotationOrderForFilter=XZY)
    Child=(Type=Null,Name="dyn_backHair_l_03_ctrl_null")
    Parents="((Item=(Type=Bone,Name=\"upperarm_twist_02_l\"),Weight=0.0138),(Item=(Type=Bone,Name=\"spine_05\"),Weight=0.0183),(Item=(Type=Bone,Name=\"clavicle_l\"),Weight=0.1185),(Item=(Type=Bone,Name=\"spine_02\"),Weight=0.1432),(Item=(Type=Bone,Name=\"spine_03\"),Weight=0.160 ...(truncated)"
    Weight=1
   )
   > &RigUnit_ParentConstraint(
    AdvancedSettings=(RotationOrderForFilter=XZY)
    Child=(Type=Null,Name="dyn_backHair_l_04_ctrl_null")
    Parents="((Item=(Type=Bone,Name=\"spine_03\"),Weight=0.016),(Item=(Type=Bone,Name=\"dyn_L_haori_side_2\"),Weight=0.0111),(Item=(Type=Bone,Name=\"dyn_R_haori_side_1\"),Weight=0.0142),(Item=(Type=Bone,Name=\"dyn_C_haori_back_1\"),Weight=0.151),(Item=(Type=Bone,Name=\"spine_02 ...(truncated)"
    Weight=1
   )
   > &RigUnit_ParentConstraint(
    AdvancedSettings=(RotationOrderForFilter=XZY)
    Child=(Type=Null,Name="dyn_backHair_l_05_ctrl_null")
    Parents="((Item=(Type=Bone,Name=\"dyn_C_haori_back_3\"),Weight=0.0144),(Item=(Type=Bone,Name=\"dyn_L_haori_side_3\"),Weight=0.0262),(Item=(Type=Bone,Name=\"dyn_C_haori_back_2\"),Weight=0.1611),(Item=(Type=Bone,Name=\"dyn_C_haori_back_1\"),Weight=0.1965),(Item=(Type=Bone,Na ...(truncated)"
    Weight=1
   )
   > &RigUnit_ParentConstraint(
    AdvancedSettings=(RotationOrderForFilter=XZY)
    Child=(Type=Null,Name="dyn_backHair_l_06_ctrl_null")
    Parents="((Item=(Type=Bone,Name=\"dyn_C_haori_back_1\"),Weight=0.0194),(Item=(Type=Bone,Name=\"dyn_L_haori_side_1\"),Weight=0.0293),(Item=(Type=Bone,Name=\"dyn_C_haori_back_4\"),Weight=0.0314),(Item=(Type=Bone,Name=\"dyn_L_haori_side_4\"),Weight=0.0489),(Item=(Type=Bone,Na ...(truncated)"
    Weight=1
   )
   > &RigUnit_ParentConstraint(
    AdvancedSettings=(RotationOrderForFilter=XZY)
    Child=(Type=Null,Name="dyn_backHair_r_01_ctrl_null")
    Filter=(TranslationFilter=(bX=True,bY=True,bZ=True),RotationFilter=(bX=True,bY=True,bZ=True),ScaleFilter=(bX=False,bY=False,bZ=False))
    Parents=((Item=(Type=Bone,Name="head"),Weight=0.5),(Item=(Type=Bone,Name="neck_01"),Weight=0.5))
    Weight=1
   )
   > &RigUnit_ParentConstraint(
    AdvancedSettings=(RotationOrderForFilter=XZY)
    Child=(Type=Null,Name="dyn_backHair_r_02_ctrl_null")
    Parents="((Item=(Type=Bone,Name=\"neck_01\"),Weight=0.0234),(Item=(Type=Bone,Name=\"spine_04\"),Weight=0.115),(Item=(Type=Bone,Name=\"spine_05\"),Weight=0.1206),(Item=(Type=Bone,Name=\"clavicle_r\"),Weight=0.3462),(Item=(Type=Bone,Name=\"upperarm_twist_01_r\"),Weight=0.3948) ...(truncated)"
    Weight=1
   )
   > &RigUnit_ParentConstraint(
    AdvancedSettings=(RotationOrderForFilter=XZY)
    Child=(Type=Null,Name="dyn_backHair_r_03_ctrl_null")
    Parents="((Item=(Type=Bone,Name=\"spine_02\"),Weight=0.1227),(Item=(Type=Bone,Name=\"clavicle_r\"),Weight=0.1355),(Item=(Type=Bone,Name=\"spine_03\"),Weight=0.171),(Item=(Type=Bone,Name=\"spine_04\"),Weight=0.1869),(Item=(Type=Bone,Name=\"upperarm_twist_01_r\"),Weight=0.3839 ...(truncated)"
    Weight=1
   )
   > &RigUnit_ParentConstraint(
    AdvancedSettings=(RotationOrderForFilter=XZY)
    Child=(Type=Null,Name="dyn_backHair_r_04_ctrl_null")
    Parents="((Item=(Type=Bone,Name=\"dyn_C_haori_back_2\"),Weight=0.0012),(Item=(Type=Bone,Name=\"dyn_R_haori_side_2\"),Weight=0.0011),(Item=(Type=Bone,Name=\"dyn_L_haori_side_1\"),Weight=0.0138),(Item=(Type=Bone,Name=\"dyn_C_haori_back_1\"),Weight=0.1391),(Item=(Type=Bone,Na ...(truncated)"
    Weight=1
   )
   > &RigUnit_ParentConstraint(
    AdvancedSettings=(RotationOrderForFilter=XZY)
    Child=(Type=Null,Name="dyn_backHair_r_05_ctrl_null")
    Parents="((Item=(Type=Bone,Name=\"dyn_C_haori_back_3\"),Weight=0.0144),(Item=(Type=Bone,Name=\"dyn_R_haori_side_3\"),Weight=0.0262),(Item=(Type=Bone,Name=\"dyn_C_haori_back_2\"),Weight=0.1612),(Item=(Type=Bone,Name=\"dyn_C_haori_back_1\"),Weight=0.1976),(Item=(Type=Bone,Na ...(truncated)"
    Weight=1
   )
   > &RigUnit_ParentConstraint(
    AdvancedSettings=(RotationOrderForFilter=XZY)
    Child=(Type=Null,Name="dyn_backHair_r_06_ctrl_null")
    Parents="((Item=(Type=Bone,Name=\"dyn_R_haori_side_1\"),Weight=0.0209),(Item=(Type=Bone,Name=\"dyn_C_haori_back_1\"),Weight=0.0286),(Item=(Type=Bone,Name=\"dyn_R_haori_side_4\"),Weight=0.0373),(Item=(Type=Bone,Name=\"dyn_C_haori_back_4\"),Weight=0.0475),(Item=(Type=Bone,Na ...(truncated)"
    Weight=1
   )
   > RigVMFunctionReturnNode()
  ]}
 CollapseNode_ContainedGraph {nodes [
   RigVMFunctionEntryNode()
   ItemArray_6 = RigVMDispatchNode(Value=((Type=Bone,Name="dyn_backHair_l_01"),(Type=Bone,Name="dyn_backHair_l_02"),(Type=Bone,Name="dyn_backHair_l_03"),(Type=Bone,Name="dyn_backHair_l_04"),(Type=Bone,Name="dyn_backHair_l_05"),(Type=Bone,Name="dyn_backHair_l_06"))) {template="DISPATCH_RigVMDispatch_Constant(io Value)"}
   > TwistBones = RigVMFunctionReferenceNode(
    Axis to Twist=(bX=True,bY=true,bZ=true)
    End=(Type=Bone,Name="neck_01")
    Start=(Type=Bone,Name="spine_02")
    Bones to Twist=@ItemArray_6.Value
   )
   ItemArray_7 = RigVMDispatchNode(Value=((Type=Bone,Name="dyn_backHair_r_01"),(Type=Bone,Name="dyn_backHair_r_02"),(Type=Bone,Name="dyn_backHair_r_03"),(Type=Bone,Name="dyn_backHair_r_04"),(Type=Bone,Name="dyn_backHair_r_05"),(Type=Bone,Name="dyn_backHair_r_06"))) {template="DISPATCH_RigVMDispatch_Constant(io Value)"}
   > TwistBones_1 = RigVMFunctionReferenceNode(
    Axis to Twist=(bX=True,bY=true,bZ=true)
    End=(Type=Bone,Name="neck_01")
    Start=(Type=Bone,Name="spine_02")
    Bones to Twist=@ItemArray_7.Value
   )
   ItemArray_8 = RigVMDispatchNode(Value=((Type=Bone,Name="dyn_backHair_c_01"),(Type=Bone,Name="dyn_backHair_c_02"),(Type=Bone,Name="dyn_backHair_c_03"),(Type=Bone,Name="dyn_backHair_c_04"),(Type=Bone,Name="dyn_backHair_c_05"),(Type=Bone,Name="dyn_backHair_c_06"))) {template="DISPATCH_RigVMDispatch_Constant(io Value)"}
   > TwistBones_2 = RigVMFunctionReferenceNode(
    Axis to Twist=(bX=True,bY=true,bZ=true)
    End=(Type=Bone,Name="neck_01")
    Start=(Type=Bone,Name="spine_02")
    Bones to Twist=@ItemArray_8.Value
   )
   > RigVMFunctionReturnNode()
   ItemArray_1 = RigVMDispatchNode(Value=((Type=Bone,Name="dyn_backHair_r_02"),(Type=Bone,Name="dyn_backHair_r_03"),(Type=Bone,Name="dyn_backHair_r_04"),(Type=Bone,Name="dyn_backHair_r_05"),(Type=Bone,Name="dyn_backHair_r_06"))) {template="DISPATCH_RigVMDispatch_Constant(io Value)"}
   ItemArray_2 = RigVMDispatchNode(Value=((Type=Bone,Name="dyn_backHair_c_02"),(Type=Bone,Name="dyn_backHair_c_03"),(Type=Bone,Name="dyn_backHair_c_04"),(Type=Bone,Name="dyn_backHair_c_05"),(Type=Bone,Name="dyn_backHair_c_06"))) {template="DISPATCH_RigVMDispatch_Constant(io Value)"}
  ]}
 CollapseNode_ContainedGraph {nodes [
   RigVMFunctionEntryNode()
   ItemArray = RigVMDispatchNode(Value=((Type=Bone,Name="dyn_backHair_l_01"),(Type=Bone,Name="dyn_backHair_l_02"),(Type=Bone,Name="dyn_backHair_l_03"),(Type=Bone,Name="dyn_backHair_l_04"),(Type=Bone,Name="dyn_backHair_l_05"),(Type=Bone,Name="dyn_backHair_l_06"))) {template="DISPATCH_RigVMDispatch_Constant(io Value)"}
   ItemArray_1 = RigVMDispatchNode(Value="((Type=Control,Name=\"dyn_backHair_l_01_ctrl\"),(Type=Control,Name=\"dyn_backHair_l_02_ctrl\"),(Type=Control,Name=\"dyn_backHair_l_03_ctrl\"),(Type=Control,Name=\"dyn_backHair_l_04_ctrl\"),(Type=Control,Name=\"dyn_backHair_l_05_ctrl\"),(Type=Control,Name=\"dyn_backHa ...(truncated)") {template="DISPATCH_RigVMDispatch_Constant(io Value)"}
   > SplineIK = RigVMFunctionReferenceNode(
    Primary Axis=(X=1,Y=0,Z=0)
    Samples Per Segment=1
    Secondary Spline Direction=(X=0,Y=0,Z=0)
    Spline Mode=BSpline
    Up Axis=(X=0,Y=0,Z=0)
    Bones=@ItemArray.Value
    Controls=@ItemArray_1.Value
   )
   ItemArray_2 = RigVMDispatchNode(Value="((Type=Control,Name=\"dyn_backHair_r_01_ctrl\"),(Type=Control,Name=\"dyn_backHair_r_02_ctrl\"),(Type=Control,Name=\"dyn_backHair_r_03_ctrl\"),(Type=Control,Name=\"dyn_backHair_r_04_ctrl\"),(Type=Control,Name=\"dyn_backHair_r_05_ctrl\"),(Type=Control,Name=\"dyn_backHa ...(truncated)") {template="DISPATCH_RigVMDispatch_Constant(io Value)"}
   ItemArray_3 = RigVMDispatchNode(Value=((Type=Bone,Name="dyn_backHair_r_01"),(Type=Bone,Name="dyn_backHair_r_02"),(Type=Bone,Name="dyn_backHair_r_03"),(Type=Bone,Name="dyn_backHair_r_04"),(Type=Bone,Name="dyn_backHair_r_05"),(Type=Bone,Name="dyn_backHair_r_06"))) {template="DISPATCH_RigVMDispatch_Constant(io Value)"}
   > SplineIK_1 = RigVMFunctionReferenceNode(
    Primary Axis=(X=1,Y=0,Z=0)
    Samples Per Segment=1
    Secondary Spline Direction=(X=0,Y=0,Z=0)
    Spline Mode=BSpline
    Up Axis=(X=0,Y=0,Z=0)
    Controls=@ItemArray_2.Value
    Bones=@ItemArray_3.Value
   )
   ItemArray_4 = RigVMDispatchNode(Value=((Type=Bone,Name="dyn_backHair_c_01"),(Type=Bone,Name="dyn_backHair_c_02"),(Type=Bone,Name="dyn_backHair_c_03"),(Type=Bone,Name="dyn_backHair_c_04"),(Type=Bone,Name="dyn_backHair_c_05"),(Type=Bone,Name="dyn_backHair_c_06"))) {template="DISPATCH_RigVMDispatch_Constant(io Value)"}
   ItemArray_5 = RigVMDispatchNode(Value="((Type=Control,Name=\"dyn_backHair_c_01_ctrl\"),(Type=Control,Name=\"dyn_backHair_c_02_ctrl\"),(Type=Control,Name=\"dyn_backHair_c_03_ctrl\"),(Type=Control,Name=\"dyn_backHair_c_04_ctrl\"),(Type=Control,Name=\"dyn_backHair_c_05_ctrl\"),(Type=Control,Name=\"dyn_backHa ...(truncated)") {template="DISPATCH_RigVMDispatch_Constant(io Value)"}
   > SplineIK_2 = RigVMFunctionReferenceNode(
    Primary Axis=(X=1,Y=0,Z=0)
    Samples Per Segment=1
    Secondary Spline Direction=(X=0,Y=0,Z=0)
    Spline Mode=BSpline
    Up Axis=(X=0,Y=0,Z=0)
    Bones=@ItemArray_4.Value
    Controls=@ItemArray_5.Value
   )
   > RigVMFunctionReturnNode()
  ]}
 CollapseNode_ContainedGraph {nodes [
   n1 = RigVMFunctionEntryNode()
   RigVMFunction_Sequence = &RigVMFunction_Sequence(ExecuteContext=@n1.ExecutePin)
   &RigUnit_ParentConstraint(
    AdvancedSettings=(RotationOrderForFilter=XZY)
    Child=(Type=Bone,Name="base_backHair_l")
    Parents=((Item=(Type=Bone,Name="spine_05"),Weight=1.0))
    Weight=1
    ExecutePin=@RigVMFunction_Sequence.A
   )
   > &RigUnit_ParentConstraint(
    AdvancedSettings=(RotationOrderForFilter=XZY)
    Child=(Type=Bone,Name="base_backHair_c")
    Parents=((Item=(Type=Bone,Name="spine_05"),Weight=1.0))
    Weight=1
   )
   > &RigUnit_ParentConstraint(
    AdvancedSettings=(RotationOrderForFilter=XZY)
    Child=(Type=Bone,Name="base_backHair_r")
    Parents=((Item=(Type=Bone,Name="spine_05"),Weight=1.0))
    Weight=1
   )
   RigVMFunction_MathBoolOr = &RigVMFunction_MathBoolOr(A=RigVMVariableNode() {variable=Pub_WearingPack}.Value, B=RigVMVariableNode() {variable=WearingCape}.Value)
   RigVMFunction_MathBoolToFloat = &RigVMFunction_MathBoolToFloat(Value=@RigVMFunction_MathBoolOr.Result)
   AlphaInterp = &RigVMFunction_AlphaInterp(InterpSpeedDecreasing=10, InterpSpeedIncreasing=10, Scale=1, bClampResult=False, Value=@RigVMFunction_MathBoolToFloat.Result)
   > &RigUnit_ModifyTransforms(
    ItemToModify="((Item=(Type=Bone,Name=\"dyn_backHair_l_03\"),Transform=(Rotation=(X=0,Y=-0.4226,Z=0,W=0.9063),Translation=(X=0,Y=0,Z=-30),Scale3D=(X=1,Y=1,Z=1))),(Item=(Type=Bone,Name=\"dyn_backHair_c_03\"),Transform=(Rotation=(X=0,Y=-0.4226,Z=0,W=0.9063),Translation=(X=0,Y= ...(truncated)"
    WeightMinimum=0
    Weight=@AlphaInterp.Result
   )
   > &RigUnit_ModifyTransforms(
    ItemToModify="((Item=(Type=Bone,Name=\"dyn_backHair_l_03\"),Transform=(Rotation=(X=0,Y=-0.2588,Z=0,W=0.9659),Translation=(X=0,Y=0,Z=0),Scale3D=(X=1,Y=1,Z=1))),(Item=(Type=Bone,Name=\"dyn_backHair_c_03\"),Transform=(Rotation=(X=0,Y=-0.2588,Z=0,W=0.9659),Translation=(X=0,Y=0, ...(truncated)"
    WeightMinimum=0
    Weight=@AlphaInterp.Result
   )
   RigVMFunctionReturnNode(ExecutePin=@RigVMFunction_Sequence.B)
  ]}
 CollapseNode_ContainedGraph {nodes [
   RigVMFunctionEntryNode()
   n1 = RigVMVariableNode() {variable=Is Targeting}
   n2 = &RigVMFunction_AlphaInterp(
    InterpSpeedDecreasing=5
    InterpSpeedIncreasing=25
    Scale=1
    bClampResult=true
    Value=&RigVMFunction_MathDoubleMul(B=&RigVMFunction_MathBoolToFloat(Value=@n1.Value).Result, A=RigVMVariableNode() {variable=HeadRotUp}.Value).Result
   )
   > &RigUnit_ModifyTransforms(
    ItemToModify=((Item=(Type=Bone,Name="base_fronthair_r_headpivot"),Transform=(Rotation=(X=0,Y=0.3007,Z=0,W=0.9537),Translation=(X=0,Y=0,Z=0),Scale3D=(X=1,Y=1,Z=1))))
    WeightMinimum=0
    Weight=@n2.Result
   )
   n3 = &RigVMFunction_AlphaInterp(
    InterpSpeedDecreasing=5
    InterpSpeedIncreasing=25
    Scale=1
    bClampResult=true
    Value=&RigVMFunction_MathBoolToFloat(Value=RigVMVariableNode() {variable=Pub_IsParachuteOpened}.Value).Result
   )
   > &RigUnit_ModifyTransforms(
    ItemToModify="((Item=(Type=Bone,Name=\"base_fronthair_l\"),Transform=(Rotation=(X=0,Y=0,Z=0,W=1),Translation=(X=0,Y=8.8894,Z=0),Scale3D=(X=1,Y=1,Z=1))),(Item=(Type=Bone,Name=\"base_fronthair_r\"),Transform=(Rotation=(X=-0.0004,Y=0.0001,Z=-0.289,W=-0.9573),Translation=(X=0.8 ...(truncated)"
    WeightMinimum=0
    Weight=@n3.Result
   )
   > GetAlphaFromQuatDelta_5 = RigVMFunctionReferenceNode(
    Rotation Order=ZYX
    absoluteValue=false
    child=(Type=Bone,Name="upperarm_r")
    debugRotation=false
    outputMinMax=(X=0,Y=1)
    parent=(Type=Null,Name="upperarm_r_null")
    premult=1
    rotationMinMax=(X=0,Y=-25)
    triggerAxis=Y
    twistAxis=(X=1,Y=0,Z=0)
   )
   RigVMFunction_MathBoolNot = &RigVMFunction_MathBoolNot(Value=RigVMVariableNode() {variable=Is Targeting}.Value)
   Multiply = &RigVMFunction_MathDoubleMul(A=@GetAlphaFromQuatDelta_5.alpha, B=&RigVMFunction_MathBoolToFloat(Value=@RigVMFunction_MathBoolNot.Result).Result)
   > &RigUnit_ModifyTransforms(
    ItemToModify="((Item=(Type=Bone,Name=\"base_fronthair_l_headpivot\"),Transform=(Rotation=(X=-0.0872,Y=0,Z=0,W=0.9962),Translation=(X=0,Y=2.9218,Z=0.0704),Scale3D=(X=1,Y=1,Z=1))),(Item=(Type=Bone,Name=\"base_fronthair_r_headpivot\"),Transform=(Rotation=(X=0,Y=0,Z=0.1736,W=0. ...(truncated)"
    WeightMinimum=0
    Weight=@Multiply.Result
   )
   AlphaInterp = &RigVMFunction_AlphaInterp(
    InterpSpeedDecreasing=5
    InterpSpeedIncreasing=5
    Scale=1
    bClampResult=False
    Value=&RigVMFunction_MathBoolToFloat(Value=RigVMVariableNode() {variable=Is Targeting}.Value).Result
   )
   > &RigUnit_ModifyTransforms(
    ItemToModify=((Item=(Type=Bone,Name="base_fronthair_r_headpivot"),Transform=(Rotation=(X=0,Y=0,Z=-0.1736,W=0.9848),Translation=(X=-5,Y=0,Z=5),Scale3D=(X=1,Y=1,Z=1))))
    WeightMinimum=0
    Weight=@AlphaInterp.Result
   )
   > RigVMFunctionReturnNode()
  ]}
 CollapseNode_ContainedGraph {nodes [
   RigVMFunctionEntryNode()
   > &RigUnit_ParentConstraint(
    AdvancedSettings=(RotationOrderForFilter=XZY)
    Child=(Type=Bone,Name="dyn_frontHair_l_01")
    Parents=((Item=(Type=Bone,Name="spine_05"),Weight=0.1577),(Item=(Type=Bone,Name="clavicle_l"),Weight=0.2406),(Item=(Type=Bone,Name="neck_01"),Weight=0.6017))
    Weight=1
   )
   > &RigUnit_ParentConstraint(
    AdvancedSettings=(RotationOrderForFilter=XZY)
    Child=(Type=Bone,Name="dyn_frontHair_l_02")
    Parents="((Item=(Type=Bone,Name=\"spine_02\"),Weight=0.0009),(Item=(Type=Bone,Name=\"upperarm_twist_02_l\"),Weight=0.011),(Item=(Type=Bone,Name=\"spine_03\"),Weight=0.0116),(Item=(Type=Bone,Name=\"neck_01\"),Weight=0.0247),(Item=(Type=Bone,Name=\"upperarm_twist_01_l\"),Weigh ...(truncated)"
    Weight=1
   )
   > &RigUnit_ParentConstraint(
    AdvancedSettings=(RotationOrderForFilter=XZY)
    Child=(Type=Bone,Name="dyn_frontHair_l_03")
    Parents="((Item=(Type=Bone,Name=\"dyn_L_haori_front_1\"),Weight=0.0034),(Item=(Type=Bone,Name=\"upperarm_twist_02_l\"),Weight=0.0284),(Item=(Type=Bone,Name=\"spine_05\"),Weight=0.0308),(Item=(Type=Bone,Name=\"spine_02\"),Weight=0.0738),(Item=(Type=Bone,Name=\"upperarm_twist ...(truncated)"
    Weight=1
   )
   > &RigUnit_ParentConstraint(
    AdvancedSettings=(RotationOrderForFilter=XZY)
    Child=(Type=Bone,Name="dyn_frontHair_l_04")
    Parents="((Item=(Type=Bone,Name=\"upperarm_twist_01_l\"),Weight=0.0312),(Item=(Type=Bone,Name=\"dyn_L_haori_side_1\"),Weight=0.0338),(Item=(Type=Bone,Name=\"spine_04\"),Weight=0.0639),(Item=(Type=Bone,Name=\"spine_01\"),Weight=0.0748),(Item=(Type=Bone,Name=\"dyn_L_haori_fro ...(truncated)"
    Weight=1
   )
   > &RigUnit_ParentConstraint(
    AdvancedSettings=(RotationOrderForFilter=XZY)
    Child=(Type=Bone,Name="dyn_frontHair_l_05")
    Parents="((Item=(Type=Bone,Name=\"dyn_L_haori_front_1\"),Weight=0.0263),(Item=(Type=Bone,Name=\"clavicle_l\"),Weight=0.0338),(Item=(Type=Bone,Name=\"base_C_skirt_a_1\"),Weight=0.0389),(Item=(Type=Bone,Name=\"base_L_skirt_b_1\"),Weight=0.0515),(Item=(Type=Bone,Name=\"pelvis\" ...(truncated)"
    Weight=1
   )
   > &RigUnit_ParentConstraint(
    AdvancedSettings=(RotationOrderForFilter=XZY)
    Child=(Type=Bone,Name="dyn_frontHair_l_06")
    Parents="((Item=(Type=Bone,Name=\"spine_01\"),Weight=0.0288),(Item=(Type=Bone,Name=\"dyn_L_skirt_a_2\"),Weight=0.0531),(Item=(Type=Bone,Name=\"base_L_skirt_c_1\"),Weight=0.0601),(Item=(Type=Bone,Name=\"base_C_skirt_a_1\"),Weight=0.1008),(Item=(Type=Bone,Name=\"dyn_L_skirt_b ...(truncated)"
    Weight=1
   )
   > &RigUnit_ParentConstraint(
    AdvancedSettings=(RotationOrderForFilter=XZY)
    Child=(Type=Bone,Name="dyn_frontHair_r_01")
    Parents="((Item=(Type=Bone,Name=\"spine_03\"),Weight=0.0003),(Item=(Type=Bone,Name=\"neck_02\"),Weight=0.0003),(Item=(Type=Bone,Name=\"upperarm_twist_02_r\"),Weight=0.0005),(Item=(Type=Bone,Name=\"spine_04\"),Weight=0.0022),(Item=(Type=Bone,Name=\"upperarm_twist_01_r\"),Weig ...(truncated)"
    Weight=1
   )
   > &RigUnit_ParentConstraint(
    AdvancedSettings=(RotationOrderForFilter=XZY)
    Child=(Type=Bone,Name="dyn_frontHair_r_02")
    Parents="((Item=(Type=Bone,Name=\"spine_02\"),Weight=0.0007),(Item=(Type=Bone,Name=\"spine_03\"),Weight=0.0081),(Item=(Type=Bone,Name=\"spine_04\"),Weight=0.0139),(Item=(Type=Bone,Name=\"upperarm_twist_02_r\"),Weight=0.0239),(Item=(Type=Bone,Name=\"neck_01\"),Weight=0.0688), ...(truncated)"
    Weight=1
   )
   > &RigUnit_ParentConstraint(
    AdvancedSettings=(RotationOrderForFilter=XZY)
    Child=(Type=Bone,Name="dyn_frontHair_r_03")
    Parents="((Item=(Type=Bone,Name=\"spine_01\"),Weight=0.0001),(Item=(Type=Bone,Name=\"upperarm_twist_02_r\"),Weight=0.0086),(Item=(Type=Bone,Name=\"spine_05\"),Weight=0.013),(Item=(Type=Bone,Name=\"upperarm_twist_01_r\"),Weight=0.0603),(Item=(Type=Bone,Name=\"spine_02\"),Weig ...(truncated)"
    Weight=1
   )
   > &RigUnit_ParentConstraint(
    AdvancedSettings=(RotationOrderForFilter=XZY)
    Child=(Type=Bone,Name="dyn_frontHair_r_04")
    Parents="((Item=(Type=Bone,Name=\"upperarm_twist_02_r\"),Weight=0.0073),(Item=(Type=Bone,Name=\"upperarm_twist_01_r\"),Weight=0.0227),(Item=(Type=Bone,Name=\"spine_04\"),Weight=0.031),(Item=(Type=Bone,Name=\"dyn_R_haori_front_1\"),Weight=0.0425),(Item=(Type=Bone,Name=\"clav ...(truncated)"
    Weight=1
   )
   > &RigUnit_ParentConstraint(
    AdvancedSettings=(RotationOrderForFilter=XZY)
    Child=(Type=Bone,Name="dyn_frontHair_r_05")
    Parents="((Item=(Type=Bone,Name=\"base_R_skirt_b_1\"),Weight=0.0078),(Item=(Type=Bone,Name=\"spine_03\"),Weight=0.0091),(Item=(Type=Bone,Name=\"base_C_skirt_a_1\"),Weight=0.0195),(Item=(Type=Bone,Name=\"dyn_R_haori_front_1\"),Weight=0.0213),(Item=(Type=Bone,Name=\"pelvis\"), ...(truncated)"
    Weight=1
   )
   > &RigUnit_ParentConstraint(
    AdvancedSettings=(RotationOrderForFilter=XZY)
    Child=(Type=Bone,Name="dyn_frontHair_r_06")
    Parents="((Item=(Type=Bone,Name=\"dyn_R_skirt_a_3\"),Weight=0.0024),(Item=(Type=Bone,Name=\"base_L_skirt_a_1\"),Weight=0.0026),(Item=(Type=Bone,Name=\"dyn_R_skirt_c_2\"),Weight=0.0034),(Item=(Type=Bone,Name=\"base_R_skirt_c_1\"),Weight=0.0218),(Item=(Type=Bone,Name=\"dyn_C_ ...(truncated)"
    Weight=1
   )
   RigVMFunction_MathBoolToFloat = &RigVMFunction_MathBoolToFloat(Value=RigVMVariableNode() {variable=Is Targeting}.Value)
   AlphaInterp = &RigVMFunction_AlphaInterp(InterpSpeedDecreasing=5, InterpSpeedIncreasing=5, Scale=1, bClampResult=False, Value=@RigVMFunction_MathBoolToFloat.Result)
   > &RigUnit_ModifyTransforms(
    ItemToModify=((Item=(Type=Bone,Name="base_fronthair_r_headpivot"),Transform=(Rotation=(X=0,Y=0,Z=0,W=-1),Translation=(X=-5,Y=0,Z=6),Scale3D=(X=1,Y=1,Z=1))))
    WeightMinimum=0
    Weight=@AlphaInterp.Result
   )
   > RigVMFunctionReturnNode()
  ]}
 New Function_ContainedGraph {nodes [
   RigVMFunctionEntryNode()
   RigVMFunction_MathBoolToInteger = &RigVMFunction_MathBoolToInteger(Value=RigVMVariableNode() {variable=onlyTwist}.Value)
   RigVMFunction_MathQuaternionSwingTwist = &RigVMFunction_MathQuaternionSwingTwist(
    Input=&RigUnit_GetRelativeTransformForItem(
     bChildInitial=False
     bParentInitial=False
     Child=RigVMVariableNode() {variable=child}.Value
     Parent=RigVMVariableNode() {variable=parent}.Value
    ).RelativeTransform.Rotation
    TwistAxis=RigVMVariableNode() {variable=twistAxis}.Value
   )
   n1 = RigVMDispatchNode(
    Values=((),(),(X=0,Y=0,Z=0,W=1))
    Index=@RigVMFunction_MathBoolToInteger.Result
    Values.0=@RigVMFunction_MathQuaternionSwingTwist.Swing
    Values.1=@RigVMFunction_MathQuaternionSwingTwist.Twist
   ) {template="DISPATCH_RigVMDispatch_SelectInt32(in Index,in Values,out Result)"}
   > RigVMVariableNode(Value=@n1.Result) {variable=entry_outputQuat}
   RigVMFunction_MathQuaternionToEuler = &RigVMFunction_MathQuaternionToEuler(Value=@n1.Result, RotationOrder=RigVMVariableNode() {variable=Rotation Order}.Value)
   > DISPATCH_RigVMDispatch_Print = RigVMDispatchNode(
    ScreenColor=(R=1,G=1,B=1,A=1)
    ScreenDuration=1.5
    Value=@RigVMFunction_MathQuaternionToEuler.Result
    Enabled=RigVMVariableNode() {variable=debugRotation}.Value
   ) {template="DISPATCH_RigVMDispatch_Print(in Prefix,in Value,in Enabled,in ScreenDuration,in ScreenColor)"}
   Cast = RigVMDispatchNode(Value=RigVMVariableNode() {variable=triggerAxis}.Value) {template="DISPATCH_RigVMDispatch_CastEnumToInt(in Value,out Result)"}
   DISPATCH_RigVMDispatch_SelectInt32 = RigVMDispatchNode(
    Values=(0,0,0)
    Index=@Cast.Result
    Values.0=@RigVMFunction_MathQuaternionToEuler.Result.X
    Values.1=@RigVMFunction_MathQuaternionToEuler.Result.Y
    Values.2=@RigVMFunction_MathQuaternionToEuler.Result.Z
   ) {template="DISPATCH_RigVMDispatch_SelectInt32(in Index,in Values,out Result)"}
   n2 = RigVMVariableNode() {variable=rotationMinMax}
   n3 = RigVMVariableNode() {variable=outputMinMax}
   Remap = &RigVMFunction_MathDoubleRemap(
    bClamp=False
    Value=@DISPATCH_RigVMDispatch_SelectInt32.Result
    SourceMinimum=@n2.Value.X
    SourceMaximum=@n2.Value.Y
    TargetMinimum=@n3.Value.X
    TargetMaximum=@n3.Value.Y
   )
   Absolute = &RigVMFunction_MathDoubleAbs(Value=@Remap.Result)
   n4 = RigVMVariableNode() {variable=absoluteValue}
   Multiply = &RigVMFunction_MathDoubleMul(
    A=&RigVMFunction_MathDoubleClamp(Value=RigVMDispatchNode(True=@Absolute.Result, False=@Remap.Result, Condition=@n4.Value) {template="DISPATCH_RigVMDispatch_If(in Condition,in True,in False,out Result)"}.Result).Result
    B=RigVMVariableNode() {variable=premult}.Value
   )
   > RigVMFunctionReturnNode(alpha=@Multiply.Result, rawQuat=RigVMVariableNode() {variable=entry_outputQuat}.Value)
  ]}
 New Function_ContainedGraph {nodes [
   n1 = RigVMFunctionEntryNode()
   > For_Each = RigVMDispatchNode(Array=@n1.transforms) {template="DISPATCH_RigVMDispatch_ArrayIterator(in Array,out Element,out Index,out Count,out Ratio)"}
   > &RigUnit_OffsetTransformForItem(Item=@For_Each.Element.Item, OffsetTransform=@For_Each.Element.Transform, Weight=@n1.weights, bPropagateToChildren=@n1.inheritsTransform)
   RigVMFunctionReturnNode(ExecuteContext=@For_Each.Completed)
  ]}
 RigVMFunctionLibrary {nodes [
   AlterTransform = RigVMCollapseNode(ExecuteContext=(), inheritsTransform=false, transforms=(), weights=1)
   GetAlphaFromQuatDelta = RigVMCollapseNode(
    ExecuteContext=()
    Rotation Order=ZYX
    child=(Type=Bone,Name="None")
    outputMinMax=(X=0,Y=1)
    parent=(Type=Bone,Name="None")
    premult=1
    rotationMinMax=(X=0,Y=0)
    twistAxis=(X=1,Y=0,Z=0)
   )
  ]}
 RigVMFunction_MathBoolOr_2_ContainedGraph {nodes [
   n1 = RigVMFunctionEntryNode()
   n2 = &RigVMFunction_MathBoolOr(A=@n1.A, B=@n1.B)
   RigVMFunctionReturnNode(Result=&RigVMFunction_MathBoolOr(B=@n1.D, A=&RigVMFunction_MathBoolOr(B=@n1.C, A=@n2.Result).Result).Result)
  ]}
 RigVMFunction_MathBoolOr_2_ContainedGraph {nodes [
   n1 = RigVMFunctionEntryNode()
   RigVMFunction_MathBoolOr = &RigVMFunction_MathBoolOr(A=@n1.A, B=@n1.B)
   RigVMFunctionReturnNode(Result=&RigVMFunction_MathBoolOr(B=@n1.C, A=@RigVMFunction_MathBoolOr.Result).Result)
  ]}
 RigVMFunction_Sequence_1_ContainedGraph {nodes [
   n1 = RigVMFunctionEntryNode()
   RigVMFunction_Sequence = &RigVMFunction_Sequence(ExecuteContext=@n1.ExecuteContext)
   n2 = &RigVMFunction_Sequence(ExecuteContext=@RigVMFunction_Sequence.B)
   n3 = &RigVMFunction_Sequence(ExecuteContext=@n2.B)
   RigVMFunctionReturnNode(A=@RigVMFunction_Sequence.A, B=@n2.A, C=@n3.A, D=@n3.B)
  ]}
 RigVMFunction_Sequence_2_ContainedGraph {nodes [
   n1 = RigVMFunctionEntryNode()
   n2 = &RigVMFunction_Sequence(ExecuteContext=@n1.ExecuteContext)
   n3 = &RigVMFunction_Sequence(ExecuteContext=@n2.B)
   RigVMFunctionReturnNode(A=@n2.A, B=@n3.A, C=@n3.B)
  ]}
 RigVMModel {nodes [
   RigUnit_BeginExecution = &RigUnit_BeginExecution()
   Set_Variables = RigVMCollapseNode(ExecuteContext=@RigUnit_BeginExecution.ExecutePin)
   RigVMFunction_ControlFlowBranch = &RigVMFunction_ControlFlowBranch(ExecuteContext=@Set_Variables.ExecuteContext, Condition=RigVMVariableNode() {variable=Is rigidbody enabled}.Value)
   Back_hair_setup = RigVMCollapseNode(ExecuteContext=@RigVMFunction_ControlFlowBranch.Completed)
   Front_hair_setup_DynamicsOff = RigVMCollapseNode(ExecuteContext=@RigVMFunction_ControlFlowBranch.False)
   Front_hair_Setup_DynamicsOn = RigVMCollapseNode(ExecuteContext=@RigVMFunction_ControlFlowBranch.True)
  ]}
]
hierarchy [
 root {loc=<0 0 0>, rot=<0 0 0 1>}
 pelvis {parent=root, loc=<0 -0.5301 98.0571>, rot=<0.0127 0.707 -0.0127 -0.707>}
 spine_01 {parent=pelvis, loc=<6.1208 -0.9328 0>, rot=<0 0 0.034 -0.9994>}
 spine_02 {parent=spine_01, loc=<6.5997 -0.2095 0>, rot=<0 0 0.0238 -0.9997>}
 spine_03 {parent=spine_02, loc=<6.4914 -0.1237 0>, rot=<0 0 0.0149 -0.9999>}
 spine_04 {parent=spine_03, loc=<6.9598 0.0794 0>, rot=<0 0 -0.0132 -0.9999>}
 spine_05 {parent=spine_04, loc=<6.9987 0 0>, rot=<0 0 -0.0015 -1>}
 neck_01 {parent=spine_05, loc=<11.3138 0.1529 0>, rot=<0 0 -0.2003 -0.9797>}
 neck_02 {parent=neck_01, loc=<4.5627 0 0>, rot=<0 0 0.0504 -0.9987>}
 head {parent=neck_02, loc=<5.3914 0 0>, rot=<0 0 0.115 -0.9934>}
 base_fronthair_l_headpivot {parent=head, loc=<0 0 0>, rot=<0.0029 -0.7071 -0.0029 -0.7071>}
 base_fronthair_l {parent=base_fronthair_l_headpivot, loc=<12.0447 -0.1614 -0.5446>, rot=<-0.1299 -0.6903 -0.1996 -0.6832>}
 dyn_frontHair_l_01 {parent=base_fronthair_l, loc=<0 0 0>, rot=<0 0 0 1>}
 dyn_frontHair_l_02 {parent=dyn_frontHair_l_01, loc=<13.1586 0 0>, rot=<0 0.0809 0 -0.9967>}
 dyn_frontHair_l_03 {parent=dyn_frontHair_l_02, loc=<13.161 0 0>, rot=<0.023 0.0049 0.2074 -0.978>}
 dyn_frontHair_l_04 {parent=dyn_frontHair_l_03, loc=<12.7197 0 0>, rot=<-0.0126 -0.0088 -0.0028 -0.9999>}
 dyn_frontHair_l_05 {parent=dyn_frontHair_l_04, loc=<13.6521 0 0>, rot=<0.03 -0.0007 0.1787 -0.9834>}
 dyn_frontHair_l_06 {parent=dyn_frontHair_l_05, loc=<13.1792 0 0>, rot=<-0.0118 0.002 -0.1666 -0.986>}
 base_fronthair_r_headpivot {parent=head, loc=<0 0 0>, rot=<0.0029 -0.7071 -0.0029 -0.7071>}
 base_fronthair_r {parent=base_fronthair_r_headpivot, loc=<-12.0447 -0.1614 -0.5446>, rot=<-0.0073 -0.702 -0.1358 -0.699>}
 dyn_frontHair_r_01 {parent=base_fronthair_r, loc=<0 0 0>, rot=<0 0 0 1>}
 dyn_frontHair_r_02 {parent=dyn_frontHair_r_01, loc=<13.1586 0 0>, rot=<0.0615 -0.0349 -0.263 -0.9622>}
 dyn_frontHair_r_03 {parent=dyn_frontHair_r_02, loc=<13.161 0 0>, rot=<0.0016 0.0271 0.3676 -0.9296>}
 dyn_frontHair_r_04 {parent=dyn_frontHair_r_03, loc=<12.7197 0 0>, rot=<0.1962 0.0275 -0.0794 -0.977>}
 dyn_frontHair_r_05 {parent=dyn_frontHair_r_04, loc=<13.6521 0 0>, rot=<0.0215 -0.1777 -0.0723 -0.9812>}
 dyn_frontHair_r_06 {parent=dyn_frontHair_r_05, loc=<13.1792 0 0>, rot=<0 0 0 1>}
 base_backHair_l {parent=head, loc=<5.2683 -10.0687 -12.51>, rot=<0.105 -0.6993 0.6993 -0.105>}
 dyn_backHair_l_01 {parent=base_backHair_l, loc=<0 0 0>, rot=<0 0.0142 0 -0.9999>}
 dyn_backHair_l_02 {parent=dyn_backHair_l_01, loc=<18.2227 0 0>, rot=<0 -0.0265 0 -0.9996>}
 dyn_backHair_l_03 {parent=dyn_backHair_l_02, loc=<18.2086 0 0>, rot=<0 -0.0483 0 -0.9988>}
 dyn_backHair_l_04 {parent=dyn_backHair_l_03, loc=<18.2211 0 0>, rot=<0 -0.0589 0 -0.9983>}
 dyn_backHair_l_05 {parent=dyn_backHair_l_04, loc=<18.2095 0 0>, rot=<0 -0.0253 0 -0.9997>}
 dyn_backHair_l_06 {parent=dyn_backHair_l_05, loc=<18.2246 0 0>, rot=<0 0 0 1>}
 base_backHair_r {parent=head, loc=<5.2683 -10.0685 12.1318>, rot=<0.105 -0.6993 0.6993 -0.105>}
 dyn_backHair_r_01 {parent=base_backHair_r, loc=<0 0 0>, rot=<0 0.0142 0 -0.9999>}
 dyn_backHair_r_02 {parent=dyn_backHair_r_01, loc=<18.2227 0 0>, rot=<0 -0.0265 0 -0.9996>}
 dyn_backHair_r_03 {parent=dyn_backHair_r_02, loc=<18.2086 0 0>, rot=<0 -0.0483 0 -0.9988>}
 dyn_backHair_r_04 {parent=dyn_backHair_r_03, loc=<18.2211 0 0>, rot=<0 -0.0589 0 -0.9983>}
 dyn_backHair_r_05 {parent=dyn_backHair_r_04, loc=<18.2095 0 0>, rot=<0 -0.0253 0 -0.9997>}
 dyn_backHair_r_06 {parent=dyn_backHair_r_05, loc=<18.2246 0 0>, rot=<0 0 0 1>}
 base_backHair_c {parent=head, loc=<5.2907 -12.7764 -0.1892>, rot=<0.105 -0.6993 0.6993 -0.105>}
 dyn_backHair_c_01 {parent=base_backHair_c, loc=<0 0 0>, rot=<0 0 0 1>}
 dyn_backHair_c_02 {parent=dyn_backHair_c_01, loc=<18.2227 0 0>, rot=<0 -0.0265 0 -0.9996>}
 dyn_backHair_c_03 {parent=dyn_backHair_c_02, loc=<18.2086 0 0>, rot=<0 -0.0483 0 -0.9988>}
 dyn_backHair_c_04 {parent=dyn_backHair_c_03, loc=<18.2211 0 0>, rot=<0 -0.0589 0 -0.9983>}
 dyn_backHair_c_05 {parent=dyn_backHair_c_04, loc=<18.2095 0 0>, rot=<0 -0.0253 0 -0.9997>}
 dyn_backHair_c_06 {parent=dyn_backHair_c_05, loc=<18.2246 0 0>, rot=<0 0 0 1>}
 base_hairtie_1 {parent=head, loc=<17.522 -0.505 -11.9164>, rot=<0.0029 -0.7071 -0.0029 -0.7071>}
 dyn_hairtie_1 {parent=base_hairtie_1, loc=<0 0 0>, rot=<0 0 0 1>}
 base_hairtie_2 {parent=head, loc=<15.6073 0.2394 -12.793>, rot=<0.0029 -0.7071 -0.0029 -0.7071>}
 dyn_hairtie_2 {parent=base_hairtie_2, loc=<0 0 0>, rot=<0 0 0 1>}
 head_null {parent=neck_02, type=Null, loc=<5.3914 0 0>, rot=<0 0 0.115 -0.9934>}
 clavicle_l {parent=spine_05, loc=<5.4831 -0.1053 -3.2946>, rot=<0.0465 -0.7289 -0.009 -0.6829>}
 upperarm_l {parent=clavicle_l, loc=<9.8658 -1.2136 -0.4105>, rot=<0.0164 -0.3245 -0.0269 -0.9454>}
 lowerarm_l {parent=upperarm_l, loc=<26.465 -0.8345 0.259>, rot=<0.0036 0.0161 -0.2298 -0.9731>}
 hand_l {parent=lowerarm_l, loc=<22.7355 -0.3203 -0.6105>, rot=<-0.6784 -0.0223 0.0735 -0.7307>}
 lowerarm_twist_02_l {parent=lowerarm_l, loc=<10.6707 0 0>, rot=<0 0 0 1>}
 lowerarm_twist_01_l {parent=lowerarm_l, loc=<18.316 0 0>, rot=<0 0 0 1>}
 upperarm_twist_01_l {parent=upperarm_l, loc=<7.2135 -0.1134 0>, rot=<0 0 0 1>}
 upperarm_twist_02_l {parent=upperarm_l, loc=<19.7998 -0.1134 0>, rot=<0 0 0 1>}
 upperarm_l_null {parent=clavicle_l, type=Null, loc=<9.8658 -1.2136 -0.4105>, rot=<0.0164 -0.3245 -0.0269 -0.9454>}
 clavicle_r {parent=spine_05, loc=<5.4831 -0.1053 3.2946>, rot=<0.7289 0.0465 -0.6829 0.009>}
 upperarm_r {parent=clavicle_r, loc=<-9.8658 1.2136 0.4105>, rot=<0.0164 -0.3245 -0.0269 -0.9454>}
 lowerarm_r {parent=upperarm_r, loc=<-26.465 0.8345 -0.259>, rot=<0.0036 0.0161 -0.2298 -0.9731>}
 hand_r {parent=lowerarm_r, loc=<-22.7355 0.3203 0.6106>, rot=<-0.6784 -0.0223 0.0735 -0.7307>}
 lowerarm_twist_02_r {parent=lowerarm_r, loc=<-10.6707 0 0>, rot=<0 0 0 1>}
 lowerarm_twist_01_r {parent=lowerarm_r, loc=<-18.316 0 0>, rot=<0 0 0 1>}
 upperarm_twist_01_r {parent=upperarm_r, loc=<-7.2135 0.1134 0>, rot=<0 0 0 1>}
 upperarm_twist_02_r {parent=upperarm_r, loc=<-19.7998 0.1134 0>, rot=<0 0 0 1>}
 upperarm_r_null {parent=clavicle_r, type=Null, loc=<-9.8658 1.2136 0.4105>, rot=<0.0164 -0.3245 -0.0269 -0.9454>}
 clavicle_l_null {parent=spine_05, type=Null, loc=<5.483 -0.1053 -3.2946>, rot=<0.0465 -0.7289 -0.009 -0.6829>}
 clavicle_r_null {parent=spine_05, type=Null, loc=<5.483 -0.1053 3.2946>, rot=<0.7289 0.0465 -0.6829 0.009>}
 thigh_l {parent=pelvis, loc=<-1.0855 -0.2109 -8.7053>, rot=<0.0009 0.0355 0.0155 -0.9993>}
 thigh_twist_01_l {parent=thigh_l, loc=<-27.952 0 0>, rot=<0.0013 0 0 -1>}
 thigh_r {parent=pelvis, loc=<-1.0855 -0.2109 8.7053>, rot=<0.0355 -0.0009 0.9993 0.0155>}
 thigh_twist_01_r {parent=thigh_r, loc=<27.952 0 0>, rot=<0.0013 0 0 -1>}
 spine_05_null {parent=pelvis, type=Null, loc=<32.9717 -4.2232 0>, rot=<0 0 0.058 -0.9983>}
 pelvis_skirt_offset {parent=pelvis, loc=<0 0 0>, rot=<0 0 0 1>}
 base_C_skirt_a_1 {parent=pelvis_skirt_offset, loc=<-0.8659 8.3629 0.0001>, rot=<0.0127 0.707 0.707 0.0127>}
 dyn_C_skirt_a_1 {parent=base_C_skirt_a_1, loc=<16.3581 0 0>, rot=<0 0 0 1>}
 dyn_C_skirt_a_2 {parent=dyn_C_skirt_a_1, loc=<0 0 0>, rot=<0 0 0 1>}
 dyn_C_skirt_a_3 {parent=dyn_C_skirt_a_2, loc=<16.5916 0 0>, rot=<0 0 0 1>}
 pelvis_skirt_haori {parent=pelvis, loc=<0 0 0>, rot=<0 0 0 1>}
 dyn_L_haori_side_mstr {parent=pelvis_skirt_haori, loc=<5.6551 -9.8488 -13.3436>, rot=<0.039 -0.8852 0.4523 -0.1018>}
 dyn_L_haori_side_1 {parent=dyn_L_haori_side_mstr, loc=<0 0 0>, rot=<0 0 0 1>}
 dyn_L_haori_side_2 {parent=dyn_L_haori_side_1, loc=<17 0 0>, rot=<0 0 0 1>}
 dyn_L_haori_side_3 {parent=dyn_L_haori_side_2, loc=<17 0 0>, rot=<0 0 0 1>}
 dyn_L_haori_side_4 {parent=dyn_L_haori_side_3, loc=<17 0 0>, rot=<0 0 0 1>}
 dyn_C_haori_back_mstr {parent=pelvis_skirt_haori, loc=<4.7337 -14.0213 0>, rot=<0.0674 -0.7039 0.7039 -0.0674>}
 dyn_C_haori_back_1 {parent=dyn_C_haori_back_mstr, loc=<0 0 0>, rot=<0 0 0 1>}
 dyn_C_haori_back_2 {parent=dyn_C_haori_back_1, loc=<17 0 0>, rot=<0 0 0 1>}
 dyn_C_haori_back_3 {parent=dyn_C_haori_back_2, loc=<17 0 0>, rot=<0 0 0 1>}
 dyn_C_haori_back_4 {parent=dyn_C_haori_back_3, loc=<17 0 0>, rot=<0 0 0 1>}
 dyn_R_haori_side_mstr {parent=pelvis_skirt_haori, loc=<5.6547 -9.8488 13.3436>, rot=<0.1018 -0.4523 0.8852 -0.039>}
 dyn_R_haori_side_1 {parent=dyn_R_haori_side_mstr, loc=<0 0 0>, rot=<0 0 0 1>}
 dyn_R_haori_side_2 {parent=dyn_R_haori_side_1, loc=<17 0 0.0001>, rot=<0 0 0 1>}
 dyn_R_haori_side_3 {parent=dyn_R_haori_side_2, loc=<17 0 -0.0001>, rot=<0 0 0 1>}
 dyn_R_haori_side_4 {parent=dyn_R_haori_side_3, loc=<17 0 0>, rot=<0 0 0 1>}
 pelvis_null {parent=root, type=Null, loc=<0 -0.5301 98.0571>, rot=<0.0127 0.707 -0.0127 -0.707>}
 dyn_backHair_l_01_ctrl_null {type=Null, loc=<12.51 -11.7992 157.2191>, rot=<-0.4226 -0.4226 0.5669 -0.5669>}
 dyn_backHair_l_01_ctrl {controlType=EULER_TRANSFORM, parent=dyn_backHair_l_01_ctrl_null, type=Control, loc=<0 0 0>, rot=<0 0 0 1>}
 dyn_backHair_l_02_ctrl_null {type=Null, loc=<12.51 -14.7397 147.3547>, rot=<-0.3985 -0.3985 0.5841 -0.5841>}
 dyn_backHair_l_02_ctrl {
  controlType=EULER_TRANSFORM
  parent=dyn_backHair_l_02_ctrl_null
  type=Control
  loc=<14.5496 -14.7397 146.6514>
  rot=<0.3736 0.4226 -0.6003 0.5669>
 }
 dyn_backHair_l_03_ctrl_null {type=Null, loc=<12.51 -18.9465 136.6214>, rot=<-0.4359 -0.4359 0.5568 -0.5568>}
 dyn_backHair_l_03_ctrl {
  controlType=EULER_TRANSFORM
  parent=dyn_backHair_l_03_ctrl_null
  type=Control
  loc=<-16.9119 -18.9465 135.5742>
  rot=<0.4715 0.3985 -0.527 0.5841>
 }
 dyn_backHair_l_04_ctrl_null {type=Null, loc=<12.51 -22.0047 124.2458>, rot=<-0.4526 -0.4526 0.5433 -0.5433>}
 dyn_backHair_l_04_ctrl {
  controlType=EULER_TRANSFORM
  parent=dyn_backHair_l_04_ctrl_null
  type=Control
  loc=<-7.7744 -22.0047 124.0025>
  rot=<0.4689 0.4359 -0.5293 0.5568>
 }
 dyn_backHair_l_05_ctrl_null {type=Null, loc=<12.51 -24.425 111.0623>, rot=<-0.4633 -0.4633 0.5342 -0.5342>}
 dyn_backHair_l_05_ctrl {
  controlType=EULER_TRANSFORM
  parent=dyn_backHair_l_05_ctrl_null
  type=Control
  loc=<-5.2821 -24.425 110.9402>
  rot=<0.4737 0.4526 -0.525 0.5433>
 }
 dyn_backHair_l_06_ctrl_null {type=Null, loc=<12.51 -26.2489 98.3071>, rot=<-0.5 -0.5 0.5 -0.5>}
 dyn_backHair_l_06_ctrl {
  controlType=EULER_TRANSFORM
  parent=dyn_backHair_l_06_ctrl_null
  type=Control
  loc=<-14.2871 -26.2489 97.2641>
  rot=<0.5342 0.4633 -0.4633 0.5342>
 }
 dyn_backHair_r_01_ctrl_null {type=Null, loc=<-12.1318 -11.799 157.2191>, rot=<-0.4226 -0.4226 0.5669 -0.5669>}
 dyn_backHair_r_01_ctrl {controlType=EULER_TRANSFORM, parent=dyn_backHair_r_01_ctrl_null, type=Control, loc=<0 0 0>, rot=<0 0 0 1>}
 dyn_backHair_r_02_ctrl_null {type=Null, loc=<-12.1318 -14.7395 147.3547>, rot=<-0.3985 -0.3985 0.5841 -0.5841>}
 dyn_backHair_r_02_ctrl {
  controlType=EULER_TRANSFORM
  parent=dyn_backHair_r_02_ctrl_null
  type=Control
  loc=<-10.0056 -14.7395 148.7151>
  rot=<0.3736 0.4226 -0.6003 0.5669>
 }
 dyn_backHair_r_03_ctrl_null {type=Null, loc=<-12.1318 -18.9464 136.6214>, rot=<-0.4359 -0.4359 0.5568 -0.5568>}
 dyn_backHair_r_03_ctrl {
  controlType=EULER_TRANSFORM
  parent=dyn_backHair_r_03_ctrl_null
  type=Control
  loc=<-41.3417 -18.9464 132.3483>
  rot=<0.4715 0.3985 -0.527 0.5841>
 }
 dyn_backHair_r_04_ctrl_null {type=Null, loc=<-12.1318 -22.0046 124.2458>, rot=<-0.4526 -0.4526 0.5433 -0.5433>}
 dyn_backHair_r_04_ctrl {
  controlType=EULER_TRANSFORM
  parent=dyn_backHair_r_04_ctrl_null
  type=Control
  loc=<-32.3708 -22.0046 122.5077>
  rot=<0.4689 0.4359 -0.5293 0.5568>
 }
 dyn_backHair_r_05_ctrl_null {type=Null, loc=<-12.1318 -24.4248 111.0623>, rot=<-0.4633 -0.4633 0.5342 -0.5342>}
 dyn_backHair_r_05_ctrl {
  controlType=EULER_TRANSFORM
  parent=dyn_backHair_r_05_ctrl_null
  type=Control
  loc=<-29.9047 -24.4248 109.9664>
  rot=<0.4737 0.4526 -0.525 0.5433>
 }
 dyn_backHair_r_06_ctrl_null {type=Null, loc=<-12.1318 -26.2487 98.3071>, rot=<-0.5 -0.5 0.5 -0.5>}
 dyn_backHair_r_06_ctrl {
  controlType=EULER_TRANSFORM
  parent=dyn_backHair_r_06_ctrl_null
  type=Control
  loc=<-38.6808 -26.2487 93.776>
  rot=<0.5342 0.4633 -0.4633 0.5342>
 }
 dyn_backHair_c_01_ctrl_null {type=Null, loc=<0.1892 -14.507 157.2191>, rot=<-0.4226 -0.4226 0.5669 -0.5669>}
 dyn_backHair_c_01_ctrl {controlType=EULER_TRANSFORM, parent=dyn_backHair_c_01_ctrl_null, type=Control, loc=<0 0 0>, rot=<0 0 0 1>}
 dyn_backHair_c_02_ctrl_null {type=Null, loc=<0.1892 -17.4475 147.3547>, rot=<-0.3985 -0.3985 0.5841 -0.5841>}
 dyn_backHair_c_02_ctrl {
  controlType=EULER_TRANSFORM
  parent=dyn_backHair_c_02_ctrl_null
  type=Control
  loc=<2.2721 -17.4475 147.6832>
  rot=<0.3736 0.4226 -0.6003 0.5669>
 }
 dyn_backHair_c_03_ctrl_null {type=Null, loc=<0.1892 -21.6544 136.6214>, rot=<-0.4359 -0.4359 0.5568 -0.5568>}
 dyn_backHair_c_03_ctrl {
  controlType=EULER_TRANSFORM
  parent=dyn_backHair_c_03_ctrl_null
  type=Control
  loc=<-29.1267 -21.6544 133.9613>
  rot=<0.4715 0.3985 -0.527 0.5841>
 }
 dyn_backHair_c_04_ctrl_null {type=Null, loc=<0.1892 -24.7126 124.2458>, rot=<-0.4526 -0.4526 0.5433 -0.5433>}
 dyn_backHair_c_04_ctrl {
  controlType=EULER_TRANSFORM
  parent=dyn_backHair_c_04_ctrl_null
  type=Control
  loc=<-20.0725 -24.7126 123.2551>
  rot=<0.4689 0.4359 -0.5293 0.5568>
 }
 dyn_backHair_c_05_ctrl_null {type=Null, loc=<0.1892 -27.1328 111.0623>, rot=<-0.4633 -0.4633 0.5342 -0.5342>}
 dyn_backHair_c_05_ctrl {
  controlType=EULER_TRANSFORM
  parent=dyn_backHair_c_05_ctrl_null
  type=Control
  loc=<-17.5933 -27.1328 110.4533>
  rot=<0.4737 0.4526 -0.525 0.5433>
 }
 dyn_backHair_c_06_ctrl_null {type=Null, loc=<0.1892 -28.9567 98.3071>, rot=<-0.5 -0.5 0.5 -0.5>}
 dyn_backHair_c_06_ctrl {
  controlType=EULER_TRANSFORM
  parent=dyn_backHair_c_06_ctrl_null
  type=Control
  loc=<-26.4838 -28.9567 95.5201>
  rot=<0.5342 0.4633 -0.4633 0.5342>
 }
]
variables [
 ArmBackFix {type=double, default=0}
 Clavicle_front_rotation_l {type=double, default=0}
 Clavicle_front_rotation_r {type=double, default=0}
 Clavicle_up_rotation_l {type=double, default=0}
 Clavicle_up_rotation_r {type=double, default=0}
 CrouchingInPlace {type=double, default=0}
 CrouchingMovingInPlace {type=double, default=0}
 HeadLookingUpCrouching {type=double, default=0}
 HeadRotDown {type=double, default=0}
 HeadRotUp {type=double, default=0}
 HeadRotUpExtreme {type=double, default=0}
 HeadYrotation {type=double, default=0}
 Is Crouching {type=bool, default=False}
 Is Jumping {type=bool, default=False}
 Is Targeting {type=bool, default=False}
 Is rigidbody enabled {type=bool, default=False}
 Pub_BasebalSliding {type=bool, default=False}
 Pub_IsFrontEnd {type=bool, default=False}
 Pub_IsParachuteOpened {type=bool, default=False}
 Pub_IsSkydive {type=bool, default=False}
 Pub_IsSwimming {type=bool, default=False}
 Pub_IsTacticalJumps {type=bool, default=False}
 Pub_IsTargeting {type=bool, default=False}
 Pub_Sprinting {type=bool, default=False}
 Pub_SwingPickaxe {type=bool, default=False}
 Pub_TacticalSprint {type=bool, default=False}
 Pub_WearingPack {type=bool, default=False}
 Pub_kneeSliding {type=bool, default=False}
 TorsoTwistRight {type=double, default=0}
 WearingCape {type=bool, default=False}
 is floating {type=bool, default=False}
]
