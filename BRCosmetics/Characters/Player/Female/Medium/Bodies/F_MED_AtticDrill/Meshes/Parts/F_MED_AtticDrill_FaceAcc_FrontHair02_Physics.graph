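# Physics asset for the FrontHair02 face accessory of the female AtticDrill character, set up for secondary motion of two front hair strands.
# Left and right six-bone capsule chains hang from base fronthair bodies, with constraints setting narrow swing limits and a zero twist limit.
# Capsules on head, neck, spine, pelvis, arms and thighs give body shapes for collision.
# Skirt bodies are also present but have no constraints.
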PhysicsAsset /BRCosmetics/Characters/Player/Female/Medium/Bodies/F_MED_AtticDrill/Meshes/Parts/F_MED_AtticDrill_FaceAcc_FrontHair02_Physics
bodies {
 base_C_skirt_a_1 [{shapeType=Capsule, length=3.593, radius=3.0782, shapeName=base_C_skirt_a_1_capsule, rotation=<rot 0 -90 0>}]
 base_fronthair_l [{shapeType=Capsule, length=1, radius=1, shapeName=base_fronthair_l_capsule}]
 base_fronthair_r [{shapeType=Capsule, length=1, radius=1, shapeName=base_fronthair_r_capsule}]
 clavicle_l [{shapeType=Capsule, length=9.1615, radius=8.6095, shapeName=clavicle_l_capsule, center=<6.148 0 -4.3194>, rotation=<rot 0 -90 88.2999>}]
 clavicle_r [{shapeType=Capsule, length=9.1615, radius=9.5346, shapeName=clavicle_r_capsule, center=<-6.148 2.4075 4.3194>, rotation=<rot 0 -90 -91.7001>}]
 dyn_C_skirt_a_1 [{shapeType=Capsule, length=14.4886, radius=10.7028, shapeName=dyn_C_skirt_a_1_capsule, rotation=<rot -83.4175 3.6269 -93.6031>}]
 dyn_C_skirt_a_2 [{shapeType=Capsule, length=14.4886, radius=10.7028, shapeName=dyn_C_skirt_a_2_capsule, rotation=<rot -83.4175 3.6269 -93.6031>}]
 dyn_C_skirt_a_3 [{shapeType=Capsule, length=20.1495, radius=10.7028, shapeName=dyn_C_skirt_a_3_capsule, rotation=<rot -83.4175 3.6269 -93.6031>}]
 dyn_frontHair_l_01 [{
   shapeType=Capsule
   length=5.7477
   radius=3.5925
   shapeName=dyn_frontHair_l_01_capsule
   center=<2.3144 0.0499 0.0633>
   rotation=<rot 0 -85.1559 -86.1102>
  }]
 dyn_frontHair_l_02 [{
   shapeType=Capsule
   length=15.2787
   radius=1.736
   shapeName=dyn_frontHair_l_02_capsule
   center=<3.447 0.1845 0.2864>
   rotation=<rot -0.2918 91.2956 92.0012>
  }]
 dyn_frontHair_l_03 [{
   shapeType=Capsule
   length=10.1433
   radius=1.5125
   shapeName=dyn_frontHair_l_03_capsule
   center=<6.2959 -0.2677 -0.1062>
   rotation=<rot 0 -91.1365 -88.2884>
  }]
 dyn_frontHair_l_04 [{
   shapeType=Capsule
   length=11.5622
   radius=1.8635
   shapeName=dyn_frontHair_l_04_capsule
   center=<5.8259 0.1479 0.0206>
   rotation=<rot 0 -87.8016 -88.189>
  }]
 dyn_frontHair_l_05 [{
   shapeType=Capsule
   length=8.8793
   radius=2.2573
   shapeName=dyn_frontHair_l_05_capsule
   center=<5.6232 -0.3721 -0.1818>
   rotation=<rot 0 -94.1159 -84.4589>
  }]
 dyn_frontHair_l_06 [{
   shapeType=Capsule
   length=11.6926
   radius=2.9398
   shapeName=dyn_frontHair_l_06_capsule
   center=<11.2874 0.3473 -1.1561>
   rotation=<rot 0 92.8376 -88.8539>
  }]
 dyn_frontHair_r_01 [{
   shapeType=Capsule
   length=4.9261
   radius=2.558
   shapeName=dyn_frontHair_r_01_capsule
   center=<2.807 -0.447 0.3285>
   rotation=<rot 0 -104.2985 -84.3891>
  }]
 dyn_frontHair_r_02 [{
   shapeType=Capsule
   length=15.2787
   radius=1.601
   shapeName=dyn_frontHair_r_02_capsule
   center=<4.8877 0.7942 0.0015>
   rotation=<rot 0 83.0014 -87.9776>
  }]
 dyn_frontHair_r_03 [{
   shapeType=Capsule
   length=7.9185
   radius=1.9858
   shapeName=dyn_frontHair_r_03_capsule
   center=<6.6127 0.6624 0.1444>
   rotation=<rot 0 91.1109 -88.0817>
  }]
 dyn_frontHair_r_04 [{
   shapeType=Capsule
   length=10.3245
   radius=1.476
   shapeName=dyn_frontHair_r_04_capsule
   center=<5.8745 0.2124 0.1749>
   rotation=<rot 0 -88.9314 -89.3938>
  }]
 dyn_frontHair_r_05 [{
   shapeType=Capsule
   length=8.2712
   radius=2.6019
   shapeName=dyn_frontHair_r_05_capsule
   center=<5.4543 0.4779 -0.5127>
   rotation=<rot 0 -90.7497 -89.3432>
  }]
 dyn_frontHair_r_06 [{
   shapeType=Capsule
   length=8.8567
   radius=2.3014
   shapeName=dyn_frontHair_r_06_capsule
   center=<8.0703 -0.0323 0.4535>
   rotation=<rot 0 -90.5649 -88.5703>
  }]
 head [{
   shapeType=Capsule
   length=4.4059
   radius=9.1631
   shapeName=head_capsule
   center=<5.8861 1.0984 -0.2516>
   rotation=<rot 0.5968 -81.3132 -87.6122>
  }]
 lowerarm_l [{shapeType=Capsule, length=16.2169, radius=8.5378, shapeName=upperarm_l_capsule, center=<9.84 0 -1.3992>, rotation=<rot 0 -90 90.8518>}]
 lowerarm_r [{shapeType=Capsule, length=16.2169, radius=8.5378, shapeName=upperarm_r_capsule, center=<-9.84 0 1.3992>, rotation=<rot 0 -90 -89.1482>}]
 neck_02 [{
   shapeType=Capsule
   length=4.6294
   radius=8.7083
   shapeName=neck_02_capsule
   center=<4.4027 -4.4246 0.2927>
   rotation=<rot 2.8696 -90 92.8458>
  }]
 pelvis [{shapeType=Capsule, length=17.1802, radius=12.4332, shapeName=pelvis_capsule, rotation=<rot 0 -90 0>}]
 spine_01 [{shapeType=Capsule, length=17.1802, radius=12.4332, shapeName=pelvis_capsule, rotation=<rot 0 -90 0>}]
 spine_02 [{shapeType=Capsule, length=13.0154, radius=12.1676, shapeName=pelvis_capsule, rotation=<rot 0 -90 0>}]
 spine_03 [{shapeType=Capsule, length=13.106, radius=10.2365, shapeName=pelvis_capsule, rotation=<rot 0 -90 0>}]
 spine_04 [{shapeType=Capsule, length=14.8856, radius=11.2357, shapeName=pelvis_capsule, rotation=<rot 0 -90 0>}]
 spine_05 [{shapeType=Capsule, length=22.4139, radius=7.5173, shapeName=pelvis_capsule, rotation=<rot 0 -90 0>}]
 thigh_l [{shapeType=Capsule, length=16.2169, radius=15.0105, shapeName=upperarm_l_capsule, center=<-18.6067 0 1.2389>, rotation=<rot 0 -90 93.6869>}]
 thigh_r [{shapeType=Capsule, length=16.2169, radius=15.0105, shapeName=upperarm_r_capsule, center=<18.6067 0 -1.2389>, rotation=<rot 0 -90 -86.3131>}]
 thigh_twist_01_l [{shapeType=Capsule, length=16.2169, radius=17.4657, shapeName=upperarm_l_capsule, center=<-13.3216 0 -5.0647>, rotation=<rot 0 -90 93.6869>}]
 thigh_twist_01_r [{shapeType=Capsule, length=16.2169, radius=17.4657, shapeName=upperarm_r_capsule, center=<13.1747 0 2.7864>, rotation=<rot 0 -90 -86.3131>}]
 upperarm_l [{
   shapeType=Capsule
   length=16.2169
   radius=7.7436
   shapeName=upperarm_l_capsule
   center=<9.8917 -0.1735 -0.5975>
   rotation=<rot 0 -90 93.6869>
  }]
 upperarm_r [{
   shapeType=Capsule
   length=16.2169
   radius=7.3643
   shapeName=upperarm_r_capsule
   center=<-9.8917 0.6362 0.5975>
   rotation=<rot 0 -90 -86.3131>
  }]
}
constraints [
 {bone1Name=dyn_frontHair_l_01, bone2Name=base_fronthair_l, swing1LimitDegrees=30, swing2LimitDegrees=30, twistLimitDegrees=0}
 {bone1Name=dyn_frontHair_l_02, bone2Name=dyn_frontHair_l_01, swing1LimitDegrees=30, swing2LimitDegrees=10, twistLimitDegrees=0}
 {bone1Name=dyn_frontHair_l_03, bone2Name=dyn_frontHair_l_02, swing1LimitDegrees=30, swing2LimitDegrees=30, twistLimitDegrees=0}
 {bone1Name=dyn_frontHair_l_04, bone2Name=dyn_frontHair_l_03, swing1LimitDegrees=30, swing2LimitDegrees=30, twistLimitDegrees=0}
 {bone1Name=dyn_frontHair_l_05, bone2Name=dyn_frontHair_l_04, swing1LimitDegrees=30, swing2LimitDegrees=30, twistLimitDegrees=0}
 {bone1Name=dyn_frontHair_l_06, bone2Name=dyn_frontHair_l_05, swing1LimitDegrees=30, swing2LimitDegrees=30, twistLimitDegrees=0}
 {bone1Name=dyn_frontHair_r_01, bone2Name=base_fronthair_r, swing1LimitDegrees=30, swing2LimitDegrees=10, twistLimitDegrees=0}
 {bone1Name=dyn_frontHair_r_02, bone2Name=dyn_frontHair_r_01, swing1LimitDegrees=30, swing2LimitDegrees=10, twistLimitDegrees=0}
 {bone1Name=dyn_frontHair_r_03, bone2Name=dyn_frontHair_r_02, swing1LimitDegrees=30, swing2LimitDegrees=30, twistLimitDegrees=0}
 {bone1Name=dyn_frontHair_r_04, bone2Name=dyn_frontHair_r_03, swing1LimitDegrees=30, swing2LimitDegrees=30, twistLimitDegrees=0}
 {bone1Name=dyn_frontHair_r_05, bone2Name=dyn_frontHair_r_04, swing1LimitDegrees=30, swing2LimitDegrees=30, twistLimitDegrees=0}
 {bone1Name=dyn_frontHair_r_06, bone2Name=dyn_frontHair_r_05, swing1LimitDegrees=30, swing2LimitDegrees=30, twistLimitDegrees=0}
]
