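# Animation Blueprint for the F_MED_AtticDrill face accessory mesh, adding secondary motion to the character's hair.
# It copies the attached parent's pose, blends in limb, haori and skirt bones from SkelMesh, then runs a Control Rig fed by movement and equipment states.
# Chained AnimDynamics springs simulate back-hair strands from that cached pose.
# Additive bone offsets, mostly bool-gated, adjust the cached dynamics result before output.
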AnimBlueprint /BRCosmetics/Characters/Player/Female/Medium/Bodies/F_MED_AtticDrill/Meshes/Parts/F_MED_AtticDrill_FaceAcc_AnimBP
target_skeleton=&/F_MED_AtticDrill_FaceAcc_Skeleton
graphs [
 AnimGraph {nodes [
   n1 = "Transform (Modify) Bone - Bone: dyn_backHair_l_02"!(
    ComponentPose=Local To Component(LocalPose=Use cached pose 'CP_Dynamics'())
    bAlphaBoolEnabled:bool=True
    Translation="0,5,15"
    Rotation="0,0,0"
   ) -> (Pose) {
    alphaInputType=Bool
    rotationMode=BMM_Additive
    rotationSpace=BCS_BoneSpace
    translationMode=BMM_Additive
    translationSpace=BCS_BoneSpace
    alphaBoolBlend {blendInTime=0.1, blendOutTime=0.1}
    alphaScaleBiasClamp {bInterpResult=true, interpSpeedDecreasing=2, interpSpeedIncreasing=5}
    boneToModify {boneName=dyn_backHair_l_02}
    translation {y=5, z=15}
   }
   n2 = "Transform (Modify) Bone - Bone: dyn_backHair_c_02"!(ComponentPose=@n1, bAlphaBoolEnabled:bool=True, Translation="0,0,15", Rotation="10,0,0") -> (Pose) {
    alphaInputType=Bool
    rotationMode=BMM_Additive
    rotationSpace=BCS_BoneSpace
    translationMode=BMM_Additive
    translationSpace=BCS_BoneSpace
    alphaBoolBlend {blendInTime=0.1, blendOutTime=0.1}
    alphaScaleBiasClamp {bInterpResult=true, interpSpeedDecreasing=2, interpSpeedIncreasing=5}
    boneToModify {boneName=dyn_backHair_c_02}
    rotation {pitch=10}
    translation {z=15}
   }
   n3 = "Transform (Modify) Bone - Bone: dyn_backHair_r_02"!(ComponentPose=@n2, bAlphaBoolEnabled:bool=True, Translation="5,-5,15", Rotation="0,0,0") -> (Pose) {
    alphaInputType=Bool
    rotationMode=BMM_Additive
    rotationSpace=BCS_BoneSpace
    translationMode=BMM_Additive
    translationSpace=BCS_BoneSpace
    alphaBoolBlend {blendInTime=0.1, blendOutTime=0.1}
    alphaScaleBiasClamp {bInterpResult=true, interpSpeedDecreasing=2, interpSpeedIncreasing=5}
    boneToModify {boneName=dyn_backHair_r_02}
    translation=<5 -5 15>
   }
   n4 = "Transform (Modify) Bone - Bone: dyn_backHair_l_03"!(ComponentPose=@n3, bAlphaBoolEnabled:bool=True, Translation="0,0,5") -> (Pose) {
    alphaInputType=Bool
    rotationMode=BMM_Additive
    rotationSpace=BCS_BoneSpace
    scaleMode=BMM_Additive
    scaleSpace=BCS_BoneSpace
    translationMode=BMM_Additive
    translationSpace=BCS_BoneSpace
    alphaBoolBlend {blendInTime=0.1, blendOutTime=0.1}
    alphaScaleBiasClamp {bInterpResult=true, interpSpeedDecreasing=5, interpSpeedIncreasing=5}
    boneToModify {boneName=dyn_backHair_l_03}
    translation {z=5}
   }
   n5 = "Transform (Modify) Bone - Bone: dyn_backHair_c_03"!(ComponentPose=@n4, bAlphaBoolEnabled:bool=True, Translation="0,0,10") -> (Pose) {
    alphaInputType=Bool
    rotationMode=BMM_Additive
    rotationSpace=BCS_BoneSpace
    scaleMode=BMM_Additive
    scaleSpace=BCS_BoneSpace
    translationMode=BMM_Additive
    translationSpace=BCS_BoneSpace
    alphaBoolBlend {blendInTime=0.1, blendOutTime=0.1}
    alphaScaleBiasClamp {bInterpResult=true, interpSpeedDecreasing=5, interpSpeedIncreasing=5}
    boneToModify {boneName=dyn_backHair_c_03}
    translation {z=10}
   }
   n6 = "Transform (Modify) Bone - Bone: dyn_backHair_r_03"!(ComponentPose=@n5, bAlphaBoolEnabled:bool=True, Translation="0,0,10") -> (Pose) {
    alphaInputType=Bool
    rotationMode=BMM_Additive
    rotationSpace=BCS_BoneSpace
    scaleMode=BMM_Additive
    scaleSpace=BCS_BoneSpace
    translationMode=BMM_Additive
    translationSpace=BCS_BoneSpace
    alphaBoolBlend {blendInTime=0.1, blendOutTime=0.1}
    alphaScaleBiasClamp {bInterpResult=true, interpSpeedDecreasing=5, interpSpeedIncreasing=5}
    boneToModify {boneName=dyn_backHair_r_03}
    translation {z=10}
   }
   n7 = "Transform (Modify) Bone - Bone: dyn_backHair_l_03"!(ComponentPose=@n6, Alpha:float=1, Translation="0,0,-5", Rotation:Rotator="0,0,0", Scale:Vector="1,1,1") -> (Pose) {
    rotationMode=BMM_Additive
    rotationSpace=BCS_BoneSpace
    scaleMode=BMM_Additive
    scaleSpace=BCS_BoneSpace
    translationMode=BMM_Additive
    translationSpace=BCS_BoneSpace
    alphaBoolBlend {blendInTime=0.1, blendOutTime=0.1}
    alphaScaleBiasClamp {bInterpResult=true, interpSpeedDecreasing=5, interpSpeedIncreasing=5}
    boneToModify {boneName=dyn_backHair_l_03}
    translation {z=-5}
   }
   n8 = "Transform (Modify) Bone - Bone: dyn_backHair_c_03"!(ComponentPose=@n7, Alpha:float=1, Translation="0,0,-5", Rotation:Rotator="0,0,0", Scale:Vector="1,1,1") -> (Pose) {
    rotationMode=BMM_Additive
    rotationSpace=BCS_BoneSpace
    scaleMode=BMM_Additive
    scaleSpace=BCS_BoneSpace
    translationMode=BMM_Additive
    translationSpace=BCS_BoneSpace
    alphaBoolBlend {blendInTime=0.1, blendOutTime=0.1}
    alphaScaleBiasClamp {bInterpResult=true, interpSpeedDecreasing=5, interpSpeedIncreasing=5}
    boneToModify {boneName=dyn_backHair_c_03}
    translation {z=-5}
   }
   n9 = "Transform (Modify) Bone - Bone: dyn_backHair_r_03"!(ComponentPose=@n8, Alpha:float=1, Translation="0,0,-5", Rotation:Rotator="0,0,0", Scale:Vector="1,1,1") -> (Pose) {
    rotationMode=BMM_Additive
    rotationSpace=BCS_BoneSpace
    scaleMode=BMM_Additive
    scaleSpace=BCS_BoneSpace
    translationMode=BMM_Additive
    translationSpace=BCS_BoneSpace
    alphaBoolBlend {blendInTime=0.1, blendOutTime=0.1}
    alphaScaleBiasClamp {bInterpResult=true, interpSpeedDecreasing=5, interpSpeedIncreasing=5}
    boneToModify {boneName=dyn_backHair_r_03}
    translation {z=-5}
   }
   n10 = "Transform (Modify) Bone - Bone: dyn_backHair_l_02"(ComponentPose=@n9, bAlphaBoolEnabled=True, Translation="0,-2,15", Rotation="0,0,0") {
    alphaInputType=Bool
    rotationMode=BMM_Additive
    rotationSpace=BCS_BoneSpace
    translationMode=BMM_Additive
    translationSpace=BCS_BoneSpace
    alphaBoolBlend {blendInTime=0.1, blendOutTime=0.1}
    alphaScaleBiasClamp {bInterpResult=true, interpSpeedDecreasing=2, interpSpeedIncreasing=5}
    boneToModify {boneName=dyn_backHair_l_02}
    translation {y=-2, z=15}
   }
   n11 = "Transform (Modify) Bone - Bone: dyn_backHair_r_02"(ComponentPose=@n10, bAlphaBoolEnabled=True, Translation="5,2,15", Rotation="0,0,0") {
    alphaInputType=Bool
    rotationMode=BMM_Additive
    rotationSpace=BCS_BoneSpace
    translationMode=BMM_Additive
    translationSpace=BCS_BoneSpace
    alphaBoolBlend {blendInTime=0.1, blendOutTime=0.1}
    alphaScaleBiasClamp {bInterpResult=true, interpSpeedDecreasing=2, interpSpeedIncreasing=5}
    boneToModify {boneName=dyn_backHair_r_02}
    translation=<5 2 15>
   }
   n12 = "Transform (Modify) Bone - Bone: dyn_frontHair_r_06"!(ComponentPose=@n11, bAlphaBoolEnabled:bool=True, Translation="0,0,-10", Rotation="0,0,0", Scale="1,1,1") -> (Pose) {
    alphaInputType=Bool
    rotationMode=BMM_Additive
    rotationSpace=BCS_BoneSpace
    scaleMode=BMM_Additive
    scaleSpace=BCS_BoneSpace
    translationMode=BMM_Additive
    translationSpace=BCS_BoneSpace
    alphaBoolBlend {blendInTime=0.1, blendOutTime=0.1}
    boneToModify {boneName=dyn_frontHair_r_06}
    translation {z=-10}
   }
   n13 = "Transform (Modify) Bone - Bone: dyn_backHair_l_02"!(ComponentPose=@n12, bAlphaBoolEnabled=True, Translation="0,0,5", Rotation="0,0,0", Scale:Vector="1,1,1") -> (Pose) {
    alphaInputType=Bool
    rotationMode=BMM_Additive
    rotationSpace=BCS_BoneSpace
    scaleMode=BMM_Additive
    scaleSpace=BCS_BoneSpace
    translationMode=BMM_Additive
    translationSpace=BCS_BoneSpace
    alphaBoolBlend {blendInTime=0.1, blendOutTime=0.1}
    boneToModify {boneName=dyn_backHair_l_02}
    translation {z=5}
   }
   n14 = "Transform (Modify) Bone - Bone: dyn_backHair_c_02"!(ComponentPose=@n13, bAlphaBoolEnabled=True, Translation="0,0,5", Rotation="0,0,0", Scale:Vector="1,1,1") -> (Pose) {
    alphaInputType=Bool
    rotationMode=BMM_Additive
    rotationSpace=BCS_BoneSpace
    scaleMode=BMM_Additive
    scaleSpace=BCS_BoneSpace
    translationMode=BMM_Additive
    translationSpace=BCS_BoneSpace
    alphaBoolBlend {blendInTime=0.1, blendOutTime=0.1}
    boneToModify {boneName=dyn_backHair_c_02}
    translation {z=5}
   }
   n15 = "Transform (Modify) Bone - Bone: dyn_backHair_r_02"!(ComponentPose=@n14, bAlphaBoolEnabled=True, Translation="0,0,5", Rotation="0,0,0", Scale:Vector="1,1,1") -> (Pose) {
    alphaInputType=Bool
    rotationMode=BMM_Additive
    rotationSpace=BCS_BoneSpace
    scaleMode=BMM_Additive
    scaleSpace=BCS_BoneSpace
    translationMode=BMM_Additive
    translationSpace=BCS_BoneSpace
    alphaBoolBlend {blendInTime=0.1, blendOutTime=0.1}
    boneToModify {boneName=dyn_backHair_r_02}
    translation {z=5}
   }
   Output Pose(Result=Component To Local(ComponentPose=@n15)) {layerGroup=None}
   n16 = Layered blend per bone(BasePose=Copy Pose From Mesh() {bUseAttachedParent=true}, BlendPoses_0=Copy Pose From Mesh(SourceMeshComponent=Get SkelMesh())) {layerSetup [{branchFilters [
       {boneName=lowerarm_l}
       {boneName=lowerarm_r}
       {boneName=upperarm_l}
       {boneName=upperarm_r}
       {boneName=clavicle_l}
       {boneName=clavicle_r}
       {boneName=thigh_l}
       {boneName=thigh_r}
       {boneName=dyn_C_haori_back_mstr}
       {boneName=dyn_C_haori_back_1}
       {boneName=dyn_C_haori_back_2}
       {boneName=dyn_C_haori_back_3}
       {boneName=dyn_C_haori_back_4}
       {boneName=dyn_L_haori_side_mstr}
       {boneName=dyn_L_haori_side_1}
       {boneName=dyn_L_haori_side_2}
       {boneName=dyn_L_haori_side_3}
       {boneName=dyn_L_haori_side_4}
       {boneName=dyn_R_haori_side_mstr}
       {boneName=dyn_R_haori_side_1}
       {boneName=dyn_R_haori_side_2}
       {boneName=dyn_R_haori_side_3}
       {boneName=dyn_R_haori_side_4}
       {boneName=base_C_skirt_a_1}
       {boneName=dyn_C_skirt_a_1}
       {boneName=dyn_C_skirt_a_2}
       {boneName=dyn_C_skirt_a_3}
       {boneName=attach_backpack}
      ]}]}
   n17 = Control Rig!(
    Source=@n16
    Alpha
    Is Targeting:bool=Get bIsTargeting()
    Is Crouching:bool=Get bIsCrouching()
    Is Jumping:bool=Get bIsJumping()
    Is rigidbody enabled:bool=Get bRigidBodyEnabled()
    is floating:bool=Get bIsSkydiveFloating()
    WearingCape:bool=Get IsWearingCape() -> (IsWearingCape:bool)
    Pub_Sprinting:bool=Get bIsSprinting()
    Pub_TacticalSprint:bool=Get bIsTacticalSprinting()
    Pub_SwingPickaxe:bool=Get bIsPlayingMeleeAnim()
    Pub_kneeSliding:bool=Get bIsKneeSliding()
    Pub_BasebalSliding:bool=Get bIsBaseballSliding()
    Pub_WearingPack:bool=Get Is Wearing Backbling()
    Pub_IsSkydive:bool=Get bIsSkyDiving()
    Pub_IsTacticalJumps:bool=Get TacticalJumping() -> (TacticalJumping:bool)
    Pub_IsParachuteOpened:bool=Get bIsParachuteOpened()
    Pub_IsFrontEnd:bool=Get bIsInFrontEnd()
   ) -> (Pose) {controlRigAssetReference {blueprintRigClass=&/F_MED_AtticDrill_FaceAcc_CtrlRig.F_MED_AtticDrill_FaceAcc_CtrlRig_C}}
   Save cached pose 'CP_InputPose'(Pose=@n17) {cacheName=CP_InputPose}
   n18 = Anim Dynamics - Bone: dyn_backHair_r_01!(
    ComponentPose=Local To Component(LocalPose=Use cached pose 'CP_InputPose'())
    Alpha=1
    GravityOverride:Vector="0,0,-980"
    SimSpaceSettings:Anim Phys Sim Space Settings Structure=(SimSpaceAngularAlpha=0,MaxAngularVelocity=10000,MaxAngularAcceleration=10000,ExternalAngularVelocity=(X=0,Y=0,Z=0))
   ) -> (Pose) {
    angularDampingOverride=1
    angularSpringConstant=20000
    bAngularSpring=true
    bOverrideAngularDamping=true
    bUseGravityOverride=true
    gravityScale=0.35
    simulationSpace=RootRelative
    alphaScaleBiasClamp {interpSpeedDecreasing=5, interpSpeedIncreasing=5}
    boundBone {boneName=dyn_backHair_r_01}
    componentAppliedLinearAccClamp=<10000 10000 10000>
    componentLinearAccScale=<1 1 1>
    componentLinearVelScale=<1 1 1>
    gravityOverride {z=-980}
    physicsBodyDefinitions [{
      boundBone {boneName=dyn_backHair_r_01}
      boxExtents=<30 45 45>
      constraintSetup {angularLimitsMax=<12 0 10>, angularLimitsMin=<-12 -30 -15>, angularTarget=<1 0 0>}
      localJointOffset=<4 0 0>
     }]
   }
   n19 = Anim Dynamics - Bone: dyn_backHair_r_02!(
    ComponentPose=@n18
    Alpha=1
    GravityOverride:Vector="0,0,-980"
    SimSpaceSettings:Anim Phys Sim Space Settings Structure=(SimSpaceAngularAlpha=0,MaxAngularVelocity=10000,MaxAngularAcceleration=10000,ExternalAngularVelocity=(X=0,Y=0,Z=0))
   ) -> (Pose) {
    angularDampingOverride=1
    angularSpringConstant=20000
    bAngularSpring=true
    bOverrideAngularDamping=true
    bUseGravityOverride=true
    gravityScale=0.35
    simulationSpace=RootRelative
    alphaScaleBiasClamp {interpSpeedDecreasing=5, interpSpeedIncreasing=5}
    boundBone {boneName=dyn_backHair_r_02}
    componentAppliedLinearAccClamp=<10000 10000 10000>
    componentLinearAccScale=<1 1 1>
    componentLinearVelScale=<1 1 1>
    gravityOverride {z=-980}
    physicsBodyDefinitions [{
      boundBone {boneName=dyn_backHair_r_02}
      boxExtents=<30 45 45>
      constraintSetup {angularLimitsMax=<12 0 10>, angularLimitsMin=<-12 -30 -15>, angularTarget=<1 0 0>}
      localJointOffset=<4 0 0>
     }]
   }
   n20 = Anim Dynamics - Bone: dyn_backHair_r_03!(
    ComponentPose=@n19
    Alpha=1
    GravityOverride:Vector="0,0,-980"
    SimSpaceSettings:Anim Phys Sim Space Settings Structure=(SimSpaceAngularAlpha=0,MaxAngularVelocity=10000,MaxAngularAcceleration=10000,ExternalAngularVelocity=(X=0,Y=0,Z=0))
   ) -> (Pose) {
    angularDampingOverride=1
    angularSpringConstant=3000
    bAngularSpring=true
    bOverrideAngularDamping=true
    bUseGravityOverride=true
    gravityScale=0.35
    simulationSpace=RootRelative
    alphaScaleBiasClamp {interpSpeedDecreasing=5, interpSpeedIncreasing=5}
    boundBone {boneName=dyn_backHair_r_03}
    componentAppliedLinearAccClamp=<10000 10000 10000>
    componentLinearAccScale=<1 1 1>
    componentLinearVelScale=<1 1 1>
    gravityOverride {z=-980}
    physicsBodyDefinitions [{
      boundBone {boneName=dyn_backHair_r_03}
      boxExtents=<25 30 30>
      constraintSetup {angularLimitsMax=<12 0 10>, angularLimitsMin=<-12 -30 -15>, angularTarget=<1 0 0>}
      localJointOffset=<2 0 0>
     }]
   }
   n21 = Anim Dynamics - Bone: dyn_backHair_r_04!(
    ComponentPose=@n20
    Alpha=1
    GravityOverride:Vector="0,0,-980"
    SimSpaceSettings:Anim Phys Sim Space Settings Structure=(SimSpaceAngularAlpha=0,MaxAngularVelocity=10000,MaxAngularAcceleration=10000,ExternalAngularVelocity=(X=0,Y=0,Z=0))
   ) -> (Pose) {
    angularDampingOverride=1
    angularSpringConstant=2000
    bAngularSpring=true
    bOverrideAngularDamping=true
    bUseGravityOverride=true
    gravityScale=0.35
    simulationSpace=RootRelative
    alphaScaleBiasClamp {interpSpeedDecreasing=5, interpSpeedIncreasing=5}
    boundBone {boneName=dyn_backHair_r_04}
    componentAppliedLinearAccClamp=<10000 10000 10000>
    componentLinearAccScale=<1 1 1>
    componentLinearVelScale=<1 1 1>
    gravityOverride {z=-980}
    physicsBodyDefinitions [{
      boundBone {boneName=dyn_backHair_r_04}
      boxExtents=<28 32 32>
      constraintSetup {angularLimitsMax=<12 0 10>, angularLimitsMin=<-12 -30 -15>, angularTarget=<1 0 0>}
      localJointOffset=<2.5 0 0>
     }]
   }
   n22 = Anim Dynamics - Bone: dyn_backHair_r_05!(
    ComponentPose=@n21
    Alpha=1
    GravityOverride:Vector="0,0,-980"
    SimSpaceSettings:Anim Phys Sim Space Settings Structure=(SimSpaceAngularAlpha=0,MaxAngularVelocity=10000,MaxAngularAcceleration=10000,ExternalAngularVelocity=(X=0,Y=0,Z=0))
   ) -> (Pose) {
    angularDampingOverride=1
    angularSpringConstant=1200
    bAngularSpring=true
    bOverrideAngularDamping=true
    bUseGravityOverride=true
    gravityScale=0.35
    simulationSpace=RootRelative
    alphaScaleBiasClamp {interpSpeedDecreasing=5, interpSpeedIncreasing=5}
    boundBone {boneName=dyn_backHair_r_05}
    componentAppliedLinearAccClamp=<10000 10000 10000>
    componentLinearAccScale=<1 1 1>
    componentLinearVelScale=<1 1 1>
    gravityOverride {z=-980}
    physicsBodyDefinitions [{
      boundBone {boneName=dyn_backHair_r_05}
      boxExtents=<28 32 32>
      constraintSetup {angularLimitsMax=<12 0 10>, angularLimitsMin=<-12 -30 -15>, angularTarget=<1 0 0>}
      localJointOffset=<2.5 0 0>
     }]
   }
   n23 = Anim Dynamics - Bone: dyn_backHair_r_06!(
    ComponentPose=@n22
    Alpha=1
    GravityOverride:Vector="0,0,-980"
    SimSpaceSettings:Anim Phys Sim Space Settings Structure=(SimSpaceAngularAlpha=0,MaxAngularVelocity=10000,MaxAngularAcceleration=10000,ExternalAngularVelocity=(X=0,Y=0,Z=0))
   ) -> (Pose) {
    angularDampingOverride=1
    angularSpringConstant=1200
    bAngularSpring=true
    bOverrideAngularDamping=true
    bUseGravityOverride=true
    gravityScale=0.35
    simulationSpace=RootRelative
    alphaScaleBiasClamp {interpSpeedDecreasing=5, interpSpeedIncreasing=5}
    boundBone {boneName=dyn_backHair_r_06}
    componentAppliedLinearAccClamp=<10000 10000 10000>
    componentLinearAccScale=<1 1 1>
    componentLinearVelScale=<1 1 1>
    gravityOverride {z=-980}
    physicsBodyDefinitions [{
      boundBone {boneName=dyn_backHair_r_06}
      boxExtents=<28 32 32>
      constraintSetup {angularLimitsMax=<12 0 10>, angularLimitsMin=<-12 -30 -15>, angularTarget=<1 0 0>}
      localJointOffset=<2.5 0 0>
     }]
   }
   n24 = Anim Dynamics - Bone: dyn_backHair_c_01!(
    ComponentPose=@n23
    Alpha=1
    GravityOverride:Vector="0,0,-980"
    SimSpaceSettings:Anim Phys Sim Space Settings Structure=(SimSpaceAngularAlpha=0,MaxAngularVelocity=10000,MaxAngularAcceleration=10000,ExternalAngularVelocity=(X=0,Y=0,Z=0))
   ) -> (Pose) {
    angularDampingOverride=1
    angularSpringConstant=20000
    bAngularSpring=true
    bOverrideAngularDamping=true
    bUseGravityOverride=true
    gravityScale=0.35
    simulationSpace=RootRelative
    alphaScaleBiasClamp {interpSpeedDecreasing=5, interpSpeedIncreasing=5}
    boundBone {boneName=dyn_backHair_c_01}
    componentAppliedLinearAccClamp=<10000 10000 10000>
    componentLinearAccScale=<1 1 1>
    componentLinearVelScale=<1 1 1>
    gravityOverride {z=-980}
    physicsBodyDefinitions [{
      boundBone {boneName=dyn_backHair_c_01}
      boxExtents=<30 45 45>
      constraintSetup {angularLimitsMax=<5 0 10>, angularLimitsMin=<-5 -30 -10>, angularTarget=<1 0 0>}
      localJointOffset=<4 0 0>
     }]
   }
   n25 = Anim Dynamics - Bone: dyn_backHair_c_02!(
    ComponentPose=@n24
    Alpha=1
    GravityOverride:Vector="0,0,-980"
    SimSpaceSettings:Anim Phys Sim Space Settings Structure=(SimSpaceAngularAlpha=0,MaxAngularVelocity=10000,MaxAngularAcceleration=10000,ExternalAngularVelocity=(X=0,Y=0,Z=0))
   ) -> (Pose) {
    angularDampingOverride=1
    angularSpringConstant=20000
    bAngularSpring=true
    bOverrideAngularDamping=true
    bUseGravityOverride=true
    gravityScale=0.35
    simulationSpace=RootRelative
    alphaScaleBiasClamp {interpSpeedDecreasing=5, interpSpeedIncreasing=5}
    boundBone {boneName=dyn_backHair_c_02}
    componentAppliedLinearAccClamp=<10000 10000 10000>
    componentLinearAccScale=<1 1 1>
    componentLinearVelScale=<1 1 1>
    gravityOverride {z=-980}
    physicsBodyDefinitions [{
      boundBone {boneName=dyn_backHair_c_02}
      boxExtents=<30 45 45>
      constraintSetup {angularLimitsMax=<5 0 10>, angularLimitsMin=<-5 -30 -10>, angularTarget=<1 0 0>}
      localJointOffset=<4 0 0>
     }]
   }
   n26 = Anim Dynamics - Bone: dyn_backHair_c_03!(
    ComponentPose=@n25
    Alpha=1
    GravityOverride:Vector="0,0,-980"
    SimSpaceSettings:Anim Phys Sim Space Settings Structure=(SimSpaceAngularAlpha=0,MaxAngularVelocity=10000,MaxAngularAcceleration=10000,ExternalAngularVelocity=(X=0,Y=0,Z=0))
   ) -> (Pose) {
    angularDampingOverride=1
    angularSpringConstant=3000
    bAngularSpring=true
    bOverrideAngularDamping=true
    bUseGravityOverride=true
    gravityScale=0.35
    simulationSpace=RootRelative
    alphaScaleBiasClamp {interpSpeedDecreasing=5, interpSpeedIncreasing=5}
    boundBone {boneName=dyn_backHair_c_03}
    componentAppliedLinearAccClamp=<10000 10000 10000>
    componentLinearAccScale=<1 1 1>
    componentLinearVelScale=<1 1 1>
    gravityOverride {z=-980}
    physicsBodyDefinitions [{
      boundBone {boneName=dyn_backHair_c_03}
      boxExtents=<25 30 30>
      constraintSetup {angularLimitsMax=<5 0 10>, angularLimitsMin=<-5 -30 -10>, angularTarget=<1 0 0>}
      localJointOffset=<2 0 0>
     }]
   }
   n27 = Anim Dynamics - Bone: dyn_backHair_c_04!(
    ComponentPose=@n26
    Alpha=1
    GravityOverride:Vector="0,0,-980"
    SimSpaceSettings:Anim Phys Sim Space Settings Structure=(SimSpaceAngularAlpha=0,MaxAngularVelocity=10000,MaxAngularAcceleration=10000,ExternalAngularVelocity=(X=0,Y=0,Z=0))
   ) -> (Pose) {
    angularDampingOverride=1
    angularSpringConstant=2000
    bAngularSpring=true
    bOverrideAngularDamping=true
    bUseGravityOverride=true
    gravityScale=0.35
    simulationSpace=RootRelative
    alphaScaleBiasClamp {interpSpeedDecreasing=5, interpSpeedIncreasing=5}
    boundBone {boneName=dyn_backHair_c_04}
    componentAppliedLinearAccClamp=<10000 10000 10000>
    componentLinearAccScale=<1 1 1>
    componentLinearVelScale=<1 1 1>
    gravityOverride {z=-980}
    physicsBodyDefinitions [{
      boundBone {boneName=dyn_backHair_c_04}
      boxExtents=<28 32 32>
      constraintSetup {angularLimitsMax=<5 0 10>, angularLimitsMin=<-5 -30 -10>, angularTarget=<1 0 0>}
      localJointOffset=<2.5 0 0>
     }]
   }
   n28 = Anim Dynamics - Bone: dyn_backHair_c_05!(
    ComponentPose=@n27
    Alpha=1
    GravityOverride:Vector="0,0,-980"
    SimSpaceSettings:Anim Phys Sim Space Settings Structure=(SimSpaceAngularAlpha=0,MaxAngularVelocity=10000,MaxAngularAcceleration=10000,ExternalAngularVelocity=(X=0,Y=0,Z=0))
   ) -> (Pose) {
    angularDampingOverride=1
    angularSpringConstant=1200
    bAngularSpring=true
    bOverrideAngularDamping=true
    bUseGravityOverride=true
    gravityScale=0.35
    simulationSpace=RootRelative
    alphaScaleBiasClamp {interpSpeedDecreasing=5, interpSpeedIncreasing=5}
    boundBone {boneName=dyn_backHair_c_05}
    componentAppliedLinearAccClamp=<10000 10000 10000>
    componentLinearAccScale=<1 1 1>
    componentLinearVelScale=<1 1 1>
    gravityOverride {z=-980}
    physicsBodyDefinitions [{
      boundBone {boneName=dyn_backHair_c_05}
      boxExtents=<28 32 32>
      constraintSetup {angularLimitsMax=<5 0 10>, angularLimitsMin=<-5 -30 -10>, angularTarget=<1 0 0>}
      localJointOffset=<2.5 0 0>
     }]
   }
   n29 = Anim Dynamics - Bone: dyn_backHair_c_06!(
    ComponentPose=@n28
    Alpha=1
    GravityOverride:Vector="0,0,-980"
    SimSpaceSettings:Anim Phys Sim Space Settings Structure=(SimSpaceAngularAlpha=0,MaxAngularVelocity=10000,MaxAngularAcceleration=10000,ExternalAngularVelocity=(X=0,Y=0,Z=0))
   ) -> (Pose) {
    angularDampingOverride=1
    angularSpringConstant=1200
    bAngularSpring=true
    bOverrideAngularDamping=true
    bUseGravityOverride=true
    gravityScale=0.35
    simulationSpace=RootRelative
    alphaScaleBiasClamp {interpSpeedDecreasing=5, interpSpeedIncreasing=5}
    boundBone {boneName=dyn_backHair_c_06}
    componentAppliedLinearAccClamp=<10000 10000 10000>
    componentLinearAccScale=<1 1 1>
    componentLinearVelScale=<1 1 1>
    gravityOverride {z=-980}
    physicsBodyDefinitions [{
      boundBone {boneName=dyn_backHair_c_06}
      boxExtents=<28 32 32>
      constraintSetup {angularLimitsMax=<5 0 10>, angularLimitsMin=<-5 -30 -10>, angularTarget=<1 0 0>}
      localJointOffset=<2.5 0 0>
     }]
   }
   n30 = Anim Dynamics - Bone: dyn_backHair_l_01!(
    ComponentPose=@n29
    Alpha=1
    GravityOverride:Vector="0,0,-980"
    SimSpaceSettings:Anim Phys Sim Space Settings Structure=(SimSpaceAngularAlpha=0,MaxAngularVelocity=10000,MaxAngularAcceleration=10000,ExternalAngularVelocity=(X=0,Y=0,Z=0))
   ) -> (Pose) {
    angularDampingOverride=1
    angularSpringConstant=20000
    bAngularSpring=true
    bOverrideAngularDamping=true
    bUseGravityOverride=true
    gravityScale=0.35
    simulationSpace=RootRelative
    alphaScaleBiasClamp {interpSpeedDecreasing=5, interpSpeedIncreasing=5}
    boundBone {boneName=dyn_backHair_l_01}
    componentAppliedLinearAccClamp=<10000 10000 10000>
    componentLinearAccScale=<1 1 1>
    componentLinearVelScale=<1 1 1>
    gravityOverride {z=-980}
    physicsBodyDefinitions [{
      boundBone {boneName=dyn_backHair_l_01}
      boxExtents=<30 45 45>
      constraintSetup {angularLimitsMax=<8 0 12>, angularLimitsMin=<-8 -40 -5>, angularTarget=<1 0 0>}
      localJointOffset=<4 0 0>
     }]
   }
   n31 = Anim Dynamics - Bone: dyn_backHair_l_02!(
    ComponentPose=@n30
    Alpha=1
    GravityOverride:Vector="0,0,-980"
    SimSpaceSettings:Anim Phys Sim Space Settings Structure=(SimSpaceAngularAlpha=0,MaxAngularVelocity=10000,MaxAngularAcceleration=10000,ExternalAngularVelocity=(X=0,Y=0,Z=0))
   ) -> (Pose) {
    angularDampingOverride=1
    angularSpringConstant=20000
    bAngularSpring=true
    bOverrideAngularDamping=true
    bUseGravityOverride=true
    gravityScale=0.35
    simulationSpace=RootRelative
    alphaScaleBiasClamp {interpSpeedDecreasing=5, interpSpeedIncreasing=5}
    boundBone {boneName=dyn_backHair_l_02}
    componentAppliedLinearAccClamp=<10000 10000 10000>
    componentLinearAccScale=<1 1 1>
    componentLinearVelScale=<1 1 1>
    gravityOverride {z=-980}
    physicsBodyDefinitions [{
      boundBone {boneName=dyn_backHair_l_02}
      boxExtents=<30 45 45>
      constraintSetup {angularLimitsMax=<8 0 12>, angularLimitsMin=<-8 -40 -5>, angularTarget=<1 0 0>}
      localJointOffset=<4 0 0>
     }]
   }
   n32 = Anim Dynamics - Bone: dyn_backHair_l_03!(
    ComponentPose=@n31
    Alpha=1
    GravityOverride:Vector="0,0,-980"
    SimSpaceSettings:Anim Phys Sim Space Settings Structure=(SimSpaceAngularAlpha=0,MaxAngularVelocity=10000,MaxAngularAcceleration=10000,ExternalAngularVelocity=(X=0,Y=0,Z=0))
   ) -> (Pose) {
    angularDampingOverride=1
    angularSpringConstant=3000
    bAngularSpring=true
    bOverrideAngularDamping=true
    bUseGravityOverride=true
    gravityScale=0.35
    simulationSpace=RootRelative
    alphaScaleBiasClamp {interpSpeedDecreasing=5, interpSpeedIncreasing=5}
    boundBone {boneName=dyn_backHair_l_03}
    componentAppliedLinearAccClamp=<10000 10000 10000>
    componentLinearAccScale=<1 1 1>
    componentLinearVelScale=<1 1 1>
    gravityOverride {z=-980}
    physicsBodyDefinitions [{
      boundBone {boneName=dyn_backHair_l_03}
      boxExtents=<25 30 30>
      constraintSetup {angularLimitsMax=<8 0 12>, angularLimitsMin=<-8 -40 -5>, angularTarget=<1 0 0>}
      localJointOffset=<2 0 0>
     }]
   }
   n33 = Anim Dynamics - Bone: dyn_backHair_l_04!(
    ComponentPose=@n32
    Alpha=1
    GravityOverride:Vector="0,0,-980"
    SimSpaceSettings:Anim Phys Sim Space Settings Structure=(SimSpaceAngularAlpha=0,MaxAngularVelocity=10000,MaxAngularAcceleration=10000,ExternalAngularVelocity=(X=0,Y=0,Z=0))
   ) -> (Pose) {
    angularDampingOverride=1
    angularSpringConstant=2000
    bAngularSpring=true
    bOverrideAngularDamping=true
    bUseGravityOverride=true
    gravityScale=0.35
    simulationSpace=RootRelative
    alphaScaleBiasClamp {interpSpeedDecreasing=5, interpSpeedIncreasing=5}
    boundBone {boneName=dyn_backHair_l_04}
    componentAppliedLinearAccClamp=<10000 10000 10000>
    componentLinearAccScale=<1 1 1>
    componentLinearVelScale=<1 1 1>
    gravityOverride {z=-980}
    physicsBodyDefinitions [{
      boundBone {boneName=dyn_backHair_l_04}
      boxExtents=<28 32 32>
      constraintSetup {angularLimitsMax=<8 0 12>, angularLimitsMin=<-8 -50 -5>, angularTarget=<1 0 0>}
      localJointOffset=<2.5 0 0>
     }]
   }
   n34 = Anim Dynamics - Bone: dyn_backHair_l_05!(
    ComponentPose=@n33
    Alpha=1
    GravityOverride:Vector="0,0,-980"
    SimSpaceSettings:Anim Phys Sim Space Settings Structure=(SimSpaceAngularAlpha=0,MaxAngularVelocity=10000,MaxAngularAcceleration=10000,ExternalAngularVelocity=(X=0,Y=0,Z=0))
   ) -> (Pose) {
    angularDampingOverride=1
    angularSpringConstant=1200
    bAngularSpring=true
    bOverrideAngularDamping=true
    bUseGravityOverride=true
    gravityScale=0.35
    simulationSpace=RootRelative
    alphaScaleBiasClamp {interpSpeedDecreasing=5, interpSpeedIncreasing=5}
    boundBone {boneName=dyn_backHair_l_05}
    componentAppliedLinearAccClamp=<10000 10000 10000>
    componentLinearAccScale=<1 1 1>
    componentLinearVelScale=<1 1 1>
    gravityOverride {z=-980}
    physicsBodyDefinitions [{
      boundBone {boneName=dyn_backHair_l_05}
      boxExtents=<28 32 32>
      constraintSetup {angularLimitsMax=<8 0 12>, angularLimitsMin=<-8 -50 -5>, angularTarget=<1 0 0>}
      localJointOffset=<2.5 0 0>
     }]
   }
   n35 = Anim Dynamics - Bone: dyn_backHair_l_06!(
    ComponentPose=@n34
    Alpha=1
    GravityOverride:Vector="0,0,-980"
    SimSpaceSettings:Anim Phys Sim Space Settings Structure=(SimSpaceAngularAlpha=0,MaxAngularVelocity=10000,MaxAngularAcceleration=10000,ExternalAngularVelocity=(X=0,Y=0,Z=0))
   ) -> (Pose) {
    angularDampingOverride=1
    angularSpringConstant=1200
    bAngularSpring=true
    bOverrideAngularDamping=true
    bUseGravityOverride=true
    gravityScale=0.35
    simulationSpace=RootRelative
    alphaScaleBiasClamp {interpSpeedDecreasing=5, interpSpeedIncreasing=5}
    boundBone {boneName=dyn_backHair_l_06}
    componentAppliedLinearAccClamp=<10000 10000 10000>
    componentLinearAccScale=<1 1 1>
    componentLinearVelScale=<1 1 1>
    gravityOverride {z=-980}
    physicsBodyDefinitions [{
      boundBone {boneName=dyn_backHair_l_06}
      boxExtents=<28 32 32>
      constraintSetup {angularLimitsMax=<8 0 12>, angularLimitsMin=<-8 -50 -5>, angularTarget=<1 0 0>}
      localJointOffset=<2.5 0 0>
     }]
   }
   n36 = Anim Dynamics - Bone: dyn_backHair_r_01(
    ComponentPose=@n35
    Alpha=1
    GravityOverride="0,0,-980"
    SimSpaceSettings=(SimSpaceAngularAlpha=0,MaxAngularVelocity=10000,MaxAngularAcceleration=10000,ExternalAngularVelocity=(X=0,Y=0,Z=0))
   ) {
    angularDampingOverride=1
    angularSpringConstant=20000
    bAngularSpring=true
    bOverrideAngularDamping=true
    bUseGravityOverride=true
    gravityScale=0.35
    simulationSpace=RootRelative
    alphaScaleBiasClamp {interpSpeedDecreasing=5, interpSpeedIncreasing=5}
    boundBone {boneName=dyn_backHair_r_01}
    componentAppliedLinearAccClamp=<10000 10000 10000>
    componentLinearAccScale=<1 1 1>
    componentLinearVelScale=<1 1 1>
    gravityOverride {z=-980}
    physicsBodyDefinitions [{
      boundBone {boneName=dyn_backHair_r_01}
      boxExtents=<30 45 45>
      constraintSetup {angularLimitsMax=<12 0 10>, angularLimitsMin=<-12 0 -15>, angularTarget=<1 0 0>}
      localJointOffset=<4 0 0>
     }]
   }
   n37 = Anim Dynamics - Bone: dyn_backHair_r_02(
    ComponentPose=@n36
    Alpha=1
    GravityOverride="0,0,-980"
    SimSpaceSettings=(SimSpaceAngularAlpha=0,MaxAngularVelocity=10000,MaxAngularAcceleration=10000,ExternalAngularVelocity=(X=0,Y=0,Z=0))
   ) {
    angularDampingOverride=1
    angularSpringConstant=20000
    bAngularSpring=true
    bOverrideAngularDamping=true
    bUseGravityOverride=true
    gravityScale=0.35
    simulationSpace=RootRelative
    alphaScaleBiasClamp {interpSpeedDecreasing=5, interpSpeedIncreasing=5}
    boundBone {boneName=dyn_backHair_r_02}
    componentAppliedLinearAccClamp=<10000 10000 10000>
    componentLinearAccScale=<1 1 1>
    componentLinearVelScale=<1 1 1>
    gravityOverride {z=-980}
    physicsBodyDefinitions [{
      boundBone {boneName=dyn_backHair_r_02}
      boxExtents=<30 45 45>
      constraintSetup {angularLimitsMax=<12 0 10>, angularLimitsMin=<-12 0 -15>, angularTarget=<1 0 0>}
      localJointOffset=<4 0 0>
     }]
   }
   n38 = Anim Dynamics - Bone: dyn_backHair_r_03(
    ComponentPose=@n37
    Alpha=1
    GravityOverride="0,0,-980"
    SimSpaceSettings=(SimSpaceAngularAlpha=0,MaxAngularVelocity=10000,MaxAngularAcceleration=10000,ExternalAngularVelocity=(X=0,Y=0,Z=0))
   ) {
    angularDampingOverride=1
    angularSpringConstant=3000
    bAngularSpring=true
    bOverrideAngularDamping=true
    bUseGravityOverride=true
    gravityScale=0.35
    simulationSpace=RootRelative
    alphaScaleBiasClamp {interpSpeedDecreasing=5, interpSpeedIncreasing=5}
    boundBone {boneName=dyn_backHair_r_03}
    componentAppliedLinearAccClamp=<10000 10000 10000>
    componentLinearAccScale=<1 1 1>
    componentLinearVelScale=<1 1 1>
    gravityOverride {z=-980}
    physicsBodyDefinitions [{
      boundBone {boneName=dyn_backHair_r_03}
      boxExtents=<25 30 30>
      constraintSetup {angularLimitsMax=<12 0 10>, angularLimitsMin=<-12 0 -15>, angularTarget=<1 0 0>}
      localJointOffset=<2 0 0>
     }]
   }
   n39 = Anim Dynamics - Bone: dyn_backHair_r_04(
    ComponentPose=@n38
    Alpha=1
    GravityOverride="0,0,-980"
    SimSpaceSettings=(SimSpaceAngularAlpha=0,MaxAngularVelocity=10000,MaxAngularAcceleration=10000,ExternalAngularVelocity=(X=0,Y=0,Z=0))
   ) {
    angularDampingOverride=1
    angularSpringConstant=2000
    bAngularSpring=true
    bOverrideAngularDamping=true
    bUseGravityOverride=true
    gravityScale=0.35
    simulationSpace=RootRelative
    alphaScaleBiasClamp {interpSpeedDecreasing=5, interpSpeedIncreasing=5}
    boundBone {boneName=dyn_backHair_r_04}
    componentAppliedLinearAccClamp=<10000 10000 10000>
    componentLinearAccScale=<1 1 1>
    componentLinearVelScale=<1 1 1>
    gravityOverride {z=-980}
    physicsBodyDefinitions [{
      boundBone {boneName=dyn_backHair_r_04}
      boxExtents=<25 30 30>
      constraintSetup {angularLimitsMax=<12 0 10>, angularLimitsMin=<-12 0 -15>, angularTarget=<1 0 0>}
      localJointOffset=<3 0 0>
     }]
   }
   n40 = Anim Dynamics - Bone: dyn_backHair_r_05(
    ComponentPose=@n39
    Alpha=1
    GravityOverride="0,0,-980"
    SimSpaceSettings=(SimSpaceAngularAlpha=0,MaxAngularVelocity=10000,MaxAngularAcceleration=10000,ExternalAngularVelocity=(X=0,Y=0,Z=0))
   ) {
    angularDampingOverride=1
    angularSpringConstant=1200
    bAngularSpring=true
    bOverrideAngularDamping=true
    bUseGravityOverride=true
    gravityScale=0.35
    simulationSpace=RootRelative
    alphaScaleBiasClamp {interpSpeedDecreasing=5, interpSpeedIncreasing=5}
    boundBone {boneName=dyn_backHair_r_05}
    componentAppliedLinearAccClamp=<10000 10000 10000>
    componentLinearAccScale=<1 1 1>
    componentLinearVelScale=<1 1 1>
    gravityOverride {z=-980}
    physicsBodyDefinitions [{
      boundBone {boneName=dyn_backHair_r_05}
      boxExtents=<25 30 30>
      constraintSetup {angularLimitsMax=<12 0 10>, angularLimitsMin=<-12 0 -15>, angularTarget=<1 0 0>}
      localJointOffset=<3 0 0>
     }]
   }
   n41 = Anim Dynamics - Bone: dyn_backHair_r_06(
    ComponentPose=@n40
    Alpha=1
    GravityOverride="0,0,-980"
    SimSpaceSettings=(SimSpaceAngularAlpha=0,MaxAngularVelocity=10000,MaxAngularAcceleration=10000,ExternalAngularVelocity=(X=0,Y=0,Z=0))
   ) {
    angularDampingOverride=1
    angularSpringConstant=1200
    bAngularSpring=true
    bOverrideAngularDamping=true
    bUseGravityOverride=true
    gravityScale=0.35
    simulationSpace=RootRelative
    alphaScaleBiasClamp {interpSpeedDecreasing=5, interpSpeedIncreasing=5}
    boundBone {boneName=dyn_backHair_r_06}
    componentAppliedLinearAccClamp=<10000 10000 10000>
    componentLinearAccScale=<1 1 1>
    componentLinearVelScale=<1 1 1>
    gravityOverride {z=-980}
    physicsBodyDefinitions [{
      boundBone {boneName=dyn_backHair_r_06}
      boxExtents=<25 30 30>
      constraintSetup {angularLimitsMax=<12 0 10>, angularLimitsMin=<-12 0 -15>, angularTarget=<1 0 0>}
      localJointOffset=<3 0 0>
     }]
   }
   n42 = Anim Dynamics - Bone: dyn_backHair_c_01(
    ComponentPose=@n41
    Alpha=1
    GravityOverride="0,0,-980"
    SimSpaceSettings=(SimSpaceAngularAlpha=0,MaxAngularVelocity=10000,MaxAngularAcceleration=10000,ExternalAngularVelocity=(X=0,Y=0,Z=0))
   ) {
    angularDampingOverride=1
    angularSpringConstant=20000
    bAngularSpring=true
    bOverrideAngularDamping=true
    bUseGravityOverride=true
    gravityScale=0.35
    simulationSpace=RootRelative
    alphaScaleBiasClamp {interpSpeedDecreasing=5, interpSpeedIncreasing=5}
    boundBone {boneName=dyn_backHair_c_01}
    componentAppliedLinearAccClamp=<10000 10000 10000>
    componentLinearAccScale=<1 1 1>
    componentLinearVelScale=<1 1 1>
    gravityOverride {z=-980}
    physicsBodyDefinitions [{
      boundBone {boneName=dyn_backHair_c_01}
      boxExtents=<30 45 45>
      constraintSetup {angularLimitsMax=<5 0 10>, angularLimitsMin=<-5 0 -10>, angularTarget=<1 0 0>}
      localJointOffset=<4 0 0>
     }]
   }
   n43 = Anim Dynamics - Bone: dyn_backHair_c_02(
    ComponentPose=@n42
    Alpha=1
    GravityOverride="0,0,-980"
    SimSpaceSettings=(SimSpaceAngularAlpha=0,MaxAngularVelocity=10000,MaxAngularAcceleration=10000,ExternalAngularVelocity=(X=0,Y=0,Z=0))
   ) {
    angularDampingOverride=1
    angularSpringConstant=20000
    bAngularSpring=true
    bOverrideAngularDamping=true
    bUseGravityOverride=true
    gravityScale=0.35
    simulationSpace=RootRelative
    alphaScaleBiasClamp {interpSpeedDecreasing=5, interpSpeedIncreasing=5}
    boundBone {boneName=dyn_backHair_c_02}
    componentAppliedLinearAccClamp=<10000 10000 10000>
    componentLinearAccScale=<1 1 1>
    componentLinearVelScale=<1 1 1>
    gravityOverride {z=-980}
    physicsBodyDefinitions [{
      boundBone {boneName=dyn_backHair_c_02}
      boxExtents=<30 45 45>
      constraintSetup {angularLimitsMax=<5 0 10>, angularLimitsMin=<-5 0 -10>, angularTarget=<1 0 0>}
      localJointOffset=<4 0 0>
     }]
   }
   n44 = Anim Dynamics - Bone: dyn_backHair_c_03(
    ComponentPose=@n43
    Alpha=1
    GravityOverride="0,0,-980"
    SimSpaceSettings=(SimSpaceAngularAlpha=0,MaxAngularVelocity=10000,MaxAngularAcceleration=10000,ExternalAngularVelocity=(X=0,Y=0,Z=0))
   ) {
    angularDampingOverride=1
    angularSpringConstant=3000
    bAngularSpring=true
    bOverrideAngularDamping=true
    bUseGravityOverride=true
    gravityScale=0.35
    simulationSpace=RootRelative
    alphaScaleBiasClamp {interpSpeedDecreasing=5, interpSpeedIncreasing=5}
    boundBone {boneName=dyn_backHair_c_03}
    componentAppliedLinearAccClamp=<10000 10000 10000>
    componentLinearAccScale=<1 1 1>
    componentLinearVelScale=<1 1 1>
    gravityOverride {z=-980}
    physicsBodyDefinitions [{
      boundBone {boneName=dyn_backHair_c_03}
      boxExtents=<25 30 30>
      constraintSetup {angularLimitsMax=<5 0 10>, angularLimitsMin=<-5 0 -10>, angularTarget=<1 0 0>}
      localJointOffset=<2 0 0>
     }]
   }
   n45 = Anim Dynamics - Bone: dyn_backHair_c_04(
    ComponentPose=@n44
    Alpha=1
    GravityOverride="0,0,-980"
    SimSpaceSettings=(SimSpaceAngularAlpha=0,MaxAngularVelocity=10000,MaxAngularAcceleration=10000,ExternalAngularVelocity=(X=0,Y=0,Z=0))
   ) {
    angularDampingOverride=1
    angularSpringConstant=2000
    bAngularSpring=true
    bOverrideAngularDamping=true
    bUseGravityOverride=true
    gravityScale=0.35
    simulationSpace=RootRelative
    alphaScaleBiasClamp {interpSpeedDecreasing=5, interpSpeedIncreasing=5}
    boundBone {boneName=dyn_backHair_c_04}
    componentAppliedLinearAccClamp=<10000 10000 10000>
    componentLinearAccScale=<1 1 1>
    componentLinearVelScale=<1 1 1>
    gravityOverride {z=-980}
    physicsBodyDefinitions [{
      boundBone {boneName=dyn_backHair_c_04}
      boxExtents=<25 30 30>
      constraintSetup {angularLimitsMax=<5 0 10>, angularLimitsMin=<-5 0 -10>, angularTarget=<1 0 0>}
      localJointOffset=<3 0 0>
     }]
   }
   n46 = Anim Dynamics - Bone: dyn_backHair_c_05(
    ComponentPose=@n45
    Alpha=1
    GravityOverride="0,0,-980"
    SimSpaceSettings=(SimSpaceAngularAlpha=0,MaxAngularVelocity=10000,MaxAngularAcceleration=10000,ExternalAngularVelocity=(X=0,Y=0,Z=0))
   ) {
    angularDampingOverride=1
    angularSpringConstant=1200
    bAngularSpring=true
    bOverrideAngularDamping=true
    bUseGravityOverride=true
    gravityScale=0.35
    simulationSpace=RootRelative
    alphaScaleBiasClamp {interpSpeedDecreasing=5, interpSpeedIncreasing=5}
    boundBone {boneName=dyn_backHair_c_05}
    componentAppliedLinearAccClamp=<10000 10000 10000>
    componentLinearAccScale=<1 1 1>
    componentLinearVelScale=<1 1 1>
    gravityOverride {z=-980}
    physicsBodyDefinitions [{
      boundBone {boneName=dyn_backHair_c_05}
      boxExtents=<25 30 30>
      constraintSetup {angularLimitsMax=<5 0 10>, angularLimitsMin=<-5 0 -10>, angularTarget=<1 0 0>}
      localJointOffset=<3 0 0>
     }]
   }
   n47 = Anim Dynamics - Bone: dyn_backHair_c_06(
    ComponentPose=@n46
    Alpha=1
    GravityOverride="0,0,-980"
    SimSpaceSettings=(SimSpaceAngularAlpha=0,MaxAngularVelocity=10000,MaxAngularAcceleration=10000,ExternalAngularVelocity=(X=0,Y=0,Z=0))
   ) {
    angularDampingOverride=1
    angularSpringConstant=1200
    bAngularSpring=true
    bOverrideAngularDamping=true
    bUseGravityOverride=true
    gravityScale=0.35
    simulationSpace=RootRelative
    alphaScaleBiasClamp {interpSpeedDecreasing=5, interpSpeedIncreasing=5}
    boundBone {boneName=dyn_backHair_c_06}
    componentAppliedLinearAccClamp=<10000 10000 10000>
    componentLinearAccScale=<1 1 1>
    componentLinearVelScale=<1 1 1>
    gravityOverride {z=-980}
    physicsBodyDefinitions [{
      boundBone {boneName=dyn_backHair_c_06}
      boxExtents=<25 30 30>
      constraintSetup {angularLimitsMax=<5 0 10>, angularLimitsMin=<-5 0 -10>, angularTarget=<1 0 0>}
      localJointOffset=<3 0 0>
     }]
   }
   n48 = Anim Dynamics - Bone: dyn_backHair_l_01(
    ComponentPose=@n47
    Alpha=1
    GravityOverride="0,0,-980"
    SimSpaceSettings=(SimSpaceAngularAlpha=0,MaxAngularVelocity=10000,MaxAngularAcceleration=10000,ExternalAngularVelocity=(X=0,Y=0,Z=0))
   ) {
    angularDampingOverride=1
    angularSpringConstant=20000
    bAngularSpring=true
    bOverrideAngularDamping=true
    bUseGravityOverride=true
    gravityScale=0.35
    simulationSpace=RootRelative
    alphaScaleBiasClamp {interpSpeedDecreasing=5, interpSpeedIncreasing=5}
    boundBone {boneName=dyn_backHair_l_01}
    componentAppliedLinearAccClamp=<10000 10000 10000>
    componentLinearAccScale=<1 1 1>
    componentLinearVelScale=<1 1 1>
    gravityOverride {z=-980}
    physicsBodyDefinitions [{
      boundBone {boneName=dyn_backHair_l_01}
      boxExtents=<30 45 45>
      constraintSetup {angularLimitsMax=<8 0 12>, angularLimitsMin=<-8 0 -5>, angularTarget=<1 0 0>}
      localJointOffset=<4 0 0>
     }]
   }
   n49 = Anim Dynamics - Bone: dyn_backHair_l_02(
    ComponentPose=@n48
    Alpha=1
    GravityOverride="0,0,-980"
    SimSpaceSettings=(SimSpaceAngularAlpha=0,MaxAngularVelocity=10000,MaxAngularAcceleration=10000,ExternalAngularVelocity=(X=0,Y=0,Z=0))
   ) {
    angularDampingOverride=1
    angularSpringConstant=20000
    bAngularSpring=true
    bOverrideAngularDamping=true
    bUseGravityOverride=true
    gravityScale=0.35
    simulationSpace=RootRelative
    alphaScaleBiasClamp {interpSpeedDecreasing=5, interpSpeedIncreasing=5}
    boundBone {boneName=dyn_backHair_l_02}
    componentAppliedLinearAccClamp=<10000 10000 10000>
    componentLinearAccScale=<1 1 1>
    componentLinearVelScale=<1 1 1>
    gravityOverride {z=-980}
    physicsBodyDefinitions [{
      boundBone {boneName=dyn_backHair_l_02}
      boxExtents=<30 45 45>
      constraintSetup {angularLimitsMax=<8 0 12>, angularLimitsMin=<-8 0 -5>, angularTarget=<1 0 0>}
      localJointOffset=<4 0 0>
     }]
   }
   n50 = Anim Dynamics - Bone: dyn_backHair_l_03(
    ComponentPose=@n49
    Alpha=1
    GravityOverride="0,0,-980"
    SimSpaceSettings=(SimSpaceAngularAlpha=0,MaxAngularVelocity=10000,MaxAngularAcceleration=10000,ExternalAngularVelocity=(X=0,Y=0,Z=0))
   ) {
    angularDampingOverride=1
    angularSpringConstant=3000
    bAngularSpring=true
    bOverrideAngularDamping=true
    bUseGravityOverride=true
    gravityScale=0.35
    simulationSpace=RootRelative
    alphaScaleBiasClamp {interpSpeedDecreasing=5, interpSpeedIncreasing=5}
    boundBone {boneName=dyn_backHair_l_03}
    componentAppliedLinearAccClamp=<10000 10000 10000>
    componentLinearAccScale=<1 1 1>
    componentLinearVelScale=<1 1 1>
    gravityOverride {z=-980}
    physicsBodyDefinitions [{
      boundBone {boneName=dyn_backHair_l_03}
      boxExtents=<25 30 30>
      constraintSetup {angularLimitsMax=<8 0 12>, angularLimitsMin=<-8 0 -5>, angularTarget=<1 0 0>}
      localJointOffset=<2 0 0>
     }]
   }
   n51 = Anim Dynamics - Bone: dyn_backHair_l_04(
    ComponentPose=@n50
    Alpha=1
    GravityOverride="0,0,-980"
    SimSpaceSettings=(SimSpaceAngularAlpha=0,MaxAngularVelocity=10000,MaxAngularAcceleration=10000,ExternalAngularVelocity=(X=0,Y=0,Z=0))
   ) {
    angularDampingOverride=1
    angularSpringConstant=2000
    bAngularSpring=true
    bOverrideAngularDamping=true
    bUseGravityOverride=true
    gravityScale=0.35
    simulationSpace=RootRelative
    alphaScaleBiasClamp {interpSpeedDecreasing=5, interpSpeedIncreasing=5}
    boundBone {boneName=dyn_backHair_l_04}
    componentAppliedLinearAccClamp=<10000 10000 10000>
    componentLinearAccScale=<1 1 1>
    componentLinearVelScale=<1 1 1>
    gravityOverride {z=-980}
    physicsBodyDefinitions [{
      boundBone {boneName=dyn_backHair_l_04}
      boxExtents=<25 30 30>
      constraintSetup {angularLimitsMax=<8 0 12>, angularLimitsMin=<-8 0 -5>, angularTarget=<1 0 0>}
      localJointOffset=<3 0 0>
     }]
   }
   n52 = Anim Dynamics - Bone: dyn_backHair_l_05(
    ComponentPose=@n51
    Alpha=1
    GravityOverride="0,0,-980"
    SimSpaceSettings=(SimSpaceAngularAlpha=0,MaxAngularVelocity=10000,MaxAngularAcceleration=10000,ExternalAngularVelocity=(X=0,Y=0,Z=0))
   ) {
    angularDampingOverride=1
    angularSpringConstant=1200
    bAngularSpring=true
    bOverrideAngularDamping=true
    bUseGravityOverride=true
    gravityScale=0.35
    simulationSpace=RootRelative
    alphaScaleBiasClamp {interpSpeedDecreasing=5, interpSpeedIncreasing=5}
    boundBone {boneName=dyn_backHair_l_05}
    componentAppliedLinearAccClamp=<10000 10000 10000>
    componentLinearAccScale=<1 1 1>
    componentLinearVelScale=<1 1 1>
    gravityOverride {z=-980}
    physicsBodyDefinitions [{
      boundBone {boneName=dyn_backHair_l_05}
      boxExtents=<25 30 30>
      constraintSetup {angularLimitsMax=<8 0 12>, angularLimitsMin=<-8 0 -5>, angularTarget=<1 0 0>}
      localJointOffset=<3 0 0>
     }]
   }
   n53 = Anim Dynamics - Bone: dyn_backHair_l_06(
    ComponentPose=@n52
    Alpha=1
    GravityOverride="0,0,-980"
    SimSpaceSettings=(SimSpaceAngularAlpha=0,MaxAngularVelocity=10000,MaxAngularAcceleration=10000,ExternalAngularVelocity=(X=0,Y=0,Z=0))
   ) {
    angularDampingOverride=1
    angularSpringConstant=1200
    bAngularSpring=true
    bOverrideAngularDamping=true
    bUseGravityOverride=true
    gravityScale=0.35
    simulationSpace=RootRelative
    alphaScaleBiasClamp {interpSpeedDecreasing=5, interpSpeedIncreasing=5}
    boundBone {boneName=dyn_backHair_l_06}
    componentAppliedLinearAccClamp=<10000 10000 10000>
    componentLinearAccScale=<1 1 1>
    componentLinearVelScale=<1 1 1>
    gravityOverride {z=-980}
    physicsBodyDefinitions [{
      boundBone {boneName=dyn_backHair_l_06}
      boxExtents=<25 30 30>
      constraintSetup {angularLimitsMax=<8 0 12>, angularLimitsMin=<-8 0 -5>, angularTarget=<1 0 0>}
      localJointOffset=<3 0 0>
     }]
   }
   n54 = RigidBody!(ComponentPose=@n53, Alpha) -> (Pose) {
    bOverrideWorldGravity=true
    overridePhysicsAsset=&/F_MED_AtticDrill_FaceAcc_FrontHair02_Physics
    simulationSpace=BaseBoneSpace
    baseBoneRef {boneName=neck_01}
    overrideWorldGravity {z=-980}
    simSpaceSettings {
     dampingAlpha=0
     maxAngularAcceleration=50
     maxAngularVelocity=7
     maxLinearAcceleration=4000
     maxLinearVelocity=800
     velocityScaleZ=0.3
     worldAlpha=0.5
     externalLinearDragV=<1.5 1.5 1.5>
    }
   }
   n55 = Anim Dynamics - Bone: dyn_hairtie_1!(
    ComponentPose=@n54
    Alpha=1
    SimSpaceSettings:Anim Phys Sim Space Settings Structure=(SimSpaceAngularAlpha=1,MaxAngularVelocity=8,MaxAngularAcceleration=40,ExternalAngularVelocity=(X=0,Y=0,Z=0))
    ExternalForce="0,0,0"
   ) -> (Pose) {
    angularDampingOverride=1
    angularSpringConstant=1000
    bAngularSpring=true
    bOverrideAngularDamping=true
    bOverrideLinearDamping=true
    bUseGravityOverride=true
    linearDampingOverride=1
    simulationSpace=RootRelative
    boundBone {boneName=dyn_hairtie_1}
    componentAppliedLinearAccClamp=<4000 4000 4000>
    gravityOverride {z=-800}
    physicsBodyDefinitions [{
      boundBone {boneName=dyn_hairtie_1}
      boxExtents=<11 11 11>
      constraintSetup {angularLimitsMax=<5 20 20>, angularLimitsMin=<-5 -20 -20>, angularTarget=<1 0 0>}
      localJointOffset=<2 0 0>
     }]
    simSpaceSettings {maxAngularAcceleration=40, maxAngularVelocity=8, simSpaceAngularAlpha=1}
   }
   n56 = Anim Dynamics - Bone: dyn_hairtie_2!(
    ComponentPose=@n55
    Alpha=1
    SimSpaceSettings:Anim Phys Sim Space Settings Structure=(SimSpaceAngularAlpha=1,MaxAngularVelocity=8,MaxAngularAcceleration=40,ExternalAngularVelocity=(X=0,Y=0,Z=0))
    ExternalForce="0,0,0"
   ) -> (Pose) {
    angularDampingOverride=1
    angularSpringConstant=1000
    bAngularSpring=true
    bOverrideAngularDamping=true
    bOverrideLinearDamping=true
    bUseGravityOverride=true
    linearDampingOverride=1
    simulationSpace=RootRelative
    boundBone {boneName=dyn_hairtie_2}
    componentAppliedLinearAccClamp=<4000 4000 4000>
    gravityOverride {z=-800}
    physicsBodyDefinitions [{
      boundBone {boneName=dyn_hairtie_2}
      boxExtents=<11 11 11>
      constraintSetup {angularLimitsMax=<5 20 20>, angularLimitsMin=<-5 -20 -20>, angularTarget=<1 0 0>}
      localJointOffset=<2 0 0>
     }]
    simSpaceSettings {maxAngularAcceleration=40, maxAngularVelocity=8, simSpaceAngularAlpha=1}
   }
   Save cached pose 'CP_Dynamics'(Pose=Component To Local(ComponentPose=@n56)) {cacheName=CP_Dynamics}
  ]}
 EventGraph {nodes [
   Event BlueprintUpdateAnimation()
   > n1 = "Gravity Override State Function (Parameter Asset) (Nativized)"(
    CharacterPartAnimInstance=Self-Reference()
    GameTimeInSeconds=GetGameTimeInSeconds()
    Parameters=/BRCosmetics/Characters/Player/Female/Medium/Bodies/F_MED_AtticDrill/Meshes/F_MED_AtticDrill_Parameters
   )
   > Set GravityOverride(GravityOverride=@n1)
   > Set RBAN Settings Out(RBAN Settings Out=@n1.OutRigidBodyAnimNodeParameters)
   > Set Alpha_Head(Alpha_Head:double=GetAlphaFromReferenceTransformDelta(SkeletalMeshComponent=Get SkelMesh(), BoneName=head, BaseName=neck_02, InputRange=(X=0,Y=-10), TwistAxis="1,0,0")) -> (Output_Get:double)
   n2 = Event OnCharacterCustomizationLoaded()
   > n3 = Sequence()
   n4 = CheckForBackbling()
   > Set IsWearingCape(IsWearingCape:bool=@n4.bIsBackblingACape) -> (Output_Get:bool)
   > Set Is Wearing Backbling(Is Wearing Backbling=@n4)
   > n5 = Sequence!(execute) -> (then_0, then_1, then_2:Exec)
   n6 = Branch(execute=@n5.then_1, Condition=Get Is Wearing Backbling())
   > Set FrontHair dynamics Alpha(FrontHair dynamics Alpha:double=0) -> (Output_Get:double)
   Set FrontHair dynamics Alpha(execute=@n6.else, FrontHair dynamics Alpha=1)
   n7 = Branch(execute=@n5.then_2, Condition=Get bIsInFrontEnd())
   > Set Sim Space Settings(Sim Space Settings:Anim Phys Sim Space Settings Structure=Get Sim Space Settings_FE() -> (Sim Space Settings_FE:Anim Phys Sim Space Settings Structure)) -> (Output_Get:Anim Phys Sim Space Settings Structure)
   Set Sim Space Settings(execute=@n7.else, Sim Space Settings=Get Sim Space Settings_BR() -> (Sim Space Settings_BR:Anim Phys Sim Space Settings Structure))
   n8 = Branch(execute=@n5.then_0, Condition=OR Boolean(A=Get IsWearingCape(), B=Get Is Wearing Backbling()))
   > Set Alpha_dynPack(Alpha_dynPack:double=1) -> (Output_Get:double)
   > Set Alpha_dyn(Alpha_dyn:double=0) -> (Output_Get:double)
   Set Alpha_dynPack(execute=@n8.else, Alpha_dynPack=0)
   > Set Alpha_dyn(Alpha_dyn=1)
   n9 = Set SkelMesh(execute=@n3.then_1, SkelMesh=Get Array Item with EFort Custom Part Type(Array=@n2.CustomMeshParts))
   > AddTickPrerequisiteComponent!(execute, self:Actor Component Object Reference=GetOwningComponent(), PrerequisiteComponent:Actor Component Object Reference=@n9) -> (then)
   Event BlueprintBeginPlay() {is=title K2Node_Event}
   > Set DynamicAlpha(DynamicAlpha:double=1) -> (Output_Get:double)
   Event OnBeginJumping()
   > Set TacticalJumping(TacticalJumping:bool=Get bIsTacticalSprinting()) -> (Output_Get:bool)
   Event OnEndFalling()
   > Set TacticalJumping(TacticalJumping=false)
  ]}
]
variables [
 Alpha_dyn {type=number}
 Alpha_dynPack {type=number}
 Alpha_Head {type=number}
 DynamicAlpha {type=number}
 FixAlpha {type=number}
 FrontHair dynamics Alpha {type=number, default=1}
 GravityOverride {type=Vector}
 Is Wearing Backbling {type=boolean}
 isAR Equipped {type=boolean}
 isShotgunEquipped {type=boolean}
 IsWearingCape {type=boolean}
 RBAN Settings Out {type=FortRigidBodyAnimNodeParameters}
 Sim Space Settings {
  type=AnimPhysSimSpaceSettings
  default {maxAngularAcceleration=10000, maxAngularVelocity=10000, simSpaceAngularAlpha=0, externalAngularVelocity=<0 0 0>}
 }
 Sim Space Settings_BR {
  type=AnimPhysSimSpaceSettings
  default {maxAngularAcceleration=10000, maxAngularVelocity=10000, simSpaceAngularAlpha=0, externalAngularVelocity=<0 0 0>}
 }
 Sim Space Settings_FE {
  type=AnimPhysSimSpaceSettings
  default {maxAngularAcceleration=10000, maxAngularVelocity=10000, simSpaceAngularAlpha=0.4, externalAngularVelocity=<0.2 0.2 0.2>}
 }
 SkelMesh {type=SkeletalMeshComponent}
 TacticalJumping {type=boolean}
]
